
using { /Fortnite.com/Devices }
using { /Verse.org/Simulation }
using { /UnrealEngine.com/Temporary/Diagnostics }
using { /UnrealEngine.com/Temporary/SpatialMath }
using { /Verse.org/Random }
using { /Fortnite.com/Characters }

using { HavocVerse.Combat }
using { HavocVerse.Devices }
using { HavocVerse.UI }
using { HavocVerse.Logging }
using { HavocVerse.Util }

# See https://dev.epicgames.com/documentation/en-us/uefn/create-your-own-device-in-verse for how to create a verse device.

# A Verse-authored creative device that can be placed in a level
crane_controller_device := class(creative_device):

    @editable CraneVolume: volume_device = volume_device{}
    
    var RewardTimer: float = 5.0
    @editable RewardTrigger: trigger_device = trigger_device{}
    @editable CraneAudio: audio_player_device = audio_player_device{}

    @editable CraneCinema: cinematic_sequence_device = cinematic_sequence_device{}
    @editable CraneCameraCinema: cinematic_sequence_device = cinematic_sequence_device{}

    var IsActive: logic = false

    var m_PlayerInCrane: [player]void = map{}

    BreakLoopEvent : event() = event(){}

    OnBegin<override>()<suspends>:void=
        CraneVolume.AgentEntersEvent.Subscribe(StartCrane)
        CraneVolume.AgentExitsEvent.Subscribe(ExitCrane)

    StartCrane(Agent:agent):void=
        if(not IsActive?):
            if:
                Player := player[Agent]
                set m_PlayerInCrane[Player] = void

            CraneAudio.Play()
            set IsActive = true
            spawn:
                GiveRewards2(Agent)

            CraneCinema.Play()
            CraneCameraCinema.Play(Agent)

    ExitCrane(Agent:agent):void=
        if (TestPlayer := player[Agent]):
            BreakLoopEvent.Signal()
            set m_PlayerInCrane = m_PlayerInCrane.RemoveByKey(TestPlayer)(0)
            set IsActive = false
            CraneAudio.Stop()
            CraneCinema.Pause()
            CraneCameraCinema.Stop(Agent)


    GiveRewards2(Agent:agent)<suspends>:void=
        race:
            block:
                loop:
                    Sleep(RewardTimer)
                    RewardTrigger.Trigger(Agent)
            block:
                BreakLoopEvent.Await()

    GiveRewards(Agent:agent)<suspends>:void=
        loop:
            if(IsActive?):
                RewardTrigger.Trigger(Agent)
                Sleep(RewardTimer)
            else:
                break
