using { /Fortnite.com/Devices }
using { /Verse.org/Simulation }
using { /UnrealEngine.com/Temporary/Diagnostics }

creature_spawn_manager_device := class(creative_device):
    @editable SpawnTriggers: []trigger_device = array{}
    @editable Location: creative_prop = creative_prop{}
    @editable SpawnDelay : float = 1.0
    @editable DisableDelay : float = 1.0
    @editable CreatureSpawner: creature_spawner_device = creature_spawner_device{}

    var m_InUse: logic = false
    
    OnBegin<override>()<suspends>:void=
        for(SpawnTrigger:SpawnTriggers):
            SpawnTrigger.TriggeredEvent.Subscribe(OnSpawnTriggerTriggered)

    OnSpawnTriggerTriggered(Agent:?agent):void=
        spawn:
            SpawnCreatures()

    SpawnCreatures()<suspends>:void=
        if (m_InUse?) {
            return
        }
        set m_InUse = true
        if (
            Location.IsValid[]
            TargetTransform := Location.GetTransform()
            CreatureSpawner.TeleportTo[TargetTransform]
        ):
            for (__ := 0..10) {
                Sleep(0.0)
                if. CreatureSpawner.TeleportTo[TargetTransform]
            }
            Sleep(SpawnDelay)
            CreatureSpawner.Enable()
            Sleep(DisableDelay)
            CreatureSpawner.Disable()
        set m_InUse = false
