
using { /Fortnite.com/Devices }
using { /Verse.org/Simulation }
using { /UnrealEngine.com/Temporary/Diagnostics }
using { /UnrealEngine.com/Temporary/SpatialMath }
using { /Verse.org/Random }

# See https://dev.epicgames.com/documentation/en-us/uefn/create-your-own-device-in-verse for how to create a verse device.

# A Verse-authored creative device that can be placed in a level
creature_world_manager_device := class(creative_device):

    @editable CreatureSpawners: []creature_spawner_device = array{}
    @editable Locations: []creative_prop = array{}

    @editable MobTimer: timer_device = timer_device{}

    # Runs when the device is started in a running game
    OnBegin<override>()<suspends>:void=
        MobTimer.SuccessEvent.Subscribe(OnMobTimerComplete)

    OnMobTimerComplete(Agent:?agent):void=
        for(Spawner:CreatureSpawners):
            MoveToRandomLocation(Spawner)
        spawn:
            MobTimerDelayedStart()

    MoveToRandomLocation(Spawner:creature_spawner_device):void=
        var Index : int = 0
        set Index = GetRandomInt(0,Locations.Length)

        if(Location:=Locations[Index]):
            TargetLocation := Location.GetTransform().Translation
            TargetRotation := Location.GetTransform().Rotation
            
            if(Spawner.TeleportTo[TargetLocation, TargetRotation]):

    MobTimerDelayedStart()<suspends>:void=
        Sleep(1.0)
        MobTimer.Start()
