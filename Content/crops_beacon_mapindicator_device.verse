
using { /Fortnite.com/Devices }
using { /Verse.org/Simulation }
using { /UnrealEngine.com/Temporary/Diagnostics }

# See https://dev.epicgames.com/documentation/en-us/uefn/create-your-own-device-in-verse for how to create a verse device.

# A Verse-authored creative device that can be placed in a level
crops_beacon_mapindicator_device := class(creative_device):

    @editable Collectible: collectible_object_device = collectible_object_device{}

    @editable MapIndicator: map_indicator_device = map_indicator_device{}
    @editable BillboardProp: creative_prop = creative_prop{}
    @editable CollectionAudio: audio_player_device = audio_player_device{}
    @editable RewardTrigger: trigger_device = trigger_device{}
    
    @editable RespawnOnDelay: logic = true
    @editable RespawnTriggers: []trigger_device = array{}

    @editable RespawnTime: float = 30.0

    # Runs when the device is started in a running game
    OnBegin<override>()<suspends>:void=
        Collectible.CollectedEvent.Subscribe(CropsCollected)
        for(Trigger:RespawnTriggers):
            Trigger.TriggeredEvent.Subscribe(EnableEverything2)

    CropsCollected(Agent:agent):void=
        CollectionAudio.Play(Agent)
        RewardTrigger.Trigger(Agent)
        DisableEverything()
        if(RespawnOnDelay?):
            spawn:
                EnableOnDelay()

    EnableOnDelay()<suspends>:void=
        Sleep(RespawnTime)
        EnableEverything()

    DisableEverything():void=
        MapIndicator.Disable()
        BillboardProp.Hide()
            
    EnableEverything():void=
        MapIndicator.Enable()
        BillboardProp.Show()
        Collectible.RespawnForAll()
            
    EnableEverything2(Agent:?agent):void=
        MapIndicator.Enable()
        BillboardProp.Show()
        Collectible.RespawnForAll()