using { /Fortnite.com/Devices }
using { /Fortnite.com/Characters }
using { /Verse.org/Simulation }
using { /UnrealEngine.com/Temporary/Diagnostics }
using { /Verse.org/Random }

# See https://dev.epicgames.com/documentation/en-us/uefn/create-your-own-device-in-verse for how to create a verse device.

# A Verse-authored creative device that can be placed in a level
events_device := class(creative_device):
    
        @editable EventTimer: timer_device = timer_device{}
    
        var EventIndex: int = 0
        @editable EventStartAudio: audio_player_device = audio_player_device{}
    
        #Blood Moon
        @editable BloodmoonTimer: timer_device = timer_device{}
        @editable BloodmoonPostProcess: post_process_device = post_process_device{}
        @editable BloodmoonStartSequence: cinematic_sequence_device = cinematic_sequence_device{}
        @editable BloodmoonEndSequence: cinematic_sequence_device = cinematic_sequence_device{}
        @editable BloodMoonCreatureSpawners: []creature_spawner_device = array{}
        @editable BloodMoonRewardTrigger: trigger_device = trigger_device{}
        @editable BloodMoonOverTrigger: trigger_device = trigger_device{}
        @editable BloodMoonActivatedTrigger: trigger_device = trigger_device{}
        @editable BloodMoonDeactivatedTrigger: trigger_device = trigger_device{}
        @editable BloodMoonBaseZombiesTrigger: trigger_device = trigger_device{}
    
        #Supply Cache
        @editable SupplyCacheHUD: hud_message_device = hud_message_device{}
        @editable ShowCache1Cine: cinematic_sequence_device = cinematic_sequence_device{}
        @editable ShowCache2Cine: cinematic_sequence_device = cinematic_sequence_device{}
        @editable ShowCache3Cine: cinematic_sequence_device = cinematic_sequence_device{}
        @editable ShowCache4Cine: cinematic_sequence_device = cinematic_sequence_device{}
        @editable HideCachesCine: cinematic_sequence_device = cinematic_sequence_device{}
        @editable CacheBeacons: []beacon_device = array{}
        @editable CacheButtons: []button_device = array{}
        @editable CacheRewardTrigger: trigger_device = trigger_device{}
    
        #Mega Brutes
        @editable MegaBruteHUD: hud_message_device = hud_message_device{}
        var MegaBruteActive: logic = false
        @editable MegaBrutePlacers: []creature_placer_device = array{}
        @editable BruteBeacons: []beacon_device = array{}
        @editable MegaBruteRewardTrigger: trigger_device = trigger_device{}

        #Supply Drops
        
        @editable OpenSupplyDropTrigger: trigger_device = trigger_device{}
        @editable SupplyDropHUD: hud_message_device = hud_message_device{}
        var SupplyDropActive: logic = false
        @editable SupplyDrops: []supply_drop_spawner_device = array{}
        @editable SupplyDropRewardTrigger: trigger_device = trigger_device{}
    
        OnBegin<override>()<suspends>:void=
            set EventIndex = GetRandomInt(1,8)

            for(CacheButton:CacheButtons):
                CacheButton.InteractedWithEvent.Subscribe(CacheCollected)
    
            EventTimer.SuccessEvent.Subscribe(OnEventTimerComplete)
            BloodmoonTimer.SuccessEvent.Subscribe(OnBloodmoonTimerComplete)
    
            for(SupplyDrop:SupplyDrops):
                SupplyDrop.OpenedEvent.Subscribe(OnSupplyDropOpened)

            for(BrutePlacer:MegaBrutePlacers):
                BrutePlacer.EliminatedEvent.Subscribe(OnMegaBruteEliminated)

        OnEventTimerComplete(Agent:?agent):void=
            spawn:
                StartNextEvent()
    
        StartNextEvent()<suspends>:void =
            if(EventIndex = 1):
                SetBloodMoonActive()
                Sleep(3.0)
                set EventIndex = 2
                
            else if(EventIndex = 2):  
                SpawnSupplyDrop()
                Sleep(3.0)
                set EventIndex = 3

            else if(EventIndex = 3):
                StartMegaBrute()
                Sleep(3.0)
                set EventIndex = 4
            
            else if(EventIndex = 4):
                SpawnSupplyCache()
                Sleep(3.0)
                set EventIndex = 5
                
            else if(EventIndex = 5):
                SetBloodMoonActive()
                Sleep(3.0)
                set EventIndex = 6
        
            else if(EventIndex = 6):
                StartMegaBrute()
                Sleep(3.0)
                set EventIndex = 7
                
            else if(EventIndex = 7):
                SpawnSupplyDrop()
                Sleep(3.0)
                set EventIndex = 8
            
            else if(EventIndex = 8):
                SpawnSupplyCache()
                Sleep(3.0)
                set EventIndex = 1
    
        ##### Supply Drops

        SpawnSupplyDrop()<suspends>:void=
            OpenSupplyDropTrigger.Trigger()
            Sleep(3.0)
            AllPlayers := GetPlayspace().GetPlayers()
            SupplyDropHUD.Show()
            
            for(Player:AllPlayers):
                EventStartAudio.Play(Player)

            var Index : int = 0
            set Index = GetRandomInt(0,SupplyDrops.Length-1)
            if(Drop:=SupplyDrops[Index]):
                Drop.Spawn()
            Sleep(1.0)
            EventTimer.Reset()
            EventTimer.Start()
    
        OnSupplyDropOpened(Agent:agent):void=
            SupplyDropRewardTrigger.Trigger(Agent)
            
    
        ##### Mega Brute
        StartMegaBrute()<suspends>:void=
            for(Beacon:BruteBeacons):
                Beacon.Disable()
            for(Brute:MegaBrutePlacers):
                Brute.Despawn()
            AllPlayers := GetPlayspace().GetPlayers()

            for(Player:AllPlayers):
                EventStartAudio.Play(Player)

            var Index : int = 0
            set Index = GetRandomInt(0,MegaBrutePlacers.Length-1)
            if(Brute:=MegaBrutePlacers[Index]):
                Brute.Spawn()
                MegaBruteHUD.Show()
                if(Beacon:=BruteBeacons[Index]):
                    Beacon.Enable()

            Sleep(1.0)
            EventTimer.Reset()
            EventTimer.Start()

        OnMegaBruteEliminated(Agent:?agent):void=
            if(MaybeAgent:=Agent?):
                MegaBruteRewardTrigger.Trigger(MaybeAgent)
            for(Beacon:BruteBeacons):
                Beacon.Disable()

        ##### Supply Caches

        SpawnSupplyCache()<suspends>:void=
            RemoveSupplyCaches()
            Sleep(2.0)

            AllPlayers := GetPlayspace().GetPlayers()
            

            
            SupplyCacheHUD.Show()
            for(Player:AllPlayers):
                EventStartAudio.Play(Player)

            var Index : int = 0
            set Index = GetRandomInt(0,3)
            if(Index=0):
                ShowCache1Cine.Play()
            if(Index=1):
                ShowCache2Cine.Play()
            if(Index=2):
                ShowCache3Cine.Play()
            if(Index=3):
                ShowCache4Cine.Play()
            if(Beacon:=CacheBeacons[Index]):
                Beacon.Enable()
            if(Button:=CacheButtons[Index]):
                Button.Enable()
        
            Sleep(1.0)
            EventTimer.Reset()
            EventTimer.Start()
    
        RemoveSupplyCaches():void=
            HideCachesCine.Play()
            for(Beacon:CacheBeacons):
                Beacon.Disable()
            for(Button:CacheButtons):
                Button.Disable()
    
        CacheCollected(Agent:agent):void=
            CacheRewardTrigger.Trigger(Agent)
            RemoveSupplyCaches()


    
        ##### Blood Moon
    
        SetBloodMoonActive()<suspends>:void=
            AllPlayers := GetPlayspace().GetPlayers()
            for(Player:AllPlayers):
                BloodMoonRewardTrigger.Trigger(Player)
            Sleep(0.2)
            for(Player:AllPlayers):
                BloodMoonActivatedTrigger.Trigger(Player)

            BloodmoonStartSequence.Play()

            BloodmoonTimer.Reset()
            BloodmoonTimer.Start()
            
            for(CreatureSpawner:BloodMoonCreatureSpawners):
                CreatureSpawner.Enable()

            spawn:
                BloodMoonBaseZombies()

        BloodMoonBaseZombies()<suspends>:void=
            Sleep(10.0)
            BloodMoonBaseZombiesTrigger.Trigger()
            Sleep(50.0)
            BloodMoonBaseZombiesTrigger.Trigger()
            Sleep(50.0)
            BloodMoonBaseZombiesTrigger.Trigger()
            Sleep(50.0)
            BloodMoonBaseZombiesTrigger.Trigger()
    
        SetBloodMoonInactive()<suspends>:void=
            AllPlayers := GetPlayspace().GetPlayers()
            for(Player:AllPlayers):
                BloodMoonOverTrigger.Trigger(Player)
            Sleep(0.2)
            for(Player:AllPlayers):
                BloodMoonDeactivatedTrigger.Trigger(Player)

            BloodmoonEndSequence.Play()

            BloodmoonTimer.Reset()
            for(CreatureSpawner:BloodMoonCreatureSpawners):
                CreatureSpawner.Disable()
                CreatureSpawner.EliminateCreatures()

        OnBloodmoonTimerComplete(Agent:?agent):void=
            EventTimer.Reset()
            EventTimer.Start()
            spawn:
                SetBloodMoonInactive()