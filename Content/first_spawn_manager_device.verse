using { /Fortnite.com/Devices }
using { /Verse.org/Simulation }
using { /UnrealEngine.com/Temporary/Diagnostics }
using { /Fortnite.com/Characters }
using { /Fortnite.com/Playspaces }
using { /UnrealEngine.com/Temporary/Diagnostics }
using { /UnrealEngine.com/Temporary/SpatialMath }
using { /Verse.org/Random }

using { HavocVerse.Debug }
using { HavocVerse.Util }
using { HavocVerse.Logging }

# See https://dev.epicgames.com/documentation/en-us/uefn/create-your-own-device-in-verse for how to create a verse device.

# A Verse-authored creative device that can be placed in a level
first_spawn_manager_device := class(creative_device, aware_device):

    @editable StartVolume: volume_device = volume_device{}

    @editable PlotInstances: []plot_instance = array{}

    @editable BackupTeleporters: []teleporter_device = array{}

    DeviceMeta<override>()<transacts>:aware_device_meta=
        aware_device_meta{
            DeviceName := "first_spawn_manager_device"
        }  

    OnBegin<override>()<suspends>:void=
        StartVolume.AgentEntersEvent.Subscribe(SpawnPlayer)
        for(PlotInstance:PlotInstances):
            PlotInstance.EventsSubscribe()

    SpawnPlayer(Agent:agent):void=
        var index: int = 0
        loop:
            if(PlotInstance:=PlotInstances[index]):
                if(PlotInstance.CanBeTeleported?):
                    PlotInstance.TeleportPlayer(Agent)
                    break
                set index += 1
            else:
                var RandomIndex: int = GetRandomInt(0,BackupTeleporters.Length-1)
                if(Teleporter:=BackupTeleporters[RandomIndex]):
                    Teleporter.Teleport(Agent)
                else if(Teleporter:=BackupTeleporters[0]):
                    Teleporter.Teleport(Agent)
                break


plot_instance<internal> := class<final><concrete>:

    var PlayersNear: int = 0
    var CanBeTeleported: logic = true
    var CurrentlyClaimed: logic = false
    @editable Teleporter: teleporter_device = teleporter_device{}
    @editable ClaimTrigger: trigger_device = trigger_device{}
    @editable RollbackTrigger: trigger_device = trigger_device{}
    @editable VolumeChecker: volume_device = volume_device{}

    EventsSubscribe():void=
        ClaimTrigger.TriggeredEvent.Subscribe(OnClaimTrigger)
        RollbackTrigger.TriggeredEvent.Subscribe(OnRollbackTrigger)
        VolumeChecker.AgentEntersEvent.Subscribe(IncPlayerNear)
        VolumeChecker.AgentEntersEvent.Subscribe(DecPlayerNear)

    TeleportPlayer(Agent:agent):void=
        Teleporter.Teleport(Agent)
        set CanBeTeleported = false

    OnClaimTrigger(Agent:?agent):void=
        set CurrentlyClaimed = true
        set CanBeTeleported = false

    OnRollbackTrigger(Agent:?agent):void=
        set CurrentlyClaimed = false
        set CanBeTeleported = true

    IncPlayerNear(Agent:agent):void=
        set PlayersNear += 1

    DecPlayerNear(Agent:agent):void=
        set PlayersNear -= 1
        if(PlayersNear<1):
            if(not CurrentlyClaimed?):
                set CanBeTeleported = true
    

        

