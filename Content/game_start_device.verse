
using { /Fortnite.com/Devices }
using { /Verse.org/Simulation }
using { /UnrealEngine.com/Temporary/Diagnostics }

# See https://dev.epicgames.com/documentation/en-us/uefn/create-your-own-device-in-verse for how to create a verse device.

# A Verse-authored creative device that can be placed in a level
game_start_device := class(creative_device):


    @editable StartTriggers: []trigger_device = array{}

    OnBegin<override>()<suspends>:void=
        for(Trig:StartTriggers):
            Trig.Trigger()