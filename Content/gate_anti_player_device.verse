
using { /Fortnite.com/Devices }
using { /Verse.org/Simulation }
using { /UnrealEngine.com/Temporary/Diagnostics }

# See https://dev.epicgames.com/documentation/en-us/uefn/create-your-own-device-in-verse for how to create a verse device.

# A Verse-authored creative device that can be placed in a level
gate_anti_player_device := class(creative_device):

    @editable Switches: []switch_device = array{}
    @editable EnableTriggers: []trigger_device = array{}
    @editable DisableTriggers: []trigger_device = array{}

    @editable Barriers: []barrier_device = array{}
    @editable DamageVolumes: []damage_volume_device = array{}

    OnBegin<override>()<suspends>:void=
        for(EnableTrigger:EnableTriggers):
            EnableTrigger.TriggeredEvent.Subscribe(EnableBarriers)
        for(DisableTrigger:DisableTriggers):
            DisableTrigger.TriggeredEvent.Subscribe(DisableBarriers)
        for(Switch:Switches):
            Switch.TurnedOnEvent.Subscribe(DisableBarriers2)
        for(Switch:Switches):
            Switch.TurnedOffEvent.Subscribe(EnableBarriers2)

    EnableBarriers(Agent:?agent):void=
        for(Barrier:Barriers):
            Barrier.Enable()

    DisableBarriers(Agent:?agent):void=
        for(Barrier:Barriers):
            Barrier.Disable()
        for(DamageVolume:DamageVolumes):
            DamageVolume.Disable()

    EnableBarriers2(Agent:agent):void=
        for(Barrier:Barriers):
            Barrier.Enable()
        for(DamageVolume:DamageVolumes):
            DamageVolume.Enable()

    DisableBarriers2(Agent:agent):void=
        for(Barrier:Barriers):
            Barrier.Disable()
        for(DamageVolume:DamageVolumes):
            DamageVolume.Disable()