###################################################################
# Automatically generated using `gen_local_assets_provider.py`.
#
# Generated by: Brad @ 2024-10-23 23:51:55 UTC.
###################################################################

using { /Fortnite.com/Devices }
using { /Verse.org/Assets }
using { /UnrealEngine.com/Temporary/SpatialMath }

using { HavocVerse.Util }

Assets<public> := module:
    VFX<public> := module:
        Import<public> := module:
BuyplateBillboards<public> := module:
DailyRewardsMenuUI<public> := module:
FontQuicksandBold<public> := module:
GameplayRewardsMenuUI<public> := module:
HavocVerseUI<public> := module:
TemplateCommon<public> := module:
    BackendTooltips<public> := module:
UI<public> := module:
    Boosts<public> := module:
    Levels<public> := module:
    Rebirths<public> := module:
    ResourceBillboards<public> := module:
    Resources<public> := module:
UIAssets<public> := module:
    Apoc<public> := module:
    Events<public> := module:
    Notifs<public> := module:
    Widgets<public> := module:

local_assets_provider_device := class(creative_device, assets_provider_device):
    CreateAssetsProvider<override>():assets_provider=
        TextureAssets := array{
            texture_metadata{Name := "Assets/VFX/level_up", Asset := Assets.VFX.level_up, Dimensions := option{vector2{X := 512.0, Y := 128.0}}}
            texture_metadata{Name := "Assets/VFX/plus_armory", Asset := Assets.VFX.plus_armory, Dimensions := option{vector2{X := 256.0, Y := 256.0}}}
            texture_metadata{Name := "Assets/VFX/Import/T_NeonCircle", Asset := Assets.VFX.Import.T_NeonCircle, Dimensions := option{vector2{X := 1024.0, Y := 1024.0}}}
            texture_metadata{Name := "Assets/VFX/Import/T_Spark3", Asset := Assets.VFX.Import.T_Spark3, Dimensions := option{vector2{X := 64.0, Y := 64.0}}}
            texture_metadata{Name := "BuyplateBillboards/bp_001", Asset := BuyplateBillboards.bp_001, Dimensions := option{vector2{X := 512.0, Y := 256.0}}}
            texture_metadata{Name := "BuyplateBillboards/bp_002", Asset := BuyplateBillboards.bp_002, Dimensions := option{vector2{X := 512.0, Y := 256.0}}}
            texture_metadata{Name := "BuyplateBillboards/bp_003", Asset := BuyplateBillboards.bp_003, Dimensions := option{vector2{X := 512.0, Y := 256.0}}}
            texture_metadata{Name := "BuyplateBillboards/bp_004", Asset := BuyplateBillboards.bp_004, Dimensions := option{vector2{X := 512.0, Y := 256.0}}}
            texture_metadata{Name := "BuyplateBillboards/bp_005", Asset := BuyplateBillboards.bp_005, Dimensions := option{vector2{X := 512.0, Y := 256.0}}}
            texture_metadata{Name := "BuyplateBillboards/bp_006", Asset := BuyplateBillboards.bp_006, Dimensions := option{vector2{X := 512.0, Y := 256.0}}}
            texture_metadata{Name := "BuyplateBillboards/bp_007", Asset := BuyplateBillboards.bp_007, Dimensions := option{vector2{X := 512.0, Y := 256.0}}}
            texture_metadata{Name := "BuyplateBillboards/bp_008", Asset := BuyplateBillboards.bp_008, Dimensions := option{vector2{X := 512.0, Y := 256.0}}}
            texture_metadata{Name := "BuyplateBillboards/bp_009", Asset := BuyplateBillboards.bp_009, Dimensions := option{vector2{X := 512.0, Y := 256.0}}}
            texture_metadata{Name := "BuyplateBillboards/bp_010", Asset := BuyplateBillboards.bp_010, Dimensions := option{vector2{X := 512.0, Y := 256.0}}}
            texture_metadata{Name := "BuyplateBillboards/bp_011", Asset := BuyplateBillboards.bp_011, Dimensions := option{vector2{X := 512.0, Y := 256.0}}}
            texture_metadata{Name := "BuyplateBillboards/bp_012", Asset := BuyplateBillboards.bp_012, Dimensions := option{vector2{X := 512.0, Y := 256.0}}}
            texture_metadata{Name := "BuyplateBillboards/bp_013", Asset := BuyplateBillboards.bp_013, Dimensions := option{vector2{X := 512.0, Y := 256.0}}}
            texture_metadata{Name := "BuyplateBillboards/bp_014", Asset := BuyplateBillboards.bp_014, Dimensions := option{vector2{X := 512.0, Y := 256.0}}}
            texture_metadata{Name := "BuyplateBillboards/bp_015", Asset := BuyplateBillboards.bp_015, Dimensions := option{vector2{X := 512.0, Y := 256.0}}}
            texture_metadata{Name := "BuyplateBillboards/bp_016", Asset := BuyplateBillboards.bp_016, Dimensions := option{vector2{X := 512.0, Y := 256.0}}}
            texture_metadata{Name := "BuyplateBillboards/bp_017", Asset := BuyplateBillboards.bp_017, Dimensions := option{vector2{X := 512.0, Y := 256.0}}}
            texture_metadata{Name := "BuyplateBillboards/bp_018", Asset := BuyplateBillboards.bp_018, Dimensions := option{vector2{X := 512.0, Y := 256.0}}}
            texture_metadata{Name := "BuyplateBillboards/bp_019", Asset := BuyplateBillboards.bp_019, Dimensions := option{vector2{X := 512.0, Y := 256.0}}}
            texture_metadata{Name := "BuyplateBillboards/bp_020", Asset := BuyplateBillboards.bp_020, Dimensions := option{vector2{X := 512.0, Y := 256.0}}}
            texture_metadata{Name := "BuyplateBillboards/bp_021", Asset := BuyplateBillboards.bp_021, Dimensions := option{vector2{X := 512.0, Y := 256.0}}}
            texture_metadata{Name := "BuyplateBillboards/bp_022", Asset := BuyplateBillboards.bp_022, Dimensions := option{vector2{X := 512.0, Y := 256.0}}}
            texture_metadata{Name := "BuyplateBillboards/bp_023", Asset := BuyplateBillboards.bp_023, Dimensions := option{vector2{X := 512.0, Y := 256.0}}}
            texture_metadata{Name := "BuyplateBillboards/bp_024", Asset := BuyplateBillboards.bp_024, Dimensions := option{vector2{X := 512.0, Y := 256.0}}}
            texture_metadata{Name := "BuyplateBillboards/bp_025", Asset := BuyplateBillboards.bp_025, Dimensions := option{vector2{X := 512.0, Y := 256.0}}}
            texture_metadata{Name := "BuyplateBillboards/bp_026", Asset := BuyplateBillboards.bp_026, Dimensions := option{vector2{X := 512.0, Y := 256.0}}}
            texture_metadata{Name := "BuyplateBillboards/bp_027", Asset := BuyplateBillboards.bp_027, Dimensions := option{vector2{X := 512.0, Y := 256.0}}}
            texture_metadata{Name := "BuyplateBillboards/bp_028", Asset := BuyplateBillboards.bp_028, Dimensions := option{vector2{X := 512.0, Y := 256.0}}}
            texture_metadata{Name := "BuyplateBillboards/bp_029", Asset := BuyplateBillboards.bp_029, Dimensions := option{vector2{X := 512.0, Y := 256.0}}}
            texture_metadata{Name := "BuyplateBillboards/bp_030", Asset := BuyplateBillboards.bp_030, Dimensions := option{vector2{X := 512.0, Y := 256.0}}}
            texture_metadata{Name := "BuyplateBillboards/bp_031", Asset := BuyplateBillboards.bp_031, Dimensions := option{vector2{X := 512.0, Y := 256.0}}}
            texture_metadata{Name := "BuyplateBillboards/bp_032", Asset := BuyplateBillboards.bp_032, Dimensions := option{vector2{X := 512.0, Y := 256.0}}}
            texture_metadata{Name := "BuyplateBillboards/bp_033", Asset := BuyplateBillboards.bp_033, Dimensions := option{vector2{X := 512.0, Y := 256.0}}}
            texture_metadata{Name := "BuyplateBillboards/bp_034", Asset := BuyplateBillboards.bp_034, Dimensions := option{vector2{X := 512.0, Y := 256.0}}}
            texture_metadata{Name := "BuyplateBillboards/bp_035", Asset := BuyplateBillboards.bp_035, Dimensions := option{vector2{X := 512.0, Y := 256.0}}}
            texture_metadata{Name := "BuyplateBillboards/bp_036", Asset := BuyplateBillboards.bp_036, Dimensions := option{vector2{X := 512.0, Y := 256.0}}}
            texture_metadata{Name := "BuyplateBillboards/bp_037", Asset := BuyplateBillboards.bp_037, Dimensions := option{vector2{X := 512.0, Y := 256.0}}}
            texture_metadata{Name := "BuyplateBillboards/bp_038", Asset := BuyplateBillboards.bp_038, Dimensions := option{vector2{X := 512.0, Y := 256.0}}}
            texture_metadata{Name := "BuyplateBillboards/bp_039", Asset := BuyplateBillboards.bp_039, Dimensions := option{vector2{X := 512.0, Y := 256.0}}}
            texture_metadata{Name := "BuyplateBillboards/bp_050", Asset := BuyplateBillboards.bp_050, Dimensions := option{vector2{X := 512.0, Y := 256.0}}}
            texture_metadata{Name := "BuyplateBillboards/bp_051", Asset := BuyplateBillboards.bp_051, Dimensions := option{vector2{X := 512.0, Y := 256.0}}}
            texture_metadata{Name := "BuyplateBillboards/bp_052", Asset := BuyplateBillboards.bp_052, Dimensions := option{vector2{X := 512.0, Y := 256.0}}}
            texture_metadata{Name := "BuyplateBillboards/bp_100", Asset := BuyplateBillboards.bp_100, Dimensions := option{vector2{X := 512.0, Y := 256.0}}}
            texture_metadata{Name := "BuyplateBillboards/bp_101", Asset := BuyplateBillboards.bp_101, Dimensions := option{vector2{X := 512.0, Y := 256.0}}}
            texture_metadata{Name := "BuyplateBillboards/bp_102", Asset := BuyplateBillboards.bp_102, Dimensions := option{vector2{X := 512.0, Y := 256.0}}}
            texture_metadata{Name := "BuyplateBillboards/bp_103", Asset := BuyplateBillboards.bp_103, Dimensions := option{vector2{X := 512.0, Y := 256.0}}}
            texture_metadata{Name := "BuyplateBillboards/bp_104", Asset := BuyplateBillboards.bp_104, Dimensions := option{vector2{X := 512.0, Y := 256.0}}}
            texture_metadata{Name := "BuyplateBillboards/bp_200", Asset := BuyplateBillboards.bp_200, Dimensions := option{vector2{X := 512.0, Y := 256.0}}}
            texture_metadata{Name := "BuyplateBillboards/bp_201", Asset := BuyplateBillboards.bp_201, Dimensions := option{vector2{X := 512.0, Y := 256.0}}}
            texture_metadata{Name := "BuyplateBillboards/bp_202", Asset := BuyplateBillboards.bp_202, Dimensions := option{vector2{X := 512.0, Y := 256.0}}}
            texture_metadata{Name := "BuyplateBillboards/bp_203", Asset := BuyplateBillboards.bp_203, Dimensions := option{vector2{X := 512.0, Y := 256.0}}}
            texture_metadata{Name := "BuyplateBillboards/bp_204", Asset := BuyplateBillboards.bp_204, Dimensions := option{vector2{X := 512.0, Y := 256.0}}}
            texture_metadata{Name := "BuyplateBillboards/bp_205", Asset := BuyplateBillboards.bp_205, Dimensions := option{vector2{X := 512.0, Y := 256.0}}}
            texture_metadata{Name := "BuyplateBillboards/bp_206", Asset := BuyplateBillboards.bp_206, Dimensions := option{vector2{X := 512.0, Y := 256.0}}}
            texture_metadata{Name := "BuyplateBillboards/bp_207", Asset := BuyplateBillboards.bp_207, Dimensions := option{vector2{X := 512.0, Y := 256.0}}}
            texture_metadata{Name := "BuyplateBillboards/bp_208", Asset := BuyplateBillboards.bp_208, Dimensions := option{vector2{X := 512.0, Y := 256.0}}}
            texture_metadata{Name := "BuyplateBillboards/bp_209", Asset := BuyplateBillboards.bp_209, Dimensions := option{vector2{X := 512.0, Y := 256.0}}}
            texture_metadata{Name := "BuyplateBillboards/bp_210", Asset := BuyplateBillboards.bp_210, Dimensions := option{vector2{X := 512.0, Y := 256.0}}}
            texture_metadata{Name := "BuyplateBillboards/bp_211", Asset := BuyplateBillboards.bp_211, Dimensions := option{vector2{X := 512.0, Y := 256.0}}}
            texture_metadata{Name := "BuyplateBillboards/bp_212", Asset := BuyplateBillboards.bp_212, Dimensions := option{vector2{X := 512.0, Y := 256.0}}}
            texture_metadata{Name := "BuyplateBillboards/bp_213", Asset := BuyplateBillboards.bp_213, Dimensions := option{vector2{X := 512.0, Y := 256.0}}}
            texture_metadata{Name := "BuyplateBillboards/bp_214", Asset := BuyplateBillboards.bp_214, Dimensions := option{vector2{X := 512.0, Y := 256.0}}}
            texture_metadata{Name := "BuyplateBillboards/bp_215", Asset := BuyplateBillboards.bp_215, Dimensions := option{vector2{X := 512.0, Y := 256.0}}}
            texture_metadata{Name := "BuyplateBillboards/bp_216", Asset := BuyplateBillboards.bp_216, Dimensions := option{vector2{X := 512.0, Y := 256.0}}}
            texture_metadata{Name := "BuyplateBillboards/bp_217", Asset := BuyplateBillboards.bp_217, Dimensions := option{vector2{X := 512.0, Y := 256.0}}}
            texture_metadata{Name := "BuyplateBillboards/bp_218", Asset := BuyplateBillboards.bp_218, Dimensions := option{vector2{X := 512.0, Y := 256.0}}}
            texture_metadata{Name := "BuyplateBillboards/bp_219", Asset := BuyplateBillboards.bp_219, Dimensions := option{vector2{X := 512.0, Y := 256.0}}}
            texture_metadata{Name := "BuyplateBillboards/bp_220", Asset := BuyplateBillboards.bp_220, Dimensions := option{vector2{X := 512.0, Y := 256.0}}}
            texture_metadata{Name := "BuyplateBillboards/bp_250", Asset := BuyplateBillboards.bp_250, Dimensions := option{vector2{X := 512.0, Y := 256.0}}}
            texture_metadata{Name := "BuyplateBillboards/bp_251", Asset := BuyplateBillboards.bp_251, Dimensions := option{vector2{X := 512.0, Y := 256.0}}}
            texture_metadata{Name := "BuyplateBillboards/bp_252", Asset := BuyplateBillboards.bp_252, Dimensions := option{vector2{X := 512.0, Y := 256.0}}}
            texture_metadata{Name := "BuyplateBillboards/bp_253", Asset := BuyplateBillboards.bp_253, Dimensions := option{vector2{X := 512.0, Y := 256.0}}}
            texture_metadata{Name := "BuyplateBillboards/bp_300", Asset := BuyplateBillboards.bp_300, Dimensions := option{vector2{X := 512.0, Y := 256.0}}}
            texture_metadata{Name := "BuyplateBillboards/bp_301", Asset := BuyplateBillboards.bp_301, Dimensions := option{vector2{X := 512.0, Y := 256.0}}}
            texture_metadata{Name := "BuyplateBillboards/bp_302", Asset := BuyplateBillboards.bp_302, Dimensions := option{vector2{X := 512.0, Y := 256.0}}}
            texture_metadata{Name := "BuyplateBillboards/bp_303", Asset := BuyplateBillboards.bp_303, Dimensions := option{vector2{X := 512.0, Y := 256.0}}}
            texture_metadata{Name := "BuyplateBillboards/bp_304", Asset := BuyplateBillboards.bp_304, Dimensions := option{vector2{X := 512.0, Y := 256.0}}}
            texture_metadata{Name := "BuyplateBillboards/bp_305", Asset := BuyplateBillboards.bp_305, Dimensions := option{vector2{X := 512.0, Y := 256.0}}}
            texture_metadata{Name := "BuyplateBillboards/bp_306", Asset := BuyplateBillboards.bp_306, Dimensions := option{vector2{X := 512.0, Y := 256.0}}}
            texture_metadata{Name := "BuyplateBillboards/bp_307", Asset := BuyplateBillboards.bp_307, Dimensions := option{vector2{X := 512.0, Y := 256.0}}}
            texture_metadata{Name := "BuyplateBillboards/bp_308", Asset := BuyplateBillboards.bp_308, Dimensions := option{vector2{X := 512.0, Y := 256.0}}}
            texture_metadata{Name := "BuyplateBillboards/bp_309", Asset := BuyplateBillboards.bp_309, Dimensions := option{vector2{X := 512.0, Y := 256.0}}}
            texture_metadata{Name := "BuyplateBillboards/bp_310", Asset := BuyplateBillboards.bp_310, Dimensions := option{vector2{X := 512.0, Y := 256.0}}}
            texture_metadata{Name := "BuyplateBillboards/bp_311", Asset := BuyplateBillboards.bp_311, Dimensions := option{vector2{X := 512.0, Y := 256.0}}}
            texture_metadata{Name := "BuyplateBillboards/bp_312", Asset := BuyplateBillboards.bp_312, Dimensions := option{vector2{X := 512.0, Y := 256.0}}}
            texture_metadata{Name := "BuyplateBillboards/bp_313", Asset := BuyplateBillboards.bp_313, Dimensions := option{vector2{X := 512.0, Y := 256.0}}}
            texture_metadata{Name := "BuyplateBillboards/bp_314", Asset := BuyplateBillboards.bp_314, Dimensions := option{vector2{X := 512.0, Y := 256.0}}}
            texture_metadata{Name := "BuyplateBillboards/bp_315", Asset := BuyplateBillboards.bp_315, Dimensions := option{vector2{X := 512.0, Y := 256.0}}}
            texture_metadata{Name := "BuyplateBillboards/bp_316", Asset := BuyplateBillboards.bp_316, Dimensions := option{vector2{X := 512.0, Y := 256.0}}}
            texture_metadata{Name := "BuyplateBillboards/bp_317", Asset := BuyplateBillboards.bp_317, Dimensions := option{vector2{X := 512.0, Y := 256.0}}}
            texture_metadata{Name := "BuyplateBillboards/bp_318", Asset := BuyplateBillboards.bp_318, Dimensions := option{vector2{X := 512.0, Y := 256.0}}}
            texture_metadata{Name := "BuyplateBillboards/bp_319", Asset := BuyplateBillboards.bp_319, Dimensions := option{vector2{X := 512.0, Y := 256.0}}}
            texture_metadata{Name := "BuyplateBillboards/bp_320", Asset := BuyplateBillboards.bp_320, Dimensions := option{vector2{X := 512.0, Y := 256.0}}}
            texture_metadata{Name := "BuyplateBillboards/bp_321", Asset := BuyplateBillboards.bp_321, Dimensions := option{vector2{X := 512.0, Y := 256.0}}}
            texture_metadata{Name := "BuyplateBillboards/bp_322", Asset := BuyplateBillboards.bp_322, Dimensions := option{vector2{X := 512.0, Y := 256.0}}}
            texture_metadata{Name := "BuyplateBillboards/bp_323", Asset := BuyplateBillboards.bp_323, Dimensions := option{vector2{X := 512.0, Y := 256.0}}}
            texture_metadata{Name := "BuyplateBillboards/bp_324", Asset := BuyplateBillboards.bp_324, Dimensions := option{vector2{X := 512.0, Y := 256.0}}}
            texture_metadata{Name := "BuyplateBillboards/bp_325", Asset := BuyplateBillboards.bp_325, Dimensions := option{vector2{X := 512.0, Y := 256.0}}}
            texture_metadata{Name := "BuyplateBillboards/bp_400", Asset := BuyplateBillboards.bp_400, Dimensions := option{vector2{X := 512.0, Y := 256.0}}}
            texture_metadata{Name := "BuyplateBillboards/bp_401", Asset := BuyplateBillboards.bp_401, Dimensions := option{vector2{X := 512.0, Y := 256.0}}}
            texture_metadata{Name := "BuyplateBillboards/bp_402", Asset := BuyplateBillboards.bp_402, Dimensions := option{vector2{X := 512.0, Y := 256.0}}}
            texture_metadata{Name := "BuyplateBillboards/bp_403", Asset := BuyplateBillboards.bp_403, Dimensions := option{vector2{X := 512.0, Y := 256.0}}}
            texture_metadata{Name := "BuyplateBillboards/bp_404", Asset := BuyplateBillboards.bp_404, Dimensions := option{vector2{X := 512.0, Y := 256.0}}}
            texture_metadata{Name := "BuyplateBillboards/bp_405", Asset := BuyplateBillboards.bp_405, Dimensions := option{vector2{X := 512.0, Y := 256.0}}}
            texture_metadata{Name := "BuyplateBillboards/bp_406", Asset := BuyplateBillboards.bp_406, Dimensions := option{vector2{X := 512.0, Y := 256.0}}}
            texture_metadata{Name := "BuyplateBillboards/bp_407", Asset := BuyplateBillboards.bp_407, Dimensions := option{vector2{X := 512.0, Y := 256.0}}}
            texture_metadata{Name := "BuyplateBillboards/bp_408", Asset := BuyplateBillboards.bp_408, Dimensions := option{vector2{X := 512.0, Y := 256.0}}}
            texture_metadata{Name := "BuyplateBillboards/bp_409", Asset := BuyplateBillboards.bp_409, Dimensions := option{vector2{X := 512.0, Y := 256.0}}}
            texture_metadata{Name := "BuyplateBillboards/bp_410", Asset := BuyplateBillboards.bp_410, Dimensions := option{vector2{X := 512.0, Y := 256.0}}}
            texture_metadata{Name := "BuyplateBillboards/bp_411", Asset := BuyplateBillboards.bp_411, Dimensions := option{vector2{X := 512.0, Y := 256.0}}}
            texture_metadata{Name := "BuyplateBillboards/bp_412", Asset := BuyplateBillboards.bp_412, Dimensions := option{vector2{X := 512.0, Y := 256.0}}}
            texture_metadata{Name := "BuyplateBillboards/bp_413", Asset := BuyplateBillboards.bp_413, Dimensions := option{vector2{X := 512.0, Y := 256.0}}}
            texture_metadata{Name := "BuyplateBillboards/bp_414", Asset := BuyplateBillboards.bp_414, Dimensions := option{vector2{X := 512.0, Y := 256.0}}}
            texture_metadata{Name := "BuyplateBillboards/bp_415", Asset := BuyplateBillboards.bp_415, Dimensions := option{vector2{X := 512.0, Y := 256.0}}}
            texture_metadata{Name := "BuyplateBillboards/bp_416", Asset := BuyplateBillboards.bp_416, Dimensions := option{vector2{X := 512.0, Y := 256.0}}}
            texture_metadata{Name := "BuyplateBillboards/bp_417", Asset := BuyplateBillboards.bp_417, Dimensions := option{vector2{X := 512.0, Y := 256.0}}}
            texture_metadata{Name := "BuyplateBillboards/bp_418", Asset := BuyplateBillboards.bp_418, Dimensions := option{vector2{X := 512.0, Y := 256.0}}}
            texture_metadata{Name := "BuyplateBillboards/bp_419", Asset := BuyplateBillboards.bp_419, Dimensions := option{vector2{X := 512.0, Y := 256.0}}}
            texture_metadata{Name := "BuyplateBillboards/bp_420", Asset := BuyplateBillboards.bp_420, Dimensions := option{vector2{X := 512.0, Y := 256.0}}}
            texture_metadata{Name := "BuyplateBillboards/bp_421", Asset := BuyplateBillboards.bp_421, Dimensions := option{vector2{X := 512.0, Y := 256.0}}}
            texture_metadata{Name := "BuyplateBillboards/bp_422", Asset := BuyplateBillboards.bp_422, Dimensions := option{vector2{X := 512.0, Y := 256.0}}}
            texture_metadata{Name := "BuyplateBillboards/bp_500", Asset := BuyplateBillboards.bp_500, Dimensions := option{vector2{X := 512.0, Y := 256.0}}}
            texture_metadata{Name := "BuyplateBillboards/bp_501", Asset := BuyplateBillboards.bp_501, Dimensions := option{vector2{X := 512.0, Y := 256.0}}}
            texture_metadata{Name := "BuyplateBillboards/bp_502", Asset := BuyplateBillboards.bp_502, Dimensions := option{vector2{X := 512.0, Y := 256.0}}}
            texture_metadata{Name := "BuyplateBillboards/bp_503", Asset := BuyplateBillboards.bp_503, Dimensions := option{vector2{X := 512.0, Y := 256.0}}}
            texture_metadata{Name := "BuyplateBillboards/bp_504", Asset := BuyplateBillboards.bp_504, Dimensions := option{vector2{X := 512.0, Y := 256.0}}}
            texture_metadata{Name := "BuyplateBillboards/bp_505", Asset := BuyplateBillboards.bp_505, Dimensions := option{vector2{X := 512.0, Y := 256.0}}}
            texture_metadata{Name := "BuyplateBillboards/bp_506", Asset := BuyplateBillboards.bp_506, Dimensions := option{vector2{X := 512.0, Y := 256.0}}}
            texture_metadata{Name := "BuyplateBillboards/bp_507", Asset := BuyplateBillboards.bp_507, Dimensions := option{vector2{X := 512.0, Y := 256.0}}}
            texture_metadata{Name := "BuyplateBillboards/bp_508", Asset := BuyplateBillboards.bp_508, Dimensions := option{vector2{X := 512.0, Y := 256.0}}}
            texture_metadata{Name := "BuyplateBillboards/bp_509", Asset := BuyplateBillboards.bp_509, Dimensions := option{vector2{X := 512.0, Y := 256.0}}}
            texture_metadata{Name := "BuyplateBillboards/bp_510", Asset := BuyplateBillboards.bp_510, Dimensions := option{vector2{X := 512.0, Y := 256.0}}}
            texture_metadata{Name := "BuyplateBillboards/bp_511", Asset := BuyplateBillboards.bp_511, Dimensions := option{vector2{X := 512.0, Y := 256.0}}}
            texture_metadata{Name := "BuyplateBillboards/bp_512", Asset := BuyplateBillboards.bp_512, Dimensions := option{vector2{X := 512.0, Y := 256.0}}}
            texture_metadata{Name := "BuyplateBillboards/bp_513", Asset := BuyplateBillboards.bp_513, Dimensions := option{vector2{X := 512.0, Y := 256.0}}}
            texture_metadata{Name := "BuyplateBillboards/bp_514", Asset := BuyplateBillboards.bp_514, Dimensions := option{vector2{X := 512.0, Y := 256.0}}}
            texture_metadata{Name := "BuyplateBillboards/bp_515", Asset := BuyplateBillboards.bp_515, Dimensions := option{vector2{X := 512.0, Y := 256.0}}}
            texture_metadata{Name := "BuyplateBillboards/bp_516", Asset := BuyplateBillboards.bp_516, Dimensions := option{vector2{X := 512.0, Y := 256.0}}}
            texture_metadata{Name := "BuyplateBillboards/bp_517", Asset := BuyplateBillboards.bp_517, Dimensions := option{vector2{X := 512.0, Y := 256.0}}}
            texture_metadata{Name := "BuyplateBillboards/bp_518", Asset := BuyplateBillboards.bp_518, Dimensions := option{vector2{X := 512.0, Y := 256.0}}}
            texture_metadata{Name := "BuyplateBillboards/bp_519", Asset := BuyplateBillboards.bp_519, Dimensions := option{vector2{X := 512.0, Y := 256.0}}}
            texture_metadata{Name := "BuyplateBillboards/bp_520", Asset := BuyplateBillboards.bp_520, Dimensions := option{vector2{X := 512.0, Y := 256.0}}}
            texture_metadata{Name := "BuyplateBillboards/bp_521", Asset := BuyplateBillboards.bp_521, Dimensions := option{vector2{X := 512.0, Y := 256.0}}}
            texture_metadata{Name := "BuyplateBillboards/bp_522", Asset := BuyplateBillboards.bp_522, Dimensions := option{vector2{X := 512.0, Y := 256.0}}}
            texture_metadata{Name := "BuyplateBillboards/bp_523", Asset := BuyplateBillboards.bp_523, Dimensions := option{vector2{X := 512.0, Y := 256.0}}}
            texture_metadata{Name := "BuyplateBillboards/bp_524", Asset := BuyplateBillboards.bp_524, Dimensions := option{vector2{X := 512.0, Y := 256.0}}}
            texture_metadata{Name := "BuyplateBillboards/bp_525", Asset := BuyplateBillboards.bp_525, Dimensions := option{vector2{X := 512.0, Y := 256.0}}}
            texture_metadata{Name := "BuyplateBillboards/bp_526", Asset := BuyplateBillboards.bp_526, Dimensions := option{vector2{X := 512.0, Y := 256.0}}}
            texture_metadata{Name := "BuyplateBillboards/bp_527", Asset := BuyplateBillboards.bp_527, Dimensions := option{vector2{X := 512.0, Y := 256.0}}}
            texture_metadata{Name := "BuyplateBillboards/bp_528", Asset := BuyplateBillboards.bp_528, Dimensions := option{vector2{X := 512.0, Y := 256.0}}}
            texture_metadata{Name := "BuyplateBillboards/bp_529", Asset := BuyplateBillboards.bp_529, Dimensions := option{vector2{X := 512.0, Y := 256.0}}}
            texture_metadata{Name := "BuyplateBillboards/bp_530", Asset := BuyplateBillboards.bp_530, Dimensions := option{vector2{X := 512.0, Y := 256.0}}}
            texture_metadata{Name := "BuyplateBillboards/bp_531", Asset := BuyplateBillboards.bp_531, Dimensions := option{vector2{X := 512.0, Y := 256.0}}}
            texture_metadata{Name := "BuyplateBillboards/bp_600", Asset := BuyplateBillboards.bp_600, Dimensions := option{vector2{X := 512.0, Y := 256.0}}}
            texture_metadata{Name := "BuyplateBillboards/bp_601", Asset := BuyplateBillboards.bp_601, Dimensions := option{vector2{X := 512.0, Y := 256.0}}}
            texture_metadata{Name := "BuyplateBillboards/bp_602", Asset := BuyplateBillboards.bp_602, Dimensions := option{vector2{X := 512.0, Y := 256.0}}}
            texture_metadata{Name := "BuyplateBillboards/bp_603", Asset := BuyplateBillboards.bp_603, Dimensions := option{vector2{X := 512.0, Y := 256.0}}}
            texture_metadata{Name := "BuyplateBillboards/bp_604", Asset := BuyplateBillboards.bp_604, Dimensions := option{vector2{X := 512.0, Y := 256.0}}}
            texture_metadata{Name := "BuyplateBillboards/bp_605", Asset := BuyplateBillboards.bp_605, Dimensions := option{vector2{X := 512.0, Y := 256.0}}}
            texture_metadata{Name := "BuyplateBillboards/bp_606", Asset := BuyplateBillboards.bp_606, Dimensions := option{vector2{X := 512.0, Y := 256.0}}}
            texture_metadata{Name := "BuyplateBillboards/bp_607", Asset := BuyplateBillboards.bp_607, Dimensions := option{vector2{X := 512.0, Y := 256.0}}}
            texture_metadata{Name := "BuyplateBillboards/bp_608", Asset := BuyplateBillboards.bp_608, Dimensions := option{vector2{X := 512.0, Y := 256.0}}}
            texture_metadata{Name := "BuyplateBillboards/bp_609", Asset := BuyplateBillboards.bp_609, Dimensions := option{vector2{X := 512.0, Y := 256.0}}}
            texture_metadata{Name := "BuyplateBillboards/bp_610", Asset := BuyplateBillboards.bp_610, Dimensions := option{vector2{X := 512.0, Y := 256.0}}}
            texture_metadata{Name := "BuyplateBillboards/bp_611", Asset := BuyplateBillboards.bp_611, Dimensions := option{vector2{X := 512.0, Y := 256.0}}}
            texture_metadata{Name := "BuyplateBillboards/bp_612", Asset := BuyplateBillboards.bp_612, Dimensions := option{vector2{X := 512.0, Y := 256.0}}}
            texture_metadata{Name := "BuyplateBillboards/bp_613", Asset := BuyplateBillboards.bp_613, Dimensions := option{vector2{X := 512.0, Y := 256.0}}}
            texture_metadata{Name := "BuyplateBillboards/bp_614", Asset := BuyplateBillboards.bp_614, Dimensions := option{vector2{X := 512.0, Y := 256.0}}}
            texture_metadata{Name := "BuyplateBillboards/bp_615", Asset := BuyplateBillboards.bp_615, Dimensions := option{vector2{X := 512.0, Y := 256.0}}}
            texture_metadata{Name := "BuyplateBillboards/bp_616", Asset := BuyplateBillboards.bp_616, Dimensions := option{vector2{X := 512.0, Y := 256.0}}}
            texture_metadata{Name := "BuyplateBillboards/bp_617", Asset := BuyplateBillboards.bp_617, Dimensions := option{vector2{X := 512.0, Y := 256.0}}}
            texture_metadata{Name := "BuyplateBillboards/bp_618", Asset := BuyplateBillboards.bp_618, Dimensions := option{vector2{X := 512.0, Y := 256.0}}}
            texture_metadata{Name := "BuyplateBillboards/bp_619", Asset := BuyplateBillboards.bp_619, Dimensions := option{vector2{X := 512.0, Y := 256.0}}}
            texture_metadata{Name := "BuyplateBillboards/bp_620", Asset := BuyplateBillboards.bp_620, Dimensions := option{vector2{X := 512.0, Y := 256.0}}}
            texture_metadata{Name := "BuyplateBillboards/bp_621", Asset := BuyplateBillboards.bp_621, Dimensions := option{vector2{X := 512.0, Y := 256.0}}}
            texture_metadata{Name := "BuyplateBillboards/bp_622", Asset := BuyplateBillboards.bp_622, Dimensions := option{vector2{X := 512.0, Y := 256.0}}}
            texture_metadata{Name := "BuyplateBillboards/bp_623", Asset := BuyplateBillboards.bp_623, Dimensions := option{vector2{X := 512.0, Y := 256.0}}}
            texture_metadata{Name := "BuyplateBillboards/bp_624", Asset := BuyplateBillboards.bp_624, Dimensions := option{vector2{X := 512.0, Y := 256.0}}}
            texture_metadata{Name := "BuyplateBillboards/bp_625", Asset := BuyplateBillboards.bp_625, Dimensions := option{vector2{X := 512.0, Y := 256.0}}}
            texture_metadata{Name := "BuyplateBillboards/bp_626", Asset := BuyplateBillboards.bp_626, Dimensions := option{vector2{X := 512.0, Y := 256.0}}}
            texture_metadata{Name := "BuyplateBillboards/bp_627", Asset := BuyplateBillboards.bp_627, Dimensions := option{vector2{X := 512.0, Y := 256.0}}}
            texture_metadata{Name := "BuyplateBillboards/bp_628", Asset := BuyplateBillboards.bp_628, Dimensions := option{vector2{X := 512.0, Y := 256.0}}}
            texture_metadata{Name := "BuyplateBillboards/bp_629", Asset := BuyplateBillboards.bp_629, Dimensions := option{vector2{X := 512.0, Y := 256.0}}}
            texture_metadata{Name := "BuyplateBillboards/bp_630", Asset := BuyplateBillboards.bp_630, Dimensions := option{vector2{X := 512.0, Y := 256.0}}}
            texture_metadata{Name := "BuyplateBillboards/bp_631", Asset := BuyplateBillboards.bp_631, Dimensions := option{vector2{X := 512.0, Y := 256.0}}}
            texture_metadata{Name := "BuyplateBillboards/bp_700", Asset := BuyplateBillboards.bp_700, Dimensions := option{vector2{X := 512.0, Y := 256.0}}}
            texture_metadata{Name := "BuyplateBillboards/bp_701", Asset := BuyplateBillboards.bp_701, Dimensions := option{vector2{X := 512.0, Y := 256.0}}}
            texture_metadata{Name := "BuyplateBillboards/bp_702", Asset := BuyplateBillboards.bp_702, Dimensions := option{vector2{X := 512.0, Y := 256.0}}}
            texture_metadata{Name := "BuyplateBillboards/bp_703", Asset := BuyplateBillboards.bp_703, Dimensions := option{vector2{X := 512.0, Y := 256.0}}}
            texture_metadata{Name := "BuyplateBillboards/bp_704", Asset := BuyplateBillboards.bp_704, Dimensions := option{vector2{X := 512.0, Y := 256.0}}}
            texture_metadata{Name := "BuyplateBillboards/bp_705", Asset := BuyplateBillboards.bp_705, Dimensions := option{vector2{X := 512.0, Y := 256.0}}}
            texture_metadata{Name := "BuyplateBillboards/bp_claim", Asset := BuyplateBillboards.bp_claim, Dimensions := option{vector2{X := 512.0, Y := 256.0}}}
            texture_metadata{Name := "DailyRewardsMenuUI/background", Asset := DailyRewardsMenuUI.background, Dimensions := option{vector2{X := 2048.0, Y := 1397.0}}}
            texture_metadata{Name := "DailyRewardsMenuUI/claimed_label", Asset := DailyRewardsMenuUI.claimed_label, Dimensions := option{vector2{X := 268.0, Y := 268.0}}}
            texture_metadata{Name := "DailyRewardsMenuUI/day_bg", Asset := DailyRewardsMenuUI.day_bg, Dimensions := option{vector2{X := 480.0, Y := 508.0}}}
            texture_metadata{Name := "DailyRewardsMenuUI/prev_day_overlay", Asset := DailyRewardsMenuUI.prev_day_overlay, Dimensions := option{vector2{X := 480.0, Y := 508.0}}}
            texture_metadata{Name := "DailyRewardsMenuUI/reward_day1", Asset := DailyRewardsMenuUI.reward_day1, Dimensions := option{vector2{X := 480.0, Y := 508.0}}}
            texture_metadata{Name := "DailyRewardsMenuUI/reward_day2", Asset := DailyRewardsMenuUI.reward_day2, Dimensions := option{vector2{X := 480.0, Y := 508.0}}}
            texture_metadata{Name := "DailyRewardsMenuUI/reward_day3", Asset := DailyRewardsMenuUI.reward_day3, Dimensions := option{vector2{X := 480.0, Y := 508.0}}}
            texture_metadata{Name := "DailyRewardsMenuUI/reward_day4", Asset := DailyRewardsMenuUI.reward_day4, Dimensions := option{vector2{X := 480.0, Y := 508.0}}}
            texture_metadata{Name := "DailyRewardsMenuUI/reward_day5", Asset := DailyRewardsMenuUI.reward_day5, Dimensions := option{vector2{X := 480.0, Y := 508.0}}}
            texture_metadata{Name := "DailyRewardsMenuUI/reward_day6", Asset := DailyRewardsMenuUI.reward_day6, Dimensions := option{vector2{X := 480.0, Y := 508.0}}}
            texture_metadata{Name := "DailyRewardsMenuUI/special_gift_claimed", Asset := DailyRewardsMenuUI.special_gift_claimed, Dimensions := option{vector2{X := 915.0, Y := 1070.0}}}
            texture_metadata{Name := "DailyRewardsMenuUI/special_gift_unclaimed", Asset := DailyRewardsMenuUI.special_gift_unclaimed, Dimensions := option{vector2{X := 915.0, Y := 1070.0}}}
            texture_metadata{Name := "DailyRewardsMenuUI/today_bg", Asset := DailyRewardsMenuUI.today_bg, Dimensions := option{vector2{X := 480.0, Y := 508.0}}}
            texture_metadata{Name := "FontQuicksandBold/a", Asset := FontQuicksandBold.a, Dimensions := option{vector2{X := 15.0, Y := 29.0}}}
            texture_metadata{Name := "FontQuicksandBold/A_upper", Asset := FontQuicksandBold.A_upper, Dimensions := option{vector2{X := 16.0, Y := 29.0}}}
            texture_metadata{Name := "FontQuicksandBold/b", Asset := FontQuicksandBold.b, Dimensions := option{vector2{X := 15.0, Y := 29.0}}}
            texture_metadata{Name := "FontQuicksandBold/B_upper", Asset := FontQuicksandBold.B_upper, Dimensions := option{vector2{X := 16.0, Y := 29.0}}}
            texture_metadata{Name := "FontQuicksandBold/c", Asset := FontQuicksandBold.c, Dimensions := option{vector2{X := 12.0, Y := 29.0}}}
            texture_metadata{Name := "FontQuicksandBold/C_upper", Asset := FontQuicksandBold.C_upper, Dimensions := option{vector2{X := 15.0, Y := 29.0}}}
            texture_metadata{Name := "FontQuicksandBold/d", Asset := FontQuicksandBold.d, Dimensions := option{vector2{X := 15.0, Y := 29.0}}}
            texture_metadata{Name := "FontQuicksandBold/D_upper", Asset := FontQuicksandBold.D_upper, Dimensions := option{vector2{X := 17.0, Y := 29.0}}}
            texture_metadata{Name := "FontQuicksandBold/e", Asset := FontQuicksandBold.e, Dimensions := option{vector2{X := 14.0, Y := 29.0}}}
            texture_metadata{Name := "FontQuicksandBold/E_upper", Asset := FontQuicksandBold.E_upper, Dimensions := option{vector2{X := 14.0, Y := 29.0}}}
            texture_metadata{Name := "FontQuicksandBold/f", Asset := FontQuicksandBold.f, Dimensions := option{vector2{X := 11.0, Y := 29.0}}}
            texture_metadata{Name := "FontQuicksandBold/F_upper", Asset := FontQuicksandBold.F_upper, Dimensions := option{vector2{X := 13.0, Y := 29.0}}}
            texture_metadata{Name := "FontQuicksandBold/g", Asset := FontQuicksandBold.g, Dimensions := option{vector2{X := 15.0, Y := 29.0}}}
            texture_metadata{Name := "FontQuicksandBold/G_upper", Asset := FontQuicksandBold.G_upper, Dimensions := option{vector2{X := 17.0, Y := 29.0}}}
            texture_metadata{Name := "FontQuicksandBold/h", Asset := FontQuicksandBold.h, Dimensions := option{vector2{X := 14.0, Y := 29.0}}}
            texture_metadata{Name := "FontQuicksandBold/H_upper", Asset := FontQuicksandBold.H_upper, Dimensions := option{vector2{X := 17.0, Y := 29.0}}}
            texture_metadata{Name := "FontQuicksandBold/i", Asset := FontQuicksandBold.i, Dimensions := option{vector2{X := 6.0, Y := 29.0}}}
            texture_metadata{Name := "FontQuicksandBold/I_upper", Asset := FontQuicksandBold.I_upper, Dimensions := option{vector2{X := 7.0, Y := 29.0}}}
            texture_metadata{Name := "FontQuicksandBold/j", Asset := FontQuicksandBold.j, Dimensions := option{vector2{X := 7.0, Y := 29.0}}}
            texture_metadata{Name := "FontQuicksandBold/J_upper", Asset := FontQuicksandBold.J_upper, Dimensions := option{vector2{X := 13.0, Y := 29.0}}}
            texture_metadata{Name := "FontQuicksandBold/k", Asset := FontQuicksandBold.k, Dimensions := option{vector2{X := 14.0, Y := 29.0}}}
            texture_metadata{Name := "FontQuicksandBold/K_upper", Asset := FontQuicksandBold.K_upper, Dimensions := option{vector2{X := 16.0, Y := 29.0}}}
            texture_metadata{Name := "FontQuicksandBold/l", Asset := FontQuicksandBold.l, Dimensions := option{vector2{X := 6.0, Y := 29.0}}}
            texture_metadata{Name := "FontQuicksandBold/L_upper", Asset := FontQuicksandBold.L_upper, Dimensions := option{vector2{X := 13.0, Y := 29.0}}}
            texture_metadata{Name := "FontQuicksandBold/m", Asset := FontQuicksandBold.m, Dimensions := option{vector2{X := 22.0, Y := 29.0}}}
            texture_metadata{Name := "FontQuicksandBold/M_upper", Asset := FontQuicksandBold.M_upper, Dimensions := option{vector2{X := 20.0, Y := 29.0}}}
            texture_metadata{Name := "FontQuicksandBold/n", Asset := FontQuicksandBold.n, Dimensions := option{vector2{X := 14.0, Y := 29.0}}}
            texture_metadata{Name := "FontQuicksandBold/N_upper", Asset := FontQuicksandBold.N_upper, Dimensions := option{vector2{X := 18.0, Y := 29.0}}}
            texture_metadata{Name := "FontQuicksandBold/num_0", Asset := FontQuicksandBold.num_0, Dimensions := option{vector2{X := 15.0, Y := 29.0}}}
            texture_metadata{Name := "FontQuicksandBold/num_1", Asset := FontQuicksandBold.num_1, Dimensions := option{vector2{X := 9.0, Y := 29.0}}}
            texture_metadata{Name := "FontQuicksandBold/num_2", Asset := FontQuicksandBold.num_2, Dimensions := option{vector2{X := 13.0, Y := 29.0}}}
            texture_metadata{Name := "FontQuicksandBold/num_3", Asset := FontQuicksandBold.num_3, Dimensions := option{vector2{X := 13.0, Y := 29.0}}}
            texture_metadata{Name := "FontQuicksandBold/num_4", Asset := FontQuicksandBold.num_4, Dimensions := option{vector2{X := 17.0, Y := 29.0}}}
            texture_metadata{Name := "FontQuicksandBold/num_5", Asset := FontQuicksandBold.num_5, Dimensions := option{vector2{X := 14.0, Y := 29.0}}}
            texture_metadata{Name := "FontQuicksandBold/num_6", Asset := FontQuicksandBold.num_6, Dimensions := option{vector2{X := 13.0, Y := 29.0}}}
            texture_metadata{Name := "FontQuicksandBold/num_7", Asset := FontQuicksandBold.num_7, Dimensions := option{vector2{X := 14.0, Y := 29.0}}}
            texture_metadata{Name := "FontQuicksandBold/num_8", Asset := FontQuicksandBold.num_8, Dimensions := option{vector2{X := 14.0, Y := 29.0}}}
            texture_metadata{Name := "FontQuicksandBold/num_9", Asset := FontQuicksandBold.num_9, Dimensions := option{vector2{X := 13.0, Y := 29.0}}}
            texture_metadata{Name := "FontQuicksandBold/o", Asset := FontQuicksandBold.o, Dimensions := option{vector2{X := 15.0, Y := 29.0}}}
            texture_metadata{Name := "FontQuicksandBold/O_upper", Asset := FontQuicksandBold.O_upper, Dimensions := option{vector2{X := 18.0, Y := 29.0}}}
            texture_metadata{Name := "FontQuicksandBold/p", Asset := FontQuicksandBold.p, Dimensions := option{vector2{X := 15.0, Y := 29.0}}}
            texture_metadata{Name := "FontQuicksandBold/P_upper", Asset := FontQuicksandBold.P_upper, Dimensions := option{vector2{X := 14.0, Y := 29.0}}}
            texture_metadata{Name := "FontQuicksandBold/q", Asset := FontQuicksandBold.q, Dimensions := option{vector2{X := 15.0, Y := 29.0}}}
            texture_metadata{Name := "FontQuicksandBold/Q_upper", Asset := FontQuicksandBold.Q_upper, Dimensions := option{vector2{X := 19.0, Y := 29.0}}}
            texture_metadata{Name := "FontQuicksandBold/r", Asset := FontQuicksandBold.r, Dimensions := option{vector2{X := 10.0, Y := 29.0}}}
            texture_metadata{Name := "FontQuicksandBold/R_upper", Asset := FontQuicksandBold.R_upper, Dimensions := option{vector2{X := 16.0, Y := 29.0}}}
            texture_metadata{Name := "FontQuicksandBold/s", Asset := FontQuicksandBold.s, Dimensions := option{vector2{X := 12.0, Y := 29.0}}}
            texture_metadata{Name := "FontQuicksandBold/S_upper", Asset := FontQuicksandBold.S_upper, Dimensions := option{vector2{X := 14.0, Y := 29.0}}}
            texture_metadata{Name := "FontQuicksandBold/symb_bang", Asset := FontQuicksandBold.symb_bang, Dimensions := option{vector2{X := 6.0, Y := 29.0}}}
            texture_metadata{Name := "FontQuicksandBold/symb_comma", Asset := FontQuicksandBold.symb_comma, Dimensions := option{vector2{X := 7.0, Y := 29.0}}}
            texture_metadata{Name := "FontQuicksandBold/symb_dot", Asset := FontQuicksandBold.symb_dot, Dimensions := option{vector2{X := 6.0, Y := 29.0}}}
            texture_metadata{Name := "FontQuicksandBold/symb_eq", Asset := FontQuicksandBold.symb_eq, Dimensions := option{vector2{X := 15.0, Y := 29.0}}}
            texture_metadata{Name := "FontQuicksandBold/symb_minus", Asset := FontQuicksandBold.symb_minus, Dimensions := option{vector2{X := 10.0, Y := 29.0}}}
            texture_metadata{Name := "FontQuicksandBold/symb_paren_close", Asset := FontQuicksandBold.symb_paren_close, Dimensions := option{vector2{X := 9.0, Y := 29.0}}}
            texture_metadata{Name := "FontQuicksandBold/symb_paren_open", Asset := FontQuicksandBold.symb_paren_open, Dimensions := option{vector2{X := 9.0, Y := 29.0}}}
            texture_metadata{Name := "FontQuicksandBold/symb_plus", Asset := FontQuicksandBold.symb_plus, Dimensions := option{vector2{X := 14.0, Y := 29.0}}}
            texture_metadata{Name := "FontQuicksandBold/symb_question", Asset := FontQuicksandBold.symb_question, Dimensions := option{vector2{X := 13.0, Y := 29.0}}}
            texture_metadata{Name := "FontQuicksandBold/symb_slash", Asset := FontQuicksandBold.symb_slash, Dimensions := option{vector2{X := 14.0, Y := 29.0}}}
            texture_metadata{Name := "FontQuicksandBold/t", Asset := FontQuicksandBold.t, Dimensions := option{vector2{X := 10.0, Y := 29.0}}}
            texture_metadata{Name := "FontQuicksandBold/T_upper", Asset := FontQuicksandBold.T_upper, Dimensions := option{vector2{X := 15.0, Y := 29.0}}}
            texture_metadata{Name := "FontQuicksandBold/u", Asset := FontQuicksandBold.u, Dimensions := option{vector2{X := 14.0, Y := 29.0}}}
            texture_metadata{Name := "FontQuicksandBold/U_upper", Asset := FontQuicksandBold.U_upper, Dimensions := option{vector2{X := 17.0, Y := 29.0}}}
            texture_metadata{Name := "FontQuicksandBold/v", Asset := FontQuicksandBold.v, Dimensions := option{vector2{X := 13.0, Y := 29.0}}}
            texture_metadata{Name := "FontQuicksandBold/V_upper", Asset := FontQuicksandBold.V_upper, Dimensions := option{vector2{X := 16.0, Y := 29.0}}}
            texture_metadata{Name := "FontQuicksandBold/w", Asset := FontQuicksandBold.w, Dimensions := option{vector2{X := 18.0, Y := 29.0}}}
            texture_metadata{Name := "FontQuicksandBold/W_upper", Asset := FontQuicksandBold.W_upper, Dimensions := option{vector2{X := 23.0, Y := 29.0}}}
            texture_metadata{Name := "FontQuicksandBold/x", Asset := FontQuicksandBold.x, Dimensions := option{vector2{X := 12.0, Y := 29.0}}}
            texture_metadata{Name := "FontQuicksandBold/X_upper", Asset := FontQuicksandBold.X_upper, Dimensions := option{vector2{X := 16.0, Y := 29.0}}}
            texture_metadata{Name := "FontQuicksandBold/y", Asset := FontQuicksandBold.y, Dimensions := option{vector2{X := 14.0, Y := 29.0}}}
            texture_metadata{Name := "FontQuicksandBold/Y_upper", Asset := FontQuicksandBold.Y_upper, Dimensions := option{vector2{X := 14.0, Y := 29.0}}}
            texture_metadata{Name := "FontQuicksandBold/z", Asset := FontQuicksandBold.z, Dimensions := option{vector2{X := 12.0, Y := 29.0}}}
            texture_metadata{Name := "FontQuicksandBold/Z_upper", Asset := FontQuicksandBold.Z_upper, Dimensions := option{vector2{X := 16.0, Y := 29.0}}}
            texture_metadata{Name := "GameplayRewardsMenuUI/background", Asset := GameplayRewardsMenuUI.background, Dimensions := option{vector2{X := 2048.0, Y := 1425.0}}}
            texture_metadata{Name := "GameplayRewardsMenuUI/card_bg", Asset := GameplayRewardsMenuUI.card_bg, Dimensions := option{vector2{X := 548.0, Y := 622.0}}}
            texture_metadata{Name := "GameplayRewardsMenuUI/reward_blue_chest", Asset := GameplayRewardsMenuUI.reward_blue_chest, Dimensions := option{vector2{X := 219.0, Y := 178.0}}}
            texture_metadata{Name := "GameplayRewardsMenuUI/reward_blue_chest_claimed", Asset := GameplayRewardsMenuUI.reward_blue_chest_claimed, Dimensions := option{vector2{X := 219.0, Y := 185.0}}}
            texture_metadata{Name := "GameplayRewardsMenuUI/reward_blue_gift", Asset := GameplayRewardsMenuUI.reward_blue_gift, Dimensions := option{vector2{X := 219.0, Y := 166.0}}}
            texture_metadata{Name := "GameplayRewardsMenuUI/reward_chest", Asset := GameplayRewardsMenuUI.reward_chest, Dimensions := option{vector2{X := 219.0, Y := 178.0}}}
            texture_metadata{Name := "GameplayRewardsMenuUI/reward_chest_claimed", Asset := GameplayRewardsMenuUI.reward_chest_claimed, Dimensions := option{vector2{X := 219.0, Y := 185.0}}}
            texture_metadata{Name := "GameplayRewardsMenuUI/reward_red_gift", Asset := GameplayRewardsMenuUI.reward_red_gift, Dimensions := option{vector2{X := 219.0, Y := 166.0}}}
            texture_metadata{Name := "GameplayRewardsMenuUI/unlocks_in_bg", Asset := GameplayRewardsMenuUI.unlocks_in_bg, Dimensions := option{vector2{X := 468.0, Y := 112.0}}}
            texture_metadata{Name := "GameplayRewardsMenuUI/up_next_label", Asset := GameplayRewardsMenuUI.up_next_label, Dimensions := option{vector2{X := 282.0, Y := 68.0}}}
            texture_metadata{Name := "HavocVerseUI/ui_black_contrast_rounded", Asset := HavocVerseUI.ui_black_contrast_rounded, Dimensions := option{vector2{X := 512.0, Y := 512.0}}}
            texture_metadata{Name := "HavocVerseUI/ui_debug_menu", Asset := HavocVerseUI.ui_debug_menu, Dimensions := option{vector2{X := 1280.0, Y := 720.0}}}
            texture_metadata{Name := "HavocVerseUI/ui_debug_menu_code_prompt", Asset := HavocVerseUI.ui_debug_menu_code_prompt, Dimensions := option{vector2{X := 450.0, Y := 542.0}}}
            texture_metadata{Name := "TemplateCommon/placeholder", Asset := TemplateCommon.placeholder, Dimensions := option{vector2{X := 512.0, Y := 256.0}}}
            texture_metadata{Name := "TemplateCommon/BackendTooltips/extra", Asset := TemplateCommon.BackendTooltips.extra, Dimensions := option{vector2{X := 1024.0, Y := 512.0}}}
            texture_metadata{Name := "TemplateCommon/BackendTooltips/integration_debug", Asset := TemplateCommon.BackendTooltips.integration_debug, Dimensions := option{vector2{X := 1024.0, Y := 1024.0}}}
            texture_metadata{Name := "TemplateCommon/BackendTooltips/progress_line", Asset := TemplateCommon.BackendTooltips.progress_line, Dimensions := option{vector2{X := 1024.0, Y := 512.0}}}
            texture_metadata{Name := "TemplateCommon/BackendTooltips/tycoon", Asset := TemplateCommon.BackendTooltips.tycoon, Dimensions := option{vector2{X := 1024.0, Y := 512.0}}}
            texture_metadata{Name := "UI/event_bloodmoon", Asset := UI.event_bloodmoon, Dimensions := option{vector2{X := 598.0, Y := 126.0}}}
            texture_metadata{Name := "UI/minimap", Asset := UI.minimap, Dimensions := option{vector2{X := 2018.0, Y := 2018.0}}}
            texture_metadata{Name := "UI/sidebar_armory", Asset := UI.sidebar_armory, Dimensions := option{vector2{X := 764.0, Y := 140.0}}}
            texture_metadata{Name := "UI/sidebar_food", Asset := UI.sidebar_food, Dimensions := option{vector2{X := 764.0, Y := 140.0}}}
            texture_metadata{Name := "UI/sidebar_garage", Asset := UI.sidebar_garage, Dimensions := option{vector2{X := 764.0, Y := 140.0}}}
            texture_metadata{Name := "UI/sidebar_meds", Asset := UI.sidebar_meds, Dimensions := option{vector2{X := 764.0, Y := 140.0}}}
            texture_metadata{Name := "UI/sidebar_scrap", Asset := UI.sidebar_scrap, Dimensions := option{vector2{X := 764.0, Y := 140.0}}}
            texture_metadata{Name := "UI/sidebar_walls", Asset := UI.sidebar_walls, Dimensions := option{vector2{X := 764.0, Y := 140.0}}}
            texture_metadata{Name := "UI/ui_keypad", Asset := UI.ui_keypad, Dimensions := option{vector2{X := 798.0, Y := 1566.0}}}
            texture_metadata{Name := "UI/ui_mapcode", Asset := UI.ui_mapcode, Dimensions := option{vector2{X := 536.0, Y := 118.0}}}
            texture_metadata{Name := "UI/Boosts/boost_time", Asset := UI.Boosts.boost_time, Dimensions := option{vector2{X := 824.0, Y := 108.0}}}
            texture_metadata{Name := "UI/Boosts/sidebar_10x", Asset := UI.Boosts.sidebar_10x, Dimensions := option{vector2{X := 94.0, Y := 826.0}}}
            texture_metadata{Name := "UI/Boosts/sidebar_2x", Asset := UI.Boosts.sidebar_2x, Dimensions := option{vector2{X := 94.0, Y := 826.0}}}
            texture_metadata{Name := "UI/Boosts/sidebar_3x", Asset := UI.Boosts.sidebar_3x, Dimensions := option{vector2{X := 94.0, Y := 826.0}}}
            texture_metadata{Name := "UI/Boosts/sidebar_4x", Asset := UI.Boosts.sidebar_4x, Dimensions := option{vector2{X := 94.0, Y := 826.0}}}
            texture_metadata{Name := "UI/Boosts/sidebar_5x", Asset := UI.Boosts.sidebar_5x, Dimensions := option{vector2{X := 94.0, Y := 826.0}}}
            texture_metadata{Name := "UI/Boosts/sidebar_6x", Asset := UI.Boosts.sidebar_6x, Dimensions := option{vector2{X := 94.0, Y := 826.0}}}
            texture_metadata{Name := "UI/Boosts/sidebar_8x", Asset := UI.Boosts.sidebar_8x, Dimensions := option{vector2{X := 94.0, Y := 826.0}}}
            texture_metadata{Name := "UI/Levels/L1", Asset := UI.Levels.L1, Dimensions := option{vector2{X := 512.0, Y := 128.0}}}
            texture_metadata{Name := "UI/Levels/L10", Asset := UI.Levels.L10, Dimensions := option{vector2{X := 512.0, Y := 128.0}}}
            texture_metadata{Name := "UI/Levels/L11", Asset := UI.Levels.L11, Dimensions := option{vector2{X := 512.0, Y := 128.0}}}
            texture_metadata{Name := "UI/Levels/L12", Asset := UI.Levels.L12, Dimensions := option{vector2{X := 512.0, Y := 128.0}}}
            texture_metadata{Name := "UI/Levels/L13", Asset := UI.Levels.L13, Dimensions := option{vector2{X := 512.0, Y := 128.0}}}
            texture_metadata{Name := "UI/Levels/L14", Asset := UI.Levels.L14, Dimensions := option{vector2{X := 512.0, Y := 128.0}}}
            texture_metadata{Name := "UI/Levels/L15", Asset := UI.Levels.L15, Dimensions := option{vector2{X := 512.0, Y := 128.0}}}
            texture_metadata{Name := "UI/Levels/L16", Asset := UI.Levels.L16, Dimensions := option{vector2{X := 512.0, Y := 128.0}}}
            texture_metadata{Name := "UI/Levels/L17", Asset := UI.Levels.L17, Dimensions := option{vector2{X := 512.0, Y := 128.0}}}
            texture_metadata{Name := "UI/Levels/L18", Asset := UI.Levels.L18, Dimensions := option{vector2{X := 512.0, Y := 128.0}}}
            texture_metadata{Name := "UI/Levels/L19", Asset := UI.Levels.L19, Dimensions := option{vector2{X := 512.0, Y := 128.0}}}
            texture_metadata{Name := "UI/Levels/L2", Asset := UI.Levels.L2, Dimensions := option{vector2{X := 512.0, Y := 128.0}}}
            texture_metadata{Name := "UI/Levels/L20", Asset := UI.Levels.L20, Dimensions := option{vector2{X := 512.0, Y := 128.0}}}
            texture_metadata{Name := "UI/Levels/L20Plus", Asset := UI.Levels.L20Plus, Dimensions := option{vector2{X := 512.0, Y := 128.0}}}
            texture_metadata{Name := "UI/Levels/L3", Asset := UI.Levels.L3, Dimensions := option{vector2{X := 512.0, Y := 128.0}}}
            texture_metadata{Name := "UI/Levels/L4", Asset := UI.Levels.L4, Dimensions := option{vector2{X := 512.0, Y := 128.0}}}
            texture_metadata{Name := "UI/Levels/L5", Asset := UI.Levels.L5, Dimensions := option{vector2{X := 512.0, Y := 128.0}}}
            texture_metadata{Name := "UI/Levels/L6", Asset := UI.Levels.L6, Dimensions := option{vector2{X := 512.0, Y := 128.0}}}
            texture_metadata{Name := "UI/Levels/L7", Asset := UI.Levels.L7, Dimensions := option{vector2{X := 512.0, Y := 128.0}}}
            texture_metadata{Name := "UI/Levels/L8", Asset := UI.Levels.L8, Dimensions := option{vector2{X := 512.0, Y := 128.0}}}
            texture_metadata{Name := "UI/Levels/L9", Asset := UI.Levels.L9, Dimensions := option{vector2{X := 512.0, Y := 128.0}}}
            texture_metadata{Name := "UI/Levels/progress_bar_background", Asset := UI.Levels.progress_bar_background, Dimensions := option{vector2{X := 1310.0, Y := 101.0}}}
            texture_metadata{Name := "UI/Levels/progress_bar_fill", Asset := UI.Levels.progress_bar_fill, Dimensions := option{vector2{X := 77.0, Y := 77.0}}}
            texture_metadata{Name := "UI/Levels/progress_bar_fill_end", Asset := UI.Levels.progress_bar_fill_end, Dimensions := option{vector2{X := 21.0, Y := 77.0}}}
            texture_metadata{Name := "UI/Levels/progress_bar_fill_start", Asset := UI.Levels.progress_bar_fill_start, Dimensions := option{vector2{X := 21.0, Y := 77.0}}}
            texture_metadata{Name := "UI/Rebirths/rebirth_01", Asset := UI.Rebirths.rebirth_01, Dimensions := option{vector2{X := 548.0, Y := 128.0}}}
            texture_metadata{Name := "UI/Rebirths/rebirth_02", Asset := UI.Rebirths.rebirth_02, Dimensions := option{vector2{X := 548.0, Y := 128.0}}}
            texture_metadata{Name := "UI/Rebirths/rebirth_03", Asset := UI.Rebirths.rebirth_03, Dimensions := option{vector2{X := 548.0, Y := 128.0}}}
            texture_metadata{Name := "UI/Rebirths/rebirth_04", Asset := UI.Rebirths.rebirth_04, Dimensions := option{vector2{X := 548.0, Y := 128.0}}}
            texture_metadata{Name := "UI/Rebirths/rebirth_05", Asset := UI.Rebirths.rebirth_05, Dimensions := option{vector2{X := 548.0, Y := 128.0}}}
            texture_metadata{Name := "UI/Rebirths/rebirth_06", Asset := UI.Rebirths.rebirth_06, Dimensions := option{vector2{X := 548.0, Y := 128.0}}}
            texture_metadata{Name := "UI/Rebirths/rebirth_07", Asset := UI.Rebirths.rebirth_07, Dimensions := option{vector2{X := 548.0, Y := 128.0}}}
            texture_metadata{Name := "UI/Rebirths/rebirth_08", Asset := UI.Rebirths.rebirth_08, Dimensions := option{vector2{X := 548.0, Y := 128.0}}}
            texture_metadata{Name := "UI/Rebirths/rebirth_09", Asset := UI.Rebirths.rebirth_09, Dimensions := option{vector2{X := 548.0, Y := 128.0}}}
            texture_metadata{Name := "UI/Rebirths/rebirth_10", Asset := UI.Rebirths.rebirth_10, Dimensions := option{vector2{X := 548.0, Y := 128.0}}}
            texture_metadata{Name := "UI/Rebirths/rebirth_100", Asset := UI.Rebirths.rebirth_100, Dimensions := option{vector2{X := 548.0, Y := 128.0}}}
            texture_metadata{Name := "UI/Rebirths/rebirth_11", Asset := UI.Rebirths.rebirth_11, Dimensions := option{vector2{X := 548.0, Y := 128.0}}}
            texture_metadata{Name := "UI/Rebirths/rebirth_12", Asset := UI.Rebirths.rebirth_12, Dimensions := option{vector2{X := 548.0, Y := 128.0}}}
            texture_metadata{Name := "UI/Rebirths/rebirth_13", Asset := UI.Rebirths.rebirth_13, Dimensions := option{vector2{X := 548.0, Y := 128.0}}}
            texture_metadata{Name := "UI/Rebirths/rebirth_14", Asset := UI.Rebirths.rebirth_14, Dimensions := option{vector2{X := 548.0, Y := 128.0}}}
            texture_metadata{Name := "UI/Rebirths/rebirth_15", Asset := UI.Rebirths.rebirth_15, Dimensions := option{vector2{X := 548.0, Y := 128.0}}}
            texture_metadata{Name := "UI/Rebirths/rebirth_16", Asset := UI.Rebirths.rebirth_16, Dimensions := option{vector2{X := 548.0, Y := 128.0}}}
            texture_metadata{Name := "UI/Rebirths/rebirth_17", Asset := UI.Rebirths.rebirth_17, Dimensions := option{vector2{X := 548.0, Y := 128.0}}}
            texture_metadata{Name := "UI/Rebirths/rebirth_18", Asset := UI.Rebirths.rebirth_18, Dimensions := option{vector2{X := 548.0, Y := 128.0}}}
            texture_metadata{Name := "UI/Rebirths/rebirth_19", Asset := UI.Rebirths.rebirth_19, Dimensions := option{vector2{X := 548.0, Y := 128.0}}}
            texture_metadata{Name := "UI/Rebirths/rebirth_20", Asset := UI.Rebirths.rebirth_20, Dimensions := option{vector2{X := 548.0, Y := 128.0}}}
            texture_metadata{Name := "UI/Rebirths/rebirth_200", Asset := UI.Rebirths.rebirth_200, Dimensions := option{vector2{X := 548.0, Y := 128.0}}}
            texture_metadata{Name := "UI/Rebirths/rebirth_21", Asset := UI.Rebirths.rebirth_21, Dimensions := option{vector2{X := 548.0, Y := 128.0}}}
            texture_metadata{Name := "UI/Rebirths/rebirth_22", Asset := UI.Rebirths.rebirth_22, Dimensions := option{vector2{X := 548.0, Y := 128.0}}}
            texture_metadata{Name := "UI/Rebirths/rebirth_23", Asset := UI.Rebirths.rebirth_23, Dimensions := option{vector2{X := 548.0, Y := 128.0}}}
            texture_metadata{Name := "UI/Rebirths/rebirth_24", Asset := UI.Rebirths.rebirth_24, Dimensions := option{vector2{X := 548.0, Y := 128.0}}}
            texture_metadata{Name := "UI/Rebirths/rebirth_25", Asset := UI.Rebirths.rebirth_25, Dimensions := option{vector2{X := 548.0, Y := 128.0}}}
            texture_metadata{Name := "UI/Rebirths/rebirth_26", Asset := UI.Rebirths.rebirth_26, Dimensions := option{vector2{X := 548.0, Y := 128.0}}}
            texture_metadata{Name := "UI/Rebirths/rebirth_300", Asset := UI.Rebirths.rebirth_300, Dimensions := option{vector2{X := 548.0, Y := 128.0}}}
            texture_metadata{Name := "UI/Rebirths/rebirth_400", Asset := UI.Rebirths.rebirth_400, Dimensions := option{vector2{X := 548.0, Y := 128.0}}}
            texture_metadata{Name := "UI/Rebirths/rebirth_50", Asset := UI.Rebirths.rebirth_50, Dimensions := option{vector2{X := 548.0, Y := 128.0}}}
            texture_metadata{Name := "UI/Rebirths/rebirth_500", Asset := UI.Rebirths.rebirth_500, Dimensions := option{vector2{X := 548.0, Y := 128.0}}}
            texture_metadata{Name := "UI/ResourceBillboards/armory", Asset := UI.ResourceBillboards.armory, Dimensions := option{vector2{X := 512.0, Y := 512.0}}}
            texture_metadata{Name := "UI/ResourceBillboards/food", Asset := UI.ResourceBillboards.food, Dimensions := option{vector2{X := 512.0, Y := 512.0}}}
            texture_metadata{Name := "UI/ResourceBillboards/garage", Asset := UI.ResourceBillboards.garage, Dimensions := option{vector2{X := 512.0, Y := 512.0}}}
            texture_metadata{Name := "UI/ResourceBillboards/meds", Asset := UI.ResourceBillboards.meds, Dimensions := option{vector2{X := 512.0, Y := 512.0}}}
            texture_metadata{Name := "UI/ResourceBillboards/Supply_Cache", Asset := UI.ResourceBillboards.Supply_Cache, Dimensions := option{vector2{X := 1024.0, Y := 1024.0}}}
            texture_metadata{Name := "UI/ResourceBillboards/walls", Asset := UI.ResourceBillboards.walls, Dimensions := option{vector2{X := 512.0, Y := 512.0}}}
            texture_metadata{Name := "UI/Resources/armory", Asset := UI.Resources.armory, Dimensions := option{vector2{X := 128.0, Y := 128.0}}}
            texture_metadata{Name := "UI/Resources/armory_marker", Asset := UI.Resources.armory_marker, Dimensions := option{vector2{X := 128.0, Y := 128.0}}}
            texture_metadata{Name := "UI/Resources/food", Asset := UI.Resources.food, Dimensions := option{vector2{X := 128.0, Y := 128.0}}}
            texture_metadata{Name := "UI/Resources/food_marker", Asset := UI.Resources.food_marker, Dimensions := option{vector2{X := 128.0, Y := 128.0}}}
            texture_metadata{Name := "UI/Resources/garage", Asset := UI.Resources.garage, Dimensions := option{vector2{X := 128.0, Y := 128.0}}}
            texture_metadata{Name := "UI/Resources/garage_marker", Asset := UI.Resources.garage_marker, Dimensions := option{vector2{X := 128.0, Y := 128.0}}}
            texture_metadata{Name := "UI/Resources/meds", Asset := UI.Resources.meds, Dimensions := option{vector2{X := 128.0, Y := 128.0}}}
            texture_metadata{Name := "UI/Resources/meds_marker", Asset := UI.Resources.meds_marker, Dimensions := option{vector2{X := 128.0, Y := 128.0}}}
            texture_metadata{Name := "UI/Resources/scrap", Asset := UI.Resources.scrap, Dimensions := option{vector2{X := 128.0, Y := 128.0}}}
            texture_metadata{Name := "UI/Resources/scrap_marker", Asset := UI.Resources.scrap_marker, Dimensions := option{vector2{X := 128.0, Y := 128.0}}}
            texture_metadata{Name := "UI/Resources/walls", Asset := UI.Resources.walls, Dimensions := option{vector2{X := 128.0, Y := 128.0}}}
            texture_metadata{Name := "UI/Resources/walls_marker", Asset := UI.Resources.walls_marker, Dimensions := option{vector2{X := 128.0, Y := 128.0}}}
            texture_metadata{Name := "UIAssets/Apoc/heli_icon", Asset := UIAssets.Apoc.heli_icon, Dimensions := option{vector2{X := 512.0, Y := 202.0}}}
            texture_metadata{Name := "UIAssets/Apoc/ui_notif_red", Asset := UIAssets.Apoc.ui_notif_red, Dimensions := option{vector2{X := 784.0, Y := 252.0}}}
            texture_metadata{Name := "UIAssets/Apoc/ui_notif_white", Asset := UIAssets.Apoc.ui_notif_white, Dimensions := option{vector2{X := 784.0, Y := 252.0}}}
            texture_metadata{Name := "UIAssets/Apoc/x_circle", Asset := UIAssets.Apoc.x_circle, Dimensions := option{vector2{X := 128.0, Y := 128.0}}}
            texture_metadata{Name := "UIAssets/Events/event_megabrute", Asset := UIAssets.Events.event_megabrute, Dimensions := option{vector2{X := 784.0, Y := 252.0}}}
            texture_metadata{Name := "UIAssets/Events/event_supply_cache", Asset := UIAssets.Events.event_supply_cache, Dimensions := option{vector2{X := 784.0, Y := 252.0}}}
            texture_metadata{Name := "UIAssets/Events/event_supply_drop", Asset := UIAssets.Events.event_supply_drop, Dimensions := option{vector2{X := 784.0, Y := 252.0}}}
            texture_metadata{Name := "UIAssets/Notifs/icecream", Asset := UIAssets.Notifs.icecream, Dimensions := option{vector2{X := 350.0, Y := 291.0}}}
            texture_metadata{Name := "UIAssets/Notifs/muscle", Asset := UIAssets.Notifs.muscle, Dimensions := option{vector2{X := 357.0, Y := 251.0}}}
            texture_metadata{Name := "UIAssets/Notifs/muscle_galaxy", Asset := UIAssets.Notifs.muscle_galaxy, Dimensions := option{vector2{X := 356.0, Y := 249.0}}}
            texture_metadata{Name := "UIAssets/Notifs/muscle_gold", Asset := UIAssets.Notifs.muscle_gold, Dimensions := option{vector2{X := 355.0, Y := 249.0}}}
            texture_metadata{Name := "UIAssets/Notifs/muscle_rainbow", Asset := UIAssets.Notifs.muscle_rainbow, Dimensions := option{vector2{X := 358.0, Y := 249.0}}}
            texture_metadata{Name := "UIAssets/Notifs/muscle_rainbow2", Asset := UIAssets.Notifs.muscle_rainbow2, Dimensions := option{vector2{X := 357.0, Y := 249.0}}}
            texture_metadata{Name := "UIAssets/Notifs/new_car_unlocked", Asset := UIAssets.Notifs.new_car_unlocked, Dimensions := option{vector2{X := 764.0, Y := 147.0}}}
            texture_metadata{Name := "UIAssets/Notifs/new_pet_unlocked", Asset := UIAssets.Notifs.new_pet_unlocked, Dimensions := option{vector2{X := 764.0, Y := 147.0}}}
            texture_metadata{Name := "UIAssets/Notifs/news", Asset := UIAssets.Notifs.news, Dimensions := option{vector2{X := 342.0, Y := 287.0}}}
            texture_metadata{Name := "UIAssets/Notifs/pet_gameplay01", Asset := UIAssets.Notifs.pet_gameplay01, Dimensions := option{vector2{X := 280.0, Y := 379.0}}}
            texture_metadata{Name := "UIAssets/Notifs/pet_gameplay02", Asset := UIAssets.Notifs.pet_gameplay02, Dimensions := option{vector2{X := 344.0, Y := 478.0}}}
            texture_metadata{Name := "UIAssets/Notifs/pet_level03", Asset := UIAssets.Notifs.pet_level03, Dimensions := option{vector2{X := 229.0, Y := 402.0}}}
            texture_metadata{Name := "UIAssets/Notifs/pet_level17", Asset := UIAssets.Notifs.pet_level17, Dimensions := option{vector2{X := 340.0, Y := 478.0}}}
            texture_metadata{Name := "UIAssets/Notifs/pet_loginstreak01", Asset := UIAssets.Notifs.pet_loginstreak01, Dimensions := option{vector2{X := 273.0, Y := 402.0}}}
            texture_metadata{Name := "UIAssets/Notifs/pet_loginstreak02", Asset := UIAssets.Notifs.pet_loginstreak02, Dimensions := option{vector2{X := 292.0, Y := 385.0}}}
            texture_metadata{Name := "UIAssets/Notifs/pet_rebirth01", Asset := UIAssets.Notifs.pet_rebirth01, Dimensions := option{vector2{X := 234.0, Y := 411.0}}}
            texture_metadata{Name := "UIAssets/Notifs/pet_rebirth03", Asset := UIAssets.Notifs.pet_rebirth03, Dimensions := option{vector2{X := 305.0, Y := 400.0}}}
            texture_metadata{Name := "UIAssets/Notifs/pet_rebirth05", Asset := UIAssets.Notifs.pet_rebirth05, Dimensions := option{vector2{X := 273.0, Y := 398.0}}}
            texture_metadata{Name := "UIAssets/Notifs/pet_rebirth10", Asset := UIAssets.Notifs.pet_rebirth10, Dimensions := option{vector2{X := 287.0, Y := 380.0}}}
            texture_metadata{Name := "UIAssets/Notifs/pet_rebirth100", Asset := UIAssets.Notifs.pet_rebirth100, Dimensions := option{vector2{X := 421.0, Y := 488.0}}}
            texture_metadata{Name := "UIAssets/Notifs/pet_rebirth15", Asset := UIAssets.Notifs.pet_rebirth15, Dimensions := option{vector2{X := 415.0, Y := 434.0}}}
            texture_metadata{Name := "UIAssets/Notifs/pet_rebirth25", Asset := UIAssets.Notifs.pet_rebirth25, Dimensions := option{vector2{X := 326.0, Y := 459.0}}}
            texture_metadata{Name := "UIAssets/Notifs/pet_rebirth50", Asset := UIAssets.Notifs.pet_rebirth50, Dimensions := option{vector2{X := 305.0, Y := 460.0}}}
            texture_metadata{Name := "UIAssets/Widgets/bonus_bg", Asset := UIAssets.Widgets.bonus_bg, Dimensions := option{vector2{X := 784.0, Y := 252.0}}}
            texture_metadata{Name := "UIAssets/Widgets/bonus_frame", Asset := UIAssets.Widgets.bonus_frame, Dimensions := option{vector2{X := 784.0, Y := 252.0}}}
            texture_metadata{Name := "UIAssets/Widgets/gift", Asset := UIAssets.Widgets.gift, Dimensions := option{vector2{X := 128.0, Y := 128.0}}}
            texture_metadata{Name := "UIAssets/Widgets/rainbow_gradient", Asset := UIAssets.Widgets.rainbow_gradient, Dimensions := option{vector2{X := 512.0, Y := 512.0}}}
            texture_metadata{Name := "UIAssets/Widgets/rainbow_spiral", Asset := UIAssets.Widgets.rainbow_spiral, Dimensions := option{vector2{X := 1024.0, Y := 1024.0}}}
            texture_metadata{Name := "UIAssets/Widgets/T_gradient", Asset := UIAssets.Widgets.T_gradient, Dimensions := option{vector2{X := 256.0, Y := 256.0}}}
            texture_metadata{Name := "UIAssets/Widgets/testtrace", Asset := UIAssets.Widgets.testtrace, Dimensions := option{vector2{X := 512.0, Y := 512.0}}}
            texture_metadata{Name := "UIAssets/Widgets/testtraceopacity", Asset := UIAssets.Widgets.testtraceopacity, Dimensions := option{vector2{X := 512.0, Y := 512.0}}}
        }

        var MappedTextureAssets: [string]texture_metadata = map{}
        for (
            TextureMeta:TextureAssets
            set MappedTextureAssets[TextureMeta.Name] = TextureMeta
        ) {}

        local_assets_provider{
            TextureAssets := MappedTextureAssets
        }

local_assets_provider := class<final>(assets_provider):
    TextureAssets: [string]texture_metadata

    GetTextureMetaAssets<override>()<transacts>:[]texture_metadata=
        TextureAssets.Values()

    GetTextureAsset<override>(Name:string)<transacts><decides>:texture=
        TextureAssets[Name].Asset

    GetTextureAssetDimensions<override>(Name:string)<transacts><decides>:vector2=
        TextureAssets[Name].Dimensions?

    GetMaterialAsset<override>(Name:string)<transacts><decides>:material=
        case(Name):
            "Assets/VFX/MI_level_up" => Assets.VFX.MI_level_up
            "Assets/VFX/MI_plus_armory" => Assets.VFX.MI_plus_armory
            "Assets/VFX/Import/M_AuraCircle" => Assets.VFX.Import.M_AuraCircle
            "Assets/VFX/Import/M_Glow_Slow" => Assets.VFX.Import.M_Glow_Slow
            "Assets/VFX/Import/MI_Glow_Slow" => Assets.VFX.Import.MI_Glow_Slow
            "Assets/VFX/Import/MI_ToolsAuraCircle" => Assets.VFX.Import.MI_ToolsAuraCircle
            "BuyplateBillboards/MI_bp_001" => BuyplateBillboards.MI_bp_001
            "BuyplateBillboards/MI_bp_002" => BuyplateBillboards.MI_bp_002
            "BuyplateBillboards/MI_bp_003" => BuyplateBillboards.MI_bp_003
            "BuyplateBillboards/MI_bp_004" => BuyplateBillboards.MI_bp_004
            "BuyplateBillboards/MI_bp_005" => BuyplateBillboards.MI_bp_005
            "BuyplateBillboards/MI_bp_006" => BuyplateBillboards.MI_bp_006
            "BuyplateBillboards/MI_bp_007" => BuyplateBillboards.MI_bp_007
            "BuyplateBillboards/MI_bp_008" => BuyplateBillboards.MI_bp_008
            "BuyplateBillboards/MI_bp_009" => BuyplateBillboards.MI_bp_009
            "BuyplateBillboards/MI_bp_010" => BuyplateBillboards.MI_bp_010
            "BuyplateBillboards/MI_bp_011" => BuyplateBillboards.MI_bp_011
            "BuyplateBillboards/MI_bp_012" => BuyplateBillboards.MI_bp_012
            "BuyplateBillboards/MI_bp_013" => BuyplateBillboards.MI_bp_013
            "BuyplateBillboards/MI_bp_014" => BuyplateBillboards.MI_bp_014
            "BuyplateBillboards/MI_bp_015" => BuyplateBillboards.MI_bp_015
            "BuyplateBillboards/MI_bp_016" => BuyplateBillboards.MI_bp_016
            "BuyplateBillboards/MI_bp_017" => BuyplateBillboards.MI_bp_017
            "BuyplateBillboards/MI_bp_018" => BuyplateBillboards.MI_bp_018
            "BuyplateBillboards/MI_bp_019" => BuyplateBillboards.MI_bp_019
            "BuyplateBillboards/MI_bp_020" => BuyplateBillboards.MI_bp_020
            "BuyplateBillboards/MI_bp_021" => BuyplateBillboards.MI_bp_021
            "BuyplateBillboards/MI_bp_022" => BuyplateBillboards.MI_bp_022
            "BuyplateBillboards/MI_bp_023" => BuyplateBillboards.MI_bp_023
            "BuyplateBillboards/MI_bp_024" => BuyplateBillboards.MI_bp_024
            "BuyplateBillboards/MI_bp_025" => BuyplateBillboards.MI_bp_025
            "BuyplateBillboards/MI_bp_026" => BuyplateBillboards.MI_bp_026
            "BuyplateBillboards/MI_bp_027" => BuyplateBillboards.MI_bp_027
            "BuyplateBillboards/MI_bp_028" => BuyplateBillboards.MI_bp_028
            "BuyplateBillboards/MI_bp_029" => BuyplateBillboards.MI_bp_029
            "BuyplateBillboards/MI_bp_030" => BuyplateBillboards.MI_bp_030
            "BuyplateBillboards/MI_bp_031" => BuyplateBillboards.MI_bp_031
            "BuyplateBillboards/MI_bp_032" => BuyplateBillboards.MI_bp_032
            "BuyplateBillboards/MI_bp_033" => BuyplateBillboards.MI_bp_033
            "BuyplateBillboards/MI_bp_034" => BuyplateBillboards.MI_bp_034
            "BuyplateBillboards/MI_bp_035" => BuyplateBillboards.MI_bp_035
            "BuyplateBillboards/MI_bp_036" => BuyplateBillboards.MI_bp_036
            "BuyplateBillboards/MI_bp_037" => BuyplateBillboards.MI_bp_037
            "BuyplateBillboards/MI_bp_038" => BuyplateBillboards.MI_bp_038
            "BuyplateBillboards/MI_bp_039" => BuyplateBillboards.MI_bp_039
            "BuyplateBillboards/MI_bp_050" => BuyplateBillboards.MI_bp_050
            "BuyplateBillboards/MI_bp_051" => BuyplateBillboards.MI_bp_051
            "BuyplateBillboards/MI_bp_052" => BuyplateBillboards.MI_bp_052
            "BuyplateBillboards/MI_bp_100" => BuyplateBillboards.MI_bp_100
            "BuyplateBillboards/MI_bp_101" => BuyplateBillboards.MI_bp_101
            "BuyplateBillboards/MI_bp_102" => BuyplateBillboards.MI_bp_102
            "BuyplateBillboards/MI_bp_103" => BuyplateBillboards.MI_bp_103
            "BuyplateBillboards/MI_bp_104" => BuyplateBillboards.MI_bp_104
            "BuyplateBillboards/MI_bp_200" => BuyplateBillboards.MI_bp_200
            "BuyplateBillboards/MI_bp_201" => BuyplateBillboards.MI_bp_201
            "BuyplateBillboards/MI_bp_202" => BuyplateBillboards.MI_bp_202
            "BuyplateBillboards/MI_bp_203" => BuyplateBillboards.MI_bp_203
            "BuyplateBillboards/MI_bp_204" => BuyplateBillboards.MI_bp_204
            "BuyplateBillboards/MI_bp_205" => BuyplateBillboards.MI_bp_205
            "BuyplateBillboards/MI_bp_206" => BuyplateBillboards.MI_bp_206
            "BuyplateBillboards/MI_bp_207" => BuyplateBillboards.MI_bp_207
            "BuyplateBillboards/MI_bp_208" => BuyplateBillboards.MI_bp_208
            "BuyplateBillboards/MI_bp_209" => BuyplateBillboards.MI_bp_209
            "BuyplateBillboards/MI_bp_210" => BuyplateBillboards.MI_bp_210
            "BuyplateBillboards/MI_bp_211" => BuyplateBillboards.MI_bp_211
            "BuyplateBillboards/MI_bp_212" => BuyplateBillboards.MI_bp_212
            "BuyplateBillboards/MI_bp_213" => BuyplateBillboards.MI_bp_213
            "BuyplateBillboards/MI_bp_214" => BuyplateBillboards.MI_bp_214
            "BuyplateBillboards/MI_bp_215" => BuyplateBillboards.MI_bp_215
            "BuyplateBillboards/MI_bp_216" => BuyplateBillboards.MI_bp_216
            "BuyplateBillboards/MI_bp_217" => BuyplateBillboards.MI_bp_217
            "BuyplateBillboards/MI_bp_218" => BuyplateBillboards.MI_bp_218
            "BuyplateBillboards/MI_bp_219" => BuyplateBillboards.MI_bp_219
            "BuyplateBillboards/MI_bp_220" => BuyplateBillboards.MI_bp_220
            "BuyplateBillboards/MI_bp_250" => BuyplateBillboards.MI_bp_250
            "BuyplateBillboards/MI_bp_251" => BuyplateBillboards.MI_bp_251
            "BuyplateBillboards/MI_bp_252" => BuyplateBillboards.MI_bp_252
            "BuyplateBillboards/MI_bp_253" => BuyplateBillboards.MI_bp_253
            "BuyplateBillboards/MI_bp_300" => BuyplateBillboards.MI_bp_300
            "BuyplateBillboards/MI_bp_301" => BuyplateBillboards.MI_bp_301
            "BuyplateBillboards/MI_bp_302" => BuyplateBillboards.MI_bp_302
            "BuyplateBillboards/MI_bp_303" => BuyplateBillboards.MI_bp_303
            "BuyplateBillboards/MI_bp_304" => BuyplateBillboards.MI_bp_304
            "BuyplateBillboards/MI_bp_305" => BuyplateBillboards.MI_bp_305
            "BuyplateBillboards/MI_bp_306" => BuyplateBillboards.MI_bp_306
            "BuyplateBillboards/MI_bp_307" => BuyplateBillboards.MI_bp_307
            "BuyplateBillboards/MI_bp_308" => BuyplateBillboards.MI_bp_308
            "BuyplateBillboards/MI_bp_309" => BuyplateBillboards.MI_bp_309
            "BuyplateBillboards/MI_bp_310" => BuyplateBillboards.MI_bp_310
            "BuyplateBillboards/MI_bp_311" => BuyplateBillboards.MI_bp_311
            "BuyplateBillboards/MI_bp_312" => BuyplateBillboards.MI_bp_312
            "BuyplateBillboards/MI_bp_313" => BuyplateBillboards.MI_bp_313
            "BuyplateBillboards/MI_bp_314" => BuyplateBillboards.MI_bp_314
            "BuyplateBillboards/MI_bp_315" => BuyplateBillboards.MI_bp_315
            "BuyplateBillboards/MI_bp_316" => BuyplateBillboards.MI_bp_316
            "BuyplateBillboards/MI_bp_317" => BuyplateBillboards.MI_bp_317
            "BuyplateBillboards/MI_bp_318" => BuyplateBillboards.MI_bp_318
            "BuyplateBillboards/MI_bp_319" => BuyplateBillboards.MI_bp_319
            "BuyplateBillboards/MI_bp_320" => BuyplateBillboards.MI_bp_320
            "BuyplateBillboards/MI_bp_321" => BuyplateBillboards.MI_bp_321
            "BuyplateBillboards/MI_bp_322" => BuyplateBillboards.MI_bp_322
            "BuyplateBillboards/MI_bp_323" => BuyplateBillboards.MI_bp_323
            "BuyplateBillboards/MI_bp_324" => BuyplateBillboards.MI_bp_324
            "BuyplateBillboards/MI_bp_325" => BuyplateBillboards.MI_bp_325
            "BuyplateBillboards/MI_bp_400" => BuyplateBillboards.MI_bp_400
            "BuyplateBillboards/MI_bp_401" => BuyplateBillboards.MI_bp_401
            "BuyplateBillboards/MI_bp_402" => BuyplateBillboards.MI_bp_402
            "BuyplateBillboards/MI_bp_403" => BuyplateBillboards.MI_bp_403
            "BuyplateBillboards/MI_bp_404" => BuyplateBillboards.MI_bp_404
            "BuyplateBillboards/MI_bp_405" => BuyplateBillboards.MI_bp_405
            "BuyplateBillboards/MI_bp_406" => BuyplateBillboards.MI_bp_406
            "BuyplateBillboards/MI_bp_407" => BuyplateBillboards.MI_bp_407
            "BuyplateBillboards/MI_bp_408" => BuyplateBillboards.MI_bp_408
            "BuyplateBillboards/MI_bp_409" => BuyplateBillboards.MI_bp_409
            "BuyplateBillboards/MI_bp_410" => BuyplateBillboards.MI_bp_410
            "BuyplateBillboards/MI_bp_411" => BuyplateBillboards.MI_bp_411
            "BuyplateBillboards/MI_bp_412" => BuyplateBillboards.MI_bp_412
            "BuyplateBillboards/MI_bp_413" => BuyplateBillboards.MI_bp_413
            "BuyplateBillboards/MI_bp_414" => BuyplateBillboards.MI_bp_414
            "BuyplateBillboards/MI_bp_415" => BuyplateBillboards.MI_bp_415
            "BuyplateBillboards/MI_bp_416" => BuyplateBillboards.MI_bp_416
            "BuyplateBillboards/MI_bp_417" => BuyplateBillboards.MI_bp_417
            "BuyplateBillboards/MI_bp_418" => BuyplateBillboards.MI_bp_418
            "BuyplateBillboards/MI_bp_419" => BuyplateBillboards.MI_bp_419
            "BuyplateBillboards/MI_bp_420" => BuyplateBillboards.MI_bp_420
            "BuyplateBillboards/MI_bp_421" => BuyplateBillboards.MI_bp_421
            "BuyplateBillboards/MI_bp_422" => BuyplateBillboards.MI_bp_422
            "BuyplateBillboards/MI_bp_500" => BuyplateBillboards.MI_bp_500
            "BuyplateBillboards/MI_bp_501" => BuyplateBillboards.MI_bp_501
            "BuyplateBillboards/MI_bp_502" => BuyplateBillboards.MI_bp_502
            "BuyplateBillboards/MI_bp_503" => BuyplateBillboards.MI_bp_503
            "BuyplateBillboards/MI_bp_504" => BuyplateBillboards.MI_bp_504
            "BuyplateBillboards/MI_bp_505" => BuyplateBillboards.MI_bp_505
            "BuyplateBillboards/MI_bp_506" => BuyplateBillboards.MI_bp_506
            "BuyplateBillboards/MI_bp_507" => BuyplateBillboards.MI_bp_507
            "BuyplateBillboards/MI_bp_508" => BuyplateBillboards.MI_bp_508
            "BuyplateBillboards/MI_bp_509" => BuyplateBillboards.MI_bp_509
            "BuyplateBillboards/MI_bp_510" => BuyplateBillboards.MI_bp_510
            "BuyplateBillboards/MI_bp_511" => BuyplateBillboards.MI_bp_511
            "BuyplateBillboards/MI_bp_512" => BuyplateBillboards.MI_bp_512
            "BuyplateBillboards/MI_bp_513" => BuyplateBillboards.MI_bp_513
            "BuyplateBillboards/MI_bp_514" => BuyplateBillboards.MI_bp_514
            "BuyplateBillboards/MI_bp_515" => BuyplateBillboards.MI_bp_515
            "BuyplateBillboards/MI_bp_516" => BuyplateBillboards.MI_bp_516
            "BuyplateBillboards/MI_bp_517" => BuyplateBillboards.MI_bp_517
            "BuyplateBillboards/MI_bp_518" => BuyplateBillboards.MI_bp_518
            "BuyplateBillboards/MI_bp_519" => BuyplateBillboards.MI_bp_519
            "BuyplateBillboards/MI_bp_520" => BuyplateBillboards.MI_bp_520
            "BuyplateBillboards/MI_bp_521" => BuyplateBillboards.MI_bp_521
            "BuyplateBillboards/MI_bp_522" => BuyplateBillboards.MI_bp_522
            "BuyplateBillboards/MI_bp_523" => BuyplateBillboards.MI_bp_523
            "BuyplateBillboards/MI_bp_524" => BuyplateBillboards.MI_bp_524
            "BuyplateBillboards/MI_bp_525" => BuyplateBillboards.MI_bp_525
            "BuyplateBillboards/MI_bp_526" => BuyplateBillboards.MI_bp_526
            "BuyplateBillboards/MI_bp_527" => BuyplateBillboards.MI_bp_527
            "BuyplateBillboards/MI_bp_528" => BuyplateBillboards.MI_bp_528
            "BuyplateBillboards/MI_bp_529" => BuyplateBillboards.MI_bp_529
            "BuyplateBillboards/MI_bp_530" => BuyplateBillboards.MI_bp_530
            "BuyplateBillboards/MI_bp_531" => BuyplateBillboards.MI_bp_531
            "BuyplateBillboards/MI_bp_600" => BuyplateBillboards.MI_bp_600
            "BuyplateBillboards/MI_bp_601" => BuyplateBillboards.MI_bp_601
            "BuyplateBillboards/MI_bp_602" => BuyplateBillboards.MI_bp_602
            "BuyplateBillboards/MI_bp_603" => BuyplateBillboards.MI_bp_603
            "BuyplateBillboards/MI_bp_604" => BuyplateBillboards.MI_bp_604
            "BuyplateBillboards/MI_bp_605" => BuyplateBillboards.MI_bp_605
            "BuyplateBillboards/MI_bp_606" => BuyplateBillboards.MI_bp_606
            "BuyplateBillboards/MI_bp_607" => BuyplateBillboards.MI_bp_607
            "BuyplateBillboards/MI_bp_608" => BuyplateBillboards.MI_bp_608
            "BuyplateBillboards/MI_bp_609" => BuyplateBillboards.MI_bp_609
            "BuyplateBillboards/MI_bp_610" => BuyplateBillboards.MI_bp_610
            "BuyplateBillboards/MI_bp_611" => BuyplateBillboards.MI_bp_611
            "BuyplateBillboards/MI_bp_612" => BuyplateBillboards.MI_bp_612
            "BuyplateBillboards/MI_bp_613" => BuyplateBillboards.MI_bp_613
            "BuyplateBillboards/MI_bp_614" => BuyplateBillboards.MI_bp_614
            "BuyplateBillboards/MI_bp_615" => BuyplateBillboards.MI_bp_615
            "BuyplateBillboards/MI_bp_616" => BuyplateBillboards.MI_bp_616
            "BuyplateBillboards/MI_bp_617" => BuyplateBillboards.MI_bp_617
            "BuyplateBillboards/MI_bp_618" => BuyplateBillboards.MI_bp_618
            "BuyplateBillboards/MI_bp_619" => BuyplateBillboards.MI_bp_619
            "BuyplateBillboards/MI_bp_620" => BuyplateBillboards.MI_bp_620
            "BuyplateBillboards/MI_bp_621" => BuyplateBillboards.MI_bp_621
            "BuyplateBillboards/MI_bp_622" => BuyplateBillboards.MI_bp_622
            "BuyplateBillboards/MI_bp_623" => BuyplateBillboards.MI_bp_623
            "BuyplateBillboards/MI_bp_624" => BuyplateBillboards.MI_bp_624
            "BuyplateBillboards/MI_bp_625" => BuyplateBillboards.MI_bp_625
            "BuyplateBillboards/MI_bp_626" => BuyplateBillboards.MI_bp_626
            "BuyplateBillboards/MI_bp_627" => BuyplateBillboards.MI_bp_627
            "BuyplateBillboards/MI_bp_628" => BuyplateBillboards.MI_bp_628
            "BuyplateBillboards/MI_bp_629" => BuyplateBillboards.MI_bp_629
            "BuyplateBillboards/MI_bp_630" => BuyplateBillboards.MI_bp_630
            "BuyplateBillboards/MI_bp_631" => BuyplateBillboards.MI_bp_631
            "BuyplateBillboards/MI_bp_700" => BuyplateBillboards.MI_bp_700
            "BuyplateBillboards/MI_bp_701" => BuyplateBillboards.MI_bp_701
            "BuyplateBillboards/MI_bp_702" => BuyplateBillboards.MI_bp_702
            "BuyplateBillboards/MI_bp_703" => BuyplateBillboards.MI_bp_703
            "BuyplateBillboards/MI_bp_704" => BuyplateBillboards.MI_bp_704
            "BuyplateBillboards/MI_bp_705" => BuyplateBillboards.MI_bp_705
            "BuyplateBillboards/MI_bp_claim" => BuyplateBillboards.MI_bp_claim
            "TemplateCommon/BillboardMat_Master" => TemplateCommon.BillboardMat_Master
            "TemplateCommon/BuyplateDiscMat_Master" => TemplateCommon.BuyplateDiscMat_Master
            "TemplateCommon/BuyplateDiscMatTransparent_Master" => TemplateCommon.BuyplateDiscMatTransparent_Master
            "TemplateCommon/BuyplateMatCostNotMet" => TemplateCommon.BuyplateMatCostNotMet
            "TemplateCommon/BuyplateMatMain" => TemplateCommon.BuyplateMatMain
            "TemplateCommon/BuyplateMatMainSub" => TemplateCommon.BuyplateMatMainSub
            "TemplateCommon/BuyplateMatOrange" => TemplateCommon.BuyplateMatOrange
            "TemplateCommon/BackendTooltips/extra_Mat" => TemplateCommon.BackendTooltips.extra_Mat
            "TemplateCommon/BackendTooltips/integration_debug_Mat" => TemplateCommon.BackendTooltips.integration_debug_Mat
            "TemplateCommon/BackendTooltips/progress_line_Mat" => TemplateCommon.BackendTooltips.progress_line_Mat
            "TemplateCommon/BackendTooltips/tycoon_Mat" => TemplateCommon.BackendTooltips.tycoon_Mat
            "UI/ResourceBillboards/MI_BillboardArmory" => UI.ResourceBillboards.MI_BillboardArmory
            "UI/ResourceBillboards/MI_BillboardCache" => UI.ResourceBillboards.MI_BillboardCache
            "UI/ResourceBillboards/MI_BillboardFarm" => UI.ResourceBillboards.MI_BillboardFarm
            "UI/ResourceBillboards/MI_BillboardGarage" => UI.ResourceBillboards.MI_BillboardGarage
            "UI/ResourceBillboards/MI_BillboardMeds" => UI.ResourceBillboards.MI_BillboardMeds
            "UI/ResourceBillboards/MI_BillboardWalls" => UI.ResourceBillboards.MI_BillboardWalls
            "UI/Resources/MI_Billboard_armory" => UI.Resources.MI_Billboard_armory
            "UI/Resources/MI_Billboard_food" => UI.Resources.MI_Billboard_food
            "UI/Resources/MI_Billboard_garage" => UI.Resources.MI_Billboard_garage
            "UI/Resources/MI_Billboard_meds" => UI.Resources.MI_Billboard_meds
            "UI/Resources/MI_Billboard_ResourceMultiplier" => UI.Resources.MI_Billboard_ResourceMultiplier
            "UI/Resources/MI_Billboard_scrap" => UI.Resources.MI_Billboard_scrap
            "UI/Resources/MI_Billboard_walls" => UI.Resources.MI_Billboard_walls
            "UI/Resources/MI_UI_armory" => UI.Resources.MI_UI_armory
            "UI/Resources/MI_UI_food" => UI.Resources.MI_UI_food
            "UI/Resources/MI_UI_garage" => UI.Resources.MI_UI_garage
            "UI/Resources/MI_UI_meds" => UI.Resources.MI_UI_meds
            "UI/Resources/MI_UI_scrap" => UI.Resources.MI_UI_scrap
            "UI/Resources/MI_UI_walls" => UI.Resources.MI_UI_walls
            "UIAssets/M_UI" => UIAssets.M_UI
            "UIAssets/M_UI_Gradient" => UIAssets.M_UI_Gradient
            "UIAssets/M_UI_Rainbow" => UIAssets.M_UI_Rainbow
            "UIAssets/M_UI_RainbowGradient" => UIAssets.M_UI_RainbowGradient
            "UIAssets/M_UI_Shimmer" => UIAssets.M_UI_Shimmer
            "UIAssets/M_UI_Twist" => UIAssets.M_UI_Twist
            "UIAssets/MI_UI_Shimmer_Text" => UIAssets.MI_UI_Shimmer_Text
            "UIAssets/Events/MI_event_megabrute" => UIAssets.Events.MI_event_megabrute
            "UIAssets/Events/MI_event_supply_cache" => UIAssets.Events.MI_event_supply_cache
            "UIAssets/Events/MI_event_supply_drop" => UIAssets.Events.MI_event_supply_drop
            "UIAssets/Notifs/MI_car_icecream" => UIAssets.Notifs.MI_car_icecream
            "UIAssets/Notifs/MI_car_muscle" => UIAssets.Notifs.MI_car_muscle
            "UIAssets/Notifs/MI_car_muscle_galaxy" => UIAssets.Notifs.MI_car_muscle_galaxy
            "UIAssets/Notifs/MI_car_muscle_gold" => UIAssets.Notifs.MI_car_muscle_gold
            "UIAssets/Notifs/MI_car_muscle_rainbow" => UIAssets.Notifs.MI_car_muscle_rainbow
            "UIAssets/Notifs/MI_car_muscle_rainbow2" => UIAssets.Notifs.MI_car_muscle_rainbow2
            "UIAssets/Notifs/MI_car_news" => UIAssets.Notifs.MI_car_news
            "UIAssets/Notifs/MI_new_car_unlocked" => UIAssets.Notifs.MI_new_car_unlocked
            "UIAssets/Notifs/MI_new_pet_unlocked" => UIAssets.Notifs.MI_new_pet_unlocked
            "UIAssets/Notifs/MI_pet_gameplay01" => UIAssets.Notifs.MI_pet_gameplay01
            "UIAssets/Notifs/MI_pet_gameplay02" => UIAssets.Notifs.MI_pet_gameplay02
            "UIAssets/Notifs/MI_pet_level03" => UIAssets.Notifs.MI_pet_level03
            "UIAssets/Notifs/MI_pet_level17" => UIAssets.Notifs.MI_pet_level17
            "UIAssets/Notifs/MI_pet_loginstreak01" => UIAssets.Notifs.MI_pet_loginstreak01
            "UIAssets/Notifs/MI_pet_loginstreak02" => UIAssets.Notifs.MI_pet_loginstreak02
            "UIAssets/Notifs/MI_pet_rebirth01" => UIAssets.Notifs.MI_pet_rebirth01
            "UIAssets/Notifs/MI_pet_rebirth03" => UIAssets.Notifs.MI_pet_rebirth03
            "UIAssets/Notifs/MI_pet_rebirth05" => UIAssets.Notifs.MI_pet_rebirth05
            "UIAssets/Notifs/MI_pet_rebirth10" => UIAssets.Notifs.MI_pet_rebirth10
            "UIAssets/Notifs/MI_pet_rebirth100" => UIAssets.Notifs.MI_pet_rebirth100
            "UIAssets/Notifs/MI_pet_rebirth15" => UIAssets.Notifs.MI_pet_rebirth15
            "UIAssets/Notifs/MI_pet_rebirth25" => UIAssets.Notifs.MI_pet_rebirth25
            "UIAssets/Notifs/MI_pet_rebirth50" => UIAssets.Notifs.MI_pet_rebirth50
            "UIAssets/Widgets/M_UI_Rainbow_BG" => UIAssets.Widgets.M_UI_Rainbow_BG
            "UIAssets/Widgets/MI_UI_Rainbow" => UIAssets.Widgets.MI_UI_Rainbow
            "UIAssets/Widgets/UI_RainbowBox" => UIAssets.Widgets.UI_RainbowBox
            "UIAssets/Widgets/UI_RainbowBox2" => UIAssets.Widgets.UI_RainbowBox2

    GetMeshAsset<override>(Name:string)<transacts><decides>:mesh=
        case(Name):
            "TemplateCommon/BuyplateDiscMesh2" => TemplateCommon.BuyplateDiscMesh2
            "TemplateCommon/BuyplateImageBackgroundMesh" => TemplateCommon.BuyplateImageBackgroundMesh
            "TemplateCommon/SM_BuyplateDevice" => TemplateCommon.SM_BuyplateDevice
            "TemplateCommon/BackendTooltips/TooltipPlane" => TemplateCommon.BackendTooltips.TooltipPlane

    GetAnimationSequenceAsset<override>(Name:string)<transacts><decides>:animation_sequence=
       None()?

    GetParticleSystemAsset<override>(Name:string)<transacts><decides>:particle_system=
        case(Name):
            "Assets/VFX/ArmoryParticle" => Assets.VFX.ArmoryParticle
            "Assets/VFX/BuyplateUnlockEffect" => Assets.VFX.BuyplateUnlockEffect
            "Assets/VFX/LevelUp" => Assets.VFX.LevelUp
            "Assets/VFX/Sparkle" => Assets.VFX.Sparkle
            "Assets/VFX/Import/VFX_Manhole" => Assets.VFX.Import.VFX_Manhole
            "Assets/VFX/Import/VFX_OrangeMagicSmall" => Assets.VFX.Import.VFX_OrangeMagicSmall
