using { /Fortnite.com/Devices }
using { /Verse.org/Simulation/Tags }

using { HavocVerse.Util }

buyplate_disc_tag := class(tag):
buyplate_billboard_tag := class(tag):

collectible_tag := class(tag):

local_tags_provider := class(creative_device, tags_provider):
    GetTag<override>(Name:string)<transacts><decides>:tag=
        case(Name):
            "buyplate_disc" => buyplate_disc_tag{}
            "buyplate_billboard" => buyplate_billboard_tag{}
            "collectible" => collectible_tag{}
