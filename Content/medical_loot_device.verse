
using { /Fortnite.com/Devices }
using { /Verse.org/Simulation }
using { /UnrealEngine.com/Temporary/Diagnostics }

# See https://dev.epicgames.com/documentation/en-us/uefn/create-your-own-device-in-verse for how to create a verse device.

# A Verse-authored creative device that can be placed in a level
medical_loot_device := class(creative_device):

    @editable StartEnabled : logic = true
    @editable DisableTrigger: trigger_device = trigger_device{}

    @editable RemovedOnCollection : logic = true

    @editable MedsProp: creative_prop = creative_prop{}
    @editable Button: button_device = button_device{}
    @editable MapIndicator: map_indicator_device = map_indicator_device{}
    @editable BillboardProp: creative_prop = creative_prop{}

    @editable RewardTrigger: trigger_device = trigger_device{}
    @editable CollectionAudio: audio_player_device = audio_player_device{}
    @editable RespawnTime: float = 30.0

    @editable RespawnTriggers: []trigger_device = array{}
    @editable RespawnOnDelay: logic = true

    # Runs when the device is started in a running game
    OnBegin<override>()<suspends>:void=
        Button.InteractedWithEvent.Subscribe(MedsCollected)
        for(Trigger:RespawnTriggers):
            Trigger.TriggeredEvent.Subscribe(EnableEverything2)

        DisableTrigger.TriggeredEvent.Subscribe(DisableEverything2)

        if(not StartEnabled?):
            DisableEverything()

    MedsCollected(Agent:agent):void=
        CollectionAudio.Play(Agent)
        RewardTrigger.Trigger(Agent)
        if(RemovedOnCollection?):
            DisableEverything()
        if(RespawnOnDelay?):
            spawn:
                EnableOnDelay()

    EnableOnDelay()<suspends>:void=
        Sleep(RespawnTime)
        EnableEverything()

    DisableEverything():void=
        Button.Disable()
        MapIndicator.Disable()
        BillboardProp.Hide()
        MedsProp.Hide()

    DisableEverything2(Agent:?agent):void=
        DisableEverything()

    EnableEverything():void=
        Button.Enable()
        MapIndicator.Enable()
        BillboardProp.Show()
        MedsProp.Show()

    EnableEverything2(Agent:?agent):void=
        EnableEverything()





