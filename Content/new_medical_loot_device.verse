using { /Fortnite.com/Devices }
using { /Verse.org/Simulation }
using { /UnrealEngine.com/Temporary/Diagnostics }
using { /Fortnite.com/Characters }
using { /Fortnite.com/Playspaces }
using { /UnrealEngine.com/Temporary/Diagnostics }
using { /UnrealEngine.com/Temporary/SpatialMath }

using { HavocVerse.Debug }
using { HavocVerse.Util }
using { HavocVerse.Logging }

# See https://dev.epicgames.com/documentation/en-us/uefn/create-your-own-device-in-verse for how to create a verse device.

# A Verse-authored creative device that can be placed in a level
MedsHandler<internal> := class() {
    Device : new_medical_loot_device
    Parameter1 : int

    HandlerFunction(Agent : agent) : void = {
        Device.OnMedsCollected(Agent,Parameter1)
    }

    SZMedsHandlerFunction(Agent : agent) : void = {
        Device.OnSZMedsCollected(Agent,Parameter1)
    }
    
    TriggerEnableHandlerFunction(Agent : ?agent) : void = {
        if(MaybeAgent:=Agent?):
            Device.OnEnableTriggerTriggered(MaybeAgent,Parameter1)
    }
    
    TriggerDisableHandlerFunction(Agent : ?agent) : void = {
        if(MaybeAgent:=Agent?):
            Device.OnDisableTriggerTriggered(MaybeAgent,Parameter1)
    }
}

new_medical_loot_device := class(creative_device, aware_device):

    @editable p_Meds: []ind_medical_loot_device = array{}
    @editable p_SZMeds: []safezone_medical_loot_device = array{}
    @editable RewardTrigger: trigger_device = trigger_device{}
    @editable PermRewardTrigger: trigger_device = trigger_device{}

    DeviceMeta<override>()<transacts>:aware_device_meta=
        aware_device_meta{
            DeviceName := "new_medical_loot"
        }  

    OnBegin<override>()<suspends>:void=
        var Index: int = 0
        for(Meds:p_Meds):
            Meds.Button.InteractedWithEvent.Subscribe(MedsHandler{Device:=Self, Parameter1:=Index}.HandlerFunction)
            set Index = Index + 1

        var SZIndex: int = 0
        for(SZMeds:p_SZMeds):
            SZMeds.Button.InteractedWithEvent.Subscribe(MedsHandler{Device:=Self, Parameter1:=SZIndex}.SZMedsHandlerFunction)
            SZMeds.EnableTrigger.TriggeredEvent.Subscribe(MedsHandler{Device:=Self, Parameter1:=SZIndex}.TriggerEnableHandlerFunction)
            SZMeds.DisableTrigger.TriggeredEvent.Subscribe(MedsHandler{Device:=Self, Parameter1:=SZIndex}.TriggerDisableHandlerFunction)
            SZMeds.DisableEverything()
            set SZIndex = SZIndex + 1

    OnMedsCollected(Agent:agent,Parameter1:int) : void =
        if(Med:=p_Meds[Parameter1]):
            RewardTrigger.Trigger(Agent)
            Med.DisableEverything()

    OnSZMedsCollected(Agent:agent,Parameter1:int) : void =
        Print("Test 3")
        if(Med:=p_SZMeds[Parameter1]):
            Print("Test 4")
            IsPerm := Med.GetIsPerm() 
            if(IsPerm?):
                PermRewardTrigger.Trigger(Agent)
            else:
                RewardTrigger.Trigger(Agent)
                Med.DisableEverything()

    OnEnableTriggerTriggered(Agent:agent,Parameter1:int) : void =
        if(Meds:=p_SZMeds[Parameter1]):
            Meds.EnableEverything()
            
    OnDisableTriggerTriggered(Agent:agent,Parameter1:int) : void =
        if(Meds:=p_SZMeds[Parameter1]):
            Meds.DisableEverything()

ind_medical_loot_device<internal> := class<final><concrete>:
    @editable Button: button_device = button_device{}
    @editable MedsProp: creative_prop = creative_prop{}
    @editable MapIndicator: map_indicator_device = map_indicator_device{}
    @editable BillboardProp: creative_prop = creative_prop{}

    DisableEverything():void=
        Button.Disable()
        MapIndicator.Disable()
        BillboardProp.Hide()
        MedsProp.Hide()
        spawn:
            RespawnMeds()

    EnableEverything():void=
        Button.Enable()
        MapIndicator.Enable()
        BillboardProp.Show()
        MedsProp.Show()

    RespawnMeds()<suspends>:void=
        Sleep(30.0)
        EnableEverything()


safezone_medical_loot_device<internal> := class<final><concrete>:
    @editable IsPerm: logic = false
    @editable EnableTrigger: trigger_device = trigger_device{}
    @editable DisableTrigger: trigger_device = trigger_device{}

    @editable Button: button_device = button_device{}
    @editable MedsProp: creative_prop = creative_prop{}
    @editable MapIndicator: map_indicator_device = map_indicator_device{}
    @editable BillboardProp: creative_prop = creative_prop{}

    DisableEverything():void=
        Button.Disable()
        MapIndicator.Disable()
        BillboardProp.Hide()
        MedsProp.Hide()

    EnableEverything():void=
        Button.Enable()
        MapIndicator.Enable()
        BillboardProp.Show()
        MedsProp.Show()

    GetIsPerm():logic=
        return IsPerm