using { /Fortnite.com/Devices }
using { /Verse.org/Simulation }
using { /UnrealEngine.com/Temporary/Diagnostics }
using { /Fortnite.com/Characters }
using { /Fortnite.com/Playspaces }
using { /UnrealEngine.com/Temporary/Diagnostics }
using { /UnrealEngine.com/Temporary/SpatialMath }

using { HavocVerse.Debug }
using { HavocVerse.Util }
using { HavocVerse.Logging }
# See https://dev.epicgames.com/documentation/en-us/uefn/create-your-own-device-in-verse for how to create a verse device.

# A Verse-authored creative device that can be placed in a level

Handler1<internal> := class() {
    Device : new_ruined_vehicle_device
    Parameter1 : int

    HandlerFunction(Agent : agent) : void = {
        Device.OnVehicleDamaged(Agent,Parameter1)
    }

    SafezoneVehHandlerFunction(Agent : agent) : void = {
        Device.OnSafezoneVehicleDamaged(Agent,Parameter1)
    }

    TriggerEnableHandlerFunction(Agent : ?agent) : void = {
        if(MaybeAgent:=Agent?):
            Device.OnEnableTriggerTriggered(MaybeAgent,Parameter1)
    }
    
    TriggerDisableHandlerFunction(Agent : ?agent) : void = {
        if(MaybeAgent:=Agent?):
            Device.OnDisableTriggerTriggered(MaybeAgent,Parameter1)
    }
}

new_ruined_vehicle_device := class(creative_device, aware_device):

    @editable p_Vehicles: []ind_ruined_vehicle_device = array{}
    @editable p_SafeonzeVehicles: []safezone_ruined_vehicle_device = array{}

    @editable HitTrigger: trigger_device = trigger_device{}
    @editable DestroyTrigger: trigger_device = trigger_device{}
    @editable ScrapHitTrigger: trigger_device = trigger_device{}

    @editable MaxDistance : float = 3000.0
    
    DeviceMeta<override>()<transacts>:aware_device_meta=
        aware_device_meta{
            DeviceName := "new_ruined_vehicle"
        }  

    OnBegin<override>()<suspends>:void=
        var Index: int = 0
        for(Vehicle:p_Vehicles):
            Vehicle.VehiclePropMani.DamagedEvent.Subscribe(Handler1{Device:=Self, Parameter1:=Index}.HandlerFunction)
            set Index = Index + 1

        var SZIndex: int = 0
        for(Vehicle:p_SafeonzeVehicles):
            Vehicle.VehiclePropMani.DamagedEvent.Subscribe(Handler1{Device:=Self, Parameter1:=SZIndex}.SafezoneVehHandlerFunction)
            Vehicle.EnableTrigger.TriggeredEvent.Subscribe(Handler1{Device:=Self, Parameter1:=SZIndex}.TriggerEnableHandlerFunction)
            Vehicle.DisableTrigger.TriggeredEvent.Subscribe(Handler1{Device:=Self, Parameter1:=SZIndex}.TriggerDisableHandlerFunction)
            Vehicle.DestroyVehicle()
            set SZIndex = SZIndex + 1
            

    OnVehicleDamaged(Agent:agent,Parameter1:int) : void =
        if(Vehicle:=p_Vehicles[Parameter1]):
            if (Vehicle.IsOnDelay[]) {
                return
            }

            if(FortCharacter := Agent.GetFortCharacter[]):
                VehProp := Vehicle.GetProp()
                VehLoc := VehProp.GetTransform().Translation
                PlayerLocation := FortCharacter.GetTransform().Translation
            
                DistanceX : float = VehLoc.X - PlayerLocation.X
                DistanceY : float = VehLoc.Y - PlayerLocation.Y
                DistanceZ : float = VehLoc.Z - PlayerLocation.Z
    
                HorVehicleDistance := Sqrt((DistanceX * DistanceX) + (DistanceY * DistanceY))
                VehicleDistance := Sqrt((HorVehicleDistance * HorVehicleDistance) + (DistanceZ * DistanceZ))
            
                if (VehicleDistance > MaxDistance):
                    return

            if (Vehicle.Health > 1):
                HitTrigger.Trigger(Agent)
                Vehicle.DamageVehicle()
            else if (Vehicle.Health <= 1):
                DestroyTrigger.Trigger(Agent)
                Vehicle.DestroyVehicle()

    OnSafezoneVehicleDamaged(Agent:agent,Parameter1:int) : void =
        if(Vehicle:=p_SafeonzeVehicles[Parameter1]):
            if (Vehicle.IsOnDelay[]) {
                return
            }

            if(FortCharacter := Agent.GetFortCharacter[]):
                VehProp := Vehicle.GetProp()
                VehLoc := VehProp.GetTransform().Translation
                PlayerLocation := FortCharacter.GetTransform().Translation
            
                DistanceX : float = VehLoc.X - PlayerLocation.X
                DistanceY : float = VehLoc.Y - PlayerLocation.Y
                DistanceZ : float = VehLoc.Z - PlayerLocation.Z
    
                HorVehicleDistance := Sqrt((DistanceX * DistanceX) + (DistanceY * DistanceY))
                VehicleDistance := Sqrt((HorVehicleDistance * HorVehicleDistance) + (DistanceZ * DistanceZ))
            
                if (VehicleDistance > MaxDistance):
                    return

            if (Vehicle.IsScrapVehicle[]):
                ScrapHitTrigger.Trigger(Agent)
                Vehicle.DamageVehicle()
            else if (Vehicle.Health > 1):
                HitTrigger.Trigger(Agent)
                Vehicle.DamageVehicle()
            else if (Vehicle.Health <= 1):
                DestroyTrigger.Trigger(Agent)
                Vehicle.DestroyVehicle()

    OnEnableTriggerTriggered(Agent:agent,Parameter1:int) : void =
        if(Vehicle:=p_SafeonzeVehicles[Parameter1]):
            Vehicle.EnableAll()

    OnDisableTriggerTriggered(Agent:agent,Parameter1:int) : void =
        if(Vehicle:=p_SafeonzeVehicles[Parameter1]):
            Vehicle.DestroyVehicle()

ruined_vehicle_instance_base := class<abstract>:
    var Health: int = 6
    var HitDelay: float = 0.5
    var OnDelay: logic = false

    GetProp():creative_prop

    IsOnDelay()<decides><transacts>:void=
        OnDelay?

    ApplyCooldown()<suspends>:void=
        set OnDelay = true
        Sleep(HitDelay)
        set OnDelay = false

ind_ruined_vehicle_device<internal> := class<final><concrete>(ruined_vehicle_instance_base):
    @editable MapIndicator: map_indicator_device = map_indicator_device{}
    @editable BillboardProp: creative_prop = creative_prop{}
    @editable VehiclePropMani : prop_manipulator_device = prop_manipulator_device{}

    DamageVehicle():void=
        set Health = Health - 1
        spawn:
            ApplyCooldown()

    GetProp<override>():creative_prop=
        return BillboardProp

    DestroyVehicle():void=
        MapIndicator.Disable()
        BillboardProp.Hide()
        VehiclePropMani.HideProps()
        spawn:
            RespawnVehicle()

    EnableAll():void=
        MapIndicator.Enable()
        BillboardProp.Show()
        VehiclePropMani.ShowProps()
    
    RespawnVehicle()<suspends>:void=
        Sleep(30.0)
        set Health = 6
        EnableAll()

safezone_ruined_vehicle_device<internal> := class<final><concrete>(ruined_vehicle_instance_base):
    @editable IsScrap: logic = false
    @editable EnableTrigger: trigger_device = trigger_device{}
    @editable DisableTrigger: trigger_device = trigger_device{}

    @editable MapIndicator: map_indicator_device = map_indicator_device{}
    @editable BillboardProp: creative_prop = creative_prop{}
    @editable VehiclePropMani : prop_manipulator_device = prop_manipulator_device{}

    DamageVehicle():void=
        if(not IsScrap?):
            set Health = Health - 1

        spawn:
            ApplyCooldown()

    IsScrapVehicle()<decides><transacts>:void=
        IsScrap?

    GetProp<override>():creative_prop=
        return BillboardProp

    DestroyVehicle():void=
        MapIndicator.Disable()
        BillboardProp.Hide()
        VehiclePropMani.HideProps()

    EnableAll():void=
        set Health = 6
        MapIndicator.Enable()
        BillboardProp.Show()
        VehiclePropMani.ShowProps()
