
using { /Fortnite.com/Devices }
using { /Verse.org/Simulation }
using { /UnrealEngine.com/Temporary/Diagnostics }

# See https://dev.epicgames.com/documentation/en-us/uefn/create-your-own-device-in-verse for how to create a verse device.

# A Verse-authored creative device that can be placed in a level
rebirth_props_device := class(creative_device):


    @editable ShowTriggers: []trigger_device = array{}
    @editable HideTriggers: []trigger_device = array{}

    @editable ShowCinema: cinematic_sequence_device = cinematic_sequence_device{}
    @editable HideCinema: cinematic_sequence_device = cinematic_sequence_device{}

    # Runs when the device is started in a running game
    OnBegin<override>()<suspends>:void=
        for(ShowTrigger:ShowTriggers):
            ShowTrigger.TriggeredEvent.Subscribe(ShowProps)
        for(HideTrigger:HideTriggers):
            HideTrigger.TriggeredEvent.Subscribe(HideProps)

    ShowProps(Agent:?agent):void=
        ShowCinema.Play()

    HideProps(Agent:?agent):void=
        HideCinema.Play()