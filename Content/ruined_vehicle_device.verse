using { /Fortnite.com/Devices }
using { /Verse.org/Simulation }
using { /UnrealEngine.com/Temporary/Diagnostics }
using { /Fortnite.com/Characters }
using { /Fortnite.com/Playspaces }
using { /UnrealEngine.com/Temporary/Diagnostics }
using { /UnrealEngine.com/Temporary/SpatialMath }

using { HavocVerse.Debug }
using { HavocVerse.Util }
using { HavocVerse.Logging }

ruined_vehicle_device := class(creative_device, aware_device):

    @editable StartEnabled : logic = true
    @editable DisableTrigger: trigger_device = trigger_device{}

    @editable AutoRespawn : logic = true
    @editable RespawnTriggers: []trigger_device = array{}
    
    @editable ScrapPile : logic = false

    @editable StartingHealth : int = 5

    @editable RespawnTime : float = 15.0
    @editable HitCooldown: float = 0.5

    @editable MaxDistance : float = 3000.0

    @editable HitTrigger: trigger_device = trigger_device{}
    @editable DestroyTrigger: trigger_device = trigger_device{}
    @editable ScrapHitTrigger: trigger_device = trigger_device{}
    
    @editable MapIndicator: map_indicator_device = map_indicator_device{}
    @editable BillboardProp: creative_prop = creative_prop{}
    
    @editable VehiclePropMani : prop_manipulator_device = prop_manipulator_device{}

    var m_CurrentHealth : int = 5
    var m_LastHit: ?float = false

    DeviceMeta<override>()<transacts>:aware_device_meta=
        aware_device_meta{
            DeviceName := "ruined_vehicle"
        }

    OnBegin<override>()<suspends>:void=
        VehiclePropMani.DamagedEvent.Subscribe(OnVehicleDamaged)
        for (Trigger:RespawnTriggers) {
            Trigger.TriggeredEvent.Subscribe(OnRespawnVehicle)
        }

        
        DisableTrigger.TriggeredEvent.Subscribe(DisableEverything2)
        
        if(not StartEnabled?):
            MapIndicator.Disable()
            BillboardProp.Hide()
            VehiclePropMani.HideProps()


    OnVehicleDamaged<private>(Agent:agent) : void =
        Now := GetSimulationElapsedTime()
        if (LastHitAt := m_LastHit?, LastHitAt + HitCooldown > Now) {
            Log(Self, MsgHitCooldown(Agent), ?Level := log_level.Verbose)
            return
        }

        if(FortCharacter := Agent.GetFortCharacter[]):
            VehLoc := BillboardProp.GetTransform().Translation
            PlayerLocation := FortCharacter.GetTransform().Translation
        
            DistanceX : float = VehLoc.X - PlayerLocation.X
            DistanceY : float = VehLoc.Y - PlayerLocation.Y
            DistanceZ : float = VehLoc.Z - PlayerLocation.Z

            HorVehicleDistance := Sqrt((DistanceX * DistanceX) + (DistanceY * DistanceY))
            VehicleDistance := Sqrt((HorVehicleDistance * HorVehicleDistance) + (DistanceZ * DistanceZ))
        
            if (VehicleDistance > MaxDistance):
                return

        set m_LastHit = Some(Now)

        if (m_CurrentHealth > 0) {
            if(ScrapPile?):
                ScrapHitTrigger.Trigger(Agent)
            else:
                HitTrigger.Trigger(Agent)
                set m_CurrentHealth -= 1
                
        } else {
            DestroyTrigger.Trigger(Agent)
            MapIndicator.Disable()
            BillboardProp.Hide()
            VehiclePropMani.HideProps()

            set m_LastHit = None()
            if (AutoRespawn?) {
                spawn:
                    ScheduleVehicleRespawn()
            }
        }

    MsgHitCooldown<private><localizes>(Agent:agent):message=
        "Hit from '{Agent}' was ignored because of cooldown"

    ScheduleVehicleRespawn<private>()<suspends>:void=
        Sleep(RespawnTime)
        RespawnVehicle()

    OnRespawnVehicle<private>(Agent:?agent):void=
        RespawnVehicle()

    RespawnVehicle<private>():void=
        set m_CurrentHealth = StartingHealth
        VehiclePropMani.ShowProps()
        MapIndicator.Enable()
        BillboardProp.Show()

    DisableEverything2(Agent:?agent):void=
        MapIndicator.Disable()
        BillboardProp.Hide()
        VehiclePropMani.HideProps()

