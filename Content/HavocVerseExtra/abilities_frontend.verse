using { /Fortnite.com/Devices }
using { /Verse.org/Simulation }
using { /UnrealEngine.com/Temporary/Diagnostics }

using { HavocVerse.Debug }
using { HavocVerse.Devices }
using { HavocVerse.Logging }
using { HavocVerse.Util }

abilities_frontend_display_config := class<concrete>:
    @editable AbilityId: string = ""
    @editable OnButton: []button_device = array{}
    @editable OnTrigger: []trigger_device = array{}
    @editable UnownedErrorHUDMessage: []hud_message_device = array{}

abilities_frontend_device := class(creative_device, aware_device):
    @editable p_DevId: string = "N/A"

    @editable p_AccessControl: ?access_control_device = false
    @editable p_AbilitiesManager: abilities_manager_device = abilities_manager_device{}
    
    @editable p_MaxActive: int = 1
    @editable p_AbilityDisplays: []abilities_frontend_display_config = array{}

    DeviceMeta<override>()<transacts>:aware_device_meta=
        aware_device_meta{
            DeviceName := "abilities_frontend"
            DevId := p_DevId
        }

    OnBegin<override>()<suspends>:void=
        AwareInitDevice(Self, OnInit)
    
    OnInit<private>()<suspends>:void=
        for (DisplayConfig:p_AbilityDisplays) {
            Closure := abilities_frontend_display_closure{
                Device := Self
                DisplayConfig := DisplayConfig
            }
            for (Button:DisplayConfig.OnButton) {
                Button.InteractedWithEvent.Subscribe(Closure.OnButtonInteractedWith)
            }
            for (Trigger:DisplayConfig.OnTrigger) {
                Trigger.TriggeredEvent.Subscribe(Closure.OnTriggered)
            }
        }

    HandleDisplayInteraction<internal>(Player:player, DisplayConfig:abilities_frontend_display_config):void=
        if (AccessControl := p_AccessControl?, not AccessControl.IsAllowed[Player]) {
            return
        }

        AbilityId := DisplayConfig.AbilityId
        if (not p_AbilitiesManager.IsKnownAbility[AbilityId]) {
            Log(Self, "Handling display interaction for inexistant ability \"{AbilityId}\"", ?Level := log_level.Error)
            return
        }
        if (ActiveMeta := p_AbilitiesManager.GetActiveAbility[Player, AbilityId]) {
            p_AbilitiesManager.Deactivate(Player, AbilityId)
            return
        } else if (p_AbilitiesManager.GetLevel[Player, AbilityId] = 0) {
            Result := p_AbilitiesManager.TryUpgrade(Player, AbilityId)
            if (Result = abilities_manager_upgrade_result.FailCost) {
                return
            }
            if (Result <> abilities_manager_upgrade_result.Success) {
                for (HUDMessage:DisplayConfig.UnownedErrorHUDMessage) {
                    HUDMessage.Show(Player)
                }
                return
            }
        }

        var OtherAbilities: [string]void = map{}
        for (
            Other:p_AbilityDisplays
            Other.AbilityId <> AbilityId
            set OtherAbilities[Other.AbilityId] = void
        ) {
        }
        var OtherActiveCount: int = 0
        for (
            ActiveMeta:p_AbilitiesManager.GetActiveAbilities(Player)
            OtherAbilities[ActiveMeta.AbilityId]
        ) {
            if (OtherActiveCount + 1 >= p_MaxActive) {
                p_AbilitiesManager.Deactivate(Player, ActiveMeta.AbilityId)
            } else {
                set OtherActiveCount += 1
            }
        }

        p_AbilitiesManager.Activate(Player, AbilityId)

abilities_frontend_display_closure := class:
    Device: abilities_frontend_device
    DisplayConfig: abilities_frontend_display_config

    OnButtonInteractedWith(Agent:agent):void=
        if (Player := player[Agent]) {
            Device.HandleDisplayInteraction(Player, DisplayConfig)
        }

    OnTriggered(Agent:?agent):void=
        if (Player := player[Agent?]) {
            Device.HandleDisplayInteraction(Player, DisplayConfig)
        }