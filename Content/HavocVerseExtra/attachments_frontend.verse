using { /Fortnite.com/Devices }
using { /Verse.org/Simulation }
using { /UnrealEngine.com/Temporary/Diagnostics }

using { HavocVerse.Debug }
using { HavocVerse.Devices }
using { HavocVerse.Logging }
using { HavocVerse.Util }
using { HavocVerse.UI }

attachments_frontend_mode := enum:
    Unlock
    Equip
    UnlockAndEquip

attachments_frontend_display_config := class<concrete>:
    @editable Attachment: string = ""
    @editable ModeOverride: ?attachments_frontend_mode = false

    @editable OnButton: []button_device = array{}
    @editable OnTrigger: []trigger_device = array{}

    @editable UnlockTrigger: []trigger_device = array{}
    @editable EquipTrigger: []trigger_device = array{}

attachments_frontend_clear_display_config := class<concrete>:
    @editable OnButton: []button_device = array{}
    @editable OnTrigger: []trigger_device = array{}

# Frontend device to unlock and equip attachments.
attachments_frontend_device := class(creative_device, aware_device):
	@editable p_DevId: string = "N/A"

	@editable p_AccessControl: ?access_control_device = false
	@editable p_AttachmentsManager: attachments_manager_device = attachments_manager_device{}
	@editable p_AttachmentKindId: string = ""
	
    @editable p_DefaultMode: attachments_frontend_mode = attachments_frontend_mode.UnlockAndEquip
	@editable p_AttachmentDisplays: []attachments_frontend_display_config = array{}
    @editable p_ClearDisplays: []attachments_frontend_clear_display_config = array{}

    @editable p_PurchaseDialog: ?popup_dialog_device = false
    @editable p_PurchaseDialogConfirmButtonIndex: int = 0
    @editable p_GeneralUnlockTrigger: []trigger_device = array{}
    @editable p_GeneralEquipTrigger: []trigger_device = array{}

    DeviceMeta<override>()<transacts>:aware_device_meta=
        aware_device_meta{
            DeviceName := "attachments_frontend"
            DevId := p_DevId
        }

    OnBegin<override>()<suspends>:void=
        AwareInitDevice(Self, OnInit)

    OnInit<private>()<suspends>:void=
        if (p_AttachmentKindId.Length = 0) {
            Log(Self, "p_AttachmentKindId must be set", ?Level := log_level.Error)
        }

        for (DisplayConfig:p_AttachmentDisplays) {
            Closure := attachments_frontend_display_closure{
                Device := Self
                DisplayConfig := DisplayConfig
            }
            for (Button:DisplayConfig.OnButton) {
                Button.InteractedWithEvent.Subscribe(Closure.OnButtonInteractedWith)
            }
            for (Trigger:DisplayConfig.OnTrigger) {
                Trigger.TriggeredEvent.Subscribe(Closure.OnTriggered)
            }
        }

        for (DisplayConfig:p_ClearDisplays) {
            Closure := attachments_frontend_clear_display_closure{
                Device := Self
                DisplayConfig := DisplayConfig
            }
            for (Button:DisplayConfig.OnButton) {
                Button.InteractedWithEvent.Subscribe(Closure.OnButtonInteractedWith)
            }
            for (Trigger:DisplayConfig.OnTrigger) {
                Trigger.TriggeredEvent.Subscribe(Closure.OnTriggered)
            }
        }

    HandleDisplayInteraction<internal>(
        Closure:attachments_frontend_display_closure,
        Player:player,
        DisplayConfig:attachments_frontend_display_config
    ):void=
        if (AccessControl := p_AccessControl?, not AccessControl.IsAllowed[Player]) {
            return
        }

        Mode := DisplayConfig.ModeOverride? or p_DefaultMode
        if (Mode = attachments_frontend_mode.Unlock or Mode = attachments_frontend_mode.UnlockAndEquip) {
            if (DialogDevice := p_PurchaseDialog?) {
                Wrapper := dialog_device_wrapper{
                    DialogDevice := DialogDevice
                }
                Wrapper.Show(Player, Some(Closure.OnConfirmClick))
                return
            } else {
                Unlock(Player, DisplayConfig)
            }
        }
        HandlePostUnlocked(Closure, Player, DisplayConfig)

    Unlock<internal>(Player:player, DisplayConfig:attachments_frontend_display_config):void=
        Success := p_AttachmentsManager.UnlockAttachment(Player, p_AttachmentKindId, DisplayConfig.Attachment)
        if (Success?) {
            for (Trigger:p_GeneralUnlockTrigger) {
                Trigger.Trigger(Player)
            }
            for (Trigger:DisplayConfig.UnlockTrigger) {
                Trigger.Trigger(Player)
            }
        }

    HandlePostUnlocked<internal>(
        Closure:attachments_frontend_display_closure,
        Player:player,
        DisplayConfig:attachments_frontend_display_config
    ):void=
        Mode := DisplayConfig.ModeOverride? or p_DefaultMode
        if (Mode = attachments_frontend_mode.Equip or Mode = attachments_frontend_mode.UnlockAndEquip) {
            Equip(Player, DisplayConfig)
        }

    Equip<internal>(Player:player, DisplayConfig:attachments_frontend_display_config):void=
        p_AttachmentsManager.SetActiveAttachment(Player, p_AttachmentKindId, Some(DisplayConfig.Attachment))

    HandleClearDisplayInteraction<internal>(
        Player:player,
        DisplayConfig:attachments_frontend_clear_display_config
    ):void=
        p_AttachmentsManager.SetActiveAttachment(Player, p_AttachmentKindId, None())

attachments_frontend_display_closure := class:
    Device: attachments_frontend_device
    DisplayConfig: attachments_frontend_display_config

    OnButtonInteractedWith(Agent:agent):void=
        if (Player := player[Agent]) {
            Device.HandleDisplayInteraction(Self, Player, DisplayConfig)
        }

    OnTriggered(Agent:?agent):void=
        if (Player := player[Agent?]) {
            Device.HandleDisplayInteraction(Self, Player, DisplayConfig)
        }

    OnConfirmClick(Event:dialog_response_event):void=
        Log(Device, "Purchase dialog event {Event}", ?Level := log_level.Debug)
        if (
            Event.ResponseType = dialog_response_type.Clicked
            Event.ButtonIndex? = Device.p_PurchaseDialogConfirmButtonIndex
            Player := player[Event.Agent]
        ) {
            Device.Unlock(Player, DisplayConfig)
            Device.HandlePostUnlocked(Self, Player, DisplayConfig)
        }

attachments_frontend_clear_display_closure := class:
    Device: attachments_frontend_device
    DisplayConfig: attachments_frontend_clear_display_config

    OnButtonInteractedWith(Agent:agent):void=
        if (Player := player[Agent]) {
            Device.HandleClearDisplayInteraction(Player, DisplayConfig)
        }

    OnTriggered(Agent:?agent):void=
        if (Player := player[Agent?]) {
            Device.HandleClearDisplayInteraction(Player, DisplayConfig)
        }
