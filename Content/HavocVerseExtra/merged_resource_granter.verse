using { /Fortnite.com/Devices }
using { /Verse.org/Simulation }
using { /UnrealEngine.com/Temporary/Diagnostics }

using { HavocVerse.Debug }
using { HavocVerse.Devices }
using { HavocVerse.Resource }
using { HavocVerse.Logging }
using { HavocVerse.Util }

merged_resource_granter_hud_messenger_config := class<concrete><final>:
    @editable HUDMessenger: hud_message_device = hud_message_device{}
    @editable_text_box{MultiLine := true}
    Template: string = "+$%Amount:money%"
    @editable NumberFormatter: number_formatter_config = number_formatter_config{}

merged_resource_granter_device := class(creative_device, aware_device):
    @editable p_ResourceManager: resource_manager_device = resource_manager_device{}

    @editable p_GiveUserActions: []merged_resource_granter_user_action_give_config = array{}
    
    DeviceMeta<override>()<transacts>:aware_device_meta=
        aware_device_meta{
            DeviceName := "merged_resource_granter"
        }

    OnBegin<override>()<suspends>:void=
        AwareInitDevice(Self, OnInit)

    OnInit<private>()<suspends>:void=
        p_ResourceManager.Setup()

        for (UserActionConfig:p_GiveUserActions) {
            UserAction := merged_resource_granter_user_action_give{
                EndpointDevice := Self
                Config := UserActionConfig
            }
            UserActionConfig.OnTrigger.TriggeredEvent.Subscribe(UserAction.OnTriggered)
        }

merged_resource_granter_user_action_give_multiplier := enum:
    NoMultiplier
    EffectiveMultiplier
    AutoGenerationRate

merged_resource_granter_user_action_give_round_method := enum:
    R_Round
    R_Floor
    R_Ceil

TooltipMultiplierSourceOverride<internal><localizes>:message = "Resource to use as the source of the multiplier"

merged_resource_granter_user_action_give_config := class<concrete>:
    @editable OnTrigger: trigger_device = trigger_device{}
    @editable ShowFeedback: logic = true
    @editable Resources: []merged_resource_granter_user_action_give_config_entry = array{}
    @editable HudMessengers: []merged_resource_granter_hud_messenger_config = array{}
    @editable Multiplier: merged_resource_granter_user_action_give_multiplier = merged_resource_granter_user_action_give_multiplier.NoMultiplier
    @editable RoundingMethod : merged_resource_granter_user_action_give_round_method = merged_resource_granter_user_action_give_round_method.R_Ceil
    @editable ConstrainAmountMin: int = 0
    @editable ConstrainAmountMax: int = 2147483647

merged_resource_granter_user_action_give_config_entry := class<concrete>:
    @editable ResourceId: string = ""
    @editable Amount: float = 1.0
    @editable ShowFeedbackOverride: ?logic = false
    @editable MultiplierSourceOverride: ?string = false

merged_resource_granter_user_action_give := class:
    EndpointDevice: merged_resource_granter_device
    Config: merged_resource_granter_user_action_give_config

    OnTriggered(Agent:?agent):void=
        Summary := merged_resource_granter_summary{}
        for (Entry:Config.Resources) {
            ProcessEntry(Agent, Entry, Summary)
        }
    
        if (Player := player[Agent?]) {
            for (MsgConfig:Config.HudMessengers) {
                NumberFormatter := MsgConfig.NumberFormatter.CreateFormatter()
                var Replacements: [string]message = map{}
                for (ResourceId->Amount:Summary.m_Amounts) {
                    set Replacements = ConcatenateMaps(Replacements, map{
                        "Amount:{ResourceId}" => ToMessage(Amount)
                    })
                }
                Message := ReplaceIntoMessage(MsgConfig.Template, Replacements)
                MsgConfig.HUDMessenger.Show(Player, Message)
            }
        }

    ProcessEntry<private>(
        Agent:?agent, Entry:merged_resource_granter_user_action_give_config_entry,
        Summary:merged_resource_granter_summary
    ):void=
        var Amount: int = if (l_Agent := Agent?) {
            MultiplierSource := if (ResourceId := Entry.MultiplierSourceOverride?) {
                ResourceId
            } else {
                Entry.ResourceId
            }
            FloatAmount := case(Config.Multiplier):
                merged_resource_granter_user_action_give_multiplier.NoMultiplier => Entry.Amount
                merged_resource_granter_user_action_give_multiplier.AutoGenerationRate =>
                    Rate := EndpointDevice.p_ResourceManager.GetAutoGenerationRate(l_Agent, MultiplierSource)
                    Log(MsgUsingGenerationRateMultiplier(l_Agent, MultiplierSource, Entry.ResourceId, Rate), ?Level := log_level.Verbose)
                    Entry.Amount * Rate
                merged_resource_granter_user_action_give_multiplier.EffectiveMultiplier =>
                    EffectiveMultiplier := EndpointDevice.p_ResourceManager.GetEffectiveMultiplier(l_Agent, MultiplierSource)
                    Log(MsgUsingEffectiveMultiplier(l_Agent, MultiplierSource, Entry.ResourceId, EffectiveMultiplier), ?Level := log_level.Verbose)
                    Entry.Amount * EffectiveMultiplier
            case(Config.RoundingMethod):
                merged_resource_granter_user_action_give_round_method.R_Round => DefaultingRound(FloatAmount)
                merged_resource_granter_user_action_give_round_method.R_Ceil => DefaultingCeil(FloatAmount)
                merged_resource_granter_user_action_give_round_method.R_Floor => DefaultingFloor(FloatAmount)
        } else {
            DefaultingFloor(Entry.Amount)
        }
        set Amount = Clamp(Amount, Config.ConstrainAmountMin, Config.ConstrainAmountMax)
        ShowFeedback := Entry.ShowFeedbackOverride? or Config.ShowFeedback
        EndpointDevice.p_ResourceManager.Give(Agent, Entry.ResourceId, Amount, ?SendFeedback := ShowFeedback)
        Summary.Add(Entry.ResourceId, Amount)

    MsgUsingGenerationRateMultiplier<private><localizes>(Agent:agent, ResourceId:string, MultiplierSource:string, Rate:float):message=
        "Using generation rate {Rate} (source: {MultiplierSource}) as multiplier when giving resource {ResourceId} to {Agent}"
    MsgUsingEffectiveMultiplier<private><localizes>(Agent:agent, ResourceId:string, MultiplierSource:string, Multiplier:float):message=
        "Using effective multiplier {Multiplier} (source: {MultiplierSource}) when giving resource {ResourceId} to {Agent}"

merged_resource_granter_summary := class<final>:
    var m_Amounts: [string]int = map{}

    Add(ResourceId:string, Amount:int):void=
        NewAmount := (m_Amounts[ResourceId] or 0) + Amount
        if. set m_Amounts[ResourceId] = NewAmount
