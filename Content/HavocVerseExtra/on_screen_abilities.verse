using { /Fortnite.com/Devices }
using { /Verse.org/Simulation }
using { /UnrealEngine.com/Temporary/Diagnostics }
using { /UnrealEngine.com/Temporary/SpatialMath }
using { /UnrealEngine.com/Temporary/UI }

using { HavocVerse.Debug }
using { HavocVerse.Devices }
using { HavocVerse.UI }
using { HavocVerse.Util }

on_screen_abilities_entry := class<concrete>:
    @editable AbilityId: string = ""
    @editable IconImage: string = ""
    @editable LevelRangeFrom: int = 0
    @editable LevelRangeTo: int = -1

on_screen_abilities_device := class(creative_device, aware_device, ability_hook):
    @editable p_DevId: string = "N/A"

    @editable p_AbilitiesManager: abilities_manager_device = abilities_manager_device{}

    @editable p_Entries: []on_screen_abilities_entry = array{}
    @editable p_IconImagesScale: float = 1.0
    @editable p_StackDirection: orientation = orientation.Vertical
    @editable p_StackGap: float = 20.0

    @editable p_HUDAccessControl: []access_control_device = array{}
    @editable p_HUDManagerConfig: hud_manager_base_config = hud_manager_base_config{
        RefreshRate := -1.0
    }
    @editable p_HUDPlacement: ui_canvas_placement_config = ui_canvas_placement_config{}

    var m_HudManager: ?on_screen_abilities_hud_manager = false

    DeviceMeta<override>()<transacts>:aware_device_meta=
        aware_device_meta{
            DeviceName := "on_screen_abilities"
            DevId := p_DevId
        }

    OnBegin<override>()<suspends>:void=
        AwareInitDevice(Self, OnInit)
    
    OnInit<private>()<suspends>:void=
        p_AbilitiesManager.RegisterHook(Self)

        HudManager := on_screen_abilities_hud_manager{
            AccessControl := option{p_HUDAccessControl[0]}
            RefreshRate := p_HUDManagerConfig.RefreshRate
            ZOrder := p_HUDManagerConfig.ZOrder
            Device := Self
        }
        set m_HudManager = Some(HudManager)
        spawn:
            HudManager.Run(GetPlayspace())

    OnAbilityActivated<override>(Player:player, Meta:active_ability_meta):void=
        if (HudManager := m_HudManager?, Player.IsActive[]) {
            HudManager.Refresh(Player)
        }

    OnAbilityDeactivated<override>(Player:player, Meta:active_ability_meta):void=
        if (HudManager := m_HudManager?, Player.IsActive[]) {
            HudManager.Refresh(Player)
        }

on_screen_abilities_hud_manager := class(hud_manager_base):
    Device: on_screen_abilities_device

    CreateHUD<override>(Player:player):?hud_manager_player=
        Placeholder := overlay:
        Canvas := canvas:
            Slots := array{
                Device.p_HUDPlacement.ToCanvasSlot(Placeholder)
            }
        Element := on_screen_abilities_hud_element:
            Widget := Canvas
            Device := Device
            Player := Player
            Placeholder := Placeholder
        Some(MakeSingleElementHUDPlayer(Player, Canvas, Element))

on_screen_abilities_hud_element := class(hud_element):
    Device: on_screen_abilities_device
    Player: player

    Placeholder: overlay

    var m_StackBox: ?stack_box = false
    var m_LastActive: []active_ability_meta = array{}

    Update<override>():void=
        Active := Device.p_AbilitiesManager.GetActiveAbilities(Player)
        if (m_LastActive = Active) {
            return
        }
        set m_LastActive = Active
        if (m_PrevStackBox := m_StackBox?) {
            Placeholder.RemoveWidget(m_PrevStackBox)
            set m_StackBox = None()
        }
        if (Active.Length = 0) {
            return
        }
        Scale := MakeUniformVec2(Device.p_IconImagesScale)
        var Index: int = -1
        StackBoxSlots := for (
            Entry:Device.p_Entries
            ActiveLevel := Device.p_AbilitiesManager.GetActiveAbility[Player, Entry.AbilityId].Level
            Entry.LevelRangeFrom <= ActiveLevel
            (Entry.LevelRangeTo < 0) or (ActiveLevel <= Entry.LevelRangeTo)
        ) {
            set Index += 1
            stack_box_slot:
                Widget := CreateStaticImage(Entry.IconImage, ?Scale := Scale)
                HorizontalAlignment := if (Device.p_StackDirection = orientation.Vertical) {
                    horizontal_alignment.Center
                } else {
                    horizontal_alignment.Left
                }
                VerticalAlignment := if (Device.p_StackDirection = orientation.Horizontal) {
                    vertical_alignment.Center
                } else {
                    vertical_alignment.Top
                }
                Padding := margin{
                    Left := if (Index > 0, Device.p_StackDirection = orientation.Horizontal) {
                        Device.p_StackGap
                    } else {
                        0.0
                    }
                    Top := if (Index > 0, Device.p_StackDirection = orientation.Vertical) {
                        Device.p_StackGap
                    } else {
                        0.0
                    }
                }
        }
        StackBox := stack_box:
            Orientation := Device.p_StackDirection
            Slots := StackBoxSlots
        set m_StackBox = Some(StackBox)
        Placeholder.AddWidget(overlay_slot{
            Widget := StackBox
        })
