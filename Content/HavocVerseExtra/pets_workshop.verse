using { /Fortnite.com/Devices }
using { /Verse.org/Simulation }
using { /UnrealEngine.com/Temporary/Diagnostics }

using { HavocVerse.Debug }
using { HavocVerse.Devices }
using { HavocVerse.Resource }
using { HavocVerse.Logging }
using { HavocVerse.Util }

pets_workshop_ability_entry := class<concrete>:
    @editable PetId: string = ""
    @editable AbilityId: string = ""

pets_workshop_resource_count_backed_entry := class<concrete>:
    @editable PetId: string = ""
    @editable ResourceId: string = ""
    @editable Amount: int = 1

pets_workshop_device<public> := class(creative_device, aware_device):
    @editable p_PetsManager<public>: pet_controller_device = pet_controller_device{}
    @editable p_AbilitiesManager<public>: abilities_manager_device = abilities_manager_device{}
    @editable p_ResourceManager<public>: resource_manager_device = resource_manager_device{}

    @editable p_Abilities: []pets_workshop_ability_entry = array{}
    @editable p_ResourceBackedPets: []pets_workshop_resource_count_backed_entry = array{}

    DeviceMeta<override>()<transacts>:aware_device_meta=
        aware_device_meta{
            DeviceName := "pets_workshop"
        }

    OnBegin<override>()<suspends>:void=
        AwareInitDevice(Self, OnInit)
    
    OnInit<private>()<suspends>:void=
        p_PetsManager.SubscribeToEquipEvents(OnPetEquipped, OnPetUnequipped)

        for (Entry:p_ResourceBackedPets) {
            ConditionProvider := pets_workshop_resource_condition_provider{
                ResourceManager := p_ResourceManager
                Entry := Entry
            }
            p_PetsManager.SetPetConditionProvider(Entry.PetId, ConditionProvider)
        }

    OnPetEquipped<private>(Player:player, PetId:string):void=
        for (Entry:p_Abilities, Entry.PetId = PetId) {
            p_AbilitiesManager.Activate(Player, Entry.AbilityId)
        }

    OnPetUnequipped<private>(Player:player, PetId:string):void=
        for (Entry:p_Abilities, Entry.PetId = PetId) {
            p_AbilitiesManager.Deactivate(Player, Entry.AbilityId)
        }

    # Returns the abilities of a pet.
    GetAbilities<public>(PetId:string)<transacts>:[]string=
        for (Entry:p_Abilities, Entry.PetId = PetId) {
            Entry.AbilityId
        }

    # Checks if a pet has an ability.
    HasAbility<public>(PetId:string, AbilityId:string)<transacts><decides>:void=
        var Found: logic = false
        for (Entry:p_Abilities, Entry.PetId = PetId, Entry.AbilityId = AbilityId) {
            set Found = true
        }
        Found?

    # Checks if a pet has any ability.
    HasAnyAbility<public>(PetId:string)<transacts><decides>:void=
        GetAbilities(PetId).Length > 0

pets_workshop_resource_condition_provider := class(pet_condition_provider):
    ResourceManager: resource_manager_device
    Entry: pets_workshop_resource_count_backed_entry

    IsPetUnlocked<override>(Player:player, PetId:string)<transacts><decides>:void=
        Resource := transactful_resource[ResourceManager.GetResource[Entry.ResourceId]]
        Resource.TransactfulGetAmount(Some(Player)) >= Entry.Amount
