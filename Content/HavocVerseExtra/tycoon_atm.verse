using { /Fortnite.com/Devices }
using { /Verse.org/Simulation }
using { /UnrealEngine.com/Temporary/Diagnostics }

using { HavocVerse.Debug }
using { HavocVerse.Devices }
using { HavocVerse.Resource }
using { HavocVerse.Storage }
using { HavocVerse.Logging }
using { HavocVerse.Util }

tycoon_atm_mode := enum:
    Deposit
    TransformResource

tycoon_atm_device := class(creative_device, aware_device, custom_device, resetable, resource_holder, persistence_recordable):
    @editable p_DevId: string = "N/A"
    @editable p_PersistenceIdOverride: ?string = false
    @editable p_PersistenceReserveBits: int = 31

    @editable p_StartEnabled: logic = true
    @editable p_AccessControl: access_control_device = access_control_device{}
    @editable p_ResourceManager: resource_manager_device = resource_manager_device{}
    @editable p_Mode: tycoon_atm_mode = tycoon_atm_mode.Deposit
    @editable p_InResourceId: string = ""
    @editable p_OutResourceId: string = ""

    @editable p_RefreshRate: float = 0.25
    @editable p_Billboards: []billboard_device = array{}
    @editable p_BillboardTemplate: string = "%Amount%\n%Rate%"
    @editable p_BillboardAmountFormatter: number_formatter_config = number_formatter_config{
        Template := "$%Val%"
    }
    @editable p_BillboardRateFormatter: number_formatter_config = number_formatter_config{
        Template := "$%Val%/s"
    }
    @editable p_HideBillboardsWhenDisabled: logic = true

    @editable p_CollectButtons: []button_device = array{}
    @editable p_CollectTriggers: []trigger_device = array{}
    @editable p_CollectVolumes: []volume_device = array{}
    @editable p_CollectStandingRate: ?float = false
    @editable p_OnCollectTriggers: []trigger_device = array{}
    @editable p_OnCollectAudioCue: ?shared_audio_cue_pool_device = false

    @editable p_UserActions: []tycoon_atm_user_action_config = array{}

    UpdateDisplaysEvent: event() = event(){}

    var m_Enabled: logic = false
    var m_Balance: int = 0

    var m_PersistenceControl: persistence_recordable_control = persistence_recordable_control{}

    DeviceMeta<override>()<transacts>:aware_device_meta=
        aware_device_meta{
            DeviceName := "tycoon_atm"
            DevId := p_DevId
        }

    OnBegin<override>()<suspends>:void=
        AwareInitDevice(Self, OnInit)

    OnInit<private>()<suspends>:void=
        if (p_Mode = tycoon_atm_mode.TransformResource, p_InResourceId.Length = 0) {
            Log(Self, "p_Mode = TransformResource, but p_InResourceId is not set", ?Level := log_level.Error)
            return
        }
        if (p_OutResourceId.Length = 0) {
            Log(Self, "p_OutResourceId is not set", ?Level := log_level.Error)
            return
        }

        p_AccessControl.RegisterUsingDevice(Self)

        for (Button:p_CollectButtons) {
            Button.InteractedWithEvent.Subscribe(OnButtonInteractedWith)
        }
        for (Trigger:p_CollectTriggers) {
            Trigger.TriggeredEvent.Subscribe(Collect)
        }   
        for (Volume:p_CollectVolumes) {
            Closure := tycoon_atm_volume_closure{
                Device := Self
                Volume := Volume
            }
            Volume.AgentEntersEvent.Subscribe(Closure.OnAgentEnters)
        }

        if (p_Billboards.Length > 0) {
            spawn:
                StartUpdateLoop()
        }

        if (p_StartEnabled?) {
            Enable()
        } else if (not (m_Enabled?), p_HideBillboardsWhenDisabled?) {
            for (Billboard:p_Billboards) {
                Billboard.HideText()
            }
        }

        for (UserActionConfig:p_UserActions) {
            UserAction := tycoon_atm_user_action{
                Device := Self
                Config := UserActionConfig
            }
            UserActionConfig.OnTrigger.TriggeredEvent.Subscribe(UserAction.OnTriggered)
        }

    Reset<override>(Agent:?agent):void=
        set m_Balance = 0
        if (p_StartEnabled?) {
            Enable()
        } else {
            Disable()
        }

    Enable<internal>():void=
        if (m_Enabled?) {
            return
        }
        set m_Enabled = true
        for (Billboard:p_Billboards) {
            Billboard.ShowText()
        }

    Disable<internal>():void=
        if (not (m_Enabled?)) {
            return
        }
        set m_Enabled = false
        for (Billboard:p_Billboards) {
            Billboard.HideText()
        }

    StartUpdateLoop<private>()<suspends>:void=
        AmountFormatter := p_BillboardAmountFormatter.CreateFormatter()
        RateFormatter := p_BillboardRateFormatter.CreateFormatter()
        var PrevAmount: string = ""
        var PrevRate: string = ""
        loop {
            race:
                Sleep(p_RefreshRate)
                UpdateDisplaysEvent.Await()

            FormattedAmount := AmountFormatter.Format(GetBalance())
            Rate := if (FirstAgent := p_AccessControl.ListAllowedAgents()[0]) {
                p_ResourceManager.GetAutoGenerationRate(FirstAgent, p_InResourceId)
            } else {
                0.0
            }
            FormattedRate := RateFormatter.Format(Rate)
            if (FormattedAmount <> PrevAmount or FormattedRate <> PrevRate) {
                set PrevAmount = FormattedAmount
                set PrevRate = FormattedRate

                NewBillboardText := ReplaceIntoMessage(p_BillboardTemplate, map{
                    "Amount" => ToMessage(FormattedAmount)
                    "Rate" => ToMessage(FormattedRate)
                })
                for (Billboard:p_Billboards) {
                    Billboard.SetText(NewBillboardText)
                }
            }
        }

    GetBalance<private>():int=
        case(p_Mode):
            tycoon_atm_mode.Deposit => m_Balance
            tycoon_atm_mode.TransformResource => p_ResourceManager.GetAmount(p_AccessControl, p_InResourceId)

    Collect<private>(Agent:?agent):void=
        if (not (m_Enabled?)) {
            return
        }
        if (l_Agent := Agent?, not p_AccessControl.IsAllowed[l_Agent]) {
            return
        }
        Balance := GetBalance()
        if (Balance <= 0) {
            return
        }
        case(p_Mode):
            tycoon_atm_mode.Deposit =>
                set m_Balance -= Balance
                Success := GiveOutResources(Agent, Balance)
                if (not (Success?)) {
                    set m_Balance += Balance
                }
            tycoon_atm_mode.TransformResource =>
                Success := p_ResourceManager.Take(p_AccessControl, p_InResourceId, Balance)
                if (Success?) {
                    GiveOutResources(Agent, Balance)
                }
        UpdateDisplaysEvent.Signal()
        for (Trigger:p_OnCollectTriggers) {
            Trigger.MaybeTrigger(Agent)
        }
        if (AudioCuePool := p_OnCollectAudioCue?, l_Agent := Agent?) {
            AudioCuePool.Play(l_Agent)
        }

    GiveOutResources<private>(Agent:?agent, Amount:int):logic=
        if (l_Agent := Agent?) {
            p_ResourceManager.Give(l_Agent, p_OutResourceId, Amount)
        } else {
            p_ResourceManager.Give(p_AccessControl, p_OutResourceId, Amount)
        }

    IsResourceAllowed<override>(ResourceKind:string)<transacts><decides>:void=
        ResourceKind = p_InResourceId

    Deposit<override>(ResourceKind:string, Amount:int):logic=
        if (p_Mode <> tycoon_atm_mode.Deposit) {
            Log(Self, "Attempted to deposit in an ATM setup with mode TransformResource", ?Level := log_level.Error)
            return false
        }
        if (ResourceKind <> p_InResourceId) {
            Log(Self, "Attempted to deposit resource \"{ResourceKind}\" in ATM expecting \"{p_InResourceId}\" (x{Amount})", ?Level := log_level.Error)
            return false
        }
        set m_Balance += Amount
        true

    OnButtonInteractedWith<private>(Agent:agent):void=
        Collect(Some(Agent))

    OnPhysicalVolumeEntered<internal>(Volume:volume_device, Agent:agent):void=
        if (not p_AccessControl.IsAllowed[Agent]) {
            return
        }
        Collect(Some(Agent))
        if (CollectRate := p_CollectStandingRate?) {
            spawn:
                HandleStandingCollectPhysicalVolume(Agent, Volume, CollectRate)
        }

    HandleStandingCollectPhysicalVolume<private>(Agent:agent, Volume:volume_device, CollectRate:float)<suspends>:void=
        race:
            Volume.AwaitAgentExit(GetPlayspace(), Agent)
            loop {
                Sleep(CollectRate)
                Collect(Some(Agent))
            }

    CustomTrigger<override>(Input:string, Agent:?agent):logic=
        if (Input = "enable") {
            Enable()
            return true
        } else if (Input = "disable") {
            Disable()
            return true
        }
        false

    GetPersistenceControl<override>()<transacts><decides>:persistence_recordable_control=
        m_PersistenceControl

    PersistenceReserve<override>(StorageMem:storage_memory)<transacts><decides>:persistence_recordable_state=
        p_Mode = tycoon_atm_mode.Deposit
        Id := "tycoon_atm_" + p_PersistenceIdOverride.UnwrapOr(p_DevId).ToLowercase()
        tycoon_atm_persistence_state{
            BalanceChunk := StorageMem.Reserve[p_PersistenceReserveBits, "{Id}_balance"]
        }

    PersistenceSave<override>(i_State:persistence_recordable_state, Agent:agent, SaveCause:persistence_save_cause):void=
        if (State := tycoon_atm_persistence_state[i_State]) {
            State.BalanceChunk.WriteInt(Agent, m_Balance)
        }

    PersistenceLoad<override>(i_State:persistence_recordable_state, Agent:agent, Query:persistence_load_query):void=
        if (State := tycoon_atm_persistence_state[i_State]) {
            SavedBalance := State.BalanceChunk.ReadInt(Agent)
            if (Query.Overwrite?) {
                set m_Balance = SavedBalance
            } else {
                set m_Balance += SavedBalance
            }
        }

tycoon_atm_persistence_state := class(persistence_recordable_state):
    BalanceChunk: storage_memory_chunk

    Reset<override>(Agent:agent):void=
        BalanceChunk.Reset(Agent)

tycoon_atm_volume_closure := class:
    Device: tycoon_atm_device
    Volume: volume_device

    OnAgentEnters(Agent:agent):void=
        Device.OnPhysicalVolumeEntered(Volume, Agent)

tycoon_atm_user_action_kind := enum:
    Enable
    Disable

tycoon_atm_user_action_config := class<concrete>:
    @editable Action: tycoon_atm_user_action_kind = tycoon_atm_user_action_kind.Enable
    @editable OnTrigger: trigger_device = trigger_device{}

tycoon_atm_user_action := class:
    Device: tycoon_atm_device
    Config: tycoon_atm_user_action_config

    OnTriggered(Agent:?agent):void=
        case(Config.Action):
            tycoon_atm_user_action_kind.Enable => Device.Enable()
            tycoon_atm_user_action_kind.Disable => Device.Disable()
