using { /Fortnite.com/Devices }
using { /Fortnite.com/Characters }
using { /Fortnite.com/Vehicles }
using { /Fortnite.com/UI }
using { /Verse.org/Simulation }
using { /UnrealEngine.com/Temporary/Diagnostics }
using { /UnrealEngine.com/Temporary/SpatialMath }
using { /UnrealEngine.com/Temporary/UI }

using { HavocVerse.Debug }
using { HavocVerse.Devices }
using { HavocVerse.UI }
using { HavocVerse.Util }
using { HavocVerse.Logging }

using { HavocVerseExtra }

vehicle_garage_car_kind<public> := enum:
    Fang

vehicle_garage_car_config<public> := class<final><concrete>:
    @editable Id<public>: string = ""
    @editable Kind<public>: vehicle_garage_car_kind = vehicle_garage_car_kind.Fang
    @editable VFXMesh<public>: ?visual_effect_powerup_device = false
    @editable IdleProp<public>: ?creative_prop_asset = false

vehicle_garage_locked_button_override := class<final><concrete>:
    @editable VehicleId<public>: string = ""
    @editable ButtonText<public>: string = ""

vehicle_garage_model_device := class<final><concrete>(creative_device, aware_device):
    @editable p_CarVFXMeshClear: visual_effect_powerup_device = visual_effect_powerup_device{}
    @editable p_CarModels: []vehicle_garage_car_config = array{}
    @editable p_CarPlaceholder: ?creative_prop_asset = false

    @editable p_AttachmentsManager: attachments_manager_device = attachments_manager_device{}
    @editable p_AttachmentKindCarModels: string = "cars"

    @editable p_ButtonTextSelect: string = "Select Car"
    @editable p_ButtonTextRecall: string = "Recall Car"

    @editable p_ShowLockedVehicles: logic = false
    @editable p_GoMenuButtonText: string = "Go"
    @editable p_LockedMenuButtonText: string = "Locked"
    @editable p_LockedMenuButtonTextOverrides: []vehicle_garage_locked_button_override = array{}

    DeviceMeta<override>()<transacts>:aware_device_meta=
        aware_device_meta{
            DeviceName := "vehicle_garage_model"
        }

vehicle_garage_active_car := class<final>:
    Index: int
    Prop: ?creative_prop
    CarConfig: vehicle_garage_car_config
    Unlocked: logic

vehicle_garage_instance_device := class(creative_device, aware_device, aware_debug_device, resetable, custom_device):
    @editable p_DevId: string = "N/A"
    @editable{ToolTip := TooltipRoutingIdOverride} p_RoutingIdOverride: ?string = false

    @editable p_Model: vehicle_garage_model_device = vehicle_garage_model_device{}

    @editable p_AccessControl: access_control_device = access_control_device{}
    @editable p_StartEnabled: logic = true

    @editable p_EnterButtons: []button_device = array{}
    @editable p_PreviewCamera: gameplay_camera_fixed_point_device = gameplay_camera_fixed_point_device{}
    @editable p_HideHudAccessControl: ?access_control_device = false

    @editable p_CarMarker: creative_prop = creative_prop{}
    @editable p_FangVehicleSpawner: vehicle_spawner_getaway_device = vehicle_spawner_getaway_device{}

    var m_ActiveCar: ?vehicle_garage_active_car = false
    var m_Menu: ?vehicle_garage_menu = false
    var m_KnownMenuInstances: [player]vehicle_garage_menu_instance = map{}
    var m_PrevHudState: [player]void = map{}

    var m_Enabled: logic = false
    var m_FangVehicleSpawnerEnabled: logic = false
    var m_VehicleSpawned: logic = false
    var m_PlaceholderProp: ?creative_prop = false

    var m_Subscriptions: []cancelable = array{}

    DeviceMeta<override>()<transacts>:aware_device_meta=
        aware_device_meta{
            DeviceName := "vehicle_garage"
            DevId := p_DevId
            RoutingIdOverride := p_RoutingIdOverride
        }

    DebugStats<override>()<transacts>:aware_debug_stats=
        aware_debug_stats{
            Length := map{
                "knownMenuInstances" => m_KnownMenuInstances.Length
                "prevHudState" => m_PrevHudState.Length
                "subscriptions" => m_Subscriptions.Length
            }
        }

    OnBegin<override>()<suspends>:void=
        AwareInitDevice(Self, OnInit)

    OnInit<private>()<suspends>:void=
        if (HudAccessControl := p_HideHudAccessControl?, HudAccessControl.p_Mode <> access_control_mode.Whitelist) {
            Log(Self, "p_HideHudAccessControl must be in whitelist mode", ?Level := log_level.Error)
        }

        p_AccessControl.RegisterUsingDevice(Self)

        if (AssetsProvider := AwareDefaultAssetsProvider[]) {
            set m_Menu = Some(vehicle_garage_menu{
                AssetsProvider := AssetsProvider
                Device := Self
            })
        }

        VehicleMeshProvider := garage_vehicle_mesh_provider{
            Device := Self
        }
        VehicleMeshEntry := vehicle_mesh_spawner_entry{
            Spawner := p_FangVehicleSpawner
            Provider := VehicleMeshProvider
        }
        VehicleMeshEntry.Init()

        GetPlayspace().PlayerRemovedEvent().Subscribe(OnPlayerRemoved)
        p_FangVehicleSpawner.SpawnedEvent.Subscribe(OnVehicleSpawned)
        p_FangVehicleSpawner.DestroyedEvent.Subscribe(OnVehicleDestroyed)

        p_Model.p_AttachmentsManager.AddUnlockListener(OnCarModelUnlocked,
            ?AttachmentKindId := Some(p_Model.p_AttachmentKindCarModels))

        if (p_StartEnabled?) {
            Enable()
        } else {
            Disable()
        }

    OnPlayerRemoved<private>(Player:player):void=
        if (ActiveMenu := m_KnownMenuInstances[Player]) {
            ActiveMenu.Close()
        }
        set m_KnownMenuInstances = m_KnownMenuInstances.RemoveByKey(Player)(0)
        set m_PrevHudState = m_PrevHudState.RemoveByKey(Player)(0)

    Enable<private>():void=
        if (m_Enabled?) {
            return
        }
        set m_Enabled = true

        for (EnterButton:p_EnterButtons) {
            EnterButton.Enable()
            set m_Subscriptions += array{
                EnterButton.InteractedWithEvent.Subscribe(OnEnterButtonInteractedWith)
            }
            EnterButton.SetInteractionText(ToMessage(p_Model.p_ButtonTextSelect))
        }

        if (not (m_ActiveCar?)) {
            CycleActiveCar(true)
        }

    Disable<private>():void=
        set m_Enabled = false
        if (m_FangVehicleSpawnerEnabled?) {
            set m_FangVehicleSpawnerEnabled = false
            p_FangVehicleSpawner.DestroyVehicle()
            p_FangVehicleSpawner.Disable()
        }
        if (Prop := m_ActiveCar?.Prop?) {
            Prop.Dispose()
        }
        set m_ActiveCar = None()

        for (EnterButton:p_EnterButtons) {
            EnterButton.Disable()
        }
        for (Subscription:m_Subscriptions) {
            Subscription.Cancel()
        }
        set m_Subscriptions = array{}

    Reset<override>(Agent:?agent):void=
        Disable()

    OnVehicleSpawned<private>(Vehicle:fort_vehicle):void=
        set m_VehicleSpawned = true
        if (Prop := m_ActiveCar?.Prop?) {
            Prop.Hide()
        }
        SpawnPlaceholderCar()
        for (EnterButton:p_EnterButtons) {
            EnterButton.SetInteractionText(ToMessage(p_Model.p_ButtonTextRecall))
        }

    OnVehicleDestroyed<private>():void=
        set m_VehicleSpawned = false
        DisposePlaceholderCar()
        if (Prop := m_ActiveCar?.Prop?) {
            Prop.Show()
        }
        for (EnterButton:p_EnterButtons) {
            EnterButton.SetInteractionText(ToMessage(p_Model.p_ButtonTextSelect))
        }

    OnCarModelUnlocked<private>(Event:attachment_unlocked_event):void=
        if (not m_Enabled?) {
            return
        }
        if (Event.AttachmentKindId = p_Model.p_AttachmentKindCarModels, not (m_ActiveCar?)) {
            CycleActiveCar(true)
        }

    IsCarUnlocked<private>(Player:player, ModelId:string)<transacts><decides>:void=
        p_Model.p_AttachmentsManager.IsUnlocked[Player, p_Model.p_AttachmentKindCarModels, ModelId]

    SpawnPlaceholderCar<private>():void=
        if (m_PlaceholderProp?) {
            return
        }
        if (Placeholder := p_Model.p_CarPlaceholder?) {
            SpawnRes := SpawnProp(Placeholder, p_CarMarker.GetTransform())
            set m_PlaceholderProp = SpawnRes(0)
        }

    DisposePlaceholderCar<private>():void=
        if (PlaceholderProp := m_PlaceholderProp?) {
            PlaceholderProp.Dispose()
            set m_PlaceholderProp = None()
        }

    CycleActiveCar<internal>(Forward:logic):void=
        Player := if (
            FirstAgent := p_AccessControl.ListAllowedAgents()[0]
            FirstPlayer := player[FirstAgent]
        ) {
            FirstPlayer
        } else {
            return
        }

        DisposePlaceholderCar()

        var Index: int = 0
        var SecludeFirst: logic = if (Current := m_ActiveCar?) {
            set Index = Current.Index
            false
        } else {
            true
        }

        StartIndex := Index
        ModelsLength := p_Model.p_CarModels.Length
        loop {
            if (not (SecludeFirst?)) {
                if (Forward?, Index < ModelsLength - 1) {
                    set Index += 1
                } else if (Forward?) {
                    set Index = 0
                } else if (Index > 0) {
                    set Index -= 1
                } else {
                    set Index = ModelsLength - 1
                }
                if (Index = StartIndex or (StartIndex < 0 and Index = 0)) {
                    # cycled back to current, nothing else is unlocked
                    return
                }
            }
            if (
                ModelConfig := p_Model.p_CarModels[Index]
                p_Model.p_ShowLockedVehicles? or IsCarUnlocked[Player, ModelConfig.Id]
            ) {
                break
            }
            set SecludeFirst = false
        }

        if (PrevProp := m_ActiveCar?.Prop?) {
            PrevProp.Dispose()
            set m_ActiveCar = None()
        }
        CarConfig := if (
            Next := p_Model.p_CarModels[Index]
            p_Model.p_ShowLockedVehicles? or IsCarUnlocked[Player, Next.Id]
        ) {
            Next
        } else {
            SpawnPlaceholderCar()
            SyncOpenMenu()
            return
        }

        Prop := if (p_CarMarker.IsValid[], IdleProp := CarConfig.IdleProp?) {
            SpawnProp(IdleProp, p_CarMarker.GetTransform())(0)
        } else {
            None()
        }
        Unlocked := if (IsCarUnlocked[Player, CarConfig.Id]) {
            true
        } else {
            false
        }
        ActiveCar := vehicle_garage_active_car{
            Index := Index
            Prop := Prop
            CarConfig := CarConfig
            Unlocked := Unlocked
        }
        set m_ActiveCar = Some(ActiveCar)
        SyncOpenMenu()

    SyncOpenMenu<private>():void=
        for (__->Instance:m_KnownMenuInstances) {
            Instance.SyncGoButton()
        }

    OnEnterButtonInteractedWith<private>(Agent:agent):void=
        if (not p_AccessControl.IsAllowed[Agent]) {
            return
        }
        if (m_VehicleSpawned?) {
            p_FangVehicleSpawner.DestroyVehicle()
        }
        DisposePlaceholderCar()
        if (
            Player := player[Agent]
            FortCharacter := Player.GetFortCharacter[]
            Menu := m_Menu?
        ) {
            p_PreviewCamera.AddTo(Agent)
            FortCharacter.Hide()

            OpenRes := Menu.Open(Player)
            if:
                Instance := vehicle_garage_menu_instance[OpenRes?]
                set m_KnownMenuInstances[Player] = Instance
        }

    CustomTrigger<override>(Action:string, Agent:?agent):logic=
        if (Action = "enable") {
            Enable()
            return true
        } else if (Action = "disable") {
            Disable()
            return true
        }
        false

garage_vehicle_mesh_provider<internal> := class(vehicle_mesh_provider):
    Device: vehicle_garage_instance_device

    GetIdlePropAsset<override>()<transacts><decides>:creative_prop_asset=
        Device.m_ActiveCar?.CarConfig.IdleProp?

    GetVFXMesh<override>()<transacts><decides>:visual_effect_powerup_device=
        Device.m_ActiveCar?.CarConfig.VFXMesh?

    GetVFXClearMesh<override>()<transacts><decides>:visual_effect_powerup_device=
        Device.p_Model.p_CarVFXMeshClear

vehicle_garage_menu := class(ui_menu):
    Device: vehicle_garage_instance_device

    CreateInstance<override>(Player:player):tuple(ui_menu_instance,canvas)=
        if (
            HudAccessControl := Device.p_HideHudAccessControl?
            HudAccessControl.IsAllowed[Player]
            set Device.m_PrevHudState[Player] = void
        ) {
            HudAccessControl.RemoveFromWhitelist(Player, false)
        }

        GoButton := CreateFortButton(fort_button_type.Loud, ToMessage(""))
        Instance := vehicle_garage_menu_instance{
            Menu := Self
            Player := Player
            Garage := Device
            GoButton := GoButton
        }

        ButtonSize := vector2{X := 40.0, Y := 600.0}
        LeftButton := CreateFortButton(fort_button_type.Regular, ToMessage("<"))
        LeftButton.OnClick().Subscribe(Instance.OnLeftClick)
        RightButton := CreateFortButton(fort_button_type.Regular, ToMessage(">"))
        RightButton.OnClick().Subscribe(Instance.OnRightClick)
        ScrollButtons := stack_box:
            Orientation := orientation.Horizontal
            Slots := array{
                stack_box_slot:
                    Widget := ApplySizeHint(LeftButton, ButtonSize)
                    HorizontalAlignment := horizontal_alignment.Left
                    VerticalAlignment := vertical_alignment.Top
                stack_box_slot:
                    Widget := ApplySizeHint(RightButton, ButtonSize)
                    HorizontalAlignment := horizontal_alignment.Left
                    VerticalAlignment := vertical_alignment.Top
                    Padding := margin{
                        Left := 1400.0
                    }
            }

        Instance.SyncGoButton()
        GoButton.OnClick().Subscribe(Instance.OnGoClick)
        CloseButton := CreateFortButton(fort_button_type.Regular, ToMessage("Back"))
        CloseButton.OnClick().Subscribe(Instance.OnCloseClick)
        ActionButtons := stack_box:
            Orientation := orientation.Vertical
            Slots := array{
                stack_box_slot:
                    Widget := ApplySizeHint(GoButton, vector2{Y := 40.0})
                    HorizontalAlignment := horizontal_alignment.Fill
                    VerticalAlignment := vertical_alignment.Top
                stack_box_slot:
                    Widget := ApplySizeHint(CloseButton, vector2{Y := 40.0})
                    HorizontalAlignment := horizontal_alignment.Fill
                    VerticalAlignment := vertical_alignment.Top
                    Padding := margin{
                        Top := 10.0
                    }
            }

        Widget := stack_box:
            Orientation := orientation.Vertical
            Slots := array{
                stack_box_slot:
                    Widget := ScrollButtons
                    HorizontalAlignment := horizontal_alignment.Center
                    VerticalAlignment := vertical_alignment.Top
                stack_box_slot:
                    Widget := ActionButtons
                    HorizontalAlignment := horizontal_alignment.Fill
                    VerticalAlignment := vertical_alignment.Top
                    Padding := margin{
                        Top := 20.0
                    }
            }

        Canvas := GetSimpleCanvas(Widget, ?Offsets := Some(margin{
            Top := 60.0
        }))
        (Instance, Canvas)

vehicle_garage_menu_instance := class(ui_menu_instance):
    Garage: vehicle_garage_instance_device
    GoButton: text_button_base

    OnClose<override>():void=
        set Garage.m_KnownMenuInstances = Garage.m_KnownMenuInstances.RemoveByKey(Player)(0)

        if (FortCharacter := Player.GetFortCharacter[]) {
            FortCharacter.Show()
        }
        Garage.p_PreviewCamera.RemoveFrom(Player)

        RemoveRes := Garage.m_PrevHudState.RemoveByKey(Player)
        set Garage.m_PrevHudState = RemoveRes(0)
        if (RemoveRes(1)?, HudAccessControl := Garage.p_HideHudAccessControl?) {
            HudAccessControl.AddToWhitelist(Player)
        }

    OnGoClick<internal>(Data:widget_message):void=
        Close()
        if (ActiveCar := Garage.m_ActiveCar?, CarConfig := ActiveCar.CarConfig) {
            if (not ActiveCar.Unlocked?) {
                Log("'{Player}' clicked 'Go' in garage but the car is not unlocked", ?Level := log_level.Debug)
                return
            }

            VehicleSpawner: vehicle_spawner_device = case(CarConfig.Kind):
                vehicle_garage_car_kind.Fang => Garage.p_FangVehicleSpawner

            case(CarConfig.Kind):
                vehicle_garage_car_kind.Fang => if (not (Garage.m_FangVehicleSpawnerEnabled?)) {
                    VehicleSpawner.Enable()
                    set Garage.m_FangVehicleSpawnerEnabled = true
                }

            VehicleSpawner.RespawnVehicle()
            spawn:
                EnterCar(CarConfig, VehicleSpawner)

            Log("'{Player}' is entering car \"{CarConfig.Id}\"", ?Level := log_level.Debug)
        } else {
            Log("'{Player}' clicked 'Go' in garage but there is no car to enter", ?Level := log_level.Debug)
        }

    EnterCar<private>(CarConfig:vehicle_garage_car_config, VehicleSpawner:vehicle_spawner_device)<suspends>:void=
        Sleep(0.5)
        VehicleSpawner.AssignDriver(Player)

    OnLeftClick<internal>(Data:widget_message):void=
        Garage.CycleActiveCar(false)

    OnRightClick<internal>(Data:widget_message):void=
        Garage.CycleActiveCar(true)

    SyncGoButton():void=
        Text := if (ActiveCar := Garage.m_ActiveCar?) {
            if (ActiveCar.Unlocked?) {
                Garage.p_Model.p_GoMenuButtonText
            } else {
                var LockedText: string = Garage.p_Model.p_LockedMenuButtonText
                for (Override:Garage.p_Model.p_LockedMenuButtonTextOverrides, Override.VehicleId = ActiveCar.CarConfig.Id) {
                    set LockedText = Override.ButtonText
                }
                LockedText
            }
        } else {
            "Empty"
        }
        GoButton.SetText(ToMessage(Text))
