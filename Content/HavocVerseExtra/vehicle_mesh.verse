using { /Fortnite.com/Devices }
using { /Fortnite.com/Devices/CreativeAnimation }
using { /Fortnite.com/Vehicles }
using { /Verse.org/Simulation }
using { /UnrealEngine.com/Temporary/Diagnostics }
using { /UnrealEngine.com/Temporary/SpatialMath }

using { HavocVerse.Debug }
using { HavocVerse.Util }

vehicle_mesh_entry_base := class<concrete>:
    @editable VFXMesh: ?visual_effect_powerup_device = false
    @editable IdleProp: ?creative_prop_asset = false

vehicle_mesh_entry_getaway := class<concrete>(vehicle_mesh_entry_base):
    @editable VehicleSpawner: vehicle_spawner_getaway_device = vehicle_spawner_getaway_device{}

vehicle_mesh_device := class(creative_device, aware_device):
    @editable p_UseIdlePropWhileDriving: logic = false
    @editable p_VFXMeshClear: ?visual_effect_powerup_device = false
    @editable p_FangEntries: []vehicle_mesh_entry_getaway = array{}

    DeviceMeta<override>()<transacts>:aware_device_meta=
        aware_device_meta{
            DeviceName := "vehicle_mesh"
        }

    OnBegin<override>()<suspends>:void=
        AwareInitDevice(Self, OnInit)

    OnInit<private>()<suspends>:void=
        for (Entry:p_FangEntries) {
            Provider := vehicle_mesh_entry{
                Device := Self
                Entry := Entry
            }
            MeshEntry := vehicle_mesh_spawner_entry{
                UseIdlePropWhileDriving := p_UseIdlePropWhileDriving
                Spawner := Entry.VehicleSpawner
                Provider := Provider
            }
            MeshEntry.Init()
        }

vehicle_mesh_entry<internal> := class<final>(vehicle_mesh_provider):
    Device: vehicle_mesh_device
    Entry: vehicle_mesh_entry_base

    GetIdlePropAsset<override>()<transacts><decides>:creative_prop_asset=
        Entry.IdleProp?

    GetVFXMesh<override>()<transacts><decides>:visual_effect_powerup_device=
        Entry.VFXMesh?

    GetVFXClearMesh<override>()<transacts><decides>:visual_effect_powerup_device=
        Device.p_VFXMeshClear?

vehicle_mesh_provider<public> := interface:
    GetIdlePropAsset<public>()<transacts><decides>:creative_prop_asset

    GetVFXMesh<public>()<transacts><decides>:visual_effect_powerup_device

    GetVFXClearMesh<public>()<transacts><decides>:visual_effect_powerup_device

vehicle_mesh_spawner_entry<public> := class<final>:
    UseIdlePropWhileDriving<public>: logic = false
    Spawner<public>: vehicle_spawner_device
    Provider<public>: vehicle_mesh_provider

    var m_Vehicle: ?fort_vehicle = false
    var m_Subscriptions: []cancelable = array{}
    var m_ActiveVFX: ?vehicle_mesh_active_entry_vfx = false
    var m_IdleState: ?vehicle_mesh_active_entry_idle_state = false

    Init<public>():void=
        set m_Subscriptions += array{
            Spawner.SpawnedEvent.Subscribe(OnVehicleSpawned)
            Spawner.DestroyedEvent.Subscribe(OnVehicleDestroyed)
            Spawner.AgentEntersVehicleEvent.Subscribe(OnAgentEntered)
            Spawner.AgentExitsVehicleEvent.Subscribe(OnAgentExited)
        }

    OnVehicleSpawned<private>(Vehicle:fort_vehicle):void=
        set m_Vehicle = Some(Vehicle)
        PassengerCount := Vehicle.GetPassengers().Length
        if (UseIdlePropWhileDriving? or PassengerCount = 0) {
            InitIdleState(Vehicle)
        }

    OnVehicleDestroyed<private>():void=
        set m_Vehicle = None()

        if (ActiveVFX := m_ActiveVFX?) {
            ActiveVFX.Cleanup()
            set m_ActiveVFX = None()
        }
        if (IdleState := m_IdleState?) {
            IdleState.Cleanup()
            set m_IdleState = None()
        }

    OnAgentEntered<private>(Agent:agent):void=
        if (IdleState := m_IdleState?, not UseIdlePropWhileDriving?) {
            IdleState.Cleanup()
            set m_IdleState = None()
        }
        if (Vehicle := m_Vehicle?) {
            PassengerCount := Vehicle.GetPassengers().Length
            if (PassengerCount = 1) {
                ApplyVFX(Agent)
            }
        }

    ApplyVFX<private>(Agent:agent):void=
        if (VFXMesh := Provider.GetVFXMesh[]) {
            VFXMesh.Pickup(Agent)
        }
        ActiveVFX := vehicle_mesh_active_entry_vfx{
            Agent := Agent
            VFXMeshClear := option{Provider.GetVFXClearMesh[]}
        }
        set m_ActiveVFX = Some(ActiveVFX)

    OnAgentExited<private>(Agent:agent):void=
        Vehicle := if (l_Vehicle := m_Vehicle?) {
            l_Vehicle
        } else {
            return
        }
        if (ActiveVFX := m_ActiveVFX?) {
            ActiveVFX.Cleanup()
            set m_ActiveVFX = None()
        }
        Passengers := Vehicle.GetPassengers()
        if (Passengers.Length = 0) {
            InitIdleState(Vehicle)
        } else if (
            AnotherPassenger := Passengers[0]
            PassengerAgent := AnotherPassenger.GetAgent[]
        ) {
            ApplyVFX(PassengerAgent)
        }

    InitIdleState<private>(Vehicle:fort_vehicle):void=
        if (m_IdleState?) {
            return
        }
        Transform := Vehicle.GetTransform()
        IdlePropRes := TrySpawnIdleProp(Transform)
        if (IdleProp := IdlePropRes?) {
            IdleState := vehicle_mesh_active_entry_idle_state{
                ActiveEntry := Self
                Prop := IdleProp
            }
            set m_IdleState = Some(IdleState)
            spawn:
                IdleState.StartLoop()
        }

    TrySpawnIdleProp<private>(Transform:transform):?creative_prop=
        if (IdlePropAsset := Provider.GetIdlePropAsset[]) {
            SpawnRes := SpawnProp(IdlePropAsset, Transform)
            if (IdleProp := SpawnRes(0)?) {
                return Some(IdleProp)
            }
        }
        None()

vehicle_mesh_active_entry_vfx<internal> := class<final>:
    Agent: agent
    VFXMeshClear: ?visual_effect_powerup_device

    Cleanup<internal>():void=
        if (l_VFXMeshClear := VFXMeshClear?) {
            l_VFXMeshClear.Pickup(Agent)
        }

vehicle_mesh_active_entry_idle_state<internal> := class<final>:
    ActiveEntry: vehicle_mesh_spawner_entry
    Prop: creative_prop

    AbortLoopEvent: event() = event(){}
    OverrideMoveToEvent: event() = event(){}

    StartLoop<internal>()<suspends>:void=
        race:
            AbortLoopEvent.Await()
            RunLoop()

    RunLoop<private>()<suspends>:void=
        loop {
            Vehicle := if (l_Vehicle := ActiveEntry.m_Vehicle?) {
                l_Vehicle
            } else {
                break
            }
            VehicleTransform := Vehicle.GetTransform()
            PropTransform := Prop.GetTransform()
            DistSq := DistanceSquared(VehicleTransform.Translation, PropTransform.Translation)
            if (DistSq >= 0.001, AnimationController := Prop.GetAnimationController[]) {
                DeltaLocation := if (DistSq >= 10.0) {
                    (VehicleTransform.Translation - PropTransform.Translation) * MakeUniformVec3(3.0)
                } else {
                    VehicleTransform.Translation - PropTransform.Translation
                }
                AnimationController.PlayNewAnimation(array{
                    keyframe_delta{
                        DeltaLocation := DeltaLocation
                        DeltaRotation := MakeShortestRotationBetween(PropTransform.Rotation, VehicleTransform.Rotation)
                        Time := 1.0
                    }
                })
            }
            Sleep(0.0)
        }

    Cleanup<internal>():void=
        AbortLoopEvent.Signal()
        Prop.Dispose()
