using { /Fortnite.com/Devices }
using { /Verse.org/Simulation }
using { /UnrealEngine.com/Temporary/Diagnostics }

using { HavocVerse.Debug }
using { HavocVerse.Devices }
using { HavocVerse.Util }

vfx_while_held_device := class(creative_device, aware_device, aware_debug_device):
    @editable p_DevId: string = "N/A"
    @editable p_AccessControl: ?access_control_device = false

    @editable p_CondButton: conditional_button_device = conditional_button_device{}
    @editable p_VFX: visual_effect_powerup_device = visual_effect_powerup_device{}
    @editable p_VFXClear: visual_effect_powerup_device = visual_effect_powerup_device{}

    var m_HoldingPlayers: [player]vfx_while_held_state = map{}

    DeviceMeta<override>()<transacts>:aware_device_meta=
        aware_device_meta{
            DeviceName := "vfx_while_held"
            DevId := p_DevId
        }

    DebugStats<override>()<transacts>:aware_debug_stats=
        aware_debug_stats{
            Length := map{
                "holdingPlayers" => m_HoldingPlayers.Length
            }
        }

    OnBegin<override>()<suspends>:void=
        AwareInitDevice(Self, OnInit)

    OnInit<private>()<suspends>:void=
        if (AccessControl := p_AccessControl?) {
            AccessControl.RegisterUsingDevice(Self)
        }

        GetPlayspace().PlayerRemovedEvent().Subscribe(OnPlayerRemoved)
        spawn:
            StartTickingLoop()

    OnPlayerRemoved<private>(Player:player):void=
        RemoveRes := m_HoldingPlayers.RemoveByKey(Player)
        set m_HoldingPlayers = RemoveRes(0)
        if (HeldState := RemoveRes(1)?) {
            HeldState.Unequip()
        }

    StartTickingLoop<private>()<suspends>:void=
        loop {
            Sleep(0.0)
            Agents := if (AccessControl := p_AccessControl?) {
                AccessControl.ListAllowedAgents()
            } else {
                GetPlayspace().GetPlayers()
            }
            for (Agent:Agents, Player := player[Agent]) {
                HoldingItem := if (p_CondButton.IsHoldingItem[Player]). true else false

                if (HoldingItem?, not m_HoldingPlayers[Player]) {
                    HeldState := vfx_while_held_state{
                        Device := Self
                        Player := Player
                    }
                    set m_HoldingPlayers = ConcatenateMaps(m_HoldingPlayers, map{
                        Player => HeldState
                    })
                    HeldState.Equip()
                } else if (not (HoldingItem?), HeldState := m_HoldingPlayers[Player]) {
                    HeldState.Unequip()
                }
            }
        }

vfx_while_held_state := class<final>:
    Device: vfx_while_held_device
    Player: player

    Equip<internal>():void=
        Device.p_VFX.Pickup(Player)

    Unequip<internal>():void=
        Device.p_VFXClear.Pickup(Player)
