using { /Fortnite.com/Devices }
using { /Verse.org/Simulation }
using { /Verse.org/Simulation/Tags }
using { /Verse.org/Assets }
using { /UnrealEngine.com/Temporary/Diagnostics }
using { /UnrealEngine.com/Temporary/SpatialMath }

using { HavocVerse.Debug }
using { HavocVerse.Devices }
using { HavocVerse.Persistence }
using { HavocVerse.Logging }
using { HavocVerse.UI }
using { HavocVerse.Util }

ui_integration_custom_font_config := class<concrete>:
    @editable FontName: string = ""
    @editable AssetsPath: string = ""
    @editable BaseFontSize: float = 24.0

havoc_verse_default_integration_device := class(creative_device, aware_integration, aware_device, aware_debug_device):
    @editable p_AwareNetConfig: aware_net_config = aware_net_config{}
    @editable p_GlobalUIConfig: global_ui_config = global_ui_config{}

    @editable p_AssetsProvider: ?creative_device = false
    @editable p_TagsProvider: ?creative_device = false

    @editable p_CustomFonts: []ui_integration_custom_font_config = array{}

    @editable p_PlayerSpawnSource: []exposed_player_spawn_device = array{}

    var m_Subscriptions: []cancelable = array{}
    var m_OnlinePlayers: [player]float = map{}

    var m_PlayerSpawnListeners: []type{__(:player):void} = array{}

    DeviceMeta<override>()<transacts>:aware_device_meta=
        aware_device_meta{
            DeviceName := "havoc_default_integration_device"
        }

    DebugStats<override>()<transacts>:aware_debug_stats=
        aware_debug_stats{
            Length := map{
                "subscriptions" => m_Subscriptions.Length
                "onlinePlayers" => m_OnlinePlayers.Length
                "listeners::playerSpawn" => m_PlayerSpawnListeners.Length
            }
        }

    GlobalDebugStats<override>()<transacts>:[string]aware_debug_stats=
        map{
            "_ AwareNet" => AwareNet_DebugStats()
            "_ Global Log" => GlobalLog_DebugStats()
            "_ Player data" => PlayerData_DebugStats()
            "_ Global UI" => GlobalUI_DebugStats()
            "_ UI Menus" => MenuUI_DebugStats()
        }

    OnBegin<override>()<suspends>:void=
        GlobalUISetup(GetPlayspace(), p_GlobalUIConfig)

        AssetsProvider := if (AssetsProviderDevice := p_AssetsProvider?) {
            if (AssetsProviderSource := assets_provider_device[AssetsProviderDevice]) {
                l_AssetsProvider := AssetsProviderSource.CreateAssetsProvider()
                Some(l_AssetsProvider)
            } else {
                Log("p_AssetsProvider of default integration device is pointing to a device that is not an assets provider source", ?Level := log_level.Error)
                None()
            }
        } else {
            None()
        }
        TagsProvider := if (TagsProviderDevice := p_TagsProvider?) {
            if (l_TagsProvider := tags_provider[TagsProviderDevice]) {
                Some(l_TagsProvider)
            } else {
                Log("p_TagsProvider of default integration device is pointing to a device that is not a tags provider", ?Level := log_level.Error)
                None()
            }
        } else {
            None()
        }

        var UICustomFonts: [string]custom_font = map{}
        if (l_AssetsProvider := AssetsProvider?) {
            for (FontConfig:p_CustomFonts) {
                CustomFont := CreateCustomFont(l_AssetsProvider, FontConfig.AssetsPath, FontConfig.BaseFontSize)
                set UICustomFonts = AddToMap(UICustomFonts, FontConfig.FontName, CustomFont)
            }
        }

        DefaultAssetsProvider := integration_assets_provider_wrapper{
            DeferTo := AssetsProvider
            UICustomFonts := UICustomFonts
        }
        DefaultTagsProvider := integration_tags_provider_wrapper{
            DeferTo := TagsProvider
        }

        set m_Subscriptions += array{
            GetPlayspace().PlayerAddedEvent().Subscribe(OnPlayerAdded),
            GetPlayspace().PlayerRemovedEvent().Subscribe(OnPlayerRemoved)
        }
        for (Player:GetPlayspace().GetPlayers()) {
            OnPlayerAdded(Player)
        }

        if (PlayerSpawnSource := p_PlayerSpawnSource[0]) {
            PlayerSpawnSource.SubscribeToSpawnEvent(OnPlayerSpawned)
        }

        AwareInitIntegration(p_AwareNetConfig, Some(Self), DefaultAssetsProvider, DefaultTagsProvider)

    OnEnd<override>():void=
        for (Subscription:m_Subscriptions) {
            Subscription.Cancel()
        }
        set m_Subscriptions = array{}
        set m_OnlinePlayers = map{}

    OnPlayerAdded<private>(Player:player):void=
        if. set m_OnlinePlayers[Player] = GetSimulationElapsedTime()
    
    OnPlayerRemoved<private>(Player:player):void=
        set m_OnlinePlayers = m_OnlinePlayers.RemoveByKey(Player)(0)

        MenuUI_HandlePlayerQuit(Player)
        PlayerData_HandlePlayerQuit(Player)

    OnPlayerSpawned<private>(Event:player_spawn_event):void=
        for (Listener:m_PlayerSpawnListeners) {
            Listener(Event.Player)
        }

    GetSimulationElapsedTimeAtJoin<override>(Player:player)<transacts>:float=
        if (JoinedAt := m_OnlinePlayers[Player]) {
            JoinedAt
        } else {
            0.0
        }

    SubscribeToPlayerSpawnEvent<override>(Listener:type{__(:player):void}):void=
        if (p_PlayerSpawnSource.Length <= 0) {
            Log("Subscribing to player spawn event but {Self} has no player spawn source set.", ?Level := log_level.Error)
        }
        set m_PlayerSpawnListeners += array{Listener}

integration_assets_provider_wrapper := class(assets_provider, ui_assets_provider):
    DeferTo: ?assets_provider 

    UICustomFonts: [string]custom_font

    GetTextureMetaAssets<override>()<transacts>:[]texture_metadata=
        if (Result := DeferTo?.GetTextureMetaAssets()) {
            Result
        } else {
            array{}
        }

    GetTextureAsset<override>(Name:string)<transacts><decides>:texture=
        if (Result := DeferTo?.GetTextureAsset[Name]) {
            Result
        } else {
            Log("Could not find texture asset \"{Name}\".", ?Level := log_level.Warning)
            None()?
        }

    GetTextureAssetDimensions<override>(Name:string)<transacts><decides>:vector2=
        if (Result := DeferTo?.GetTextureAssetDimensions[Name]) {
            Result
        } else {
            Log("Could not find dimensions of texture asset \"{Name}\".", ?Level := log_level.Warning)
            None()?
        }

    GetMaterialAsset<override>(Name:string)<transacts><decides>:material=
        if (Result := DeferTo?.GetMaterialAsset[Name]) {
            Result
        } else {
            Log("Could not find material asset \"{Name}\".", ?Level := log_level.Warning)
            None()?
        }

    GetMeshAsset<override>(Name:string)<transacts><decides>:mesh=
        if (Result := DeferTo?.GetMeshAsset[Name]) {
            Result
        } else {
            Log("Could not find mesh asset \"{Name}\".", ?Level := log_level.Warning)
            None()?
        }

    GetAnimationSequenceAsset<override>(Name:string)<transacts><decides>:animation_sequence=
        if (Result := DeferTo?.GetAnimationSequenceAsset[Name]) {
            Result
        } else {
            Log("Could not find animation sequence asset \"{Name}\".", ?Level := log_level.Warning)
            None()?
        }

    GetParticleSystemAsset<override>(Name:string)<transacts><decides>:particle_system=
        if (Result := DeferTo?.GetParticleSystemAsset[Name]) {
            Result
        } else {
            Log("Could not find particle system asset \"{Name}\".", ?Level := log_level.Warning)
            None()?
        }

    GetCustomFont<override>(Name:string)<transacts><decides>:custom_font=
        if (Result := UICustomFonts[Name]) {
            Result
        } else {
            Log("Could not find custom font \"{Name}\".", ?Level := log_level.Warning)
            None()?
        }

integration_tags_provider_wrapper := class(tags_provider):
    DeferTo: ?tags_provider

    GetTag<override>(Name:string)<transacts><decides>:tag=
        if (Result := DeferTo?.GetTag[Name]) {
            Result
        } else {
            Log("Could not find tag \"{Name}\".", ?Level := log_level.Warning)
            None()?
        }
