using { /Fortnite.com/UI }
using { /Verse.org/Simulation }
using { /UnrealEngine.com/Temporary/UI }

callback_agent_to_player_closure<public> := class:
    Callback<public>: type{__(:player):void}

    CallWithAgent<public>(Agent:agent):void=
        if (Player := player[Agent]) {
            Callback(Player)
        }

    CallOptAgent<public>(Agent:?agent):void=
        if (Player := player[Agent?]) {
            Callback(Player)
        }

# Verse generics are broken, this is needed to pass around Callbacks
button_click_callback_wrapper<public> := class:
    Callback<public>: type{__(:widget_message):void}

    Bind<public>(Button:text_button_base):void=
        Button.OnClick().Subscribe(OnClicked)

    OnClicked<private>(Data:widget_message):void=
        Callback(Data)

BindButtonCallback<public>(Button:text_button_base, Callback:type{__(:widget_message):void}):void=
    Closure := button_click_callback_wrapper{
        Callback := Callback
    }
    Closure.Bind(Button)

# NOTE: This is timing out session because Verse is a broken language,
#       and the previous attempt using a generic :ret type was crashing the editor lmao.
#       Leaving this commented out as a reminder to myself that generics are f'ed up
#       and sadly there isn't much I can do other than _closure types everywhere.
### Wrap void callbacks {{{
#WrapCallback<public>(
#    Callback:type{__(:t1):void},
#    Param1:t1 where t1:type
#):type{__():void}=
#    Closure := wrapped_callback_p1_void(t1){
#        Callback := Callback
#        Param1 := Param1
#    }
#    Closure.Execute
#
#wrapped_callback_p1_void<internal>(t1:type) := class:
#    Callback: type{__(:t1):void}
#    Param1: t1
#
#    Execute():void=
#        Callback(Param1)
#
#WrapCallback<public>(
#    Callback:type{__(:t1,:t2):void},
#    Param1:t1, Param2:t2 where t1:type, t2:type
#):type{__():void}=
#    Closure := wrapped_callback_p2_void(t1,t2){
#        Callback := Callback
#        Param1 := Param1
#        Param2 := Param2
#    }
#    Closure.Execute
#
#wrapped_callback_p2_void<internal>(t1:type, t2:type) := class:
#    Callback: type{__(:t1,:t2):void}
#    Param1: t1
#    Param2: t2
#
#    Execute():void=
#        Callback(Param1, Param2)
### }}}
#
### Wrap widget callbacks {{{
#WrapWidgetCallback<public>(
#    Callback:type{__(:t1):widget},
#    Param1:t1 where t1:type
#):type{__():widget}=
#    Closure := wrapped_callback_p1_widget(t1){
#        Callback := Callback
#        Param1 := Param1
#    }
#    Closure.Execute
#
#wrapped_callback_p1_widget<internal>(t1:type) := class:
#    Callback: type{__(:t1):widget}
#    Param1: t1
#
#    Execute():widget=
#        Callback(Param1)
#
#WrapWidgetCallback<public>(
#    Callback:type{__(:t1,:t2):widget},
#    Param1:t1, Param2:t2 where t1:type, t2:type
#):type{__():widget}=
#    Closure := wrapped_callback_p2_widget(t1,t2){
#        Callback := Callback
#        Param1 := Param1
#        Param2 := Param2
#    }
#    Closure.Execute
#
#wrapped_callback_p2_widget<internal>(t1:type, t2:type) := class:
#    Callback: type{__(:t1,:t2):widget}
#    Param1: t1
#    Param2: t2
#
#    Execute():widget=
#        Callback(Param1, Param2)
### }}}
