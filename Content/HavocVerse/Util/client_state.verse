using { /Verse.org/Simulation } 
using { /Fortnite.com/Devices } 
using { /Fortnite.com/Devices/CreativeAnimation } 
using { /Fortnite.com/Playspaces } 
using { /UnrealEngine.com/Temporary/SpatialMath }
using { /UnrealEngine.com/Temporary/Diagnostics }

using { HavocVerse.Debug }
using { HavocVerse.Logging }

mpc_handle<public> := class<final>:
	Context<public>: string
	NoSetupWarn<public>: logic = true
	ColdStartDelay<public>: float = 0.2

	var m_PlayerQuitSubscription<private>: ?cancelable = false
	var m_SeqControllerInstances<private>: []mpc_handle_seq_instance = array{}

	RegisterSeqControllerInstance<public>(SeqController:cinematic_sequence_device):void=
		Instance := mpc_handle_seq_instance{
			Index := m_SeqControllerInstances.Length
			SeqController := SeqController
		}
		set m_SeqControllerInstances += array{Instance}

	Setup<public>(Playspace:fort_playspace):void = Setup(Playspace, array{})
	Setup<public>(Playspace:fort_playspace, SeqControllers:[]cinematic_sequence_device):void=
		if (m_PlayerQuitSubscription?) {
			Log("Attempting to setup MPC handle ({Context}) more than once", ?Level := log_level.Error)
			return
		}
		Subscription := Playspace.PlayerRemovedEvent().Subscribe(OnPlayerQuit)
		set m_PlayerQuitSubscription = Some(Subscription)
		for (SeqController:SeqControllers) {
			RegisterSeqControllerInstance(SeqController)
		}

	OnPlayerQuit<private>(Player:player):void=
		ClearPlayer(Player)

	InitPlayer<public>(Player:player):void=
		if (not FindPlayerInstance[Player], NewInstance := FindFreeInstance[]) {
			InitPlayer0(Player, NewInstance, None())
		} else {
			Log("Unable to reserve MPC sequence controller ({Context}) for '{Player}': all {m_SeqControllerInstances.Length} are in use", ?Level := log_level.Error)
		}

	InitPlayer0<private>(Player:player, NewInstance:mpc_handle_seq_instance, StartingValue:?int):void=
		set NewInstance.m_Player = Some(Player)
		set NewInstance.m_Cold = true
		set NewInstance.m_ScheduledChange = StartingValue
		NewInstance.SeqController.SetPlayRate(0.0)
		SendValueFrame(NewInstance.SeqController, Player, 0)
		spawn:
			SchedulePostColdChange(Player, NewInstance)

		if (not m_PlayerQuitSubscription?, NoSetupWarn?) {
			Log("Using MPC handle ({Context}) without prior setup; ClearPlayer needs to be manually called", ?Level := log_level.Warning)
		}

	ClearPlayer<public>(Player:player):void=
		if (Instance := FindPlayerInstance[Player]) {
			set Instance.m_Player = None()
			set Instance.m_Cold = true
			set Instance.m_ScheduledChange = None()
		}

	SetValue<public>(Player:player, Value:int):void=
		if (Instance := FindPlayerInstance[Player]) {
			if (Instance.m_Cold?) {
				set Instance.m_ScheduledChange = Some(Value)
			} else {
				SendValueFrame(Instance.SeqController, Player, Value)
			}
		} else if (Player.IsActive[], NewInstance := FindFreeInstance[]) {
			InitPlayer0(Player, NewInstance, Some(Value))
		} else {
			Log("Unable to reserve MPC sequence controller ({Context}) for '{Player}': all {m_SeqControllerInstances.Length} are in use", ?Level := log_level.Error)
		}

	SchedulePostColdChange<private>(Player:player, Instance:mpc_handle_seq_instance)<suspends>:void=
		Sleep(ColdStartDelay)
		if (Instance.m_Player? = Player, NewValue := Instance.m_ScheduledChange?) {
			set Instance.m_Cold = false
			set Instance.m_ScheduledChange = None()
			SendValueFrame(Instance.SeqController, Player, NewValue)
		}
	
	SendValueFrame<private>(SeqController:cinematic_sequence_device, Player:player, Frame:int):void=
		SeqController.SetPlaybackFrame(Frame)
		SeqController.Play(Player)
		SeqController.Pause(Player)

	FindPlayerInstance<private>(Player:player)<decides><transacts>:mpc_handle_seq_instance=
		var Found: ?mpc_handle_seq_instance = None()
		for (Instance:m_SeqControllerInstances, not Found?, Instance.m_Player? = Player) {
			set Found = Some(Instance)
		}
		Found?

	FindFreeInstance<private>()<decides><transacts>:mpc_handle_seq_instance=
		var Found: ?mpc_handle_seq_instance = None()
		for (Instance:m_SeqControllerInstances, not Found?, not Instance.m_Player?) {
			set Found = Some(Instance)
		}
		Found?

mpc_handle_seq_instance<internal> := class<final>:
	Index: int
	SeqController: cinematic_sequence_device

	var m_Player: ?player = false
	var m_Cold: logic = true
	var m_ScheduledChange: ?int = false
