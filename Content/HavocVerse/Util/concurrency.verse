using { /Verse.org/Concurrency }

JoinAll<public>(TaskList:[]task(t) where t:type)<suspends>:void=
    for (T:TaskList) {
        T.Await()
    }
