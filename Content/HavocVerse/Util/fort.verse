using { /Fortnite.com/Devices }
using { /Fortnite.com/Characters }
using { /Verse.org/Simulation }
using { /Fortnite.com/Teams }
using { /Fortnite.com/Playspaces }
using { /UnrealEngine.com/Temporary/SpatialMath }

# Returns the team index of an agent.
# Reminder: Team indices are 1-based.
(TeamCollection:fort_team_collection).GetTeamIndex<public>(Agent:agent)<transacts><decides>:int=
    Team := TeamCollection.GetTeam[Agent]
    TeamCollection.GetTeams().Find[Team]+1

# Sets the team of an agent by team index.
# Reminder: Team indices are 1-based.
(TeamCollection:fort_team_collection).SetTeam<public>(InAgent:agent, NewTeamIndex:int)<transacts><decides>:void=
    Team := TeamCollection.GetTeams()[NewTeamIndex-1]
    TeamCollection.AddToTeam[InAgent, Team]

# Checks whether an agent is on a team by team index.
# Reminder: Team indices are 1-based.
(TeamCollection:fort_team_collection).IsOnTeam<public>(InAgent:agent, InTeamIndex:int)<transacts><decides>:void=
    Team := TeamCollection.GetTeams()[InTeamIndex-1]
    TeamCollection.IsOnTeam[InAgent, Team]

(Agent:agent).IsOnline<public>()<decides><transacts>:void=
    player[Agent].IsActive[]

(Trigger:trigger_device).MaybeTrigger<public>(Agent:?agent):void=
    if (l_Agent := Agent?) {
        Trigger.Trigger(l_Agent)
    } else {
        Trigger.Trigger()
    }

(FortCharacter:fort_character).StallingTeleportTo<public>(Position:vector3, Rotation:rotation)<decides><transacts>:void=
    FortCharacter.TeleportTo[Position, Rotation]
    FortCharacter.PutInStasis(stasis_args{
        AllowTurning := true
        AllowFalling := true
    })
    FortCharacter.ReleaseFromStasis()
