(InputMap:[k]v where k:subtype(comparable), v:type).GetOrDefault<public>(Key:k, Default:v)<transacts>:v=
    if (Value := InputMap[Key]) {
        Value
    } else {
        Default
    }

(Input:[]t where t:type).ElementOrDefault<public>(Index:int, Default:t)<transacts>:t=
    if (Value := Input[Index]) {
        Value
    } else {
        Default
    }

(Input:[]t where t:type, k:subtype(comparable)).FindBy<public>(Derive:type{__(:t)<transacts>:k}, Key:k)<decides><transacts>:t=
    var Result: ?t = None()
    loop {
        for (Element:Input, Derive(Element) = Key) {
            set Result = Some(Element)
            break
        }
        break
    }
    Result?

(InputMap:[k]v where k:subtype(comparable), v:type).RemoveByKey<public>(ElementToRemove:k)<transacts>:tuple([k]v,?v)=
    if (not InputMap[ElementToRemove]) {
        return (InputMap, false)
    }
    var FilteredMap: [k]v = map{}
    var Val: ?v = false
    for (Key->Value:InputMap) {
        if (Key = ElementToRemove) {
            set Val = Some(Value)
        } else {
            set FilteredMap = ConcatenateMaps(FilteredMap, map{Key => Value})
        }
    }
    (FilteredMap, Val)

operator'-'<public>(InputMap:[k]v, ElementToRemove:k where k:subtype(comparable), v:type)<transacts>:[k]v=
    if (not InputMap[ElementToRemove]) {
        return InputMap
    }
    var FilteredMap: [k]v = map{}
    for (Key->Value:InputMap, Key <> ElementToRemove) {
        set FilteredMap = ConcatenateMaps(FilteredMap, map{Key => Value})
    }
    FilteredMap

(InputMap:[k]v where k:subtype(comparable), v:type).Keys<public>()<transacts>:[]k=
    for (Key->__:InputMap) {
        Key
    }

(InputMap:[k]v where k:subtype(comparable), v:type).Values<public>()<transacts>:[]v=
    for (__->Value:InputMap) {
        Value
    }

(Input:[]t where t:subtype(comparable)).ToSet<public>()<transacts>:[t]void=
    var Set: [t]void = map{}
    for (Element:Input) {
        set Set = ConcatenateMaps(Set, map{Element => void})
    }
    Set

AddToMap<public>(Map:[k]v, Key:k, Value:v where k:subtype(comparable), v:type)<transacts>:[k]v=
    var NewMap: [k]v = Map
    if. set NewMap[Key] = Value
    NewMap

AddToLazyMap<public>(Map:?[k]v, Key:k, Value:v where k:subtype(comparable), v:type)<transacts>:[k]v=
    if (l_Map := Map?) {
        var NewMap: [k]v = l_Map
        if. set NewMap[Key] = Value
        NewMap
    } else {
        map{
            Key => Value
        }
    }

LazyMapUpsert<public>(Map:[k]v, Key:k, Provider:type{__(:k):v} where k:subtype(comparable), v:type):tuple([k]v,v)=
    if (Val := Map[Key]) {
        (Map, Val)
    } else {
        Val := Provider(Key)
        NewMap := AddToMap(Map, Key, Val) 
        (NewMap, Val)
    }

MapContains<public>(Map:[k]v, Key:k where k:subtype(comparable), v:type)<transacts>:logic=
    if (Map[Key]). true else false

(Input:[]t where t:type).Pop<public>(Index:int)<transacts><decides>:tuple([]t,t)=
    l_Index := if (Index < 0). Input.Length-1 else Index
    Res := if (Element := Input[l_Index], Lhs := Input.Slice[0, l_Index], Rhs := Input.Slice[l_Index+1, Input.Length]) {
        Some(Lhs + Rhs, Element)
    } else {
        None()
    }
    Res?

# Returns a new array with the elements reversed.
(Input:[]t where t:type).Reverse<public>()<computes>:[]t=
    for (Index->__:Input, Element := Input[Input.Length-Index-1]) {
        Element
    }

# Returns an array of arrays, each of which is a chunk of the original array.
(Input:[]t where t:type).Chunks<public>(ChunkSize:int)<computes>:[][]t=
    for (ChunkIndex := 0..Ceil(Input.Length / ChunkSize)-1) {
        StartIndex := ChunkIndex * ChunkSize
        EndIndex := Min(StartIndex + ChunkSize, Input.Length)
        if (Slice := Input.Slice[StartIndex, EndIndex]) {
            Slice
        } else {
            array{}
        }
    }

# Returns a slice of the input array, or the full array if there aren't too many elements.
(Input:[]t where t:type).EnforceMaxLength<public>(MaxLength:int)<computes>:[]t=
    for (Index->Element:Input, Index < MaxLength) {
        Element
    }

# Returns an array of a given length, filled with a repeated element.
MakeArray<public>(Length:int, Element:t where t:type)<computes>:[]t=
    for (Index := 1..Length) {
        Element
    }

# Returns an array with the elements of the input array repeated a given number of times.
(Input:[]t where t:type).Repeat<public>(Times:int)<computes>:[]t=
    Concatenate(for (__ := 1..Times). Input)

# Splits an array using an element as a delimiter.
(Input:[]t where t:subtype(comparable)).Split<public>(Delimiter:t, ?MaxSplits:int = 0)<transacts>:[][]t=
    var Result: [][]t = array{}
    var StartIndex: int = 0
    var SplitCount: int = 0
    for (Index := 0..Input.Length) {
        if (
            Index = Input.Length or (Input[Index] = Delimiter and (MaxSplits = 0 or SplitCount < MaxSplits))
            Substr := Input.Slice[StartIndex, Index]
        ) {
            set Result += array{Substr}
            set StartIndex = Index + 1
            set SplitCount += 1
        }
    }
    Result

# Splits an array only once.
# Fails if the delimiter is not found.
(Input:[]t where t:subtype(comparable)).SplitOnce<public>(Delimiter:t)<transacts><decides>:tuple([]t,[]t)=
    SplitRes := Input.Split(Delimiter, ?MaxSplits := 1)
    SplitRes.Length = 2
    (SplitRes[0], SplitRes[1])

# Splits an array using a sequence of elements as a delimiter.
(Input:[]t where t:subtype(comparable)).SeqSplit<public>(Delimiter:[]t, ?MaxSplits:int = 0)<transacts>:[][]t=
    PatStart := if (l_PatStart := Delimiter[0]) {
        l_PatStart
    } else {
        return array{Input}
    }

    var Result: [][]t = array{}
    var StartIndex: int = 0
    var SplitCount: int = 0

    var Index: int = 0
    loop {
        if (Element := Input[Index]) {
            if (Element = PatStart and (MaxSplits = 0 or SplitCount < MaxSplits)) {
                var IsDelimiter: logic = true
                for (PatIndex := 1..Delimiter.Length-1, IsDelimiter?, PatChar := Delimiter[PatIndex]) {
                    if (not (Input[Index + PatIndex] = PatChar)) {
                        set IsDelimiter = false
                    }
                }
                if (IsDelimiter?, Substr := Input.Slice[StartIndex, Index]) {
                    set Result += array{Substr}
                    set StartIndex = Index + Delimiter.Length
                    set Index += Delimiter.Length - 1 # +1 added by the loop
                    set SplitCount += 1
                }
            } else if (not Input[Index+1], Substr := Input.Slice[StartIndex, Index+1]) {
                set Result += array{Substr}
                set StartIndex = Index
                set SplitCount += 1
            }
        } else {
            break
        }
        set Index += 1
    }
    Result

# Splits an array using SeqSplit only once.
# Fails if the delimiter is not found.
(Input:[]t where t:subtype(comparable)).SeqSplitOnce<public>(Delimiter:[]t)<transacts><decides>:tuple([]t,[]t)=
    SplitRes := Input.SeqSplit(Delimiter, ?MaxSplits := 1)
    SplitRes.Length = 2
    (SplitRes[0], SplitRes[1])

(Input:[][]t where t:subtype(comparable)).Join<public>(Delimiter:[]t)<transacts>:[]t=
    var Result: []t = array{}
    for (Part:Input) {
        if (Result.Length > 0) {
            set Result += Delimiter
        }
        set Result += Part
    }
    Result

(Input:[]t where t:type).SafeSlice<public>(StartIndex:int, StopIndex:int)<transacts>:[]t=
    if (l_Slice := Input.Slice[StartIndex, Min(StopIndex, Input.Length)]) {
        l_Slice
    } else {
        array{}
    }

(Input:[]t where t:subtype(comparable)).StartsWith<public>(Value:[]t)<computes><decides>:void=
    Input.Slice[0, Value.Length] = Value

(Input:[]t where t:subtype(comparable)).EndsWith<public>(Value:[]t)<computes><decides>:void=
    Input.Slice[Input.Length-Value.Length, Input.Length] = Value

(Input:[]t where t:subtype(comparable)).StripPrefix<public>(Value:[]t)<computes><decides>:[]t=
    Input.Slice[0, Value.Length] = Value
    Input.Slice[Value.Length]

(Input:[]t where t:subtype(comparable)).StripSuffix<public>(Value:[]t)<computes><decides>:[]t=
    SuffixStart := Input.Length - Value.Length
    Input.Slice[SuffixStart, Input.Length] = Value
    Input.Slice[0, SuffixStart+1]

(Input:[]t where t:type, v:subtype(comparable)).Fold<public>(StartValue:v, AccFn:type{__(:v,:t)<transacts>:v})<transacts>:v=
    var Result: v = StartValue
    for (Element:Input) {
        set Result = AccFn(Result, Element)
    }
    Result

fAdd<public>(Acc:float, Value:float)<computes>:float = Acc + Value
iAdd<public>(Acc:int, Value:int)<computes>:int = Acc + Value
