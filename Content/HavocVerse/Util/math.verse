using { /UnrealEngine.com/Temporary/SpatialMath }

DefaultingMod<public>(X:int, Y:int, ?Default: int = 0)<transacts>:int=
    if (Res := Mod[X, Y]). Res else Default

DefaultingFloor<public>(Val:float, ?Default: int = 0)<transacts>:int=
    if (Res := Floor[Val]). Res else Default

DefaultingCeil<public>(Val:float, ?Default: int = 0)<transacts>:int=
    if (Res := Ceil[Val]). Res else Default

DefaultingRound<public>(Val:float, ?Default: int = 0)<transacts>:int=
    if (Res := Round[Val]). Res else Default

DivDefaultingFloor<public>(Q:int, Dividend:int, ?Default: int = 0)<transacts>:int=
    if (Res := Floor(Q / Dividend)). Res else Default

DivDefaultingCeil<public>(Q:int, Dividend:int, ?Default: int = 0)<transacts>:int=
    if (Res := Ceil(Q / Dividend)). Res else Default

ToFloat<public>(Val:int)<computes>:float = Val*1.0

MakeUniformVec2<public>(Val:float)<computes>:vector2 = vector2{
    X := Val, Y := Val
}

MakeUniformVec3<public>(Val:float)<computes>:vector3 = vector3{
    X := Val, Y := Val, Z := Val
}

Normalize<public>(Vec3:vector3)<transacts>:vector3=
    if (Res := Vec3.MakeUnitVector[]) {
        Res
    } else {
        vector3{X := 0.0, Y := 0.0, Z := 0.0}
    }

(InVec:vector2).WithX<public>(NewX:float)<transacts>:vector2=
    vector2{X := NewX, Y := InVec.Y}

(InVec:vector2).WithY<public>(NewY:float)<transacts>:vector2=
    vector2{X := InVec.X, Y := NewY}

(Vec3:vector3).GetAxisValue<public>(AxisIndex:int)<computes>:float=
    case(AxisIndex):
        0 => Vec3.X
        1 => Vec3.Y
        2 => Vec3.Z
        _ => 0.0

(Vec3:vector3).ReplaceAxisValue<public>(AxisIndex:int, Val:float)<transacts>:vector3=
    case(AxisIndex):
        0 => vector3{X := Val, Y := Vec3.Y, Z := Vec3.Z}
        1 => vector3{X := Vec3.X, Y := Val, Z := Vec3.Z}
        2 => vector3{X := Vec3.X, Y := Vec3.Y, Z := Val}
        _ => Vec3

(InVec:vector3).WithX<public>(NewX:float)<transacts>:vector3=
    vector3{X := NewX, Y := InVec.Y, Z := InVec.Z}

(InVec:vector3).WithY<public>(NewY:float)<transacts>:vector3=
    vector3{X := InVec.X, Y := NewY, Z := InVec.Z}

(InVec:vector3).WithZ<public>(NewZ:float)<transacts>:vector3=
    vector3{X := InVec.X, Y := InVec.Y, Z := NewZ}
    
(InVec:vector3).AddX<public>(DeltaX:float)<transacts>:vector3=
    vector3{X := InVec.X + DeltaX, Y := InVec.Y, Z := InVec.Z}

(InVec:vector3).AddY<public>(DeltaY:float)<transacts>:vector3=
    vector3{X := InVec.X, Y := InVec.Y + DeltaY, Z := InVec.Z}

(InVec:vector3).AddZ<public>(DeltaZ:float)<transacts>:vector3=
    vector3{X := InVec.X, Y := InVec.Y, Z := InVec.Z + DeltaZ}

(InTransform:transform).WithTranslation<public>(NewTranslation:vector3)<transacts>:transform=
    transform{
        Translation := NewTranslation
        Rotation := InTransform.Rotation
        Scale := InTransform.Scale
    }

(InTransform:transform).AddTranslation<public>(Delta:vector3)<transacts>:transform=
    transform{
        Translation := InTransform.Translation + Delta
        Rotation := InTransform.Rotation
        Scale := InTransform.Scale
    }

(InTransform:transform).WithX<public>(NewX:float)<transacts>:transform=
    transform{
        Translation := InTransform.Translation.WithX(NewX)
        Rotation := InTransform.Rotation
        Scale := InTransform.Scale
    }

(InTransform:transform).WithY<public>(NewY:float)<transacts>:transform=
    transform{
        Translation := InTransform.Translation.WithY(NewY)
        Rotation := InTransform.Rotation
        Scale := InTransform.Scale
    }

(InTransform:transform).WithZ<public>(NewZ:float)<transacts>:transform=
    transform{
        Translation := InTransform.Translation.WithZ(NewZ)
        Rotation := InTransform.Rotation
        Scale := InTransform.Scale
    }

(InTransform:transform).WithRotation<public>(NewRotation:rotation)<transacts>:transform=
    transform{
        Translation := InTransform.Translation
        Rotation := NewRotation
        Scale := InTransform.Scale
    }

(InTransform:transform).WithScale<public>(NewScale:vector3)<transacts>:transform=
    transform{
        Translation := InTransform.Translation
        Rotation := InTransform.Rotation
        Scale := NewScale
    }

(InTransform:transform).WithUniformScale<public>(NewScale:float)<transacts>:transform=
    transform{
        Translation := InTransform.Translation
        Rotation := InTransform.Rotation
        Scale := MakeUniformVec3(NewScale)
    }

(InTransform:transform).Rescale<public>(Rescale:float)<transacts>:transform=
    transform{
        Translation := InTransform.Translation
        Rotation := InTransform.Rotation
        Scale := InTransform.Scale * Rescale
    }

# Ensures a safe scale, that cannot make an object get removed
# by the engine for having a size of 0.0.
SafeScale<public>(Scale:vector3)<transacts>:vector3=
    vector3{
        X := Max(Scale.X, 0.00001)
        Y := Max(Scale.Y, 0.00001)
        Z := Max(Scale.Z, 0.00001)
    }
