using { /Verse.org/Simulation }

ToMessage<public><localizes>(Val:int): message = "{Val}"
ToMessage<public><localizes>(Text:string): message = "{Text}"
ToMessage<public><localizes>(Agent:agent): message = "{Agent}"

ConcatMessage<public><localizes>(A:message, B:message): message = "{A}{B}"
ConcatMessage<public><localizes>(A:string, B:message): message = "{A}{B}"
ConcatMessage<public><localizes>(A:message, B:string): message = "{A}{B}"

ConcatMessage<public><localizes>(A:message, B:message, C:message): message = "{A}{B}{C}"
ConcatMessage<public><localizes>(A:message, B:string, C:message): message = "{A}{B}{C}"
ConcatMessage<public><localizes>(A:message, B:message, C:string): message = "{A}{B}{C}"

EmptyMessage<public><localizes>():message = ""

ReplaceIntoMessage<public>(
    InTemplate:string,
    Replacements:[string]message,
    ?OpenDelimiter:string = "%",
    ?CloseDelimiter:string = "%",
    ?CommonEscapes:logic = true
):message=
    Lookup := replace_into_msg_map_lookup{
        Replacements := Replacements
    }
    ReplaceIntoMessageWithFn(InTemplate, Lookup.Replace,
        ?OpenDelimiter := OpenDelimiter, ?CloseDelimiter := CloseDelimiter,
        ?CommonEscapes := CommonEscapes)

replace_into_msg_map_lookup<internal> := class<final>:
    Replacements: [string]message

    Replace(VarName:string):?message=
        option{Replacements[VarName]}

ReplaceIntoMessageWithFn<public>(
    InTemplate:string,
    ReplaceFn:type{__(:string):?message},
    ?OpenDelimiter:string = "%",
    ?CloseDelimiter:string = "%",
    ?CommonEscapes:logic = true
):message=
    var Result: message = EmptyMessage()
    if (InTemplate.Length = 0) {
        return Result
    }

    Template := if (CommonEscapes?) {
        InTemplate
            .ReplaceAll(array{'\\', 'n'}, array{'\n'})
            .ReplaceAll(array{'\\', 't'}, array{'\t'})
    } else {
        InTemplate
    }
    var Buf: string = ""
    var Index: int = 0
    loop {
        if (Char := Template[Index]) {
            if (
                Char = OpenDelimiter[0]
                OpenDelimiter = Template.Slice[Index, Index+OpenDelimiter.Length]
            ) {
                set Index += OpenDelimiter.Length-1
                var VarName: string = ""
                var FoundCloseDelimiter: logic = false
                loop {
                    if (VarChar := Template[Index+1]) {
                        set Index += 1
                        if (
                            VarChar = CloseDelimiter[0]
                            CloseDelimiter = Template.Slice[Index, Index+CloseDelimiter.Length]
                        ) {
                            set Index += CloseDelimiter.Length-1
                            set FoundCloseDelimiter = true
                            break
                        } else {
                            set VarName += array{VarChar}
                        }
                    } else {
                        break
                    }
                }
                if (FoundCloseDelimiter?) {
                    ReplacementRes := ReplaceFn(VarName)
                    if (Replacement := ReplacementRes?) {
                        set Result = ConcatMessage(Result, Buf, Replacement)
                        set Buf = ""
                    } else {
                        set Buf += OpenDelimiter + VarName
                        if (FoundCloseDelimiter?) {
                            set Buf += CloseDelimiter
                        }
                    }
                } else {
                    set Buf += OpenDelimiter + VarName
                    if (FoundCloseDelimiter?) {
                        set Buf += CloseDelimiter
                    }
                }
            } else {
                set Buf += array{Char}
            }
        } else {
            break
        }
        set Index += 1
    }

    if (Buf.Length > 0) {
        set Result = ConcatMessage(Result, Buf)
    }

    Result
