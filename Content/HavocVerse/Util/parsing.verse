# Parses a string into an integer.
# Fails if the input is not a valid integer.
ParseInt<public>(Input:string, ?AllowNegative:logic = true)<transacts><decides>:int=
    var StartIndex: int = 0
    var IsNegative: logic = false
    if (AllowNegative? and Input[0] = '-') {
        set IsNegative = true
        set StartIndex += 1
    }
    Input.Length > StartIndex

    var Result: int = 0
    var OnlyValidChars: logic = true
    for (Index := StartIndex..Input.Length-1, OnlyValidChars?, Char := Input[Index]) {
        if (Digit := GetDigit[Char]) {
            set Result = Result*10 + Digit
        } else {
            set OnlyValidChars = false
        }
    }
    OnlyValidChars?

    if (IsNegative?) {
        -Result
    } else {
        Result
    }

# Parses a character into a digit.
GetDigit<public>(Char:char)<transacts><decides>:int=
    case(Char):
        '0' => 0
        '1' => 1
        '2' => 2
        '3' => 3
        '4' => 4
        '5' => 5
        '6' => 6
        '7' => 7
        '8' => 8
        '9' => 9

# Parses a string into a float.
# Fails if the input is not a valid float.
ParseFloat<public>(Input:string, ?AllowNegative:logic = true, ?OptionalDot:logic = true)<transacts><decides>:float=
    var StartIndex: int = 0
    var IsNegative: logic = false
    if (AllowNegative? and Input[0] = '-') {
        set IsNegative = true
        set StartIndex += 1
    }
    Input.Length > StartIndex

    var NextIndex: int = StartIndex
    var IntegerEndIndex: int = NextIndex
    loop {
        if (GetDigit[Input[NextIndex]]) {
            set NextIndex += 1
        } else {
            set IntegerEndIndex = NextIndex
            break
        }
    }

    IntegerPart := if (IntegerEndIndex > StartIndex) {
        ToFloat(ParseInt[Input.Slice[StartIndex, IntegerEndIndex]])
    } else {
        0.0
    }

    Result := if (Input[NextIndex] = '.') {
        set NextIndex += 1
        FactionalPart := ParseInt[Input.Slice[NextIndex, Input.Length]]
        IntegerPart + (ToFloat(FactionalPart) / Pow(10.0, ToFloat(Input.Length-NextIndex)))
    } else {
        OptionalDot?
        NextIndex >= Input.Length
        IntegerPart
    }

    if (IsNegative?) {
        -Result
    } else {
        Result
    }
