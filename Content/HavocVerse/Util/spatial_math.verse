using { /UnrealEngine.com/Temporary/SpatialMath }

IntersectLineWithSphere<public>(
    CurrentPosition:vector3, Direction:vector3,
    SphereCenter:vector3, SphereRadius:float
)<transacts><decides>:float=
    a := DotProduct(Direction, Direction)
    VecToCenter := CurrentPosition - SphereCenter
    b := 2 * DotProduct(Direction, VecToCenter)
    c := DotProduct(VecToCenter, VecToCenter) - (SphereRadius * SphereRadius)

    Discriminant := (b*b) - (4 * a * c)
    Discriminant >= 0.0

    t1 := (-b + Sqrt(Discriminant)) / (2 * a)
    t2 := (-b - Sqrt(Discriminant)) / (2 * a)
    t1 >= 0.0 or t2 >= 0.0

    if (t1 >= 0.0 and t2 >= 0.0) {
        Min(t1, t2)
    } else if (t1 >= 0.0) {
        t1
    } else {
        t2
    }
