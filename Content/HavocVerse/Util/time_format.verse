ToTimerMMSS<public>(Seconds:int)<transacts>:string=
    ToTimerMMSS(ToFloat(Seconds))

ToTimerMMSS<public>(Seconds:float)<transacts>:string=
    Minutes := DefaultingFloor(Seconds / 60.0)
    RemainingSeconds := DefaultingCeil(Seconds - (Minutes * 60.0))
    MM := if (Minutes > 9). ToString(Minutes) else "0{Minutes}"
    SS := if (RemainingSeconds > 9). ToString(RemainingSeconds) else "0{RemainingSeconds}"
    "{MM}:{SS}"

human_duration_parts<public> := class:
    Days<public>: logic = false
    Hours<public>: logic = false
    Minutes<public>: logic = false
    Seconds<public>: logic = false

human_duration_opts<public> := class:
    Delimiter<public>: string = " "
    DaysFmt<public>: tuple(string,string) = ("day", "days")
    HoursFmt<public>: tuple(string,string) = ("hr", "hrs")
    MinutesFmt<public>: tuple(string,string) = ("min", "mins")
    SecondsFmt<public>: tuple(string,string) = ("s", "s")

ToHumanDuration<public>(Seconds:float, Parts:human_duration_parts, Opts:human_duration_opts)<transacts>:string=
    var ResultParts: []string = array{}

    var w_Seconds: float = Abs(Seconds)
    l_Parts := for (
        PartConfig:array{
            (Parts.Days, Opts.DaysFmt, 86400.0),
            (Parts.Hours, Opts.HoursFmt, 3600.0),
            (Parts.Minutes, Opts.MinutesFmt, 60.0)
        }
        PartConfig(0)?
    ) {
        PartValue := DefaultingFloor(w_Seconds / PartConfig(2))
        set w_Seconds -= PartValue * PartConfig(2)
        Suffix := if (PartValue > 1). PartConfig(1)(1) else PartConfig(1)(0)
        set ResultParts += array{"{ToString(PartValue)}{Suffix}"}
    }
    if (Parts.Seconds?) {
        RemainValue := DefaultingCeil(w_Seconds)
        RemainSuffix := if (RemainValue > 1). Opts.SecondsFmt(1) else Opts.SecondsFmt(0)
        set ResultParts += array{"{ToString(RemainValue)}{RemainSuffix}"}
    }

    Result := ResultParts.Join(Opts.Delimiter)
    if (Seconds >= 0.0) {
        Result
    } else {
        "-{Result}"
    }

ToHumanDurationAutoScale<public>(Seconds:float, PartCount:int, Opts:human_duration_opts)<transacts>:string=
    var w_Seconds: float = Abs(Seconds)

    var ActiveParts: int = 0
    var Days: logic = false
    var Hours: logic = false
    var Minutes: logic = false
    var Secs: logic = false

    if (w_Seconds > 86400.0 or (PartCount >= 4 and ActiveParts = 0)) {
        set Days = true
        set ActiveParts += 1
        set w_Seconds -= DefaultingFloor(w_Seconds / 86400.0) * 86400.0
    }
    if (ActiveParts < PartCount, w_Seconds > 3600.0 or (PartCount >= 3 and ActiveParts = 0)) {
        set Hours = true
        set ActiveParts += 1
        set w_Seconds -= DefaultingFloor(w_Seconds / 3600.0) * 3600.0
    }
    if (ActiveParts < PartCount, w_Seconds > 60.0 or (PartCount >= 2 and ActiveParts = 0)) {
        set Minutes = true
        set ActiveParts += 1
        set w_Seconds -= DefaultingFloor(w_Seconds / 60.0) * 60.0
    }
    if (ActiveParts < PartCount, w_Seconds > 0.0 or ActiveParts = 0) {
        set Secs = true
    }

    Parts := human_duration_parts{
        Days := Days,
        Hours := Hours,
        Minutes := Minutes,
        Seconds := Secs
    }
    ToHumanDuration(Seconds, Parts, Opts)