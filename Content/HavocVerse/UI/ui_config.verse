using { /Verse.org/Simulation }
using { /UnrealEngine.com/Temporary/UI }
using { /UnrealEngine.com/Temporary/SpatialMath }
using { /Verse.org/Colors }
using { /Fortnite.com/UI }

using { HavocVerse.Util }

ui_canvas_placement_config<public> := class<concrete>:
    @editable Anchor<public>: vector2 = vector2{}
    @editable Offsets<public>: ui_margin_config = ui_margin_config{}
    @editable SizeToContent<public>: logic = true
    @editable Alignment<public>: vector2 = vector2{}
    @editable ZOrder<public>: type{_X:int where 0 <= _X, _X <= 2147483647} = 0

    ToCanvasSlot<public>(Widget:widget, ?IntuitiveOffsets:logic = true)<transacts>:canvas_slot=
        canvas_slot{
            Widget := Widget
            Anchors := anchors{
                Minimum := Anchor
                Maximum := Anchor
            }
            Offsets := Offsets.ToMargin(?IntuitiveOffsets := IntuitiveOffsets)
            SizeToContent := SizeToContent
            Alignment := Alignment
            ZOrder := ZOrder
        }

ui_margin_config<public> := class<concrete>:
    @editable Left<public>: float = 0.0
    @editable Top<public>: float = 0.0
    @editable Right<public>: float = 0.0
    @editable Bottom<public>: float = 0.0

    ToMargin<public>(?IntuitiveOffsets:logic = false, ?UniformScale:float = 1.0)<transacts>:margin=
        if (IntuitiveOffsets?) {
            margin{
                Left := (Left - Right) * UniformScale
                Top := (Top - Bottom) * UniformScale
            }
        } else {
            margin{
                Left := Left * UniformScale
                Top := Top * UniformScale
                Right := Right * UniformScale
                Bottom := Bottom * UniformScale
            }
        }

ui_text_style_config<public> := class<concrete>:
    @editable Color<public>: color = NamedColors.White
    @editable Justification<public>: text_justification = text_justification.Left

    @editable ShadowEnabled<public>: logic = true
    @editable ShadowOffset<public>: vector2 = vector2{X := 2.0, Y := 2.0}
    @editable ShadowColor<public>: color = NamedColors.Black
    @editable ShadowOpacity<public>: float = 0.65

    ToTextBlock<public>(DefaultText:string):text_block=
        ToTextBlock(ToMessage(DefaultText))

    ToTextBlock<public>(DefaultText:message):text_block=
        Widget := text_block{
            DefaultText := DefaultText
            DefaultTextColor := Color
            DefaultJustification := Justification
            DefaultShadowOffset := if (ShadowEnabled?). Some(ShadowOffset) else false
            DefaultShadowColor := ShadowColor
        }
        if (ShadowEnabled?) {
            Widget.SetShadowOpacity(0.65)
        }
        Widget
