using { /Verse.org/Assets }

using { /Verse.org/Colors }
using { /Verse.org/Simulation }
using { /UnrealEngine.com/Temporary/SpatialMath }
using { /UnrealEngine.com/Temporary/UI }
using { /Fortnite.com/UI }

using { HavocVerse.Debug }
using { HavocVerse.Util }

## Size Constraints {{{
FortButtonDefaultHeight<public>: float = 80.0

ApplySizeHint<public>(
    Widget:widget,
    SizeHint:vector2,
    ?HorizontalAlignment:horizontal_alignment = horizontal_alignment.Fill,
    ?VerticalAlignment:vertical_alignment = vertical_alignment.Fill
):widget=
    if (text_button_base[Widget], SizeHint.Y >= 10.0, SizeHint.Y <= 100.0, VerticalAlignment = vertical_alignment.Fill) {
        ApplyHeightConstraint(Widget, SizeHint, ?HorizontalAlignment := HorizontalAlignment)
    } else if (SizeHint.X > 0.0) {
        l_SizeHint := if (SizeHint.Y > 0.0) {
            SizeHint
        } else {
            vector2{X := SizeHint.X, Y := 1.0}
        }
        l_VerticalAlignment := if (SizeHint.Y > 0.0) {
            VerticalAlignment
        } else {
            vertical_alignment.Top
        }
        overlay{
            Slots := array{
                overlay_slot{
                    Widget := color_block{
                        DefaultDesiredSize := SizeHint
                        DefaultOpacity := 0.0
                    }
                    HorizontalAlignment := horizontal_alignment.Left
                    VerticalAlignment := vertical_alignment.Top
                }
                overlay_slot{
                    Widget := Widget
                    HorizontalAlignment := HorizontalAlignment
                    VerticalAlignment := l_VerticalAlignment
                }
            }
        }
    } else if (SizeHint.Y > 0.0) {
        ApplyHeightConstraint(Widget, SizeHint, ?HorizontalAlignment := HorizontalAlignment)
    } else {
        Widget
    }

ApplyHeightConstraint<public>(
    Widget:widget,
    SizeHint:vector2,
    ?HorizontalAlignment:horizontal_alignment = horizontal_alignment.Fill
):widget=
    YSizeHack := 0.5 * SizeHint.Y + 40.0
    stack_box{
        Orientation := orientation.Vertical
        Slots := array{
            stack_box_slot{
                Widget := color_block{
                    DefaultDesiredSize := SizeHint 
                    DefaultOpacity := 0.0
                }
                HorizontalAlignment := horizontal_alignment.Fill
                VerticalAlignment := vertical_alignment.Top
                Distribution := Some(1.0)
                Padding := margin{
                    Top := 0.0
                    Bottom := 0.0
                }
            }
            stack_box_slot{
                Widget := Widget
                HorizontalAlignment := HorizontalAlignment
                VerticalAlignment := vertical_alignment.Fill
                Distribution := Some(1.0)
                Padding := margin{
                    Top := -YSizeHack
                    Bottom := 0.0
                }
            }
        }
    }
## }}}

## Common Canvas {{{
GetSimpleCanvas<public>(
    Widget:widget,
    ?Anchor:vector2 = vector2{X := 0.5, Y := 0.5},
    ?Alignment:vector2 = vector2{X := 0.5, Y := 0.5},
    ?Offsets:?margin = false,
    ?ZOrder:type{_X:int where 0 <= _X, _X <= 2147483647} = 1000
):canvas=
    Canvas := canvas{
        Slots := array{
            canvas_slot{
                Anchors := anchors{
                    Minimum := Anchor
                    Maximum := Anchor
                }
                Offsets := Offsets.UnwrapOr(margin{})
                Alignment := Alignment
                SizeToContent := true
                ZOrder := ZOrder
                Widget := Widget
            }
        }
    }

## }}}

## Placeholders {{{
widget_placeholder<public> := class:
    Parent<public>: ?widget = false
    Overlay<public>: overlay = overlay{}
    HorizontalAlignment<public>: horizontal_alignment = horizontal_alignment.Fill
    VerticalAlignment<public>: vertical_alignment = vertical_alignment.Fill

    var m_Widget: ?widget = false

    ParentWidget<public>()<transacts>:widget=
        Parent? or Overlay

    GetWidget<public>()<decides><transacts>:widget=
        m_Widget?

    SetWidget<public>(Widget:widget):void=
        SetWidget(Some(Widget))

    SetWidget<public>(Widget:?widget):void=
        if (PrevWidget := m_Widget?) {
            Overlay.RemoveWidget(PrevWidget)
        }
        set m_Widget = Widget
        if (l_Widget := Widget?) {
            Overlay.AddWidget(overlay_slot{
                Widget := l_Widget
                HorizontalAlignment := HorizontalAlignment
                VerticalAlignment := VerticalAlignment
            })
        }

PlaceholderSizeHint<public>(
    SizeHint:vector2,
    StrictHeight:logic,
    ?HorizontalAlignment:horizontal_alignment = horizontal_alignment.Fill,
    ?VerticalAlignment:vertical_alignment = vertical_alignment.Fill
):widget_placeholder=
    Overlay := overlay{
        Slots := array{}
    }
    Parent := if (StrictHeight?) {
        ApplyHeightConstraint(Overlay, SizeHint, ?HorizontalAlignment := HorizontalAlignment)
    } else {
        Overlay.AddWidget(overlay_slot{
            Widget := color_block{
                DefaultDesiredSize := SizeHint
                DefaultOpacity := 0.0
            }
            HorizontalAlignment := horizontal_alignment.Center
            VerticalAlignment := vertical_alignment.Center
        })
        Overlay
    }
    widget_placeholder{
        Parent := Some(Parent)
        Overlay := Overlay
        HorizontalAlignment := HorizontalAlignment
        VerticalAlignment := VerticalAlignment
    }

## }}}

## Defaulting texture block {{{
defaulting_texture_block<public> := class:
    Placeholder<public>: widget_placeholder = widget_placeholder{}
    DynamicSize<public>: logic = false
    DynamicSizeScale<public>: float = 1.0

    var m_CachingTextureName: ??string = false
    var m_Size: ?vector2 = false

    Widget<public>()<transacts>:widget=
        Placeholder.Parent? or Placeholder.Overlay

    SetTexture<public>(Texture:texture):void=
        set m_CachingTextureName = None()
        SetTexture0(Some(Texture))

    SetTexture<public>(Texture:?texture):void=
        set m_CachingTextureName = None()
        SetTexture0(Texture)

    SetSize<public>(Size:vector2):void=
        set m_Size = Some(Size)

    SetTexture0<private>(Texture:?texture):void=
        if (l_Texture := Texture?) {
            if (TextureBlock := texture_block[Placeholder.GetWidget[]]) {
                TextureBlock.SetImage(l_Texture)
            } else {
                Placeholder.SetWidget(texture_block{
                    DefaultImage := l_Texture
                    DefaultDesiredSize := m_Size.UnwrapOr(vector2{})
                })
            }
        } else {
            Placeholder.SetWidget(None())
        }

    SetCachingTexture<public>(TextureName:?string):void=
        if (AssetsProvider := AwareDefaultAssetsProvider[]) {
            SetCachingTexture(AssetsProvider, TextureName)
        } else {
            set m_CachingTextureName = Some(TextureName)
        }

    SetCachingTexture<public>(AssetsProvider:assets_provider, TextureName:?string):void=
        if (PrevTextureName := m_CachingTextureName?, PrevTextureName = TextureName) {
            return
        }
        set m_CachingTextureName = Some(TextureName)
        if (Name := TextureName?, Texture := AssetsProvider.GetTextureAsset[Name]) {
            if (DynamicSize?, Dimensions := AssetsProvider.GetTextureAssetDimensions[Name]) {
                SetSize(Dimensions * DynamicSizeScale)
            }
            SetTexture0(Some(Texture))
        } else {
            SetTexture0(None())
        }

# Creates a defaulting texture block from a texture name and desired size.
CreateDefaultingTextureBlock<public>(
    TextureName:string,
    SizeHint:vector2,
    ?HorizontalAlignment:horizontal_alignment = horizontal_alignment.Fill,
    ?VerticalAlignment:vertical_alignment = vertical_alignment.Fill
):defaulting_texture_block=
    DefaultImage := option:
        TextureName.Length > 0
        AwareDefaultAssetsProvider[].GetTextureAsset[TextureName]
    CreateDefaultingTextureBlock(SizeHint, DefaultImage,
        ?HorizontalAlignment := HorizontalAlignment, ?VerticalAlignment := VerticalAlignment)

CreateDefaultingTextureBlock<public>(
    SizeHint:vector2,
    DefaultImage:?texture,
    ?HorizontalAlignment:horizontal_alignment = horizontal_alignment.Fill,
    ?VerticalAlignment:vertical_alignment = vertical_alignment.Fill
):defaulting_texture_block=
    Placeholder := PlaceholderSizeHint(SizeHint, false, ?HorizontalAlignment := HorizontalAlignment, ?VerticalAlignment := VerticalAlignment)
    TextureBlock := defaulting_texture_block{
        Placeholder := Placeholder
    }
    TextureBlock.SetTexture(DefaultImage)
    TextureBlock
## }}}

EmptyRect<public>(SizeHint:vector2):widget=
    color_block{
        DefaultDesiredSize := SizeHint
        DefaultOpacity := 0.0
    }

## Static images {{{
# Creates a texture block from a texture name.
# If the image is found, it will use its dimensions, otherwise it will be a 0x0 block.
CreateStaticImage<public>(
    TextureName:string,
    ?Scale:vector2 = vector2{X := 1.0, Y := 1.0},
    ?UniformScale:float = 1.0
)<transacts>:widget=
    if (
        TextureName.Length > 0
        AssetsProvider := AwareDefaultAssetsProvider[]
        Texture := AssetsProvider.GetTextureAsset[TextureName]
        Dimensions := AssetsProvider.GetTextureAssetDimensions[TextureName]
    ) {
        texture_block:
            DefaultImage := Texture
            DefaultDesiredSize := vector2{X := Dimensions.X * Scale.X * UniformScale, Y := Dimensions.Y * Scale.Y * UniformScale}
    } else {
        color_block:
            DefaultDesiredSize := vector2{X := 0.0, Y := 0.0}
            DefaultOpacity := 0.0
    }

# Creates a texture block from a texture name with a desired size.
CreateStaticImage<public>(TextureName:string, DesiredSize:vector2):widget=
    if (
        TextureName.Length > 0
        AssetsProvider := AwareDefaultAssetsProvider[]
        Texture := AssetsProvider.GetTextureAsset[TextureName]
    ) {
        texture_block:
            DefaultImage := Texture
            DefaultDesiredSize := DesiredSize
    } else {
        color_block:
            DefaultDesiredSize := DesiredSize
            DefaultOpacity := 0.0
    }

CreateOptionalImage<public>(ImageTexture:string, Size:vector2):?texture_block=
    if (AssetsProvider := AwareDefaultAssetsProvider[]) {
        CreateOptionalImage(AssetsProvider, ImageTexture, Size)
    } else {
        None()
    }

CreateOptionalImage<public>(AssetsProvider:assets_provider, ImageTexture:string, Size:vector2):?texture_block=
    if (
        ImageTexture.Length > 0
        Texture := AssetsProvider.GetTextureAsset[ImageTexture]
    ) {
        Some(texture_block{
            DefaultImage := Texture
            DefaultDesiredSize := Size
        })
    } else {
        None()
    }
## }}}

## Text blocks {{{
CreateTextBlock<public>(
    DefaultText:string,
    ?TextColor:color = NamedColors.White,
    ?Justification:text_justification = text_justification.Left
):text_block=
    CreateTextBlock(ToMessage(DefaultText), ?TextColor := TextColor, ?Justification := Justification)

CreateTextBlock<public>(
    DefaultText:message,
    ?TextColor:color = NamedColors.White,
    ?Justification:text_justification = text_justification.Left
):text_block=
    Widget := text_block{
        DefaultText := DefaultText
        DefaultTextColor := TextColor
        DefaultJustification := Justification
        DefaultShadowOffset := option{vector2{
            X := 2.0
            Y := 2.0
        }}
        DefaultShadowColor := NamedColors.Black
    }
    Widget.SetShadowOpacity(0.65)
    Widget
## }}}

## Cached text blocks {{{
MakeCachedTextBlock<public>(Widget:text_block)<transacts>:cached_text_block=
    cached_text_block{
        Widget := Widget
    }

MakeCachedTextBlock<public>(Widget:?text_block)<transacts>:?cached_text_block=
    option{cached_text_block{
        Widget := Widget?
    }}

cached_text_block<public> := class:
    Widget<public>: text_block

    var m_Text: ?string = false
    var m_TextColor: ?color = false

    SetText<public>(Text:string):void=
        if (PrevText := m_Text?, PrevText = Text) {
            return
        }
        Widget.SetText(ToMessage(Text))
        set m_Text = Some(Text)

    SetTextMessage<public>(Text:message):void=
        Widget.SetText(Text)
        set m_Text = None()

    SetTextColor<public>(Color:color):void=
        if (PrevColor := m_TextColor?, EqColor[PrevColor, Color]) {
            return
        }
        Widget.SetTextColor(Color)
        set m_TextColor = Some(Color)
## }}}

EqColor<public>(A:color, B:color)<transacts><decides>:void=
    A.R = B.R and A.G = B.G and A.B = B.B

## Fort Buttons {{{
fort_button_type<public> := enum:
    Regular
    Loud
    Quiet

(Button:text_button_base).GetButtonType<public>()<decides><transacts>:fort_button_type=
    if (button_regular[Button]) {
        fort_button_type.Regular
    } else if (button_loud[Button]) {
        fort_button_type.Loud
    } else if (button_quiet[Button]) {
        fort_button_type.Quiet
    }
    None()?

CreateFortButton<public>(ButtonType:fort_button_type, DefaultText:message):text_button_base=
    case(ButtonType):
        fort_button_type.Regular => button_regular{
            DefaultText := DefaultText
        }
        fort_button_type.Loud => button_loud{
            DefaultText := DefaultText
        }
        fort_button_type.Quiet => button_quiet{
            DefaultText := DefaultText
        }
## }}}

## Grid {{{
CreateFlexRow<public>(
    Row:[]widget,
    ?GapX:float = 0.0
):stack_box=
    StackSlots := for (ColumnIndex->Widget:Row) {
        stack_box_slot{
            Widget := Widget
            HorizontalAlignment := horizontal_alignment.Left
            VerticalAlignment := vertical_alignment.Top
            Padding := margin{
                Left := if (ColumnIndex > 0). GapX else 0.0
            }
        }
    }
    stack_box:
        Orientation := orientation.Horizontal
        Slots := StackSlots

CreateUIGrid<public>(
    Rows:[][]widget,
    ?GapX:float = 0.0,
    ?GapY:float = 0.0,
    ?AlignX:horizontal_alignment = horizontal_alignment.Center
):stack_box=
    GridSlots := for (RowIndex->Row:Rows) {
        StackSlots := for (ColumnIndex->Widget:Row) {
            stack_box_slot{
                Widget := Widget
                HorizontalAlignment := horizontal_alignment.Left
                VerticalAlignment := vertical_alignment.Top
                Padding := margin{
                    Left := if (ColumnIndex > 0). GapX else 0.0
                }
            }
        }
        stack_box_slot:
            Widget := stack_box:
                Orientation := orientation.Horizontal
                Slots := StackSlots
            HorizontalAlignment := AlignX
            VerticalAlignment := vertical_alignment.Top
            Padding := margin{
                Top := if (RowIndex > 0). GapY else 0.0
            }
    }
    stack_box:
        Orientation := orientation.Vertical
        Slots := GridSlots
## }}}

GetFixedWidgetDimensions<public>(Widget:widget)<transacts>:vector2=
    if (ColorBlock := color_block[Widget]) {
        ColorBlock.DefaultDesiredSize
    } else if (TextureBlock := texture_block[Widget]) {
        TextureBlock.DefaultDesiredSize
    } else {
        vector2{}
    }

MakeMargin<public>(X:float):margin=
    margin:
        Top := X
        Right := X
        Bottom := X
        Left := X

MakeMargin<public>(X:float, Y:float):margin=
    margin:
        Top := Y
        Right := X
        Bottom := Y
        Left := X

MakeOverlay<public>(Slots:[]overlay_slot):overlay=
    overlay:
        Slots := Slots
