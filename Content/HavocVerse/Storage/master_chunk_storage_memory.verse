using { /Verse.org/Simulation }
using { /UnrealEngine.com/Temporary/Diagnostics }

using { HavocVerse.Logging }
using { HavocVerse.Util }

# Storage memory implementation, using a master chunk for storage.
master_chunk_storage_memory<public> := class(storage_memory):
    MasterChunk<public>: storage_memory_chunk
    FixedLayout<public>: storage_memory_fixed_layout_config
    UniqueReservation<public>: logic = true

    ContextStartSentence<public>: string
    ContextInSentence<public>: string

    var m_Layout: ?storage_memory_fixed_layout = false
    var m_Chunks: []storage_memory_subchunk = array{}

    EnsureReady<override>():void=
        if (m_Layout?) {
            return
        }
        if (Layout := FixedLayout.Parse[]) {
            Log("{ContextStartSentence} read fixed layout with {Layout.Chunks.Length} chunks. Size: {Size()} bits")
            set m_Layout = Some(Layout)
            for (Chunk:Layout.Chunks) {
                var FoundOverlap: ?storage_memory_subchunk = false
                for (ExistingChunk:m_Chunks, not (FoundOverlap?)) {
                    if (ExistingChunk.Start < Chunk.End and Chunk.Start < ExistingChunk.End) {
                        set FoundOverlap = Some(ExistingChunk)
                    }
                }
                if (OverlappingCchunk := FoundOverlap?) {
                    Log("Found overlapping chunks when loading fixed layout of {ContextInSentence}: chunks \"{Chunk.LayoutId}\" and \"{OverlappingCchunk.LayoutId}\". Ignoring \"{Chunk.LayoutId}\".", ?Level := log_level.Error)
                } else {
                    set m_Chunks += array{storage_memory_subchunk{
                        Size := Chunk.End - Chunk.Start
                        Start := Chunk.Start
                        End := Chunk.End
                        ParentChunk := MasterChunk
                        LayoutId := Chunk.LayoutId
                    }}
                }
            }
        } else {
            Log("Could not parse fixed layout for {ContextInSentence}", ?Level := log_level.Error)
            set m_Layout = option{storage_memory_fixed_layout{
                Chunks := array{}
            }}
        }

    Size<override>()<computes>:int= MasterChunk.Size

    Reserve<override>(ChunkSize:int, LayoutId:string)<transacts><decides>:storage_memory_chunk=
        if (not (m_Layout?)) {
            Log("{ContextStartSentence} is not yet initialized (trying to reserve for chunk layout ID \"{LayoutId}\")", ?Level := log_level.Error)
            false?
        }
        var PreReservedChunk: ?storage_memory_subchunk = false
        for (l_Chunk:m_Chunks, not (PreReservedChunk?)) {
            if (l_Chunk.LayoutId = LayoutId) {
                set PreReservedChunk = Some(l_Chunk)
            }
        }
        if (Chunk := PreReservedChunk?) {
            if (Chunk.m_Reserved > 0) {
                if (UniqueReservation?) {
                    Log("Storage memory reservation conflict for {ContextInSentence}: chunk with layout ID \"{LayoutId}\" is already reserved", ?Level := log_level.Error)
                    false?
                } else {
                    Log("Storage memory reservation re-use for {ContextInSentence}: chunk with layout ID \"{LayoutId}\" now has {Chunk.m_Reserved+1} shared usages", ?Level := log_level.Verbose)
                }
            }
            if (Chunk.Size < ChunkSize) {
                Log("Storage memory reservation conflict for {ContextInSentence}: chunk with layout ID \"{LayoutId}\" requested {ChunkSize} bits, but fixed layout makes only {Chunk.Size} available", ?Level := log_level.Error)
                false?
            }
            set Chunk.m_Reserved += 1
            Chunk
        } else {
            var ReservedChunk: ?storage_memory_subchunk = false
            for (Gap:FindGaps(), not (ReservedChunk?)) {
                GapSize := Gap(1) - Gap(0)
                if (GapSize >= ChunkSize) {
                    Chunk := storage_memory_subchunk{
                        Size := ChunkSize
                        Start := Gap(0)
                        End := Gap(0) + ChunkSize
                        ParentChunk := MasterChunk
                        LayoutId := LayoutId
                    }
                    Log("Reserved unfixed memory chunk (layout ID: \"{LayoutId}\") from {ContextInSentence}. Update the fixed layout to ensure consistency", ?Level := log_level.Warning)
                    set m_Chunks += array{Chunk}
                    set ReservedChunk = Some(Chunk)
                }
            }
            if (not (ReservedChunk?)) {
                Log("Not enough space to reserve {ChunkSize} bits from {ContextInSentence} (chunk layout ID: \"{LayoutId}\")", ?Level := log_level.Error)
            }
            ReservedChunk?
        }

    FindGaps<internal>()<transacts>:[]tuple(int,int)=
        Capacity := Size()
        if (m_Chunks.Length = 0) {
            return array{(0, Capacity)}
        }

        var Result: []tuple(int,int) = array{}

        var SmallestStart: int = Capacity
        for (Chunk:m_Chunks) {
            if (Chunk.Start < SmallestStart) {
                set SmallestStart = Chunk.Start
            }
        }
        if (SmallestStart > 0) {
            set Result += array{(0, SmallestStart)}
        }

        var CurrentEnd: int = SmallestStart
        loop {
            if (CurrentEnd >= Capacity) {
                break
            }
            var NextStart: int = Capacity
            var NextEnd: int = Capacity
            for (Chunk:m_Chunks) {
                if (Chunk.Start >= CurrentEnd and Chunk.Start < NextStart) {
                    set NextStart = Chunk.Start
                    set NextEnd = Chunk.End
                }
            }

            if (NextStart > CurrentEnd) {
                set Result += array{(CurrentEnd, NextStart)}
            }
            set CurrentEnd = Max(CurrentEnd, NextEnd)
        }
        Result

    ToFixedLayout()<transacts>:storage_memory_fixed_layout=
        Res := for (Chunk:m_Chunks) {
            storage_memory_fixed_layout_chunk{
                LayoutId := Chunk.LayoutId
                Start := Chunk.Start
                End := Chunk.End
            }
        }
        storage_memory_fixed_layout{
            Chunks := Res.QuickSort(CmpStorageMemoryFixedLayoutChunk)
        }

    GetChunks<public>()<transacts>:[]storage_memory_subchunk=
        m_Chunks

    Dump<override>(Agent:agent, Detailed:logic):void=
        Capacity := Size()
        Log("-".Repeat(82))
        Log("| Storage Contents Dump - Capacity: {Capacity} bits")
        Log(MsgDumpAgentRow(Agent, ContextInSentence))

        if (Detailed?) {
            var DumpChunks: []storage_memory_dump_chunk = array{}
            for (Index->Chunk:m_Chunks) {
                set DumpChunks += array{storage_memory_dump_chunk{
                    ChunkIndex := Some(Index)
                    MemChunk := Chunk
                }}
            }
            for (Gap:FindGaps()) {
                set DumpChunks += array{storage_memory_dump_chunk{
                    ChunkIndex := false
                    MemChunk := storage_memory_subchunk{
                        Size := Gap(1) - Gap(0)
                        Start := Gap(0)
                        End := Gap(1)
                        ParentChunk := MasterChunk
                        LayoutId := ""
                    }
                }}
            }
            set DumpChunks = DumpChunks.QuickSort(CmpStorageMemoryDumpChunk)
            for (DumpChunk:DumpChunks) {
                Chunk := DumpChunk.MemChunk
                Log("|")
                if (ChunkIndex := DumpChunk.ChunkIndex?) {
                    Log("| Chunk #{ChunkIndex} ({Chunk.Start}..{Chunk.End}; {Chunk.End-Chunk.Start} bits) - Layout ID: {Chunk.LayoutId}")
                } else {
                    Log("| Free ({Chunk.Start}..{Chunk.End}; {Chunk.End-Chunk.Start} bits)")
                }
                DumpChunkBits(Chunk.Read(Agent), "> ")
            }
        }

        Log("|")
        Log("| Effective Layout: {ToFixedLayout().Serialize()}")
        Log("-".Repeat(82))

    DumpChunkBits<private>(Bits:[]logic, Prefix:string):void=
        BitsStr := BitsToString(Bits)
        for (StrChunk:BitsStr.Chunks(80)) {
            Log("{Prefix}{StrChunk}")
        }
        if (Bits.Length = 1, Bit := Bits[0]) {
            Log("{Prefix}As Boolean: {LogicToString(Bit)}")
        } else if (Bits.Length < 64) {
            Log("{Prefix}As Integer: {FromBits(Bits)} | Max Possible Integer: {FromBits(array{true}.Repeat(Bits.Length))}")
        }

    MsgDumpAgentRow<private><localizes>(Agent:agent, AtStr:string):message=
        "| @ {AtStr} for '{Agent}'"

storage_memory_subchunk<public> := class(storage_memory_bounded_chunk):
    ParentChunk<public>: storage_memory_chunk
    LayoutId<public>: string

    var m_Reserved: int = 0

    Read<override>(Agent:agent, ?Offset:int = 0, ?Limit:int = -1, ?Raw:logic = false):[]logic=
        ReadFrom := Start + Offset
        ReadTo := if (Limit < 0) {
            Start + Size
        } else {
            Min(Start + Size, ReadFrom + Limit)
        }
        ParentChunk.Read(Agent, ?Offset := ReadFrom, ?Limit := ReadTo - ReadFrom, ?Raw := Raw)

    Write<override>(Agent:agent, Data:[]logic, ?Offset:int = 0, ?Save:logic = true):void=
        OffsetStart := Start + Offset
        InBoundsData := if (Data.Length > Size - Offset) {
            if (Offset < Size, Sliced := Data.Slice[OffsetStart, OffsetStart+Size-Offset]) {
                Sliced
            } else {
                return
            }
        } else {
            Data
        }
        ParentChunk.Write(Agent, InBoundsData, ?Offset := OffsetStart)

storage_memory_dump_chunk := class:
    ChunkIndex: ?int
    MemChunk: storage_memory_subchunk

CmpStorageMemoryDumpChunk(A:storage_memory_dump_chunk, B:storage_memory_dump_chunk)<computes>:int=
    CmpInt(A.MemChunk.Start, B.MemChunk.Start)
