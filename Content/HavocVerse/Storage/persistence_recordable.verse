using { /Verse.org/Simulation }
using { /Fortnite.com/Devices }

using { HavocVerse.Util }

persistence_save_cause<public> := enum:
    Manual
    AutoSave
    AgentUntracked
    AgentGameQuit

persistence_load_query<public> := class:
    Overwrite<public>: logic
    FirstLoad<public>: logic

persistence_recordable_control<public> := class:
    var m_RecorderHandles: []persistence_recorder_recordable_handle = array{}

    Save<public>(Agent:agent, ?SaveCause:persistence_save_cause = persistence_save_cause.Manual):void=
        for (Handle:m_RecorderHandles) {
            Handle.Save(Agent, ?SaveCause := SaveCause)
        }

persistence_recordable_base<public> := interface:

persistence_recordable<public> := interface(persistence_recordable_base):
    GetPersistenceControl<public>()<transacts><decides>:persistence_recordable_control

    PersistenceReserve<public>(StorageMem:storage_memory)<transacts><decides>:persistence_recordable_state

    PersistenceSave<public>(State:persistence_recordable_state, Agent:agent, SaveCause:persistence_save_cause):void

    PersistenceLoad<public>(State:persistence_recordable_state, Agent:agent, Query:persistence_load_query):void

persistence_recordable_state<public> := interface:
    Reset<public>(Agent:agent):void

empty_persistence_recordable_state<public> := class(persistence_recordable_state):
    Reset<override>(Agent:agent):void=
        {}

persistence_reset_query<public> := class:
    SourceDevice<public>: creative_device
    ResetFilter<public>: ?type{__(:persistence_recordable_base)<transacts><decides>:void} = false

    ShouldReset<internal>(Recordable:persistence_recordable_base)<transacts><decides>:void=
        not (ResetFilter?) or (ResetFilter?)[Recordable]

persistence_resetable<public> := interface(persistence_recordable_base):
    PersistenceReset<public>(State:persistence_recordable_state, Agent:agent, Query:persistence_reset_query):void

ToString<public>(SaveCause:persistence_save_cause)<transacts>:string=
    case(SaveCause):
        persistence_save_cause.Manual => "Manual"
        persistence_save_cause.AutoSave => "AutoSave"
        persistence_save_cause.AgentUntracked => "AgentUntracked"
        persistence_save_cause.AgentGameQuit => "AgentGameQuit"

ToString<public>(Query:persistence_load_query)<transacts>:string=
    "\{Overwrite: {LogicToString(Query.Overwrite)}, FirstLoad: {LogicToString(Query.FirstLoad)}\}"
