using { /Fortnite.com/Devices }
using { /Fortnite.com/Characters }
using { /Verse.org/Simulation }
using { /UnrealEngine.com/Temporary/Diagnostics }

using { HavocVerse.Debug }
using { HavocVerse.Devices }
using { HavocVerse.Logging }
using { HavocVerse.Util }
using { HavocVerse.Util.Task }

persistence_recorder_player_tracking_mode<public> := enum:
    TrackList
    AccessControlWhitelist

persistence_recorder_save_mode<public> := enum:
    AccessControlAgents
    Individual

persistence_recorder_load_mode<public> := enum:
    FirstPlayer

persistence_recorder_save_slot := class:
    SharedStorage: shared_storage_chunk_device
    SavePoint: ?exposed_save_point_device

    MasterChunk: storage_memory_chunk

    BaseMemChunk: storage_memory_chunk
    BuyPlatesMemChunk: ?storage_memory_chunk
    var RecordDevices: []persistence_recorder_slot_record_device

persistence_recorder_recordable_handle<public> := class:
    PersistenceRecorder<public>: persistence_recorder_device
    HandleId<public>: int

    Save<public>(Agent:agent, ?SaveCause:persistence_save_cause = persistence_save_cause.Manual):void=
        if (
            SaveSlotIndex := PersistenceRecorder.m_AgentsSaveSlot[Agent]
            SaveSlot := PersistenceRecorder.m_SaveSlots[SaveSlotIndex]?
            RecordDevice := SaveSlot.RecordDevices[HandleId]
            ReservedState := RecordDevice.ReservedState?
            Recordable := persistence_recordable[RecordDevice.Device]
        ) {
            Recordable.PersistenceSave(ReservedState, Agent, SaveCause)
            Log("Saved agent persistence data for specific device {RecordDevice.AsString()} with handle {HandleId} ({SaveCause})", ?Level := log_level.Verbose)
        }

persistence_recorder_slot_record_device := class:
    Device: persistence_recordable_base
    ReservedState: ?persistence_recordable_state

    AsString<public>()<transacts>:string=
        if (CreativeDevice := creative_device[Device]) {
            ToString(CreativeDevice)
        } else {
            "<unknown_recordable>"
        }

player_save_slot_selected_event<public> := class:
    Player<public>: player
    SaveSlotIndex<public>: int
    Reset<public>: logic
    AlreadyPlayed<public>: logic

# Device that persistently store the state of devices for agents' progress.
persistence_recorder_device<public> := class(creative_device, aware_device, aware_debug_device, custom_device):
    @editable p_DevId<public>: string = "N/A"

    @editable p_PlayerTrackingMode<public>: persistence_recorder_player_tracking_mode = persistence_recorder_player_tracking_mode.TrackList
    @editable p_LoadMode<public>: persistence_recorder_load_mode = persistence_recorder_load_mode.FirstPlayer
    @editable p_SaveMode<public>: persistence_recorder_save_mode = persistence_recorder_save_mode.AccessControlAgents
    @editable p_AccessControl<public>: access_control_device = access_control_device{}
    @editable p_PlayerLoadDelayFromJoin<public>: type{_X:float where 0.1 <= _X, _X <= 120.0} = 0.5

    @editable p_StorageSaveSlots<public>: []shared_storage_chunk_device = array{}
    @editable p_StorageSaveSlotsMigrateFrom<public>: []shared_storage_chunk_device = array{}
    @editable p_SavePoints<public>: []exposed_save_point_device = array{}
    @editable p_AutoSaveInterval<public>: float = -1.0
    @editable p_OnlySaveValidPlayers<public>: logic = true

    @editable p_UseDefaultSelectMenu<public>: logic = true
    @editable p_UseDialogSelectMenu<public>: []persistence_recorder_popup_dialog_menu_config = array{}
    @editable p_DefaultSelectMenuConfig<public>: default_persistence_recorder_load_select_menu_config = default_persistence_recorder_load_select_menu_config{}
    @editable p_SkipSelectMenuOnFirstPlay<public>: logic = true

    # Devices whose status should be recorded.
    @editable p_DisableBuyplatesRecording<public>: logic = false
    @editable var p_BuyPlates: []buy_plate_device = array{}
    @editable p_CustomBuyPlatesReserveBits<public>: int = -1

    @editable p_BuyPlateRecoveryChunkSize<public>: int = -1
    @editable p_BuyPlateRecoveryChunkDelay<public>: float = 0.1

    @editable p_RecordDevices: []creative_device = array{} # gets discarded for m_RecordableDevices at runtime

    @editable p_UserActions: []persistence_recorder_user_action_config = array{}

    TracklistAgentAddedEvent<public>: event(agent) = event(agent){}
    TracklistAgentRemovedEvent<public>: event(agent) = event(agent){}

    PlayerSaveSlotSelectedEvent<public>: event(player_save_slot_selected_event) = event(player_save_slot_selected_event){}

    var m_RecordableDevices: []persistence_recordable_base = array{}

    var m_Initialized: logic = false
    var m_Tracklist: [agent]logic = map{}

    var m_LoadSelectMenu: ?persistence_recorder_load_select_menu = false
    var m_AgentsSaveSlot: [agent]int = map{}
    var m_SaveSlots: []?persistence_recorder_save_slot = array{}

    var m_KnownPlayersCount: int = 0

    var m_StartedAt: float = -1.0

    DeviceMeta<override>()<transacts>:aware_device_meta=
        aware_device_meta{
            DeviceName := "persistence_recorder"
            DevId := p_DevId
            PreferSequentialLoad := true
        }

    DebugStats<override>()<transacts>:aware_debug_stats=
        aware_debug_stats{
            Length := map{
                "recordableDevices" => m_RecordableDevices.Length
                "agentsSaveSlot" => m_AgentsSaveSlot.Length
                "saveSlots" => m_SaveSlots.Length
            }
        }

    OnBegin<override>()<suspends>:void=
        AwareInitDevice(Self, OnInit)

    OnInit<private>()<suspends>:void=
        if (m_Initialized?) {
            return
        }
        for (Index->Device:p_RecordDevices, not persistence_recordable_base[Device]) {
            Log("Persistence recorder device \"{p_DevId}\" is configured with a device that is not recordable (p_RecordDevices, index {Index})", ?Level := log_level.Error)
        }

        for (
            Index->SaveSlot:p_StorageSaveSlots
            MigrateFrom := p_StorageSaveSlotsMigrateFrom[Index]
            SaveSlot.Size() <> MigrateFrom.Size()
        ) {
            Log("{Self} is migrating")
        }

        var HandleIndex: int = 0
        set m_RecordableDevices += for (Device:p_RecordDevices, RecordDevice := persistence_recordable_base[Device]) {
            if (
                Recordable := persistence_recordable[RecordDevice]
                PersistenceControl := Recordable.GetPersistenceControl[]
            ) {
                set PersistenceControl.m_RecorderHandles += array{persistence_recorder_recordable_handle{
                    PersistenceRecorder := Self
                    HandleId := HandleIndex
                }}
            }
            set HandleIndex += 1
            RecordDevice
        }

        if (p_UseDefaultSelectMenu?) {
            if (DialogMenuConfig := p_UseDialogSelectMenu[0]) {
                set m_LoadSelectMenu = Some(persistence_recorder_popup_dialog_menu{
                    Config := DialogMenuConfig
                })
            } else if (AssetsProvider := AwareDefaultAssetsProvider[]) {
                set m_LoadSelectMenu = Some(default_persistence_recorder_load_select_menu{
                    AssetsProvider := AssetsProvider
                    Config := p_DefaultSelectMenuConfig
                })
            }
            spawn:
                InitMenu()
        }

        set m_SaveSlots = for (Index->SharedStorage:p_StorageSaveSlots) {
            SharedStorage.EnsureReady()
            SavePoint := option{p_SavePoints[Index]}
            option{ReserveSaveSlot[SharedStorage, SavePoint]}
        }
        set m_Initialized = true

        for (Index->BuyPlate:p_BuyPlates) {
            Closure := persistence_recorder_closure{
                Device := Self
                Index := Index
            }
            if (BuyPlate.SubscribeToEvent["unlock", Closure.Call]) {}
        }

        set m_StartedAt = GetSimulationElapsedTime()
        if (p_PlayerTrackingMode = persistence_recorder_player_tracking_mode.AccessControlWhitelist) {
            p_AccessControl.SubscribeToPlayerJoinEvents(OnAgentWhitelisted, OnAgentRemovedFromWhitelist)
        } else if (p_PlayerTrackingMode = persistence_recorder_player_tracking_mode.TrackList) {
            GetPlayspace().PlayerRemovedEvent().Subscribe(OnPlayerRemoved)
        }

        for (UserActionConfig:p_UserActions) {
            UserAction := persistence_recorder_user_action{
                Device := Self
                Config := UserActionConfig
            }
            UserActionConfig.OnTrigger.TriggeredEvent.Subscribe(UserAction.OnTriggered)
        }

        if (p_AutoSaveInterval > 0.0) {
            spawn:
                StartAutoSaveLoop(p_AutoSaveInterval)
        }

    InitMenu<private>()<suspends>:void=
        if (SelectMenu := m_LoadSelectMenu?) {
            SelectMenu.Init()
        }

    StartAutoSaveLoop<private>(Interval:float)<suspends>:void=
        loop {
            Sleep(Interval)
            for (Agent->SaveSlotIndex:m_AgentsSaveSlot, Agent.IsOnline[], SaveSlot := m_SaveSlots[SaveSlotIndex]?) {
                SaveAgent(Agent, SaveSlot, persistence_save_cause.AutoSave)
            }
        }

    OnEnd<override>():void=
        for (Agent->__:m_AgentsSaveSlot) {
            UnconsiderAgent(Agent, false)
        }

    RegisterDeviceToRecord<public>(Device:creative_device, ?Silent:logic = false):?persistence_recorder_recordable_handle=
        if (BuyPlate := buy_plate_device[Device]) {
            RegisterBuyPlateToRecord(BuyPlate)
        } else if (RecordDevice := persistence_recordable_base[Device]) {
            return Some(RegisterRecordable(RecordDevice))
        } else if (not (Silent?)) {
            Log("Persistence recorder device \"{p_DevId}\" is registering a device that is not recordable (RegisterDeviceToRecord)", ?Level := log_level.Error)
        }
        None()

    RegisterRecordable<public>(RecordDevice:persistence_recordable_base):persistence_recorder_recordable_handle=
        set m_RecordableDevices += array{RecordDevice}
        if (m_Initialized?) {
            for (SaveSlot:m_SaveSlots, l_SaveSlot := SaveSlot?) {
                ReservedState := option{persistence_recordable[RecordDevice].PersistenceReserve[l_SaveSlot.SharedStorage]}
                if (not (ReservedState?)) {
                    Log("Persistence recorder device \"{p_DevId}\" could not reserve storage for a record device", ?Level := log_level.Error)
                }
                set l_SaveSlot.RecordDevices += array{persistence_recorder_slot_record_device{
                    Device := RecordDevice
                    ReservedState := ReservedState
                }}
            }
        }
        Handle := persistence_recorder_recordable_handle{
            PersistenceRecorder := Self
            HandleId := m_RecordableDevices.Length - 1
        }
        if (
            Recordable := persistence_recordable[RecordDevice]
            PersistenceControl := Recordable.GetPersistenceControl[]
        ) {
            set PersistenceControl.m_RecorderHandles += array{Handle}
        }
        Handle

    RegisterBuyPlateToRecord<public>(BuyPlate:buy_plate_device, ?FixedIndex:int = -1):void=
        TargetIndex := if (FixedIndex >= 0) {
            FixedIndex
        } else {
            p_BuyPlates.Length
        }
        # fill in gaps if any
        for (__ := p_BuyPlates.Length..TargetIndex-1) {
            set p_BuyPlates += array{buy_plate_device{
                p_CustomData := array{"x-persistence-placeholder"}
            }}
        }
        if (TargetIndex = p_BuyPlates.Length) {
            set p_BuyPlates += array{BuyPlate}
        } else if (
            Prev := p_BuyPlates[TargetIndex]
            set p_BuyPlates[TargetIndex] = BuyPlate
        ) {
            if (not Prev.p_CustomData.Find["x-persistence-placeholder"]) {
                Log("Persistence recorder device \"{p_DevId}\" is overwriting a buyplate at index {TargetIndex}, {BuyPlate} replaces {Prev}", ?Level := log_level.Warning)
            }
        } else {
            Log("Persistence recorder device \"{p_DevId}\" could not register buyplate at index {TargetIndex}", ?Level := log_level.Error)
            return
        }
        if (m_Initialized?) {
            Closure := persistence_recorder_closure{
                Device := Self
                Index := TargetIndex
            }
            if. BuyPlate.SubscribeToEvent["unlock", Closure.Call]
        }

    OnBuyPlateUnlocked<internal>(Index:int, UnlockAgent:agent):void=
        if (not p_AccessControl.IsAllowed[UnlockAgent]) {
            return
        }
        AffectedAgents := case(p_SaveMode):
            persistence_recorder_save_mode.AccessControlAgents => p_AccessControl.ListAllowedAgents()
            persistence_recorder_save_mode.Individual => array{UnlockAgent}
        for (Agent:AffectedAgents, Player := player[Agent]) {
            if (
                SaveSlotIndex := m_AgentsSaveSlot[Agent]
                MemChunk := m_SaveSlots[SaveSlotIndex]?.BuyPlatesMemChunk?
            ) {
                MemChunk.Write(Player, array{true}, ?Offset := Index)
            }
        }

    ReserveSaveSlot<private>(SharedStorage:shared_storage_chunk_device, SavePoint:?exposed_save_point_device)<transacts><decides>:persistence_recorder_save_slot=
        BuyPlateBits := if (p_CustomBuyPlatesReserveBits >= 0) {
            p_CustomBuyPlatesReserveBits
        } else {
            p_BuyPlates.Length
        }
        persistence_recorder_save_slot{
            SharedStorage := SharedStorage
            SavePoint := SavePoint

            MasterChunk := SharedStorage.GetMemChunk[]

            BaseMemChunk := SharedStorage.Reserve[1, "base"]
            BuyPlatesMemChunk := if (not (p_DisableBuyplatesRecording?)) {
                Some(SharedStorage.Reserve[BuyPlateBits, "buyplates"])
            } else {
                None()
            }
            RecordDevices := for (RecordDevice:m_RecordableDevices, Recordable := persistence_recordable[RecordDevice]) {
                ReservedState := option{Recordable.PersistenceReserve[SharedStorage]}
                if (not (ReservedState?)) {
                    Log("Persistence recorder device \"{p_DevId}\" could not reserve storage for a record device", ?Level := log_level.Error)
                }
                persistence_recorder_slot_record_device{
                    Device := RecordDevice
                    ReservedState := ReservedState
                }
            }
        }

    OnAgentWhitelisted(Agent:player):void = ConsiderAgent(Agent)
    OnAgentRemovedFromWhitelist(Agent:player, GameQuit:logic):void = UnconsiderAgent(Agent, GameQuit)

    OnPlayerRemoved(Player:player):void = RemoveFromTracklist(Player, true)

    ConsiderAgent(Agent:agent):void=
        if (p_StorageSaveSlots.Length = 0) {
            return
        }
        if (p_LoadMode = persistence_recorder_load_mode.FirstPlayer and m_KnownPlayersCount > 0) {
            return
        }
        set m_KnownPlayersCount += 1
        spawn:
            LoadAgent(Agent)

    LoadAgent(Agent:agent)<suspends>:void=
        if (Player := player[Agent]) {
            # trackers need delay to load, how much exactly is unknown
            AwareEnsureDelaySincePlayerJoined(Player, p_PlayerLoadDelayFromJoin)
        }
        if (Menu := m_LoadSelectMenu?, Player := player[Agent]) {
            var AlreadyPlayed: logic = false
            SaveSlots := for (Index->i_SaveSlot:m_SaveSlots, SaveSlot := i_SaveSlot?) {
                Data := SaveSlot.BaseMemChunk.Read(Agent, ?Limit := 1)
                ContainsData := Data.ElementOrDefault(0, false)
                if (ContainsData?) {
                    set AlreadyPlayed = true
                }
                persistence_recorder_load_save_slot_state{
                    SaveSlotIndex := Index
                    ContainsData := ContainsData
                }
            }
            if (SaveSlots.Length = 0) {
                return
            }
            if (not (AlreadyPlayed?) and p_SkipSelectMenuOnFirstPlay) {
                SelectSave(Agent, 0, false, true)
            } else {
                Log(MsgOpenPrompt(p_DevId, Agent), ?Level := log_level.Debug)
                Closure := persistence_recorder_load_select_menu_closure{
                    Device := Self
                    Agent := Agent
                }
                Menu.Open(Player, SaveSlots, Closure.SelectSave)
            }
        } else {
            SelectSave(Agent, 0, false, true)
        }

    UnconsiderAgent(Agent:agent, GameQuit:logic):void=
        RemoveRes := m_AgentsSaveSlot.RemoveByKey(Agent)
        set m_AgentsSaveSlot = RemoveRes(0)
        set m_KnownPlayersCount -= 1
        if (SaveSlotIndex := RemoveRes(1)?, SaveSlot := m_SaveSlots[SaveSlotIndex]?) {
            SaveCause := if (GameQuit?) {
                persistence_save_cause.AgentGameQuit
            } else {
                persistence_save_cause.AgentUntracked
            }
            SaveAgent(Agent, SaveSlot, SaveCause)
        }

    SaveAgent<public>(Agent:agent, ?SaveCause:persistence_save_cause = persistence_save_cause.Manual):void=
        if (SaveSlotIndex := m_AgentsSaveSlot[Agent], SaveSlot := m_SaveSlots[SaveSlotIndex]?) {
            SaveAgent(Agent, SaveSlot, SaveCause)
        }

    SaveAgent<private>(Agent:agent, SaveSlot:persistence_recorder_save_slot, SaveCause:persistence_save_cause):void=
        if (SaveCause = persistence_save_cause.AgentGameQuit) {
            Log("Skipped saving agent persistence data on game quit.", ?Level := log_level.Debug)
            return
        }
        if (Player := player[Agent], not Player.IsActive[]) {
            Log(MsgSkippedSaveNotOnline(Agent), ?Level := log_level.Warning)
            return
        }
        if (FortCharacter := player[Agent].GetFortCharacter[], not FortCharacter.IsActive[]) {
            Log(MsgSkippedSaveNotActive(Agent), ?Level := log_level.Debug)
            return
        }
        var m_Saved: int = 0
        for (
            RecordDevice:SaveSlot.RecordDevices
            ReservedState := RecordDevice.ReservedState?
            Recordable := persistence_recordable[RecordDevice.Device]
        ) {
            Recordable.PersistenceSave(ReservedState, Agent, SaveCause)
            set m_Saved += 1
        }
        Log("Saved agent persistence data ({SaveCause}). {m_Saved} saved devices out of {SaveSlot.RecordDevices.Length} registered.", ?Level := log_level.Verbose)
        if (SavePoint := SaveSlot.SavePoint?, Player := player[Agent]) {
            SavePoint.SavePlayer(Player)
            Log("Saved agent persistence save point.", ?Level := log_level.Verbose)
        }

    MsgSkippedSaveNotOnline<private><localizes>(Agent:agent):message=
        "Skipped saving player '{Agent}' because they are no longer online."

    MsgSkippedSaveNotActive<private><localizes>(Agent:agent):message=
        "Skipped saving player '{Agent}' because they are inactive (dead or offline)."

    ResetAgent<public>(Agent:agent, Query:persistence_reset_query, ?ResetSavePoint:logic = false):logic=
        Log(MsgResetAgent(p_DevId, Agent), ?Level := log_level.Debug)
        if (SaveSlotIndex := m_AgentsSaveSlot[Agent], SaveSlot := m_SaveSlots[SaveSlotIndex]?) {
            if (BuyPlatesMemChunk := SaveSlot.BuyPlatesMemChunk?) {
                BuyPlatesMemChunk.Reset(Agent)
            }
            for (
                RecordDevice:SaveSlot.RecordDevices
                Query.ShouldReset[RecordDevice.Device]
            ) {
                if (CustomReset := persistence_resetable[RecordDevice.Device]) {
                    Log(MsgResetDevice(p_DevId, Agent, RecordDevice.AsString(), "yes"), ?Level := log_level.Debug)
                    if (ReservedState := RecordDevice.ReservedState?) {
                        CustomReset.PersistenceReset(ReservedState, Agent, Query)
                    } else {
                        CustomReset.PersistenceReset(empty_persistence_recordable_state{}, Agent, Query)
                    }
                } else if (
                    Recordable := persistence_recordable[RecordDevice.Device]
                    ReservedState := RecordDevice.ReservedState?
                ) {
                    Log(MsgResetDevice(p_DevId, Agent, RecordDevice.AsString(), "no"), ?Level := log_level.Debug)
                    ReservedState.Reset(Agent)
                    Recordable.PersistenceLoad(ReservedState, Agent, persistence_load_query{
                        Overwrite := true
                        FirstLoad := false
                    })
                }
            }
            if (ResetSavePoint?) {
                if (Player := player[Agent], SavePoint := SaveSlot.SavePoint?) {
                    SavePoint.ClearDataForPlayer(Player)
                    SavePoint.LoadPlayer(Player)
                }
            }
            return true
        }
        false

    MsgResetAgent<private><localizes>(DevId:string, Agent:agent):message=
        "Resetting persistence data of '{Agent}' for persistence recorder \"{DevId}\""
    MsgResetDevice<private><localizes>(DevId:string, Agent:agent, DeviceStr:string, CustomResetStr:string):message=
        "Resetting {DeviceStr} data for '{Agent}' (CustomReset: {CustomResetStr}) for persistence recorder \"{DevId}\""

    # Sets the load select menu. If no menu was set, the first save slot will always be used.
    SetLoadselectMenu<public>(Menu:persistence_recorder_load_select_menu):void=
        if (SelectMenu := m_LoadSelectMenu?) {
            SelectMenu.Discard()
        }
        Menu.Init()
        set m_LoadSelectMenu = Some(Menu)

    # Removes the load select menu.
    RemoveLoadselectMenu<public>():void=
        PrevSelectMenu := m_LoadSelectMenu
        set m_LoadSelectMenu = false
        if (SelectMenu := PrevSelectMenu?) {
            SelectMenu.Discard()
        }

    SelectSave<internal>(Agent:agent, SaveSlotIndex:int, Reset:logic, Automatic:logic):void=
        Log(MsgSelectSave(p_DevId, Agent, SaveSlotIndex, LogicToString(Reset), LogicToString(Automatic)), ?Level := log_level.Debug)
        if (not p_AccessControl.IsAllowed[Agent] or m_AgentsSaveSlot[Agent]) {
            return
        }
        ClampedSlotIndex := Min(Max(SaveSlotIndex, 0), 1)
        if (
            Player := player[Agent]
            SaveSlot := m_SaveSlots[ClampedSlotIndex]?
            set m_AgentsSaveSlot[Agent] = ClampedSlotIndex
        ) {
            Data := SaveSlot.BaseMemChunk.Read(Agent, ?Limit := 1)
            AlreadyPlayed := Data.ElementOrDefault(0, false)
            if (Reset?) {
                SaveSlot.MasterChunk.Reset(Player)
                if (SavePoint := SaveSlot.SavePoint?) {
                    SavePoint.ClearDataForPlayer(Player)
                }
            }
            SaveSlot.BaseMemChunk.Write(Player, array{true})
            if (AlreadyPlayed?, not (Reset?), SavePoint := SaveSlot.SavePoint?) {
                SavePoint.LoadPlayer(Player)
            }
            if (BuyPlatesMemChunk := SaveSlot.BuyPlatesMemChunk?) {
                spawn:
                    RecoverBuyplates(Player, BuyPlatesMemChunk)
            }
            var m_Loaded: int = 0
            for (
                RecordDevice:SaveSlot.RecordDevices
                ReservedState := RecordDevice.ReservedState?
                Recordable := persistence_recordable[RecordDevice.Device]
            ) {
                Recordable.PersistenceLoad(ReservedState, Player, persistence_load_query{
                    Overwrite := false
                    FirstLoad := true
                })
                set m_Loaded += 1
            }
            Log(MsgLoadedDevices(p_DevId, Agent, m_Loaded, SaveSlot.RecordDevices.Length), ?Level := log_level.Debug)
            Event := player_save_slot_selected_event{
                Player := Player
                SaveSlotIndex := ClampedSlotIndex
                Reset := Reset
                AlreadyPlayed := AlreadyPlayed
            }
            PlayerSaveSlotSelectedEvent.Signal(Event)
        }

    RecoverBuyplates<private>(Agent:player, MemChunk:storage_memory_chunk)<suspends>:void=
        MemData := MemChunk.Read(Agent)
        var TotalUnlocked: int = 0
        var CurrentChunkUnlocks: int = 0
        # try to enable buyplates following the enabling chains
        loop {
            var UnlockedCount: int = 0
            for (
                Index->BuyPlate:p_BuyPlates
                MemData[Index] = true
                BuyPlate.IsEnabled[]
            ) {
                if (p_BuyPlateRecoveryChunkSize > 0) {
                    if (CurrentChunkUnlocks >= p_BuyPlateRecoveryChunkSize) {
                        set CurrentChunkUnlocks = 0
                        Sleep(p_BuyPlateRecoveryChunkDelay)
                    }
                    set CurrentChunkUnlocks += 1
                }
                BuyPlate.Unlock(Agent, ?UserTriggered := false, ?FirstUnlock := false)
                set UnlockedCount += 1
                set TotalUnlocked += 1
            }
            if (UnlockedCount <= 0) {
                break
            }
        }
        # unlock remaining buyplates that are not part of an enabling chain
        for (
            Index->BuyPlate:p_BuyPlates
            MemData[Index] = true
            not BuyPlate.IsUnlocked[]
        ) {
            if (p_BuyPlateRecoveryChunkSize > 0) {
                if (CurrentChunkUnlocks >= p_BuyPlateRecoveryChunkSize) {
                    set CurrentChunkUnlocks = 0
                    Sleep(p_BuyPlateRecoveryChunkDelay)
                }
                set CurrentChunkUnlocks += 1
            }
            BuyPlate.Unlock(Agent, ?UserTriggered := false, ?FirstUnlock := false)
            set TotalUnlocked += 1
        }
        Log(MsgLoadedBuyplates(p_DevId, Agent, TotalUnlocked), ?Level := log_level.Debug)

    AddToTracklist<public>(Agent:agent):void=
        if (not m_Tracklist[Agent], set m_Tracklist[Agent] = true) {
            if (p_PlayerTrackingMode = persistence_recorder_player_tracking_mode.TrackList) {
                ConsiderAgent(Agent)
            }
            TracklistAgentAddedEvent.Signal(Agent)
        }

    RemoveFromTracklist<public>(Agent:agent, GameQuit:logic):void=
        RemoveResult := m_Tracklist.RemoveByKey(Agent)
        set m_Tracklist = RemoveResult(0)
        if (RemoveResult(1)?) {
            if (p_PlayerTrackingMode = persistence_recorder_player_tracking_mode.TrackList) {
                UnconsiderAgent(Agent, GameQuit)
            }
            TracklistAgentRemovedEvent.Signal(Agent)
        }

    ToggleFromTracklist<public>(Agent:agent):void=
        if (m_Tracklist[Agent]) {
            RemoveFromTracklist(Agent, false)
        } else {
            AddToTracklist(Agent)
        }

    CustomTrigger<override>(Action:string, Agent:?agent):logic=
        if (l_Agent := Agent?) {
            if (Action = "tracklist_add") {
                AddToTracklist(l_Agent)
            } else if (Action = "tracklist_remove") {
                RemoveFromTracklist(l_Agent, false)
            } else if (Action = "tracklist_toggle") {
                ToggleFromTracklist(l_Agent)
            } else {
                return false
            }
            true
        } else {
            false
        }

    MsgOpenPrompt<private><localizes>(DevId:string, Agent:agent):message=
        "Opening select menu for '{Agent}' for persistence recorder \"{DevId}\""
    MsgSelectSave<private><localizes>(DevId:string, Agent:agent, SaveSlotIndex:int, StrReset:string, StrAuto:string):message=
        "'{Agent}' selected save slot #{SaveSlotIndex} (reset: {StrReset}, auto: {StrAuto}) for persistence recorder \"{DevId}\""
    MsgLoadedBuyplates<private><localizes>(DevId:string, Agent:agent, TotalUnlocked:int):message=
        "Loaded persistence data of '{Agent}'. {TotalUnlocked} buyplates recovered from persistence recorder \"{DevId}\"."
    MsgLoadedDevices<private><localizes>(DevId:string, Agent:agent, Loaded:int, Total:int):message=
        "Loaded persistence data of '{Agent}'. {Loaded} loaded devices out of {Total} registered from persistence recorder \"{DevId}\"."

persistence_recorder_closure := class:
    Device: persistence_recorder_device
    Index: int

    Call(Agent:?agent):void=
        if (l_Agent := Agent?) {
            Device.OnBuyPlateUnlocked(Index, l_Agent)
        }

persistence_recorder_user_action_kind := enum:
    TracklistAdd
    TracklistRemove
    TracklistToggle

persistence_recorder_user_action_config := class<concrete>:
    @editable Action: persistence_recorder_user_action_kind = persistence_recorder_user_action_kind.TracklistAdd
    @editable OnTrigger: trigger_device = trigger_device{}

persistence_recorder_user_action := class:
    Device: persistence_recorder_device
    Config: persistence_recorder_user_action_config

    OnTriggered(Agent:?agent):void=
        if (l_Agent := Agent?) {
            if (Config.Action = persistence_recorder_user_action_kind.TracklistAdd) {
                Device.AddToTracklist(l_Agent)
            } else if (Config.Action = persistence_recorder_user_action_kind.TracklistRemove) {
                Device.RemoveFromTracklist(l_Agent, false)
            } else if (Config.Action = persistence_recorder_user_action_kind.TracklistToggle) {
                Device.ToggleFromTracklist(l_Agent)
            }
        }

persistence_recorder_load_select_menu_closure := class:
    Device: persistence_recorder_device
    Agent: agent

    SelectSave(SaveSlotIndex:int, Reset:logic):void=
        Device.SelectSave(Agent, SaveSlotIndex, Reset, false)
