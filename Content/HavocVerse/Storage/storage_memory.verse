using { /Verse.org/Simulation }

using { HavocVerse.Util }

# Persistent storage memory.
storage_memory<public> := interface:
    # Size in bits.
    Size<public>()<transacts>:int

    # Ensures the storage is ready.
    EnsureReady<public>():void

    # Reserves a chunk of memory.
    # `LayoutId` is used to consistently identify the chunk when a memory layout is set.
    # This cannot be "freed" as the written state is meant to be persistent.
    # Fails for various reasons, including if the storage is not ready. See `EnsureReady`.
    Reserve<public>(ChunkSize:int, LayoutId:string)<transacts><decides>:storage_memory_chunk

    # Dumps the storage memory state for an agent.
    Dump<public>(Agent:agent, Detailed:logic):void

storage_memory_device<public> := interface(storage_memory):
    GetID<public>()<transacts>:string

    Fill<public>(Agent:agent, Value:logic):void

storage_memory_chunk<public> := class<abstract>:
    # Size of the memory chunk in bits.
    Size<public>: int

    # Read raw data from the memory chunk.
    Read<public>(Agent:agent, ?Offset:int = 0, ?Limit:int = -1, ?Raw:logic = false):[]logic
    # Write raw data to the memory chunk.
    Write<public>(Agent:agent, Data:[]logic, ?Offset:int = 0):void

    # Read the memory chunk as a single integer.
    ReadInt<public>(Agent:agent, ?Offset:int = 0, ?Limit:int = -1):int=
        FromBits(Read(Agent, ?Offset := Offset, ?Limit := Limit))
    # Write a single integer to the memory chunk.
    WriteInt<public>(Agent:agent, Val:int, ?Offset:int = 0):void=
        Write(Agent, ToBits(Val, ?PadLength := Size - Offset), ?Offset := Offset)

    # Zeros out the memory chunk.
    Reset<public>(Agent:agent):void = Write(Agent, array{false}.Repeat(Size))

    EmptyState<protected>():[]logic=
        for (__ := 0..Size). false

storage_memory_bounded_chunk<public> := class<abstract>(storage_memory_chunk):
    Start<public>: int
    End<public>: int

# Fixed layout with assigned chunks and set sizes in storage memory.
# Layouts should be used for consistency.
storage_memory_fixed_layout<public> := class:
    Chunks<public>: []storage_memory_fixed_layout_chunk

    # Serializes a `storage_memory_fixed_layout` to its string representation.
    Serialize()<transacts>:string=
        var Result: string = ""
        for (Chunk:Self.Chunks) {
            ChunkId := Chunk.LayoutId.ReplaceAll("\"", "")
            set Result += "\"{ChunkId}\":{Chunk.Start}-{Chunk.End};"
        }
        Result

storage_memory_fixed_layout_chunk<public> := class:
    # Chunk layout ID.
    LayoutId<public>: string
    Start<public>: int
    End<public>: int

# Storage memory fixed layout device config.
storage_memory_fixed_layout_config<public> := class<concrete>:
    @editable Layout<public>: string = ""

    # Parses a `storage_memory_fixed_layout` from its string representation.
    Parse<public>()<transacts><decides>:storage_memory_fixed_layout=
        var Index: int = 0
        var l_Chunks: []storage_memory_fixed_layout_chunk = array{}
        var Failed: logic = false
        loop {
            if (Index >= Layout.Length) {
                break
            }
            if (ChunkRes := ParseChunk[Layout, Index]) {
                set l_Chunks += array{ChunkRes(0)}
                set Index = ChunkRes(1)
            } else {
                set Failed = true
                break
            }
        }
        not (Failed?)
        storage_memory_fixed_layout{
            Chunks := l_Chunks
        }

    ParseChunk<private>(Input:string, StartIndex:int)<transacts><decides>:tuple(storage_memory_fixed_layout_chunk,int)=
        var Index: int = StartIndex

        Input[Index] = '"'
        set Index += 1

        var LayoutId: string = ""
        loop {
            if (Char := Input[Index], Char <> '"') {
                set LayoutId += array{Char}
                set Index += 1
            } else {
                break
            }
        }
        Input[Index] = '"'
        Input[Index+1] = ':'
        set Index += 2

        var Data: string = ""
        loop {
            if (Char := Input[Index], Char <> ';') {
                set Data += array{Char}
                set Index += 1
            } else {
                set Index += 1
                break
            }
        }

        StartEndSplit := Data.SplitOnce['-']
        Start := ParseInt[StartEndSplit(0), ?AllowNegative := false]
        End := ParseInt[StartEndSplit(1), ?AllowNegative := false]
        Start >= 0
        End >= Start

        Chunk := storage_memory_fixed_layout_chunk{
            LayoutId := LayoutId
            Start := Start
            End := End
        }
        (Chunk, Index)

# Compare two `storage_memory_fixed_layout_chunk`s.
CmpStorageMemoryFixedLayoutChunk(A:storage_memory_fixed_layout_chunk, B:storage_memory_fixed_layout_chunk)<computes>:int=
    CmpInt(A.Start, B.Start)
