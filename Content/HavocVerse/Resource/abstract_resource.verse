using { /Fortnite.com/Devices }
using { /Verse.org/Simulation }
using { /UnrealEngine.com/Temporary/Diagnostics }

using { HavocVerse.Devices }
using { HavocVerse.Logging }
using { HavocVerse.Util }

virtual_resource<public> := class<abstract>(resource_base, transactful_resource):
    ResourceId<public>: string

    TryGive<public>(Agent:?agent, Amount:int)<decides><transacts>:void

    TryTake<public>(Agent:?agent, Amount:int)<decides><transacts>:void

    Give<override>(Agent:?agent, Amount:int):logic=
        if (TryGive[Agent, Amount]). true else false

    Take<override>(Agent:?agent, Amount:int):logic=
        if (TryTake[Agent, Amount]). true else false

# Virtual resource shared by all players.
global_virtual_resource<public> := class(virtual_resource):
    var m_Amount: int = 0

    TransactfulGetAmount<override>(Agent:?agent)<transacts>:int=
        m_Amount

    GetAmount<override>(Agent:?agent):int=
        m_Amount

    TryGive<override>(Agent:?agent, Amount:int)<decides><transacts>:void=
        set m_Amount += Amount
    
    TryTake<override>(Agent:?agent, Amount:int)<decides><transacts>:void=
        m_Amount >= Amount
        set m_Amount -= Amount

# Virtual resource supporting several balances, with abstraction
# to derive the storage key from the agent.
bound_virtual_resource<public> := class<abstract>(virtual_resource):
    # TODO : Turn this into a parametric class once mutable members get implemented...
    #        Verse is a joke of a language.

    var m_Balances<scoped {HavocVerse}>: [comparable]int = map{}

    # Returns the derivative storage key for a given agent.
    GetAgentKey(Agent:?agent)<transacts><decides>:comparable

    TransactfulGetAmount<override>(Agent:?agent)<transacts>:int=
        if (StorageKey := GetAgentKey[Agent], Balance := m_Balances[StorageKey]) {
            Balance
        } else {
            0
        }

    GetAmount<override>(Agent:?agent):int=
        if (StorageKey := GetAgentKey[Agent], Balance := m_Balances[StorageKey]) {
            Balance
        } else {
            0
        }

    TryGive<override>(Agent:?agent, Amount:int)<decides><transacts>:void=
        StorageKey := GetAgentKey[Agent]
        set m_Balances[StorageKey] = m_Balances.GetOrDefault(StorageKey, 0) + Amount

    TryTake<override>(Agent:?agent, Amount:int)<decides><transacts>:void=
        StorageKey := GetAgentKey[Agent]
        m_Balances[StorageKey] >= Amount
        set m_Balances[StorageKey] = m_Balances.GetOrDefault(StorageKey, 0) - Amount

per_player_virtual_resource<public> := class(bound_virtual_resource):
    GetAgentKey<override>(Agent:?agent)<transacts><decides>:comparable=
        player[Agent?]

using { /Fortnite.com/Teams }
per_team_virtual_resource<public> := class(bound_virtual_resource):
    TeamCollection<public>: fort_team_collection

    GetAgentKey<override>(Agent:?agent)<transacts><decides>:comparable=
        Player := player[Agent?]
        TeamCollection.GetTeam[Player]
