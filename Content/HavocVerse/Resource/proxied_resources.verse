using { /Fortnite.com/Devices }
using { /Verse.org/Simulation }
using { /UnrealEngine.com/Temporary/Diagnostics }

using { HavocVerse.Logging }

proxied_score_mode<public> := enum:
    Add
    Set

proxied_score_resource<public> := class(resource_base, transactful_resource):
    ResourceId<public>: string
    ScoreManager<public>: score_manager_device
    Mode<public>: proxied_score_mode = proxied_score_mode.Add

    TransactfulGetAmount<override>(Agent:?agent)<transacts>:int=
        if (l_Agent := Agent?) {
            ScoreManager.GetCurrentScore(l_Agent)
        } else {
            LogAgentlessOperation("GetAmount")
            -1
        }

    GetAmount<override>(Agent:?agent):int=
        if (l_Agent := Agent?) {
            ScoreManager.GetCurrentScore(l_Agent)
        } else {
            LogAgentlessOperation("GetAmount")
            -1
        }

    Give<override>(Agent:?agent, Amount:int):logic=
        if (l_Agent := Agent?) {
            PrevScoreAward := ScoreManager.GetScoreAward()
            case(Mode):
                proxied_score_mode.Add =>
                    ScoreManager.SetScoreAward(Amount)
                    ScoreManager.Activate(l_Agent)
                proxied_score_mode.Set =>
                    CurrentScore := ScoreManager.GetCurrentScore(l_Agent)
                    ScoreManager.SetScoreAward(CurrentScore + Amount)
                    ScoreManager.Activate(l_Agent)
            ScoreManager.SetScoreAward(PrevScoreAward)
            return true
        } else {
            LogAgentlessOperation("Give")
        }
        false

    Take<override>(Agent:?agent, Amount:int):logic=
        if (l_Agent := Agent?) {
            CurrentScore := ScoreManager.GetCurrentScore(l_Agent)
            if (CurrentScore < Amount) {
                return false
            }
            PrevScoreAward := ScoreManager.GetScoreAward()
            case(Mode):
                proxied_score_mode.Add =>
                    Log("Unable to take from a proxied score resource in Add mode", ?Level := log_level.Error)
                proxied_score_mode.Set =>
                    NewAmount := CurrentScore - Amount
                    ScoreManager.SetScoreAward(NewAmount)
                    ScoreManager.Activate(l_Agent)
                    ScoreManager.SetScoreAward(PrevScoreAward)
                    return true
        } else {
            LogAgentlessOperation("Take")
        }
        false

    LogAgentlessOperation<private>(Op:string)<transacts>:void=
        Log("Score proxied resources don't support agent-less {Op} (trace from {ResourceId})", ?Level := log_level.Warning)
