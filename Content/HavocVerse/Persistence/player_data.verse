using { /Verse.org/Simulation }
using { /UnrealEngine.com/Temporary/Diagnostics }

using { HavocVerse.Debug }
using { HavocVerse.Storage }
using { HavocVerse.Util }
using { HavocVerse.Logging }

var g_HavocVersePlayerData<internal>: weak_map(player,havoc_verse_player_data) = map{}

LogPlayerOffline<public>(OpName:string, Player:player)<transacts>:void=
    Log("Discarding operation \"{OpName}\" for '{Player}' because they are offline", ?Level := log_level.Warning)

havoc_verse_player_data<internal> := class<final><persistable>:
    Version: int = 1

    v0_DebugDevice: ?debug_device_persistent_player_data = false
    v0_RawBitsStores: [string]verse_bits_storage_data = map{}
    v0_Generic: [string]generic_player_data_backend = map{}

    v1_AttachmentKinds: [string]attachment_kind_player_data = map{}
    v1_LevelGroups: [string]level_group_player_data = map{}

MakeHavocVersePlayerData<internal><constructor>(
    CopyOf:havoc_verse_player_data
)<transacts> := havoc_verse_player_data:
    Version := CopyOf.Version

    v0_DebugDevice := CopyOf.v0_DebugDevice
    v0_RawBitsStores := CopyOf.v0_RawBitsStores
    v0_Generic := CopyOf.v0_Generic

    v1_AttachmentKinds := CopyOf.v1_AttachmentKinds
    v1_LevelGroups := CopyOf.v1_LevelGroups

GetPlayerData<internal>(Player:player)<transacts>:havoc_verse_player_data=
    (Player.IsActive[]; g_HavocVersePlayerData[Player]) or havoc_verse_player_data{}

SetPlayerData<internal>(Player:player, NewPlayerData:havoc_verse_player_data)<transacts>:void=
    if (Player.IsActive[]) {
        if. set g_HavocVersePlayerData[Player] = NewPlayerData
    } else {
        Log("Won't save Verse player data for '{Player}' because they are offline", ?Level := log_level.Warning)
    }

## v0 {{{
### Debug Device {{{
PlayerDataGetDebugDevice<public>(Player:player)<transacts><decides>:debug_device_persistent_player_data=
    PlayerData := GetPlayerData(Player)
    PlayerData.v0_DebugDevice?

PlayerDataSetDebugDevice<public>(Player:player, NewData:?debug_device_persistent_player_data)<transacts>:?debug_device_persistent_player_data=
    PlayerData := GetPlayerData(Player)
    SetPlayerData(Player, havoc_verse_player_data{
        MakeHavocVersePlayerData<constructor>(PlayerData)
        v0_DebugDevice := NewData
    })
    NewData
### }}}

### Raw Bits Storage {{{
PlayerDataGetRawBitsStore<public>(Player:player, StoreId:string)<transacts><decides>:verse_bits_storage_data=
    PlayerData := GetPlayerData(Player)
    PlayerData.v0_RawBitsStores[StoreId]

PlayerDataGetRawBitsStoreOrDefault<public>(Player:player, StoreId:string)<transacts><decides>:verse_bits_storage_data=
    PlayerDataGetRawBitsStore[Player, StoreId] or verse_bits_storage_data{}

PlayerDataSetRawBitsStore<public>(Player:player, StoreId:string, NewData:verse_bits_storage_data)<transacts>:verse_bits_storage_data=
    PlayerData := GetPlayerData(Player)
    SetPlayerData(Player, havoc_verse_player_data{
        MakeHavocVersePlayerData<constructor>(PlayerData)
        v0_RawBitsStores := AddToMap(PlayerData.v0_RawBitsStores, StoreId, NewData)
    })
    NewData
### }}}

### Generic Documents {{{
generic_player_data_backend<internal> := class<final><persistable>:
    Version: int = 0

    v0_Values: [string]generic_player_data_backend_value = map{}
    v0_Children: [string]generic_player_data_backend = map{}

generic_player_data_backend_value<internal> := class<final><persistable>:
    I: ?int = false
    F: ?float = false
    S: ?string = false
    Li: ?[]generic_player_data_backend_value = false

MakeGenericPlayerDataBackend<internal><constructor>(
    CopyOf:generic_player_data_backend
)<transacts> := generic_player_data_backend:
    Version := CopyOf.Version

    v0_Values := CopyOf.v0_Values
    v0_Children := CopyOf.v0_Children

var g_LocalGenericCache<internal>: weak_map(session,[player]player_data_document) = map{}

PlayerData_DebugStats<scoped {HavocVerse}>()<transacts>:aware_debug_stats=
    var Result: [string]int = map{}
    if (LocalGenericCache := g_LocalGenericCache[GetSession()]) {
        set Result = ConcatenateMaps(Result, map{
            "players" => LocalGenericCache.Length
        })
    }
    aware_debug_stats{
        Length := Result
    }

PlayerData_HandlePlayerQuit<scoped {HavocVerse}>(Player:player):void=
    if (LocalGenericCache := g_LocalGenericCache[GetSession()]) {
        NewLocalGenericCache := LocalGenericCache.RemoveByKey(Player)(0)
        if. set g_LocalGenericCache[GetSession()] = NewLocalGenericCache
    }

PlayerDataGetDocument<public>(
    Player:player, TopLevelKey:string,
    ?IgnoreCache:logic = false
)<transacts>:player_data_document=
    var LocalGenericCache: [player]player_data_document = g_LocalGenericCache[GetSession()] or (
        l_LocalGenericCache := map{}
        if. set g_LocalGenericCache[GetSession()] = l_LocalGenericCache
        l_LocalGenericCache
    )
    if (not IgnoreCache?, Cached := LocalGenericCache[Player]) {
        return Cached
    }
    PlayerData := GetPlayerData(Player)
    Backend := PlayerData.v0_Generic[TopLevelKey] or generic_player_data_backend{}
    Document := player_data_document{
        Player := Some(Player)
        TopLevelKey := Some(TopLevelKey)
        m_Backend := Backend
    }
    if:
        set LocalGenericCache[Player] = Document
        set g_LocalGenericCache[GetSession()] = LocalGenericCache
    Document
### }}}
## }}}

## v1 {{{
### Attachments {{{
PlayerDataGetAttachmentKind<scoped {HavocVerse}>(Player:player, AttachmentKindId:string)<transacts><decides>:attachment_kind_player_data=
    PlayerData := GetPlayerData(Player)
    PlayerData.v1_AttachmentKinds[AttachmentKindId]

PlayerDataGetAttachmentKindOrDefault<scoped {HavocVerse}>(Player:player, AttachmentKindId:string)<transacts>:attachment_kind_player_data=
    PlayerDataGetAttachmentKind[Player, AttachmentKindId] or attachment_kind_player_data{}

PlayerDataSetAttachmentKind<scoped {HavocVerse}>(Player:player, AttachmentKindId:string, NewData:attachment_kind_player_data)<transacts>:attachment_kind_player_data=
    PlayerData := GetPlayerData(Player)
    SetPlayerData(Player, havoc_verse_player_data{
        MakeHavocVersePlayerData<constructor>(PlayerData)
        v1_AttachmentKinds := AddToMap(PlayerData.v1_AttachmentKinds, AttachmentKindId, NewData)
    })
    NewData
### }}}

### Levels {{{
PlayerDataGetLevelGroup<scoped {HavocVerse}>(Player:player, Group:string)<transacts><decides>:level_group_player_data=
    PlayerData := GetPlayerData(Player)
    PlayerData.v1_LevelGroups[Group]

PlayerDataSetLevelGroup<scoped {HavocVerse}>(Player:player, Group:string, NewData:level_group_player_data)<transacts>:level_group_player_data=
    PlayerData := GetPlayerData(Player)
    SetPlayerData(Player, havoc_verse_player_data{
        MakeHavocVersePlayerData<constructor>(PlayerData)
        v1_LevelGroups := AddToMap(PlayerData.v1_LevelGroups, Group, NewData)
    })
    NewData
### }}}
## }}}
