using { /Verse.org/Random }
using { /Verse.org/Simulation }

int_range<public> := class<concrete>:
    @editable
    Minimum<public>: int = 0
    @editable
    Maximum<public>: int = 0

    GetRandomValue<public>()<transacts>:int=
        if (Maximum <= Minimum) {
            Minimum
        } else {
            GetRandomInt(Minimum, Maximum)
        }

float_range<public> := class<concrete>:
    @editable
    Minimum<public>: float = 0.0
    @editable
    Maximum<public>: float = 0.0

    GetRandomValue<public>()<transacts>:float=
        if (Maximum <= Minimum) {
            Minimum
        } else {
            GetRandomFloat(Minimum, Maximum)
        }