using { /Fortnite.com/Devices }
using { /Verse.org/Simulation }
using { /UnrealEngine.com/Temporary/Diagnostics }

using { HavocVerse.Debug }
using { HavocVerse.Resource }
using { HavocVerse.Logging }
using { HavocVerse.Util }
using { HavocVerse.Util.Task }

access_control_mode<public> := enum:
    AllowEveryone
    TeamIndex
    Whitelist

# Device that can limit access to custom devices to specific players or teams.
access_control_device<public> := class(creative_device, aware_device, aware_debug_device, custom_device, resetable):
    @editable p_DevId<public>: string = "N/A"

    @editable p_Mode<public>: access_control_mode = access_control_mode.AllowEveryone
    @editable p_AllowedTeamIndex: int = 1

    @editable p_WhitelistPlayerOnJoin: logic = false
    @editable p_BuildDeviceBackrefs: logic = false

    @editable p_UserActions: []access_control_user_action_config = array{}

    AgentWhitelistedEvent<public>: event(agent) = event(agent){}
    AgentRemovedFromWhitelistEvent<public>: event(agent) = event(agent){}

    var m_Whitelist: [agent]logic = map{}
    var m_WhitelistAddListeners: []type{__(:player):void} = array{}
    var m_WhitelistRemoveListeners: []type{__(:player,:logic):void} = array{}

    var m_DeviceBackrefs: []creative_device = array{}
    var m_DeviceBackrefsCallbacks: []type{__(:creative_device):void} = array{}

    DeviceMeta<override>()<transacts>:aware_device_meta=
        aware_device_meta{
            DeviceName := "access_control"
            DevId := p_DevId
        }

    DebugStats<override>()<transacts>:aware_debug_stats=
        aware_debug_stats{
            Length := map{
                "whitelist" => m_Whitelist.Length
                "listeners::whitelistAdd" => m_WhitelistAddListeners.Length
                "listeners::whitelistRemove" => m_WhitelistRemoveListeners.Length
                "deviceBackrefs" => m_DeviceBackrefs.Length
                "listeners::deviceBackrefAdd" => m_DeviceBackrefsCallbacks.Length
            }
        }

    OnBegin<override>()<suspends>:void=
        AwareInitDevice(Self, OnInit)

    OnInit<private>()<suspends>:void=
        for (UserActionConfig:p_UserActions) {
            UserAction := access_control_user_action{
                Device := Self
                Config := UserActionConfig
            }
            UserActionConfig.OnTrigger.TriggeredEvent.Subscribe(UserAction.OnTriggered)
        }
        if (p_Mode = access_control_mode.Whitelist) {
            Playspace := GetPlayspace()
            Playspace.PlayerRemovedEvent().Subscribe(OnPlayerRemoved)
            if (p_WhitelistPlayerOnJoin?) {
                Playspace.PlayerAddedEvent().Subscribe(OnPlayerAdded)
                for (Player:Playspace.GetPlayers()) {
                    OnPlayerAdded(Player)
                }
            }
        }

    Reset<override>(Agent:?agent):void=
        if (l_Agent := Agent?) {
            set m_Whitelist = m_Whitelist.RemoveByKey(l_Agent)(0)
        } else {
            set m_Whitelist = map{}
        }

    RegisterUsingDevice<public>(Device:creative_device):void=
        if (p_BuildDeviceBackrefs?) {
            set m_DeviceBackrefs += array{Device}
            for (Callback:m_DeviceBackrefsCallbacks) {
                Callback(Device)
            }
        }

    GetUsingDevices<public>()<transacts><decides>:[]creative_device=
        p_BuildDeviceBackrefs?
        m_DeviceBackrefs

    SubscribeToDevicesBackrefs<public>(
        DeviceRegisteredCallback:type{__(:creative_device):void},
        ?ReplayKnown:logic = true
    ):logic=
        if (not (p_BuildDeviceBackrefs?)) {
            return false
        }
        set m_DeviceBackrefsCallbacks += array{DeviceRegisteredCallback}
        if (ReplayKnown?) {
            for (Known:m_DeviceBackrefs) {
                DeviceRegisteredCallback(Known)
            }
        }
        true
    
    SubscribeToPlayerWhitelistEvents<public>(
        PlayerAddedCallback:type{__(:player):void},
        PlayerRemovedCallback:type{__(:player,:logic):void}
    ):void=
        set m_WhitelistAddListeners += array{PlayerAddedCallback}
        set m_WhitelistRemoveListeners += array{PlayerRemovedCallback}

    SubscribeToPlayerJoinEvents<public>(
        PlayerAddedCallback:type{__(:player):void},
        PlayerRemovedCallback:type{__(:player,:logic):void},
        ?ReplayAllowed:logic = true
    ):void=
        if (p_Mode = access_control_mode.Whitelist) {
            SubscribeToPlayerWhitelistEvents(PlayerAddedCallback, PlayerRemovedCallback)
        } else {
            Closure1 := access_control_player_join_closure{
                AccessControl := Self
                Callback := PlayerAddedCallback
            }
            GetPlayspace().PlayerAddedEvent().Subscribe(Closure1.Call)
            Closure2 := access_control_player_quit_closure{
                AccessControl := Self
                Callback := PlayerRemovedCallback
            }
            GetPlayspace().PlayerRemovedEvent().Subscribe(Closure2.Call)
        }
        if (ReplayAllowed?) {
            for (Agent:ListAllowedAgents(), Player := player[Agent]) {
                PlayerAddedCallback(Player)
            }
        }

    OnPlayerAdded<private>(Player:player):void=
        if (p_WhitelistPlayerOnJoin?) {
            AddToWhitelist(Player)
        }

    OnPlayerRemoved<private>(Player:player):void=
        RemoveFromWhitelist(Player, true)

    # Checks whether the given agent is allowed access.
    IsAllowed<public>(Agent:agent)<transacts><decides>:void=
        if (p_Mode = access_control_mode.AllowEveryone) {
            return
        } else if (p_Mode = access_control_mode.TeamIndex) {
            TeamCollection := GetPlayspace().GetTeamCollection()
            TeamCollection.IsOnTeam[Agent, p_AllowedTeamIndex]
        } else if (p_Mode = access_control_mode.Whitelist) {
            m_Whitelist[Agent]
        }
    
    # Returns a list of all online agents allowed access.
    ListAllowedAgents<public>()<transacts>:[]agent=
        if (p_Mode = access_control_mode.AllowEveryone) {
            GetPlayspace().GetPlayers()
        } else if (
            p_Mode = access_control_mode.TeamIndex
            TeamCollection := GetPlayspace().GetTeamCollection()
            Team := TeamCollection.GetTeams()[p_AllowedTeamIndex]
            Agents := TeamCollection.GetAgents[Team]
        ) {
            Agents
        } else if (p_Mode = access_control_mode.Whitelist) {
            for (Agent->__:m_Whitelist) {
                Agent
            }
        } else {
            array{}
        }

    AddToWhitelist<public>(Agent:agent):void=
        if (not m_Whitelist[Agent], set m_Whitelist[Agent] = true) {
            for (Callback:m_WhitelistAddListeners, Player := player[Agent]) {
                Callback(Player)
            }
            AgentWhitelistedEvent.Signal(Agent)
        }

    RemoveFromWhitelist<public>(Agent:agent, GameQuit:logic):void=
        RemoveResult := m_Whitelist.RemoveByKey(Agent)
        set m_Whitelist = RemoveResult(0)
        if (RemoveResult(1)?) {
            for (Callback:m_WhitelistRemoveListeners, Player := player[Agent]) {
                Callback(Player, GameQuit)
            }
            AgentRemovedFromWhitelistEvent.Signal(Agent)
        }

    SyncWhitelist<public>(Agents:[]agent):void=
        var KeepWhitelisted: [agent]logic = map{}
        for (Agent:Agents, set KeepWhitelisted[Agent] = true) {
            AddToWhitelist(Agent)
        }
        for (Agent->__:m_Whitelist, not KeepWhitelisted[Agent]) {
            RemoveFromWhitelist(Agent, false)
        }

    ClearWhitelist<public>():void=
        for (Agent->__:m_Whitelist) {
            for (Callback:m_WhitelistRemoveListeners, Player := player[Agent]) {
                Callback(Player, false)
            }
            AgentRemovedFromWhitelistEvent.Signal(Agent)
        }
        set m_Whitelist = map{}

    ToggleFromWhitelist<public>(Agent:agent):void=
        if (m_Whitelist[Agent]) {
            RemoveFromWhitelist(Agent, false)
        } else {
            AddToWhitelist(Agent)
        }

    CustomTrigger<override>(Action:string, Agent:?agent):logic=
        if (l_Agent := Agent?) {
            if (Action = "whitelist_add") {
                AddToWhitelist(l_Agent)
            } else if (Action = "whitelist_remove") {
                RemoveFromWhitelist(l_Agent, false)
            } else if (Action = "whitelist_toggle") {
                ToggleFromWhitelist(l_Agent)
            } else {
                return false
            }
            true
        } else {
            false
        }

    MsgUserAction<localizes>(DeviceStr:string, Agent:agent, ActionKindStr:string, NewState:string):message=
        "Handled user action {ActionKindStr} on {DeviceStr} for '{Agent}'. New state: {NewState}"

access_control_user_action_kind := enum:
    WhitelistAgent
    RemoveAgentFromWhitelist
    ToggleAgentWhitelist

access_control_user_action_config := class<concrete>:
    @editable Action: access_control_user_action_kind = access_control_user_action_kind.WhitelistAgent
    @editable OnTrigger: trigger_device = trigger_device{}

access_control_user_action := class:
    Device: access_control_device
    Config: access_control_user_action_config

    OnTriggered(Agent:?agent):void=
        if (l_Agent := Agent?) {
            if (Config.Action = access_control_user_action_kind.WhitelistAgent) {
                Device.AddToWhitelist(l_Agent)
            } else if (Config.Action = access_control_user_action_kind.RemoveAgentFromWhitelist) {
                Device.RemoveFromWhitelist(l_Agent, false)
            } else if (Config.Action = access_control_user_action_kind.ToggleAgentWhitelist) {
                Device.ToggleFromWhitelist(l_Agent)
            }
            NewState := if (Device.IsAllowed[l_Agent]) {
                "allowed"
            } else {
                "not allowed"
            }
            Log(Device.MsgUserAction(ToString(Device), l_Agent, ToString(Config.Action), NewState), ?Level := log_level.Debug)
        }

access_control_player_join_closure := class:
    AccessControl: access_control_device
    Callback: type{__(:player):void}

    Call(Player:player):void=
        if (AccessControl.IsAllowed[Player]) {
            Callback(Player)
        }

access_control_player_quit_closure := class:
    AccessControl: access_control_device
    Callback: type{__(:player,:logic):void}

    Call(Player:player):void=
        if (AccessControl.IsAllowed[Player]) {
            Callback(Player, true)
        }

ToString(ActionKind:access_control_user_action_kind)<transacts>:string=
    case(ActionKind):
        access_control_user_action_kind.WhitelistAgent => "WhitelistAgent"
        access_control_user_action_kind.RemoveAgentFromWhitelist => "RemoveAgentFromWhitelist"
        access_control_user_action_kind.ToggleAgentWhitelist => "ToggleAgentWhitelist"
