using { /Fortnite.com/Devices }
using { /Verse.org/Simulation }
using { /UnrealEngine.com/Temporary/Diagnostics }

using { HavocVerse.Debug }
using { HavocVerse.Logging }

analytics_controller_entry := class<concrete>:
    @editable AnalyticName: string = ""
    @editable AnalyticDevice: analytics_device = analytics_device{}

analytics_controller_device<public> := class(creative_device, aware_device, aware_debug_device):
    @editable p_Analytics: []analytics_controller_entry = array{}

    var m_MappedDevices: [string]analytics_device = map{}

    DeviceMeta<override>()<transacts>:aware_device_meta=
        aware_device_meta{
            DeviceName := "analytics_controller"
        }

    DebugStats<override>()<transacts>:aware_debug_stats=
        aware_debug_stats{
            Length := map{
                "mappedDevices" => m_MappedDevices.Length
            }
        }

    OnBegin<override>()<suspends>:void=
        EnsureSetup()

    EnsureSetup<private>():void=
        if (m_MappedDevices.Length = p_Analytics.Length) {
            return
        }
        for (Entry:p_Analytics) {
            if:
                set m_MappedDevices[Entry.AnalyticName] = Entry.AnalyticDevice
        }
        Log("{p_Analytics.Length} analytics events mapped to devices", ?Level := log_level.Debug)

    Emit<public>(Agent:agent, EventName:string):void=
        EnsureSetup()
        if (AnalyticDevice := m_MappedDevices[EventName]) {
            AnalyticDevice.Submit(Agent)
            Log(MsgEmitted(Agent, EventName), ?Level := log_level.Debug)
        } else {
            Log("Could not emit analytics event \"{EventName}\" because it has no bound device", ?Level := log_level.Warning)
        }

    MsgEmitted<private><localizes>(Agent:agent, EventName:string):message=
        "Emitted event \"{EventName}\" for '{Agent}'"
