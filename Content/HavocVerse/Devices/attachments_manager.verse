using { /Fortnite.com/Devices }
using { /Verse.org/Simulation }
using { /UnrealEngine.com/Temporary/Diagnostics }

using { HavocVerse.Debug }
using { HavocVerse.Devices.GameSpecific }
using { HavocVerse.Storage }
using { HavocVerse.Persistence }
using { HavocVerse.Logging }
using { HavocVerse.Util }

attachments_manager_attachment_kind_config := class<final><concrete>:
    @editable AttachmentKindId: string = ""
    @editable DefaultActive: ?string = false
    @editable DefaultUnlocked: []string = array{}
    @editable PersistenceKeepAcrossTycoonRebirths: logic = false

    @editable UserActions: []attachments_manager_attachment_kind_user_action_config = array{}

attachments_manager_unlock_listener<internal> := class<final>:
    AttachmentKind: ?string
    Callback: type{__(:attachment_unlocked_event):void}

attachment_unlocked_event<public> := class<final>:
    Player<public>: player
    AttachmentKindId<public>: string
    Attachment<public>: string

attachments_manager_equip_listener<internal> := class<final>:
    AttachmentKind: ?string
    Callback: type{__(:attachment_equip_event):void}

attachment_equip_event<public> := class<final>:
    Player<public>: player
    AttachmentKindId<public>: string
    Attachment<public>: ?string

attachments_manager_device<public> := class(creative_device, aware_device, aware_debug_device, persistence_resetable):
    @editable p_AttachmentKinds: []attachments_manager_attachment_kind_config = array{}

    var m_UnlockListeners: []attachments_manager_unlock_listener = array{}
    var m_EquipListeners: []attachments_manager_equip_listener = array{}

    DeviceMeta<override>()<transacts>:aware_device_meta=
        aware_device_meta{
            DeviceName := "attachments_manager"
        }

    DebugStats<override>()<transacts>:aware_debug_stats=
        aware_debug_stats{
            Length := map{
                "listeners::unlock" => m_UnlockListeners.Length
                "listeners::equip" => m_EquipListeners.Length
            }
        }

    OnBegin<override>()<suspends>:void=
        AwareInitDevice(Self, OnInit)
    
    OnInit<private>()<suspends>:void=
        for (AttachmentKindConfig:p_AttachmentKinds) {
            for (UserActionConfig:AttachmentKindConfig.UserActions) {
                UserAction := attachments_manager_attachment_kind_user_action{
                    Device := Self
                    AttachmentKind := AttachmentKindConfig.AttachmentKindId
                    Config := UserActionConfig
                }
                UserActionConfig.OnTrigger.TriggeredEvent.Subscribe(UserAction.OnTriggered)
            }
        }

    GetAttachmentKindConfig<private>(AttachmentKindId:string)<transacts><decides>:attachments_manager_attachment_kind_config=
        var Result: ?attachments_manager_attachment_kind_config = false
        for (Config:p_AttachmentKinds, not (Result?), Config.AttachmentKindId = AttachmentKindId) {
            set Result = Some(Config)
        }
        Result?

    # Checks whether a player has unlocked an attachment for a given attachment kind.
    IsUnlocked<public>(Player:player, AttachmentKindId:string, Attachment:string)<transacts><decides>:void=
        if (not GetAttachmentKindConfig[AttachmentKindId]) {
            Log(Self, "Attachment kind \"{AttachmentKindId}\" is not configured: IsUnlocked(\"{Attachment}\")", ?Level := log_level.Warning)
        }
        GetAttachmentKindConfig[AttachmentKindId].DefaultUnlocked.Find[Attachment] or (
            PlayerData := PlayerDataGetAttachmentKindOrDefault(Player, AttachmentKindId)
            PlayerData.v0_Unlocked[Attachment] or PlayerData.v0_LastActive? = Attachment
        )

    # Returns a player's unlocked attachments for a given attachment kind.
    GetUnlockedAttachments<public>(Player:player, AttachmentKindId:string)<transacts>:[]string=
        if (not GetAttachmentKindConfig[AttachmentKindId]) {
            Log(Self, "Attachment kind \"{AttachmentKindId}\" is not configured: GetUnlockedAttachments", ?Level := log_level.Warning)
        }
        PlayerData := PlayerDataGetAttachmentKindOrDefault(Player, AttachmentKindId)
        var Unlocked: []string = PlayerData.v0_Unlocked.Keys()
        if (Config := GetAttachmentKindConfig[AttachmentKindId]) {
            for (DefaultUnlocked:Config.DefaultUnlocked, not Unlocked.Find[DefaultUnlocked]) {
                set Unlocked = array{DefaultUnlocked} + Unlocked
            }
        }
        Unlocked

    # Unlocks a player's attachment for a given attachment kind.
    UnlockAttachment<public>(Player:player, AttachmentKindId:string, Attachment:string):logic=
        if (not Player.IsActive[]) {
            LogPlayerOffline("UnlockAttachment({AttachmentKindId}, {Attachment})", Player)
            return false
        }

        if (not GetAttachmentKindConfig[AttachmentKindId]) {
            Log(Self, "Attachment kind \"{AttachmentKindId}\" is not configured: UnlockAttachment(\"{Attachment}\")", ?Level := log_level.Warning)
        }
        PlayerData := PlayerDataGetAttachmentKindOrDefault(Player, AttachmentKindId)
        if (PlayerData.v0_Unlocked[Attachment] or PlayerData.v0_LastActive? = Attachment) {
            return false
        }
        PlayerDataSetAttachmentKind(Player, AttachmentKindId, attachment_kind_player_data{
            MakeAttachmentKindPlayerData<constructor>(PlayerData)
            v0_Unlocked := AddToMap(PlayerData.v0_Unlocked, Attachment, void)
        })
        Log(MsgUnlocked("{AttachmentKindId}:{Attachment}", Player), ?Level := log_level.Debug)
        for (
            Listener:m_UnlockListeners
            (not Listener.AttachmentKind?) or (Listener.AttachmentKind? = AttachmentKindId)
        ) {
            Event := attachment_unlocked_event{
                Player := Player
                AttachmentKindId := AttachmentKindId
                Attachment := Attachment
            }
            Listener.Callback(Event)
        }
        true

    # Undoes the unlock of a player's attachment for a given attachment kind.
    LockAttachment<public>(Player:player, AttachmentKindId:string, Attachment:string):void=
        if (not Player.IsActive[]) {
            LogPlayerOffline("LockAttachment({AttachmentKindId}, {Attachment})", Player)
            return
        }

        if (not GetAttachmentKindConfig[AttachmentKindId]) {
            Log(Self, "Attachment kind \"{AttachmentKindId}\" is not configured: LockAttachment", ?Level := log_level.Warning)
        }
        PlayerData := PlayerDataGetAttachmentKindOrDefault(Player, AttachmentKindId)
        PlayerDataSetAttachmentKind(Player, AttachmentKindId, attachment_kind_player_data{
            MakeAttachmentKindPlayerData<constructor>(PlayerData)
            v0_Unlocked := PlayerData.v0_Unlocked.RemoveByKey(Attachment)(0)
        })

    # Returns a player's active attachment for a given attachment kind.
    GetActiveAttachment<public>(Player:player, AttachmentKindId:string)<transacts><decides>:string=
        if (not GetAttachmentKindConfig[AttachmentKindId]) {
            Log(Self, "Attachment kind \"{AttachmentKindId}\" is not configured: GetActiveAttachment", ?Level := log_level.Warning)
        }
        if (CurrentActive := PlayerDataGetAttachmentKind[Player, AttachmentKindId].v0_LastActive?) {
            CurrentActive
        } else {
            Config := GetAttachmentKindConfig[AttachmentKindId]
            Config.DefaultActive?
        }

    # Sets a player's active attachment for a given attachment kind.
    SetActiveAttachment<public>(Player:player, AttachmentKindId:string, Attachment:?string):logic=
        if (not Player.IsActive[]) {
            AttachmentStr := Attachment? or "<none>"
            LogPlayerOffline("SetActiveAttachment({AttachmentKindId}, {AttachmentStr})", Player)
            return false
        }

        if (not GetAttachmentKindConfig[AttachmentKindId]) {
            Log(Self, "Attachment kind \"{AttachmentKindId}\" is not configured: SetActiveAttachment", ?Level := log_level.Warning)
        }
        PlayerData := PlayerDataGetAttachmentKindOrDefault(Player, AttachmentKindId)
        if (PlayerData.v0_LastActive = Attachment) {
            return false
        }
        PlayerDataSetAttachmentKind(Player, AttachmentKindId, attachment_kind_player_data{
            MakeAttachmentKindPlayerData<constructor>(PlayerData)
            v0_LastActive := Attachment
        })
        if (l_Attachment := Attachment?) {
            Log(MsgSetActive("{AttachmentKindId}:{l_Attachment}", Player), ?Level := log_level.Debug)
        } else {
            Log(MsgDelActive(AttachmentKindId, Player), ?Level := log_level.Debug)
        }
        for (
            Listener:m_EquipListeners
            (not Listener.AttachmentKind?) or (Listener.AttachmentKind? = AttachmentKindId)
        ) {
            Event := attachment_equip_event{
                Player := Player
                AttachmentKindId := AttachmentKindId
                Attachment := Attachment
            }
            Listener.Callback(Event)
        }
        true

    MsgUnlocked<private><localizes>(Id:string, Player:player):message=
        "Unlocked attachment \"{Id}\" for player \"{Player}\""

    MsgSetActive<private><localizes>(Id:string, Player:player):message=
        "Set active attachment \"{Id}\" for player \"{Player}\""
    MsgDelActive<private><localizes>(KindId:string, Player:player):message=
        "Set active attachment of kind \"{KindId}\" to none for player \"{Player}\""

    AddUnlockListener<public>(
        Callback:type{__(:attachment_unlocked_event):void},
        ?AttachmentKindId:?string = false
    ):void=
        set m_UnlockListeners += array{attachments_manager_unlock_listener{
            AttachmentKind := AttachmentKindId
            Callback := Callback
        }}

    AddEquipListener<public>(
        Callback:type{__(:attachment_equip_event):void},
        ?AttachmentKindId:?string = false
    ):void=
        set m_EquipListeners += array{attachments_manager_equip_listener{
            AttachmentKind := AttachmentKindId
            Callback := Callback
        }}

    PersistenceReset<override>(__:persistence_recordable_state, Agent:agent, Query:persistence_reset_query):void=
        for (Config:p_AttachmentKinds, Player := player[Agent]) {
            if (Config.PersistenceKeepAcrossTycoonRebirths?, tycoon_manager_rebirth_persistence_reset_query[Query]) {
                Log("Skipping reset of attachment kind \"{Config.AttachmentKindId}\" persistence state because it is initialized by a rebirth", ?Level := log_level.Debug)
            } else {
                PlayerDataSetAttachmentKind(Player, Config.AttachmentKindId, attachment_kind_player_data{})
            }
        }

attachments_manager_attachment_kind_user_action_kind := enum:
    Unlock
    Lock
    SetActive

attachments_manager_attachment_kind_user_action_config := class<final><concrete>:
    @editable OnTrigger: trigger_device = trigger_device{}
    @editable Action: attachments_manager_attachment_kind_user_action_kind = attachments_manager_attachment_kind_user_action_kind.Unlock
    @editable Attachment: string = ""

attachments_manager_attachment_kind_user_action := class<final>:
    Device: attachments_manager_device
    AttachmentKind: string
    Config: attachments_manager_attachment_kind_user_action_config

    OnTriggered(Agent:?agent):void=
        if (Player := player[Agent?]) {
            case(Config.Action):
                attachments_manager_attachment_kind_user_action_kind.Unlock =>
                    Device.UnlockAttachment(Player, AttachmentKind, Config.Attachment)
                attachments_manager_attachment_kind_user_action_kind.Lock =>
                    Device.LockAttachment(Player, AttachmentKind, Config.Attachment)
                attachments_manager_attachment_kind_user_action_kind.SetActive =>
                    l_Attachment := if (Config.Attachment.Length > 0) {
                        Some(Config.Attachment)
                    } else {
                        None()
                    }
                    Device.SetActiveAttachment(Player, AttachmentKind, l_Attachment)
        }
