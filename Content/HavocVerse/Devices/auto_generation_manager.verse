using { /Fortnite.com/Devices }
using { /Verse.org/Simulation }
using { /UnrealEngine.com/Temporary/Diagnostics }

using { HavocVerse.Debug }
using { HavocVerse.Logging }
using { HavocVerse.Util }

AutoGenerationDefaultMultiplierChannel<public>: string = "default"

auto_generation_ticket_updated_event<public> := struct:
    Ticket<public>: auto_generation_ticket
    Delta<public>: float

auto_generation_multiplier_updated_event<public> := struct:
    ResourceId<public>: ?string
    Channel<public>: string
    NewMultiplier<public>: float
    NewTotalMultiplier<public>: float

# Auto generation behavior when giving resources to several agents.
auto_generation_distribution_behavior<public> := enum:
    # Only give resources to the first agent.
    FirstPlayer
    # Give the full amount of resources to each agent.
    FullAmountToEach
    # Split the amount of resources evenly between all agents.
    Split

auto_generation_score_award_config<public> := struct<concrete>:
    @editable ResourceId: string = ""
    @editable Multiplier: float = 1.0

# Multipliers may be set by several devices at once. This
# splits a multiplier into several "channels". They are
# all added together to get the final multiplier.
split_multiplier<public> := class:
    ResourceId<public>: ?string = false

    var m_AutoGenerationManager: ?auto_generation_manager = false
    var m_Channels: [string]float = map{}

    SetMultiplier<public>(Channel:string, Multiplier:float):void=
        SetMultiplier(Channel, Some(Multiplier))

    SetMultiplier<public>(Channel:string, Multiplier:?float):void=
        if:
            l_Multiplier := Multiplier?
            set m_Channels[Channel] = l_Multiplier
        else:
            set m_Channels = m_Channels.RemoveByKey(Channel)(0)
        if (Manager := m_AutoGenerationManager?) {
            Manager.MultiplierUpdatedEvent.Signal(auto_generation_multiplier_updated_event{
                ResourceId := ResourceId
                Channel := Channel
                NewMultiplier := m_Channels.GetOrDefault(Channel, 1.0)
                NewTotalMultiplier := GetTotalMultiplier()
            })
        }

    GetMultiplier<public>(Channel:string)<transacts>:float=
        m_Channels.GetOrDefault(Channel, 1.0)

    GetTotalMultiplier<public>()<transacts>:float=
        var Total: float = 1.0
        for (Channel->Multiplier:m_Channels, Channel <> AutoGenerationDefaultMultiplierChannel) {
            set Total += Multiplier - 1.0
        }
        GeneralMultiplier := GetMultiplier(AutoGenerationDefaultMultiplierChannel)
        Total * GeneralMultiplier

auto_generation_manager<public> := class(creative_device, aware_device, aware_debug_device, custom_device, resetable):
    @editable p_AccessControl: ?access_control_device = false
    @editable p_ResourceManager<public>: resource_manager_device = resource_manager_device{}
    @editable p_SendGiveFeedback: logic = false
    @editable p_SendRateChangeFeedback: logic = true

    @editable p_DistributionBehavior<public>: auto_generation_distribution_behavior = auto_generation_distribution_behavior.FirstPlayer
    @editable p_GiveRate<public>: float = 1.0

    @editable p_ScoreManager: ?score_manager_device = false
    @editable p_DefaultScoreAwards: []auto_generation_score_award_config = array{}

    @editable p_UserActions: []auto_generation_manager_user_action_config = array{}

    TicketRateUpdatedEvent<public>: event(auto_generation_ticket_updated_event) = event(auto_generation_ticket_updated_event){}
    MultiplierUpdatedEvent<public>: event(auto_generation_multiplier_updated_event) = event(auto_generation_multiplier_updated_event){}

    var m_IdCounter: int = 0
    var m_Tickets: [int]auto_generation_ticket = map{}

    var m_GlobalMultiplier: split_multiplier = split_multiplier{ResourceId := false}
    var m_ResourceMultpliers: [string]split_multiplier = map{}
    var m_LastUnevenSplitGaveTo: int = -1

    DeviceMeta<override>()<transacts>:aware_device_meta=
        aware_device_meta{
            DeviceName := "auto_generation_manager"
        }

    DebugStats<override>()<transacts>:aware_debug_stats=
        aware_debug_stats{
            Length := map{
                "tickets" => m_Tickets.Length
                "resourceMultipliers" => m_ResourceMultpliers.Length
            }
        }

    OnBegin<override>()<suspends>:void=
        AwareInitDevice(Self, OnInit)

    OnInit<private>()<suspends>:void=
        for (UserActionConfig:p_UserActions) {
            UserAction := auto_generation_manager_user_action{
                Device := Self
                Config := UserActionConfig
            }
            UserActionConfig.OnTrigger.TriggeredEvent.Subscribe(UserAction.OnTriggered)
            for (Button:UserActionConfig.OnButton) {
                Button.InteractedWithEvent.Subscribe(UserAction.OnInteractedWith)
            }
        }

        if (p_DefaultScoreAwards.Length > 0 and not (p_ScoreManager?)) {
            Log("Auto generation manager has score awards but no score manager", ?Level := log_level.Error)
        }

        p_ResourceManager.RegisterAutoGenerationManager(Self)
        spawn:
            StartGenerationLoop()

    StartGenerationLoop<private>()<suspends>:void=
        loop {
            Generate()
            Sleep(p_GiveRate)
        }

    Reset<override>(Agent:?agent):void=
        set m_GlobalMultiplier = split_multiplier{ResourceId := false}
        set m_ResourceMultpliers = map{}
        set m_LastUnevenSplitGaveTo = -1
    
    IncludesPlayer<public>(Player:player)<decides><transacts>:void=
        if (AccessControl := p_AccessControl?) {
            AccessControl.IsAllowed[Player]
        }

    GetAgents<private>()<transacts>:[]agent=
        if (AccessControl := p_AccessControl?) {
            AccessControl.ListAllowedAgents()
        } else {
            GetPlayspace().GetPlayers()
        }

    Generate<private>():void=
        Now := GetSimulationElapsedTime()
        var Generated: [string]int = map{}
        for (Ticket:m_Tickets, not (Ticket.Opts.Vanity?), Generation := Ticket.m_Generation?) {
            Expected := Generation.Rate * (Now - Generation.m_ResetAt)
            if (
                ResourceCount := Floor[Expected]
                ToGive := ResourceCount - Generation.m_Given
                ToGive > 0
                set Generated[Ticket.ResourceId] = Generated.GetOrDefault(Ticket.ResourceId, 0) + ToGive
            ) {
                if (Expected = ResourceCount) {
                    set Generation.m_ResetAt = Now
                    set Generation.m_Given = 0
                } else {
                    set Generation.m_Given = ResourceCount
                }
            }
        }
        if (Generated.Length = 0) {
            return
        }
        Agents := GetAgents()
        if (Agents.Length = 0) {
            return
        }
        for (ResourceId->OriginalAmount:Generated) {
            Amount := DefaultingRound(OriginalAmount * GetEffectiveMultiplier(ResourceId))
            case(p_DistributionBehavior):
                auto_generation_distribution_behavior.FirstPlayer => 
                    if (Agent := Agents[0]) {
                        Give(Agent, ResourceId, Amount)
                    }
                auto_generation_distribution_behavior.FullAmountToEach => 
                    for (Agent:Agents) {
                        Give(Agent, ResourceId, Amount)
                    }
                auto_generation_distribution_behavior.Split => 
                    GiveToEach := DivDefaultingFloor(Amount, Agents.Length)
                    for (Agent:Agents) {
                        Give(Agent, ResourceId, GiveToEach)
                    }
                    Remaining := Amount - GiveToEach
                    if (Remaining > 0) {
                        if (
                            m_LastUnevenSplitGaveTo >= 0
                            GiveRemainingTo := Mod[m_LastUnevenSplitGaveTo + 1, Agents.Length]
                            Agent := Agents[GiveRemainingTo]
                        ) {
                            Give(Agent, ResourceId, Remaining)
                            set m_LastUnevenSplitGaveTo = GiveRemainingTo
                        } else if (Agent := Agents[0]) {
                            Give(Agent, ResourceId, Remaining)
                            set m_LastUnevenSplitGaveTo = 0
                        }
                    }
        }

    Give<private>(Agent:agent, ResourceId:string, Amount:int):void=
        p_ResourceManager.Give(Agent, ResourceId, Amount, ?SendFeedback := p_SendGiveFeedback)
        for (ScoreAward:p_DefaultScoreAwards) {
            if (ScoreAward.ResourceId = ResourceId) {
                if (
                    ScoreManager := p_ScoreManager?
                    Score := Floor[ScoreAward.Multiplier * Amount]
                    Score > 0
                ) {
                    ScoreManager.SetScoreAward(Score)
                    ScoreManager.Activate(Agent)
                    ScoreManager.SetScoreAward(0)
                }
                return
            }
        }

    OnTicketUpdated<internal>(Event:auto_generation_ticket_updated_event):void=
        if (not (p_SendRateChangeFeedback?)) {
            return
        }
        FeedbackKind := if (Event.Delta > 0.0) {
            resource_feedback_kind.RateIncrease
        } else if (Event.Delta < 0.0) {
            resource_feedback_kind.RateDecrease
        } else {
            return
        }
        Agents := GetAgents()
        case(p_DistributionBehavior):
            auto_generation_distribution_behavior.FirstPlayer =>
                if (FirstAgent := Agents[0]) {
                    p_ResourceManager.SendFeedbackNotification(FeedbackKind, FirstAgent, Event.Ticket.ResourceId, Event.Delta)
                }
            auto_generation_distribution_behavior.FullAmountToEach => 
                for (Agent:Agents) {
                    p_ResourceManager.SendFeedbackNotification(FeedbackKind, Agent, Event.Ticket.ResourceId, Event.Delta)
                }
            auto_generation_distribution_behavior.Split =>
                PerAgentDelta := Event.Delta / ToFloat(Agents.Length)
                for (Agent:Agents) {
                    p_ResourceManager.SendFeedbackNotification(FeedbackKind, Agent, Event.Ticket.ResourceId, PerAgentDelta)
                }

    # Checks if this manager may generate resources for the given agent.
    IsGeneratingFor(Agent:agent)<transacts><decides>:void=
        if (AccessControl := p_AccessControl?) {
            AccessControl.IsAllowed[Agent]
        }

    # Returns the global generation multiplier instance. This can be used to get or update the actual multiplier.
    GetGlobalMultiplierInstance<public>()<transacts>:split_multiplier=
        m_GlobalMultiplier

    # Returns the generation multiplier instance of a given resource. This can be used to get or update the actual multiplier.
    GetResourceMultiplierInstance<public>(ResourceId:string)<transacts>:split_multiplier=
        if (ResourceMultiplier := m_ResourceMultpliers[ResourceId]) {
            ResourceMultiplier
        } else {
            ResourceMultiplier := split_multiplier{ResourceId := Some(ResourceId)}
            if (set m_ResourceMultpliers[ResourceId] = ResourceMultiplier) {}
            ResourceMultiplier
        }

    # Returns the effective generation multiplier for a given resource.
    GetEffectiveMultiplier<public>(ResourceId:string)<transacts>:float=
        var EffectiveMultiplier: float = m_GlobalMultiplier.GetTotalMultiplier()
        if (ResourceMultiplier := m_ResourceMultpliers[ResourceId]) {
            set EffectiveMultiplier *= ResourceMultiplier.GetTotalMultiplier()
        }
        EffectiveMultiplier

    # Returns the indicative amount of the given resource generated per second.
    GetGenerationRate<public>(ResourceId:string, Agent:?agent)<transacts>:float=
        if (l_Agent := Agent?, AccessControl := p_AccessControl?, not AccessControl.IsAllowed[l_Agent]) {
            return 0.0
        }
        var MultiplierAffectedAmount: float = 0.0
        var MultiplierUnaffectedAmount: float = 0.0
        for (
            Ticket:m_Tickets
            Ticket.ResourceId = ResourceId
            Generation := Ticket.m_Generation?
        ) {
            if (Ticket.Opts.Vanity?, Ticket.Opts.VanityIgnoreMultipliers?) {
                set MultiplierUnaffectedAmount += Generation.Rate
            } else {
                set MultiplierAffectedAmount += Generation.Rate
            }
        }
        Rate := (MultiplierAffectedAmount * GetEffectiveMultiplier(ResourceId)) + MultiplierUnaffectedAmount
        if (l_Agent := Agent?) {
            Agents := GetAgents()
            case(p_DistributionBehavior):
                auto_generation_distribution_behavior.FirstPlayer => 
                    if (FirstAgent := Agents[0], FirstAgent = l_Agent) {
                        Rate
                    } else {
                        0.0
                    }
                auto_generation_distribution_behavior.FullAmountToEach => Rate
                auto_generation_distribution_behavior.Split => Rate / ToFloat(Agents.Length)
        } else {
            Rate
        }
    
    # Returns a new ticket for the given resource.
    # Fails if the resource doesn't exist or if no new ticket can be allocated.
    CreateTicket<public>(ResourceId:string, Vanity:logic)<decides><transacts>:auto_generation_ticket=
        CreateTicket[ResourceId, auto_generation_ticket_opts{Vanity := Vanity}]

    # Returns a new ticket for the given resource.
    # Fails if the resource doesn't exist or if no new ticket can be allocated.
    CreateTicket<public>(ResourceId:string, Opts:auto_generation_ticket_opts)<decides><transacts>:auto_generation_ticket=
        p_ResourceManager.GetResource[ResourceId]
        TicketId := m_IdCounter
        set m_IdCounter += 1
        Ticket := auto_generation_ticket{
            Manager := Self
            Id := TicketId
            ResourceId := ResourceId
            Opts := Opts
        }
        set m_Tickets[TicketId] = Ticket
        Ticket

    DeleteTicket<public>(Ticket:auto_generation_ticket):logic=
        if (not m_Tickets[Ticket.Id]) {
            return false
        }
        set Ticket.m_Deleted = true
        Ticket.Disable()
        set m_Tickets = m_Tickets.RemoveByKey(Ticket.Id)(0)
        true

    CustomTrigger<override>(Input:string, Agent:?agent):logic=
        SplitRes := Input.Split(':')
        if (SplitRes[0] = "set_multiplier", Channel := SplitRes[1], ResourceId := SplitRes[2], Multiplier := ParseFloat[SplitRes[3]]) {
            l_Channel := if (Channel.Length > 0) {
                Channel
            } else {
                AutoGenerationDefaultMultiplierChannel
            }
            GetResourceMultiplierInstance(ResourceId).SetMultiplier(l_Channel, Multiplier)
            true
        } else if (SplitRes[0] = "set_multiplier", Channel := SplitRes[1], Multiplier := ParseFloat[SplitRes[2]]) {
            l_Channel := if (Channel.Length > 0) {
                Channel
            } else {
                AutoGenerationDefaultMultiplierChannel
            }
            GetGlobalMultiplierInstance().SetMultiplier(l_Channel, Multiplier)
            true
        } else {
            false
        }

auto_generation_ticket_opts<public> := class:
    # Whether this generation rate is purely for vanity purposes and won't be
    # used to actually generate resources from the auto generation manager.
    Vanity<public>: logic
    # With VanityRate set to true, ignore resource multipliers when calculating generation rate.
    VanityIgnoreMultipliers<public>: logic = true

# A "ticket" that devices get to request auto generation and update their demands.
# The resource may not be changed.
auto_generation_ticket<public> := class:
    Manager<public>: auto_generation_manager
    Id<public>: int
    ResourceId<public>: string
    Opts<public>: auto_generation_ticket_opts

    var m_Generation: ?auto_generation = false
    var m_Deleted: logic = false

    # Sets the generation rate, enabling auto generation if currently disabled.
    SetRate<public>(Rate:float):void=
        if (m_Deleted?) {
            return
        }
        l_Rate := Max(Rate, 0.0)
        PrevRate := if (Generation := m_Generation?) {
            Generation.Rate
        } else {
            0.0
        }
        if (Opts.Vanity?) {
            set m_Generation = option{auto_generation{
                Rate := l_Rate
                m_ResetAt := 0.0
            }}
        } else {
            set m_Generation = option{auto_generation{
                Rate := l_Rate
                m_ResetAt := GetSimulationElapsedTime()
            }}
        }
        Event := auto_generation_ticket_updated_event{
            Ticket := Self
            Delta := l_Rate - PrevRate
        }
        Manager.TicketRateUpdatedEvent.Signal(Event)
        Manager.OnTicketUpdated(Event)

    ModifyRate<public>(Delta:float):void=
        if (m_Deleted?) {
            return
        }
        PrevRate := if (Generation := m_Generation?) {
            Generation.Rate
        } else {
            0.0
        }
        SetRate(PrevRate + Delta)

    # Disables auto generation for this ticket.
    Disable<public>():void=
        if (Generation := m_Generation?) {
            PrevRate := Generation.Rate
            set m_Generation = false
            Event := auto_generation_ticket_updated_event{
                Ticket := Self
                Delta := -PrevRate
            }
            Manager.TicketRateUpdatedEvent.Signal(Event)
            Manager.OnTicketUpdated(Event)
        } else {
            set m_Generation = false
        }

    # Deletes this ticket, making it invalid and subsequent calls to `SetRate` useless.
    Delete<public>():void=
        Manager.DeleteTicket(Self)

auto_generation<internal> := class:
    # Generation rate
    Rate: float

    # Simulation elapsed time of last reset.
    var m_ResetAt: float
    # How much of the resource was actually given since the last reset.
    var m_Given: int = 0

auto_generation_manager_user_action_kind := enum:
    SetMultiplier

auto_generation_manager_user_action_config := class<concrete>:
    @editable Action: auto_generation_manager_user_action_kind = auto_generation_manager_user_action_kind.SetMultiplier
    @editable OnTrigger: trigger_device = trigger_device{}
    @editable OnButton: []button_device = array{}
    @editable ResourceId: string = ""
    @editable Channel: string = AutoGenerationDefaultMultiplierChannel
    @editable Value: float = 1.0

auto_generation_manager_user_action := class:
    Device: auto_generation_manager
    Config: auto_generation_manager_user_action_config

    OnTriggered(Agent:?agent):void=
        if (Player := player[Agent?]) {
            DoAction(Player)
        }

    OnInteractedWith(Agent:agent):void=
        if (Player := player[Agent]) {
            DoAction(Player)
        }

    DoAction<private>(Player:player):void=
        case(Config.Action):
            auto_generation_manager_user_action_kind.SetMultiplier => if (Config.ResourceId.Length > 0) {
                Device.GetResourceMultiplierInstance(Config.ResourceId).SetMultiplier(Config.Channel, Config.Value)
            } else {
                Device.GetGlobalMultiplierInstance().SetMultiplier(Config.Channel, Config.Value)
            }
