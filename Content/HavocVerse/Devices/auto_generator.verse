using { /Fortnite.com/Devices }
using { /Verse.org/Simulation }
using { /UnrealEngine.com/Temporary/Diagnostics }

using { HavocVerse.Debug }
using { HavocVerse.Logging }
using { HavocVerse.Util }

auto_generator_device<public> := class(creative_device, aware_device, custom_device, resetable):
    @editable p_AutoGenerationManager<public>: auto_generation_manager = auto_generation_manager{}
    @editable p_ResourceId<public>: string = ""

    @editable p_StartEnabled: logic = true
    @editable p_DefaultRate: float = 0.0

    @editable p_UserActions: []auto_generator_user_action_config = array{}

    var m_Enabled: logic = false
    var m_Ticket: ?auto_generation_ticket = false
    var m_Rate: ?float = false

    DeviceMeta<override>()<transacts>:aware_device_meta=
        aware_device_meta{
            DeviceName := "auto_generator"
        }

    OnBegin<override>()<suspends>:void=
        AwareInitDevice(Self, OnInit)

    OnInit<private>()<suspends>:void=
        for (UserActionConfig:p_UserActions) {
            UserAction := auto_generator_user_action{
                Device := Self
                Config := UserActionConfig
            }
            UserActionConfig.OnTrigger.TriggeredEvent.Subscribe(UserAction.OnTriggered)
        }

        if (p_StartEnabled?) {
            Enable()
        }

    Reset<override>(Agent:?agent):void=
        set m_Rate = false
        if (p_StartEnabled?) {
            if (m_Enabled?, Ticket := m_Ticket?) {
                Ticket.SetRate(p_DefaultRate)
            } else {
                Enable()
            }
        } else {
            Disable()
        }

    Enable<public>():void=
        if (m_Enabled?) {
            return
        }
        set m_Enabled = true
        Ticket := if (l_Ticket := m_Ticket?) {
            l_Ticket
        } else if (l_Ticket := p_AutoGenerationManager.CreateTicket[p_ResourceId, false]) {
            set m_Ticket = Some(l_Ticket)
            l_Ticket
        } else {
            Log("Failed to create auto generation ticket for auto generator of resource {p_ResourceId}", ?Level := log_level.Error)
            return
        }
        if (Rate := m_Rate?) {
            Ticket.SetRate(Rate)
        } else if (p_DefaultRate > 0.0) {
            Ticket.SetRate(p_DefaultRate)
        }

    Disable<public>():void=
        if (not (m_Enabled?)) {
            return
        }
        set m_Enabled = false
        if (Ticket := m_Ticket?) {
            Ticket.Disable()
        }

    ToggleEnable<public>():void=
        if (m_Enabled?) {
            Disable()
        } else {
            Enable()
        }

    SetRate<public>(Rate:float):void=
        set m_Rate = Some(Rate)
        if (m_Enabled?, Ticket := m_Ticket?) {
            Ticket.SetRate(Rate)
        }
    
    ModifyRate<public>(Delta:float):void=
        Prev := if (Rate := m_Rate?) {
            Rate
        } else {
            p_DefaultRate
        }
        NewRate := Prev + Delta
        set m_Rate = Some(NewRate)
        if (m_Enabled?, Ticket := m_Ticket?) {
            Ticket.SetRate(NewRate)
        }
    
    CustomTrigger<override>(Input:string, Agent:?agent):logic=
        if (SplitRes := Input.SplitOnce[':']) {
            Action := SplitRes(0)
            Data := SplitRes(1)
            if (Action = "modify_rate", ModifyBy := ParseFloat[Data]) {
                ModifyRate(ModifyBy)
                return true
            } else if (Action = "set_rate", Rate := ParseFloat[Data]) {
                SetRate(Rate)
                return true
            }
        } else if (Input = "enable") {
            Enable()
            return true
        } else if (Input = "disable") {
            Disable()
            return true
        } else if (Input = "toggle") {
            ToggleEnable()
            return true
        }
        false

auto_generator_user_action_kind := enum:
    Enable
    Disable
    ToggleEnable
    SetRate
    ModifyRate

auto_generator_user_action_config := class<concrete>:
    @editable Action: auto_generator_user_action_kind = auto_generator_user_action_kind.SetRate
    @editable OnTrigger: trigger_device = trigger_device{}
    @editable Value: float = 0.0

auto_generator_user_action := class:
    Device: auto_generator_device
    Config: auto_generator_user_action_config

    OnTriggered(Agent:?agent):void=
        case(Config.Action):
            auto_generator_user_action_kind.Enable => Device.Enable()
            auto_generator_user_action_kind.Disable => Device.Disable()
            auto_generator_user_action_kind.ToggleEnable => Device.ToggleEnable()
            auto_generator_user_action_kind.SetRate => Device.SetRate(Config.Value)
            auto_generator_user_action_kind.ModifyRate => Device.ModifyRate(Config.Value)
