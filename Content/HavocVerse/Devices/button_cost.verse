using { /Fortnite.com/Devices }
using { /Verse.org/Simulation }

using { HavocVerse.Debug }
using { HavocVerse.Devices }

button_cost_device := class(creative_device, aware_device):
    @editable p_ResourceManager: resource_manager_device = resource_manager_device{}
    @editable p_Buttons: []button_device = array{}
    @editable p_CostResource: string = "money"
    @editable p_CostAmount: int = 0
    @editable p_SuccessTrigger: trigger_device = trigger_device{}

    DeviceMeta<override>()<transacts>:aware_device_meta=
        aware_device_meta{
            DeviceName := "button_cost"
        }

    OnBegin<override>()<suspends>:void=
        AwareInitDevice(Self, OnInit)

    OnInit<private>()<suspends>:void=
        for (Button:p_Buttons) {
            Button.InteractedWithEvent.Subscribe(OnButtonInteractedWith)
        }

    OnButtonInteractedWith<private>(Agent:agent):void=
        Success := p_ResourceManager.Take(Agent, p_CostResource, p_CostAmount, ?NotifyNotEnough := true, ?SendFeedback := true)
        if (Success?) {
            p_SuccessTrigger.Trigger(Agent)
        }
