using { /Fortnite.com/Devices }
using { /Fortnite.com/Devices/CreativeAnimation }
using { /Fortnite.com/Game }
using { /Fortnite.com/Characters }
using { /Fortnite.com/Playspaces }
using { /Verse.org/Simulation }
using { /Verse.org/Assets }
using { /Verse.org/Colors }
using { /UnrealEngine.com/Temporary/Diagnostics }
using { /UnrealEngine.com/Temporary/SpatialMath }

using { HavocVerse.Debug }
using { HavocVerse.Math }
using { HavocVerse.Logging }
using { HavocVerse.Util }

buy_plate_position_source := enum:
    BuyPlateDevice
    StaticLinkedProp

buy_plate_manager_group<public> := class<concrete>:
    @editable GroupName: string = ""
    @editable AccessControl: ?access_control_device = false
    @editable Priority: int = 0
    @editable AutoVolumes: []bounds_config = array{}
    @editable ExtraBuyPlates: []creative_device = array{}

buy_plate_linked_prop_kind := enum:
    StaticLinkedProp
    PooledLinkedProp

buy_plate_cost_met_reflection_config := class<concrete>:
    @editable LinkedPropKind: buy_plate_linked_prop_kind = buy_plate_linked_prop_kind.StaticLinkedProp
    @editable LinkedPropIndex: int = 0
    @editable MetMesh: string = ""
    @editable MetMaterial: string = ""
    @editable MetAdditionalMaterials: []string = array{}
    @editable UnmetMesh: string = ""
    @editable UnmetMaterial: string = ""
    @editable UnmetAdditionalMaterials: []string = array{}

buy_plate_manager_unlock_animation_config := class<concrete>:
    @editable PropCustomDataFilter: ?string = false
    @editable FilterInvert: logic = false
    @editable Priority: int = 0
    @editable StartOffset: vector3 = vector3{}
    @editable StartScale: vector3 = vector3{X := 1.0, Y := 1.0, Z := 1.0}
    @editable Keyframes: []buy_plate_manager_unlock_animation_config_keyframe = array{}

    MatchesFilter<internal>(CustomData:[]string)<transacts><decides>:void=
        if (Filter := PropCustomDataFilter?) {
            if (FilterInvert?) {
                not CustomData.Find[Filter]
            } else {
                CustomData.Find[Filter]
            }
        }

    ToKeyframeDeltas<internal>(InitialTransform:transform, Origin:transform)<transacts>:[]keyframe_delta=
        var LastTransform: transform = Origin
        for (KeyframeConfig:Keyframes) {
            TargetPosition := InitialTransform.Translation + InitialTransform.Rotation.RotateVector(KeyframeConfig.TargetAbsoluteOffset)
            NewTransform := transform{
                Translation := TargetPosition
                Rotation := InitialTransform.Rotation
                Scale := SafeScale(KeyframeConfig.TargetScale)
            }
            DeltaLocation := TargetPosition - LastTransform.Translation
            set LastTransform = NewTransform
            keyframe_delta{
                DeltaLocation := DeltaLocation
                DeltaRotation := IdentityRotation()
                DeltaScale := NewTransform.Scale / InitialTransform.Scale
                Time := KeyframeConfig.Duration
                Interpolation := GetInterpolationType(KeyframeConfig.Interpolation)
            }
        }

TooltipVFX<internal><localizes>:message = "Particle System asset path"
TooltipEffectProp<internal><localizes>:message = "Spawns a prop instead of a raw Particle System"

buy_plate_manager_unlock_effect_config := class<concrete>:
    @editable PropCustomDataFilter: ?string = false
    @editable FilterInvert: logic = false
    @editable Priority: int = 0

    @editable{ToolTip := TooltipVFX} VFX: string = ""
    @editable VFXDespawnDelay: float = 1.5
    @editable ParticleSpawnOffset: vector3 = vector3{Z := 100.0}

    @editable{ToolTip := TooltipEffectProp} EffectProp: ?creative_prop_asset = false
    @editable PropDespawnDelay: float = 1.5
    @editable PropSpawnOffset: vector3 = vector3{Z := 100.0}

    MatchesFilter<internal>(CustomData:[]string)<transacts><decides>:void=
        if (Filter := PropCustomDataFilter?) {
            if (FilterInvert?) {
                not CustomData.Find[Filter]
            } else {
                CustomData.Find[Filter]
            }
        }

buy_plate_manager_unlock_trigger_config := class<concrete>:
    @editable CustomDataFilter: ?string = false
    @editable FilterInvert: logic = false
    @editable Trigger: trigger_device = trigger_device{}
    @editable OnlyFirstUnlock: logic = false

    MatchesFilter<internal>(CustomData:[]string, FirstUnlock:logic)<transacts><decides>:void=
        (not (OnlyFirstUnlock?)) or (FirstUnlock?)
        if (Filter := CustomDataFilter?) {
            if (FilterInvert?) {
                not CustomData.Find[Filter]
            } else {
                CustomData.Find[Filter]
            }
        }

buy_plate_manager_unlock_animation_config_keyframe := class<concrete>:
    @editable TargetAbsoluteOffset: vector3 = vector3{}
    @editable TargetScale: vector3 = vector3{X := 1.0, Y := 1.0, Z := 1.0}
    @editable Duration: float = 1.0
    @editable Interpolation: animation_interpolation_kind = animation_interpolation_kind.AIK_EaseIn

buy_plate_cost_met_filter := enum:
    WhenCostMet
    WhenCostUnmet

TooltipCostMetFilter<internal><localizes>:message = "Pool and release this prop based on cost met reflection state"

buy_plate_manager_linked_props_pool_config := class<concrete>:
    @editable GroupFilter: ?string = false
    @editable{ToolTip := TooltipCostMetFilter} CostMetFilter: ?buy_plate_cost_met_filter = false
    @editable FilterExactMatch: logic = false
    @editable FilterInvert: logic = false

    @editable PropInstances: []creative_prop = array{}
    @editable PropInstancesFromTag: string = ""
    @editable Offset: vector3 = vector3{}

    @editable DynamicMesh: string = ""
    @editable DynamicMaterial: string = ""

    var m_UniquePropPool: ?prop_pool = false
    var m_TagPropPool: ?prop_pool = false

    ReserveProp<internal>()<transacts><decides>:prop_pool_entry=
        if (Entry := m_UniquePropPool?.Reserve[]) {
            Entry
        } else {
            m_TagPropPool?.Reserve[]
        }

    MatchesFilter<internal>(GroupName:string)<transacts><decides>:void=
        if (Filter := GroupFilter?) {
            Res := if (FilterExactMatch?, GroupName = Filter) {
                true
            } else if (not (FilterExactMatch?), GroupName.Contains[Filter]) {
                true
            } else {
                false
            }
            if (FilterInvert?) {
                not (Res?)
            } else {
                Res?
            }
        }

buy_plate_manager_device<public> := class(creative_device, aware_device, aware_debug_device):
    @editable p_DevId<public>: string = "N/A"

    @editable p_Groups: []buy_plate_manager_group = array{}

    @editable p_PositionSource: buy_plate_position_source = buy_plate_position_source.BuyPlateDevice
    @editable p_PositionSourceIndex: int = 0
    @editable p_PositionSourceOffset: vector3 = vector3{}
    @editable p_PositionSourceOffsetUseLocalRotation: logic = true

    @editable p_UseFakeVolumes: logic = false
    @editable p_FakeVolumeRadius: float = 75.0
    @editable p_FakeVolumeHeight: float = 20.0

    @editable p_EnableCostMetReflection: logic = false
    @editable p_CostMetReflectionConfig: buy_plate_cost_met_reflection_config = buy_plate_cost_met_reflection_config{}

    @editable p_UnlockAnimations: []buy_plate_manager_unlock_animation_config = array{}
    @editable p_UnlockEffects: []buy_plate_manager_unlock_effect_config = array{}
    @editable p_UnlockTriggers: []buy_plate_manager_unlock_trigger_config = array{}

    @editable p_PooledLinkedProps: []buy_plate_manager_linked_props_pool_config = array{}

    @editable p_LegacyAccessControls: logic = false

    var m_Groups: []buy_plate_manager_active_group = array{}

    DeviceMeta<override>()<transacts>:aware_device_meta=
        aware_device_meta{
            DeviceName := "buy_plate_manager"
            DevId := p_DevId
        }

    DebugStats<override>()<transacts>:aware_debug_stats=
        aware_debug_stats{
            Length := map{
                "groups" => m_Groups.Length
            }
        }

    OnBegin<override>()<suspends>:void=
        AwareInitDevice(Self, OnInit)

    OnInit<private>()<suspends>:void=
        if (p_EnableCostMetReflection?) {
            if (StringEmptyXOR[p_CostMetReflectionConfig.MetMesh, p_CostMetReflectionConfig.UnmetMesh]) {
                Log("{Self} has cost met reflection setup partially. MetMesh and UnmetMesh are not both set", ?Level := log_level.Error)
            }
            if (StringEmptyXOR[p_CostMetReflectionConfig.MetMaterial, p_CostMetReflectionConfig.UnmetMaterial]) {
                Log("{Self} has cost met reflection setup partially. MetMaterial and UnmetMaterial are not both set", ?Level := log_level.Error)
            }
            if (p_CostMetReflectionConfig.MetAdditionalMaterials.Length <> p_CostMetReflectionConfig.UnmetAdditionalMaterials.Length) {
                Log("{Self} has cost met reflection setup partially. MetAdditionalMaterials and UnmetAdditionalMaterials are not the same length", ?Level := log_level.Error)
            }
        }

        set m_Groups = for (GroupConfig:p_Groups) {
            TargetVolumes := for (BoundsConfig:GroupConfig.AutoVolumes) {
                BoundsConfig.ToBounds()
            }
            buy_plate_manager_active_group{
                Playspace := GetPlayspace()
                Manager := Self
                GroupName := GroupConfig.GroupName
                Config := GroupConfig
                TargetVolumes := TargetVolumes
                AccessControl := GroupConfig.AccessControl
            }
        }
        for (Index->LinkedPropConfig:p_PooledLinkedProps) {
            if (LinkedPropConfig.PropInstances.Length > 0) {
                PropPool := prop_pool{}
                for (PropInstance:LinkedPropConfig.PropInstances) {
                    PropPool.Add(PropInstance)
                }
                set LinkedPropConfig.m_UniquePropPool = Some(PropPool)
            }
            if (
                LinkedPropConfig.PropInstancesFromTag.Length > 0
                TagsProvider := AwareDefaultTagsProvider[]
                Tag := TagsProvider.GetTag[LinkedPropConfig.PropInstancesFromTag]
            ) {
                var SharedPropPool: ?prop_pool = false
                for (
                    J->Other:p_PooledLinkedProps
                    J < Index
                    not (SharedPropPool?)
                    Other.PropInstancesFromTag = LinkedPropConfig.PropInstancesFromTag
                ) {
                    set SharedPropPool = Other.m_TagPropPool
                }
                if (PropPool := SharedPropPool?) {
                    Log(Self, "Linked prop pool {Index} is sharing {PropPool.Size()} props tagged \"{LinkedPropConfig.PropInstancesFromTag}\" with another pool.", ?Level := log_level.Debug)
                    set LinkedPropConfig.m_TagPropPool = SharedPropPool
                } else {
                    PropPool := prop_pool{}
                    PropPool.AddFromTag(Tag)
                    set LinkedPropConfig.m_TagPropPool = Some(PropPool)
                }
            }
            PropCountFromTag := option{LinkedPropConfig.m_TagPropPool?.Size()}.UnwrapOr(0)
            PropCount := option{LinkedPropConfig.m_UniquePropPool?.Size()}.UnwrapOr(0) + PropCountFromTag
            FromTagInfo := if (LinkedPropConfig.PropInstancesFromTag.Length > 0) {
                "{PropCountFromTag} from tag \"{LinkedPropConfig.PropInstancesFromTag}\""
            } else {
                "{PropCountFromTag} from tag"
            }
            Log(Self, "Loaded {PropCount} props ({FromTagInfo}) for linked prop pool {Index}")
        }

        for (Index->Group:m_Groups) {
            if (AccessControl := Group.AccessControl?) {
                AccessControl.SubscribeToPlayerJoinEvents(Group.OnPlayerAdded, Group.OnPlayerRemoved)
            }

            for (ExtraIndex->BuyPlateDevice:Group.Config.ExtraBuyPlates) {
                Group.SetupBuyplate(BuyPlateDevice, true)
                if (not buy_plate_device[BuyPlateDevice]) {
                    Log(Self, "ExtraBuyPlates[{ExtraIndex}] of {Group} is not a buyplate: {BuyPlateDevice}")
                }
            }

            if (p_UseFakeVolumes?) {
                spawn:
                    Group.StartFakeVolumesLoop()
            }
        }

        AwareSubscribeToDeviceInit(SetupBuyplate, Some(aware_device_filter{
            DeviceName := "buy_plate"
        }))

    SetupBuyplate<private>(BuyPlateDevice:creative_device):void=
        for (Group:m_Groups) {
            Group.SetupBuyplate(BuyPlateDevice, false)
        }

    GetBuyplatePosition<public>(BuyPlate:buy_plate_device, ?ExtraOffset:vector3 = vector3{})<transacts><decides>:vector3=
        Transform := case(p_PositionSource):
            buy_plate_position_source.BuyPlateDevice => BuyPlate.GetTransform()
            buy_plate_position_source.StaticLinkedProp =>
                if (not BuyPlate.p_LinkedProps[p_PositionSourceIndex]) {
                    Log("{Self} has a uses static linked prop position source but {BuyPlate} has no linked prop with index {p_PositionSourceIndex}", ?Level := log_level.Error)
                }
                LinkedProp := BuyPlate.p_LinkedProps[p_PositionSourceIndex]
                if (not LinkedProp.IsValid[]) {
                    Log("{Self} has a uses static linked prop position source but {BuyPlate} has an invalid (broken link?) linked prop with index {p_PositionSourceIndex}", ?Level := log_level.Error)
                }
                LinkedProp.IsValid[]
                LinkedProp.GetTransform()
        if (p_PositionSourceOffsetUseLocalRotation?) {
            Transform.Translation + Transform.Rotation.RotateVector(p_PositionSourceOffset) + Transform.Rotation.RotateVector(ExtraOffset)
        } else {
            Transform.Translation + p_PositionSourceOffset + ExtraOffset
        }

    GetUnlockAnimation<internal>(CustomData:[]string)<transacts><decides>:buy_plate_manager_unlock_animation_config=
        MatchingConfigs := for (Config:p_UnlockAnimations, Config.MatchesFilter[CustomData]) {
            Config
        }
        MatchingConfigs.QuickSortDesc(CmpBuyplateManagerUnlockAnimationConfig)[0]

    GetUnlockEffect<internal>(CustomData:[]string)<transacts><decides>:buy_plate_manager_unlock_effect_config=
        MatchingConfigs := for (Config:p_UnlockEffects, Config.MatchesFilter[CustomData]) {
            Config
        }
        MatchingConfigs.QuickSortDesc(CmpBuyplateManagerUnlockEffectConfig)[0]

    DebugDrawFakeVolumes<public>(DebugDraw:debug_draw, Args:debug_draw_args, OnlyEnabled:logic):void=
        if (not (p_UseFakeVolumes?)) {
            return
        }
        for (Group:m_Groups) {
            Group.DebugDrawFakeVolumes(DebugDraw, Args, OnlyEnabled)
        }

    DebugDrawAutoVolumesBoundaries<public>(DebugDraw:debug_draw, Args:debug_draw_args):void=
        Colors := array{NamedColors.Red, NamedColors.Green, NamedColors.Blue, NamedColors.Yellow, NamedColors.Purple, NamedColors.Cyan, NamedColors.Orange, NamedColors.Magenta}
        for (Index->Group:m_Groups) {
            Color := option{Colors[DefaultingMod(Index, Colors.Length)]}.UnwrapOr(NamedColors.White)
            Group.DebugDrawTargetVolumes(DebugDraw, debug_draw_args{
                MakeDebugDrawArgs<constructor>(Args)
                Color := Color
            })
        }

buy_plate_manager_active_group<public> := class<unique>:
    Playspace<public>: fort_playspace
    Manager<public>: buy_plate_manager_device
    GroupName<public>: string
    Config<public>: buy_plate_manager_group

    TargetVolumes<public>: []bounds
    AccessControl<public>: ?access_control_device

    var m_VolumeBuyplates: []buy_plate_device = array{}
    var m_InFakeVolumes: [int][player]void = map{}

    IsApplicable<internal>(BuyPlate:buy_plate_device, Position:vector3)<transacts><decides>:void=
        IsWithinAnyBounds[TargetVolumes, Position] or IsInExtraBuyplates[BuyPlate]

    IsInExtraBuyplates<private>(BuyPlate:buy_plate_device)<transacts><decides>:void=
        var Found: logic = false
        for (
            ExtraBuyPlate:Config.ExtraBuyPlates
            not (Found?)
            buy_plate_device[ExtraBuyPlate] = BuyPlate
        ) {
            set Found = true
        }
        Found?

    SetupBuyplate<internal>(BuyPlateDevice:creative_device, Force:logic):void=
        if (BuyPlate := buy_plate_device[BuyPlateDevice]) {
            if (Force?) {
                SetupBuyplate0(BuyPlate)
            } else if (
                Position := Manager.GetBuyplatePosition[BuyPlate]
                IsWithinAnyBounds[TargetVolumes, Position]
            ) {
                var Proceed: logic = true
                for (
                    Group:Manager.m_Groups
                    Group <> Self
                    Proceed?
                    Group.Config.Priority > Self.Config.Priority
                    Group.IsApplicable[BuyPlate, Position]
                ) {
                    set Proceed = false
                }
                if (Proceed?) {
                    SetupBuyplate0(BuyPlate)
                }
            }
        }

    SetupBuyplate0<internal>(BuyPlate:buy_plate_device):void=
        if (BuyPlate.p_ExpectsManager?) {
            Success := BuyPlate.Setup(Some(Self))
            if (Success?, Manager.p_UseFakeVolumes?, not BuyPlate.p_StealthPlate?) {
                set m_VolumeBuyplates += array{BuyPlate}
            }
        } else {
            Log("{BuyPlate} is not expecting a manager but was a potential target of {Self}.", ?Level := log_level.Debug)
        }

    OnPlayerAdded<internal>(Player:player):void=
        {}

    ## Fake Volumes {{{
    OnPlayerRemoved<internal>(Player:player, GameQuit:logic):void=
        for (BPIndex->KnownPlayersIn:m_InFakeVolumes) {
            if (KnownPlayersIn[Player]) {
                FilteredKnownPlayersIn := KnownPlayersIn.RemoveByKey(Player)(0)
                if (FilteredKnownPlayersIn.Length > 0) {
                    if. set m_InFakeVolumes[BPIndex] = FilteredKnownPlayersIn
                } else {
                    set m_InFakeVolumes = m_InFakeVolumes.RemoveByKey(BPIndex)(0)
                }
                if (BuyPlate := m_VolumeBuyplates[BPIndex]) {
                    HandleFakeVolumeLeave(BuyPlate, Player)
                }
            }
        }

    var m_LastFakeVolumesTick: float = 0.0

    StartFakeVolumesLoop<internal>()<suspends>:void=
        loop {
            AbortEvent := event(){}
            set m_LastFakeVolumesTick = GetSimulationElapsedTime()
            race:
                RunFakeVolumesLoop(AbortEvent)
                loop {
                    Sleep(1.0)
                    ElapsedSinceLastTick := GetSimulationElapsedTime() - m_LastFakeVolumesTick
                    if (ElapsedSinceLastTick >= 5.0) {
                        Log("Detected stale fake volumes loop on {Self} for more than 5s. Restarting loop..", ?Level := log_level.Warning)
                        break
                    }
                }
        }

    RunFakeVolumesLoop<private>(AbortEvent:event())<suspends>:void=
        WithinDistSqr := Manager.p_FakeVolumeRadius * Manager.p_FakeVolumeRadius
        ZTolerance := Manager.p_FakeVolumeHeight + 80.0
        race:
            AbortEvent.Await()
            loop {
                Sleep(0.0)
                set m_LastFakeVolumesTick = GetSimulationElapsedTime()
                PlayersAndPos := for (
                    Player:Playspace.GetPlayers()
                    Player.IsActive[]
                    (not (AccessControl?)) or (AccessControl?.IsAllowed[Player])
                    FortCharacter := Player.GetFortCharacter[]
                    FortCharacter.IsActive[]
                ) {
                    (Player, FortCharacter.GetTransform().Translation)
                }
                if (PlayersAndPos.Length > 0) {
                    TickFakeVolumesLoop(PlayersAndPos, WithinDistSqr, ZTolerance)
                }
            }

    TickFakeVolumesLoop<private>(
        PlayersAndPos:[]tuple(player,vector3),
        WithinDistSqr:float,
        ZTolerance:float
    ):void=
        var Events: []tuple(buy_plate_device, player, logic) = array{}
        for (
            BPIndex->BuyPlate:m_VolumeBuyplates
            BuyPlate.IsEnabled[]
            BuyPlatePosition := Manager.GetBuyplatePosition[BuyPlate]
        ) {
            var KnownPlayersIn: ?[player]void = option{m_InFakeVolumes[BPIndex]}
            for (PlayerEntry:PlayersAndPos, Player := PlayerEntry(0)) {
                if (
                    DistanceSquaredXY(BuyPlatePosition, PlayerEntry(1)) <= WithinDistSqr
                    Abs(BuyPlatePosition.Z - PlayerEntry(1).Z) <= ZTolerance
                ) {
                    if (not (KnownPlayersIn?[Player])) {
                        set KnownPlayersIn = Some(AddToLazyMap(KnownPlayersIn, Player, void))
                        set Events += array{(BuyPlate, Player, true)}
                    }
                } else if (l_KnownPlayersIn := KnownPlayersIn?, l_KnownPlayersIn[Player]) {
                    set KnownPlayersIn = Some(l_KnownPlayersIn.RemoveByKey(Player)(0))
                    set Events += array{(BuyPlate, Player, false)}
                }
            }
            if (l_KnownPlayersIn := KnownPlayersIn?) {
                if (l_KnownPlayersIn.Length > 0) {
                    set m_InFakeVolumes = AddToMap(m_InFakeVolumes, BPIndex, l_KnownPlayersIn)
                } else {
                    set m_InFakeVolumes = m_InFakeVolumes.RemoveByKey(BPIndex)(0)
                }
            } else if (m_InFakeVolumes[BPIndex]) {
                set m_InFakeVolumes = m_InFakeVolumes.RemoveByKey(BPIndex)(0)
            }
        }
        if (Events.Length = 0) {
            return
        }
        for (Event:Events) {
            if (Event(2)?) {
                HandleFakeVolumeEnter(Event(0), Event(1))
            } else {
                HandleFakeVolumeLeave(Event(0), Event(1))
            }
        }

    HandleFakeVolumeEnter<private>(BuyPlate:buy_plate_device, Player:player):void=
        Log(MsgFakeVolumeEvent(Player, "entering", ToString(BuyPlate)), ?Level := log_level.Verbose)
        BuyPlate.OnAgentInteract(Player)

    HandleFakeVolumeLeave<private>(BuyPlate:buy_plate_device, Player:player):void=
        Log(MsgFakeVolumeEvent(Player, "leaving", ToString(BuyPlate)), ?Level := log_level.Verbose)

    MsgFakeVolumeEvent<private><localizes>(Player:player, Action:string, DeviceStr:string):message=
        "Detected '{Player}' {Action} fake volume of {DeviceStr}"

    DebugDrawFakeVolumes<internal>(DebugDraw:debug_draw, Args:debug_draw_args, OnlyEnabled:logic):void=
        for (
            BuyPlate:m_VolumeBuyplates
            BuyPlatePosition := Manager.GetBuyplatePosition[BuyPlate]
            (not (OnlyEnabled?)) or (BuyPlate.IsEnabled[])
        ) {
            DebugDraw.DrawCylinder(
                BuyPlatePosition,
                BuyPlatePosition + vector3{Z := Manager.p_FakeVolumeHeight},
                ?Radius := Manager.p_FakeVolumeRadius,
                ?NumSegments := 24,
                ?Color := Args.Color,
                ?Thickness := Args.Thickness,
                ?DrawDurationPolicy := Args.DrawDurationPolicy,
                ?Duration := Args.Duration
            )
        }

    DebugDrawTargetVolumes<internal>(DebugDraw:debug_draw, Args:debug_draw_args):void=
        for (Bounds:TargetVolumes) {
            Bounds.DrawDebugBox(DebugDraw, Args)
        }
    ## }}}

ToString<public>(Group:buy_plate_manager_active_group):string=
    "{Group.Manager}>{Group.GroupName}"

CmpBuyplateManagerUnlockAnimationConfig<internal>(
    A:buy_plate_manager_unlock_animation_config,
    B:buy_plate_manager_unlock_animation_config
)<transacts>:int=
    PriorityRes := CmpInt(A.Priority, B.Priority)
    if (PriorityRes <> 0) {
        PriorityRes
    } else if (A.PropCustomDataFilter?, not (B.PropCustomDataFilter?)) {
        1
    } else if (B.PropCustomDataFilter?, not (A.PropCustomDataFilter?)) {
        -1
    } else {
        0
    }

CmpBuyplateManagerUnlockEffectConfig<internal>(
    A:buy_plate_manager_unlock_effect_config,
    B:buy_plate_manager_unlock_effect_config
)<transacts>:int=
    PriorityRes := CmpInt(A.Priority, B.Priority)
    if (PriorityRes <> 0) {
        PriorityRes
    } else if (A.PropCustomDataFilter?, not (B.PropCustomDataFilter?)) {
        1
    } else if (B.PropCustomDataFilter?, not (A.PropCustomDataFilter?)) {
        -1
    } else {
        0
    }
