using { /Fortnite.com/Devices }
using { /Fortnite.com/Characters }
using { /Verse.org/Simulation }
using { /Verse.org/Colors }
using { /UnrealEngine.com/Temporary/Diagnostics }
using { /UnrealEngine.com/Temporary/UI }
using { /UnrealEngine.com/Temporary/SpatialMath }

using { HavocVerse.Debug }
using { HavocVerse.Logging }
using { HavocVerse.UI }
using { HavocVerse.Util }

cinematics_player_instance_config := class<concrete>:
    @editable CinematicSequence: cinematic_sequence_device = cinematic_sequence_device{}
    @editable HUDController: []hud_controller_device = array{}

cinematics_player_hud_controller_update_mode := enum:
    DontUpdate
    PlayerClass
    PlayerTeam

cinematics_player_device := class(creative_device, aware_device, aware_debug_device):
    @editable p_DevId: string = "N/A"

    @editable p_PlayOnJoin: logic = false
    @editable p_JoinAwaitPostInit: logic = true
    @editable p_PlayTriggers: []trigger_device = array{}
    @editable p_Delay: float = -1.0
    @editable p_ApplyStateAtDelayStart: logic = true
    @editable p_ShortenDelayWithDetection: logic = true
    @editable p_Instances: []cinematics_player_instance_config = array{}
    @editable p_HUDControllerUpdateMode: cinematics_player_hud_controller_update_mode = cinematics_player_hud_controller_update_mode.DontUpdate
    @editable p_AutoStasis: logic = true
    @editable p_FailsafeCutoffDelay: float = 45.0

    @editable p_SkipInputTrigger: []input_trigger_device = array{}
    @editable p_SkipInputHoldTime: float = 0.5

    @editable p_SkipTextEnabled: logic = true
    @editable p_SkipTextPlacement: ui_canvas_placement_config = ui_canvas_placement_config{
        Anchor := vector2{X := 0.5, Y := 0.5}
        Alignment := vector2{X := 0.5, Y := 0.5}
    }
    @editable p_SkipTextStyle: ui_text_style_config = ui_text_style_config{
        Color := NamedColors.White
    }
    @editable p_SkipTextFormatter: number_formatter_config = number_formatter_config{
        Template := "Skipping in %Val%s.."
        MinDecimals := 2
        MaxDecimals := 2
    }

    @editable p_OnCinematicStartTrigger: []trigger_device = array{}
    @editable p_OnCinematicStopTrigger: []trigger_device = array{}

    var m_Entries: []cinematics_player_entry = array{}

    DeviceMeta<override>()<transacts>:aware_device_meta=
        aware_device_meta{
            DeviceName := "cinematics_player"
            DevId := p_DevId
        }

    DebugStats<override>()<transacts>:aware_debug_stats=
        aware_debug_stats{
            Length := map{
                "entries" => m_Entries.Length
            }
        }

    OnBegin<override>()<suspends>:void=
        AwareInitDevice(Self, OnInit)

    OnInit<private>()<suspends>:void=
        set m_Entries = for (Config:p_Instances) {
            Entry := cinematics_player_entry{
                Device := Self
                SequenceDevice := Config.CinematicSequence
                HUDController := option{Config.HUDController[0]}
            }
            Entry.SequenceDevice.StoppedEvent.Subscribe(Entry.OnSequenceStopped)
            Entry
        }

        GetPlayspace().PlayerRemovedEvent().Subscribe(OnPlayerRemoved)
        for (Trigger:p_PlayTriggers) {
            Trigger.TriggeredEvent.Subscribe(OnTriggered)
        }

        if (InputTrigger := p_SkipInputTrigger[0]) {
            InputTrigger.PressedEvent.Subscribe(OnSkipInputTriggerPressed)
        }

        if (p_JoinAwaitPostInit?, p_PlayOnJoin?) {
            AwareSubscribeToPostInit(OnPostInit)
        } else if (p_PlayOnJoin?) {
            SetupJoinHook()
        }

    OnPostInit<private>():void=
        if (p_PlayOnJoin?) {
            SetupJoinHook()
        }

    SetupJoinHook<private>():void=
        GetPlayspace().PlayerAddedEvent().Subscribe(OnPlayerAdded)
        for (Player:GetPlayspace().GetPlayers()) {
            OnPlayerAdded(Player)
        }

    OnTriggered<private>(Agent:?agent):void=
        if (Player := player[Agent?]) {
            spawn:
                Play(Player, false)
        }

    OnPlayerAdded<private>(Player:player):void=
        if (p_PlayOnJoin?) {
            spawn:
                Play(Player, false)
        }

    OnPlayerRemoved<private>(Player:player):void=
        for (Entry:m_Entries, Entry.m_Playing?.Player = Player) {
            set Entry.m_Playing = None()
            Entry.SequenceDevice.GoToEndAndStop(Player)
        }

    OnSkipInputTriggerPressed<private>(Agent:agent):void=
        if (Player := player[Agent]) {
            var InCinematic: logic = false
            for (Entry:m_Entries, Entry.m_Playing?.Player = Player) {
                set InCinematic = true
            }
            if (not (InCinematic?)) {
                return
            }
            if (p_SkipInputHoldTime > 0.0, InputTrigger := p_SkipInputTrigger[0]) {
                spawn:
                    StartSkipHoldLoop(InputTrigger, Player)
            } else {
                Skip(Player)
            }
        }

    StartSkipHoldLoop<private>(InputTrigger:input_trigger_device, Player:player)<suspends>:void=
        StartedAt := GetSimulationElapsedTime()
        Feedback := if (p_SkipTextEnabled?) {
            TextWidget := p_SkipTextStyle.ToTextBlock("")
            Canvas := canvas{
                Slots := array{
                    p_SkipTextPlacement.ToCanvasSlot(TextWidget)
                }
            }
            GlobalUIAddWidget(Player, Canvas)
            Some((Canvas, TextWidget))
        } else {
            None()
        }
        Formatter := p_SkipTextFormatter.CreateFormatter()
        loop {
            Sleep(0.0)
            if (not InputTrigger.IsHeld[Player]) {
                break
            }
            ElapsedTime := GetSimulationElapsedTime() - StartedAt
            if (ElapsedTime >= p_SkipInputHoldTime) {
                Skip(Player)
                break
            }
            if (l_Feedback := Feedback?) {
                TextWidget := l_Feedback(1)
                RemainingSeconds := p_SkipInputHoldTime - ElapsedTime
                Text := Formatter.Format(RemainingSeconds)
                TextWidget.SetText(ToMessage(Text))
            }
        }
        if (l_Feedback := Feedback?) {
            GlobalUIRemoveWidget(Player, l_Feedback(0))
        }

    Skip<private>(Player:player):void=
        for (Entry:m_Entries, Entry.m_Playing?.Player = Player) {
            Entry.StopPlaying("skip", false)
        }

    AwaitDelay<private>(Player:player)<suspends>:void=
        if (p_ShortenDelayWithDetection?) {
            race:
                Sleep(p_Delay)
                block:
                    DelayShortenDetection(Player)
                    Log(Self, "Shortened delay for '{Player}' because movement was detected", ?Level := log_level.Debug)
        } else {
            Sleep(p_Delay)
        }

    DelayShortenDetection<private>(Player:player)<suspends>:void=
        if (FortCharacter := Player.GetFortCharacter[]) {
            var LastViewRot: rotation = FortCharacter.GetViewRotation()
            loop {
                Sleep(0.0)
                ViewRot := FortCharacter.GetViewRotation()
                if (Distance(LastViewRot, ViewRot) > 0.0) {
                    return
                }
                set LastViewRot = ViewRot
            }
        }

    Play<public>(Player:player, Immediate:logic)<suspends>:void=
        if (not Immediate?, not p_ApplyStateAtDelayStart?, p_Delay >= 0.0) {
            AwaitDelay(Player)
        }

        for (Entry:m_Entries, Entry.m_Playing?.Player = Player) {
            Log(MsgAlreadyPlaying(Player, ToString(Self)), ?Level := log_level.Debug)
            return
        }

        var AvailableEntry: ?cinematics_player_entry = None()
        for (Entry:m_Entries, not (AvailableEntry?), not (Entry.m_Playing?)) {
            set AvailableEntry = Some(Entry)
        }
        if (not (AvailableEntry?)) {
            # re use the oldest in progress
            Sorted := m_Entries.QuickSort(CmpCinematicsEntry)
            if (ReuseEntry := Sorted[0]) {
                Log(MsgReuseEntry(Player, ToString(Self)), ?Level := log_level.Warning)
                ReuseEntry.StopPlaying("re-use", false)
                set AvailableEntry = Some(ReuseEntry)
            }
        }

        if (l_Entry := AvailableEntry?) {
            Log(MsgStart(Player, ToString(Self)), ?Level := log_level.Debug)
            StasisStatus := if (p_AutoStasis?, FortCharacter := Player.GetFortCharacter[]) {
                FortCharacter.PutInStasis(stasis_args{
                    AllowTurning := false
                    AllowFalling := false
                    AllowEmotes := false
                })
                WasVulnerable := if (FortCharacter.IsVulnerable[]). true else false
                FortCharacter.SetVulnerability(false)
                Some(cinematics_player_stasis_status{
                    Vulnerable := WasVulnerable
                })
            } else {
                None()
            }
            Playing := cinematics_player_entry_playing{
                Player := Player
                StartedAt := GetSimulationElapsedTime()
                StasisStatus := StasisStatus
            }
            set l_Entry.m_Playing = Some(Playing)
            if (HUDController := l_Entry.HUDController?) {
                case(p_HUDControllerUpdateMode):
                    cinematics_player_hud_controller_update_mode.DontUpdate => {}
                    cinematics_player_hud_controller_update_mode.PlayerClass =>
                        HUDController.UpdateAffectedClass(Player)
                    cinematics_player_hud_controller_update_mode.PlayerTeam =>
                        HUDController.UpdateAffectedTeam(Player)
                HUDController.Enable()
            }

            if (not Immediate?, p_ApplyStateAtDelayStart?, p_Delay >= 0.0) {
                AwaitDelay(Player)
            }
            if (InputTrigger := p_SkipInputTrigger[0]) {
                InputTrigger.Register(Player)
            }

            for (StartTrigger:p_OnCinematicStartTrigger) {
                StartTrigger.Trigger(Player)
            }
            l_Entry.SequenceDevice.Play(Player)
            if (p_FailsafeCutoffDelay > 0.0) {
                spawn:
                    StartFailsafeCutoff(l_Entry, Playing)
            }
        } else {
            Log(MsgNoAvailableEntry(Player, ToString(Self)), ?Level := log_level.Error)
        }

    StartFailsafeCutoff<private>(Entry:cinematics_player_entry, Playing:cinematics_player_entry_playing)<suspends>:void=
        race:
            Playing.StopEvent.Await()
            block:
                Sleep(p_FailsafeCutoffDelay)
                if (Entry.m_Playing? = Playing) {
                    Entry.StopPlaying("failsafe", false)
                }

    MsgAlreadyPlaying<private><localizes>(Player:player, DeviceStr:string):message=
        "{DeviceStr} is already playing a cinematic for '{Player}'"
    MsgNoAvailableEntry<private><localizes>(Player:player, DeviceStr:string):message=
        "{DeviceStr} couldn't find or reuse a cinematic sequence to play to '{Player}'"
    MsgReuseEntry<private><localizes>(Player:player, DeviceStr:string):message=
        "{DeviceStr} is reusing an in-progress cinematic sequence to play to '{Player}'"
    MsgStart<private><localizes>(Player:player, DeviceStr:string):message=
        "{DeviceStr} starting cinematic for '{Player}'"

cinematics_player_entry := class<final>:
    Device: cinematics_player_device

    SequenceDevice: cinematic_sequence_device
    HUDController: ?hud_controller_device

    var m_Playing: ?cinematics_player_entry_playing = false

    OnSequenceStopped<internal>():void=
        StopPlaying("finished", true)

    StopPlaying<internal>(Cause:string, Finished:logic):void=
        if (Playing := m_Playing?) {
            Log(MsgStopPlaying(Playing.Player, Cause, ToString(Device)), ?Level := log_level.Debug)
            if (not (Finished?)) {
                SequenceDevice.GoToEndAndStop(Playing.Player)
            }
            if (l_HUDController := HUDController?) {
                l_HUDController.Disable()
                case(Device.p_HUDControllerUpdateMode):
                    cinematics_player_hud_controller_update_mode.DontUpdate => {}
                    cinematics_player_hud_controller_update_mode.PlayerClass =>
                        l_HUDController.ResetAffectedClass()
                    cinematics_player_hud_controller_update_mode.PlayerTeam =>
                        l_HUDController.ResetAffectedTeam()
            }
            if (InputTrigger := Device.p_SkipInputTrigger[0]) {
                InputTrigger.Unregister(Playing.Player)
            }
            for (StopTrigger:Device.p_OnCinematicStopTrigger) {
                StopTrigger.Trigger(Playing.Player)
            }
            set m_Playing = None()
            if (StasisStatus := Playing.StasisStatus?, FortCharacter := Playing.Player.GetFortCharacter[]) {
                FortCharacter.ReleaseFromStasis()
                FortCharacter.SetVulnerability(StasisStatus.Vulnerable)
            }
            Playing.StopEvent.Signal()
        }

    MsgStopPlaying<private><localizes>(Player:player, Cause:string, DeviceStr:string):message=
        "{DeviceStr} handling cinematic end for '{Player}' ({Cause})"

cinematics_player_entry_playing := class<final><unique>:
    Player: player
    StartedAt: float
    StasisStatus: ?cinematics_player_stasis_status

    StopEvent: event() = event(){}

cinematics_player_stasis_status := class<final>:
    Vulnerable: logic

CmpCinematicsEntry(A:cinematics_player_entry, B:cinematics_player_entry)<transacts>:int=
    if (A_Playing := A.m_Playing?, B_Playing := B.m_Playing?) {
        CmpFloat(A_Playing.StartedAt, B_Playing.StartedAt)
    } else {
        0
    }
