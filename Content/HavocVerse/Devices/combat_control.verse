using { /Fortnite.com/Devices }
using { /Fortnite.com/Characters }
using { /Fortnite.com/Game }
using { /Verse.org/Simulation }
using { /UnrealEngine.com/Temporary/Diagnostics }

using { HavocVerse.Debug }
using { HavocVerse.Logging }
using { HavocVerse.Util }

combat_control_damage_result := enum:
    Ignore
    Allow
    Disallow

combat_control_volume_config := class<concrete>:
    @editable Volume: volume_device = volume_device{}
    
    var m_TrackedInside: [agent]logic = map{}

    IsInVolume<public>(Agent:agent)<transacts><decides>:void=
        m_TrackedInside[Agent]

combat_control_device := class(creative_device, aware_device):
    @editable p_Volumes: []combat_control_volume_config = array{}

    @editable p_ControlOutsideVolumes: logic = true
    @editable p_OutsideVolumesAllowSwitch: []switch_device = array{}

    @editable p_AllowSwitch: switch_device = switch_device{}
    @editable p_IndividualChoiceSwitch: []switch_device = array{}

    @editable p_IgnoreSelfDamage: logic = true

    DeviceMeta<override>()<transacts>:aware_device_meta=
        aware_device_meta{
            DeviceName := "combat_control"
        }

    OnBegin<override>()<suspends>:void=
        AwareInitDevice(Self, OnInit)

    OnInit<private>()<suspends>:void=
        for (VolumeConfig:p_Volumes) {
            Closure := combat_control_volume_closure{
                VolumeConfig := VolumeConfig
            }
            VolumeConfig.Volume.AgentEntersEvent.Subscribe(Closure.OnAgentEntered)
            VolumeConfig.Volume.AgentExitsEvent.Subscribe(Closure.OnAgentExited)
        }

        GetPlayspace().PlayerAddedEvent().Subscribe(OnPlayerAdded)
        GetPlayspace().PlayerRemovedEvent().Subscribe(OnPlayerRemoved)
        for (Player:GetPlayspace().GetPlayers()) {
            OnPlayerAdded(Player)
        }

    OnPlayerAdded<private>(Player:player):void=
        if (FortCharacter := Player.GetFortCharacter[]) {
            Closure := combat_control_damaged_closure{
                Device := Self
                FortPlayer := FortCharacter
            }
            FortCharacter.DamagedEvent().Subscribe(Closure.OnDamaged)
            FortCharacter.DamagedShieldEvent().Subscribe(Closure.OnShieldDamaged)
        }

    OnPlayerRemoved<private>(Player:player):void=
        for (VolumeConfig:p_Volumes) {
            set VolumeConfig.m_TrackedInside = VolumeConfig.m_TrackedInside.RemoveByKey(Player)(0)
        }

    HandleDamage<internal>(Victim:agent, Damager:agent, DamageResult:damage_result):combat_control_damage_result=
        AffectedArea := if (p_Volumes.Length > 0) {
            var VictimInVolume: logic = false
            var DamagerInVolume: logic = false
            for (Volume:p_Volumes) {
                if (Volume.IsInVolume[Victim]) {
                    set VictimInVolume = true
                }
                if (Volume.IsInVolume[Damager]) {
                    set DamagerInVolume = true
                }
            }

            if (VictimInVolume? and DamagerInVolume?) {
                true
            } else {
                false
            }
        } else {
            true
        }
        if (not (AffectedArea?)) {
            if (p_ControlOutsideVolumes?, AllowSwitch := p_OutsideVolumesAllowSwitch[0]) {
                if (AllowSwitch.GetCurrentState[]) {
                    return combat_control_damage_result.Allow
                } else {
                    return combat_control_damage_result.Disallow
                }
            } else {
                return combat_control_damage_result.Ignore
            }
        }

        if (
            ChoiceSwitch := p_IndividualChoiceSwitch[0]
            (not ChoiceSwitch.GetCurrentState[Victim]) or (not ChoiceSwitch.GetCurrentState[Damager])
        ) {
            return combat_control_damage_result.Disallow
        }
        if (p_AllowSwitch.GetCurrentState[]) {
            combat_control_damage_result.Allow
        } else {
            combat_control_damage_result.Disallow
        }

combat_control_volume_closure := class:
    VolumeConfig: combat_control_volume_config

    OnAgentEntered(Agent:agent):void=
        if (set VolumeConfig.m_TrackedInside[Agent] = true) {}

    OnAgentExited(Agent:agent):void=
        set VolumeConfig.m_TrackedInside = VolumeConfig.m_TrackedInside.RemoveByKey(Agent)(0)

combat_control_damaged_closure := class:
    Device: combat_control_device
    FortPlayer: fort_character

    OnDamaged(DamageResult:damage_result):void=
        if (
            FortCharacter := fort_character[DamageResult.Instigator?]
            Victim := FortPlayer.GetAgent[]
            Damager := player[FortCharacter.GetAgent[]]
            (not (Device.p_IgnoreSelfDamage?)) or (Victim <> Damager)
        ) {
            Result := Device.HandleDamage(Victim, Damager, DamageResult)
            case(Result):
                combat_control_damage_result.Disallow => FortPlayer.Heal(DamageResult.Amount)
                _ => {}
            Log(MsgDebugDamaged(Victim, Damager, DamageResult.Amount, ToString(Result)), ?Level := log_level.Debug)
        }

    OnShieldDamaged(DamageResult:damage_result):void=
        if (
            FortCharacter := fort_character[DamageResult.Instigator?]
            Victim := FortPlayer.GetAgent[]
            Damager := player[FortCharacter.GetAgent[]]
            (not (Device.p_IgnoreSelfDamage?)) or (Victim <> Damager)
        ) {
            case(Device.HandleDamage(Victim, Damager, DamageResult)):
                combat_control_damage_result.Disallow => FortPlayer.SetShield(FortPlayer.GetShield() + DamageResult.Amount)
                _ => {}
        }

MsgDebugDamaged<localizes>(Victim:agent, Damager:agent, DamageAmount:float, StrResult:string):message=
    "'{Damager}' dealt {DamageAmount} damage to '{Victim}'. Result: {StrResult}"

ToString(Result:combat_control_damage_result)<transacts>:string=
    case(Result):
        combat_control_damage_result.Ignore => "Ignore"
        combat_control_damage_result.Allow => "Allow"
        combat_control_damage_result.Disallow => "Disallow"
