using { /Fortnite.com/Devices }
using { /Verse.org/Simulation }
using { /UnrealEngine.com/Temporary/Diagnostics }

using { HavocVerse.Debug }
using { HavocVerse.Resource }
using { HavocVerse.Logging }
using { HavocVerse.Util }

custom_device<public> := interface:
    # Attempts to trigger a custom action on the device.
    # Returns `false` if the data could not map to an action.
    CustomTrigger<public>(Data:string, Agent:?agent):logic

(InDevice:creative_device).AttemptCustomTrigger<public>(
    Data:string, Agent:?agent,
    ?Context:string
):void=
    if (CustomDevice := custom_device[InDevice]) {
        Success := CustomDevice.CustomTrigger(Data, Agent)
        if (not (Success?)) {
            Log("Custom trigger failed, to {InDevice} [data: {Data}, context: {Context}]", ?Level := log_level.Warning)
        }
    } else {
        Log("Custom trigger failed, {InDevice} is not a custom-device [data: {Data}, context: {Context}]", ?Level := log_level.Error)
    }

custom_device_trigger_config<public> := class<concrete>:
    @editable Device<public>: []creative_device = array{}
    @editable Data<public>: string = ""

    AttemptTrigger<public>(Agent:?agent, Context:string):void=
        if (Self.Device.Length = 0) {
            Log("Trying to trigger a custom device but no device is configured ({Context}) [data: {Data}]", ?Level := log_level.Warning)
            return
        }
        for (Index->l_Device:Self.Device) {
            l_Device.AttemptCustomTrigger(Self.Data, Agent, ?Context := Context)
        }

# A device to perform a trigger on a `custom_device`.
custom_trigger_device<public> := class(creative_device, aware_device):
    @editable p_DevId<public>: string = "N/A"

    @editable p_OnTrigger: trigger_device = trigger_device{}
    @editable p_MaxTriggers: int = -1
    @editable p_Actions: []custom_device_trigger_config = array{}

    var m_TriggerCount: int = 0

    DeviceMeta<override>()<transacts>:aware_device_meta=
        aware_device_meta{
            DeviceName := "custom_trigger_device"
            DevId := p_DevId
        }

    OnBegin<override>()<suspends>:void=
        AwareInitDevice(Self, OnInit)

    OnInit<private>()<suspends>:void=
        p_OnTrigger.TriggeredEvent.Subscribe(OnTriggered)

    OnTriggered(Agent:?agent):void=
        if (p_MaxTriggers > 0 and m_TriggerCount >= p_MaxTriggers) {
            return
        }
        for (CustomTrigger:p_Actions) {
            CustomTrigger.AttemptTrigger(Agent, "custom trigger device: {p_DevId}")
        }
        set m_TriggerCount += 1
