using { /Fortnite.com/Devices }
using { /Verse.org/Simulation }
using { /UnrealEngine.com/Temporary/Diagnostics }

using { HavocVerse.Debug }

join_action_device := class(creative_device, aware_device):
    @editable p_Delay: float = 0.0
    @editable p_Trigger: trigger_device = trigger_device{}

    DeviceMeta<override>()<transacts>:aware_device_meta=
        aware_device_meta{
            DeviceName := "join_action"
        }

    OnBegin<override>()<suspends>:void=
        AwareInitDevice(Self, OnInit)

    OnInit<private>()<suspends>:void=
        GetPlayspace().PlayerAddedEvent().Subscribe(OnPlayerAdded)
        for (Player:GetPlayspace().GetPlayers()) {
            OnPlayerAdded(Player)
        }

    OnPlayerAdded(Player:player):void=
        spawn:
            HandlePlayerJoin(Player)

    HandlePlayerJoin(Player:player)<suspends>:void=
        if (p_Delay > 0.0) {
            Sleep(p_Delay)
        }
        p_Trigger.Trigger(Player)
