using { /Fortnite.com/Devices }
using { /Verse.org/Simulation }
using { /UnrealEngine.com/Temporary/Diagnostics }

using { HavocVerse.Debug }
using { HavocVerse.Logging }
using { HavocVerse.Util }

level_group_endpoint_device<public> := class(creative_device, aware_device):
    @editable p_LevelGroupsManager<public>: level_groups_manager_device = level_groups_manager_device{}
    @editable p_LevelGroupId<public>: string = ""
    
    @editable p_UserActions: []level_group_endpoint_user_action_config = array{}

    DeviceMeta<override>()<transacts>:aware_device_meta=
        aware_device_meta{
            DeviceName := "level_group_endpoint"
        }

    OnBegin<override>()<suspends>:void=
        AwareInitDevice(Self, OnInit)
    
    OnInit<private>()<suspends>:void=
        for (UserActionConfig:p_UserActions) {
            UserAction := level_group_endpoint_user_action{
                Device := Self
                Config := UserActionConfig
            }
            UserActionConfig.OnTrigger.TriggeredEvent.Subscribe(UserAction.OnTriggered)
            for (Button:UserActionConfig.OnButton) {
                Button.InteractedWithEvent.Subscribe(UserAction.OnInteractedWith)
            }
        }
    
level_group_endpoint_user_action_kind := enum:
    GiveExp

level_group_endpoint_user_action_config := class<final><concrete>:
    @editable Action: level_group_endpoint_user_action_kind = level_group_endpoint_user_action_kind.GiveExp
    @editable OnTrigger: trigger_device = trigger_device{}
    @editable OnButton: []button_device = array{}
    @editable Value: int = 0

level_group_endpoint_user_action := class<final>:
    Device: level_group_endpoint_device
    Config: level_group_endpoint_user_action_config

    OnTriggered(Agent:?agent):void=
        if (Player := player[Agent?]) {
            DoAction(Player)
        }

    OnInteractedWith(Agent:agent):void=
        if (Player := player[Agent]) {
            DoAction(Player)
        }

    DoAction<private>(Player:player):void=
        GroupId := Device.p_LevelGroupId
        case(Config.Action):
            level_group_endpoint_user_action_kind.GiveExp =>
                Device.p_LevelGroupsManager.GiveExp(Player, GroupId, Config.Value)
