using { /Fortnite.com/Devices }
using { /Verse.org/Simulation }
using { /Verse.org/Colors }
using { /UnrealEngine.com/Temporary/Diagnostics }
using { /UnrealEngine.com/Temporary/SpatialMath }
using { /UnrealEngine.com/Temporary/UI }
using { /Fortnite.com/UI }

using { HavocVerse.Debug }
using { HavocVerse.UI }
using { HavocVerse.Logging }
using { HavocVerse.Util }

on_screen_timer_visibility := enum:
    Always
    HideWhenComplete

on_screen_timer_device := class(creative_device, aware_device):
    @editable p_AccessControl: []access_control_device = array{}

    @editable p_Timer: timer_device = timer_device{}

    @editable p_Visibility: on_screen_timer_visibility = on_screen_timer_visibility.HideWhenComplete
    @editable p_TextTemplate: string = "In %Timer%"
    @editable p_TextPlacement: ui_canvas_placement_config = ui_canvas_placement_config{}
    @editable p_TextStyle: ui_text_style_config = ui_text_style_config{}

    @editable p_HUDManagerConfig: hud_manager_base_config = hud_manager_base_config{
        RefreshRate := 1.0
    }
    @editable p_BackgroundImage: string = ""
    @editable p_BackgroundSize: vector2 = vector2{X := 100.0, Y := 50.0}
    @editable p_BackgroundPlacement: ui_canvas_placement_config = ui_canvas_placement_config{}

    var m_HudManager: ?on_screen_timer_hud_manager = false

    DeviceMeta<override>()<transacts>:aware_device_meta=
        aware_device_meta{
            DeviceName := "on_screen_timer"
        }

    OnBegin<override>()<suspends>:void=
        AwareInitDevice(Self, OnInit)

    OnInit<private>()<suspends>:void=
        HudManager := on_screen_timer_hud_manager{
            AccessControl := option{p_AccessControl[0]}
            RefreshRate := p_HUDManagerConfig.RefreshRate
            ZOrder := p_HUDManagerConfig.ZOrder
            Device := Self
        }
        set m_HudManager = Some(HudManager)
        spawn:
            HudManager.Run(GetPlayspace())

on_screen_timer_hud_manager := class(hud_manager_base):
    Device: on_screen_timer_device

    CreateHUD<override>(Player:player):?hud_manager_player=
        BackgroundWidget := if (
            Device.p_BackgroundImage.Length > 0
            Texture := AwareDefaultAssetsProvider[].GetTextureAsset[Device.p_BackgroundImage]
        ) {
            option{texture_block{
                DefaultImage := Texture
                DefaultDesiredSize := Device.p_BackgroundSize
            }}
        } else {
            None()
        }
        TextWidget := Device.p_TextStyle.ToTextBlock("")
        Canvas := canvas{
            Slots := array{
                Device.p_TextPlacement.ToCanvasSlot(TextWidget)
            }
        }
        if (l_BackgroundWidget := BackgroundWidget?) {
            Canvas.AddWidget(Device.p_BackgroundPlacement.ToCanvasSlot(l_BackgroundWidget))
        }
        Element := on_screen_timer_hud_element{
            Widget := Canvas
            Device := Device
            Player := Player
            TextWidget := MakeCachedTextBlock(TextWidget)
        }
        Some(MakeSingleElementHUDPlayer(Player, Canvas, Element))

on_screen_timer_hud_element := class(hud_element):
    Device: on_screen_timer_device
    Player: player

    TextWidget: cached_text_block

    Update<override>():void=
        RemainingSeconds := if (Device.p_Timer.IsStatePerAgent[]) {
            Device.p_Timer.GetActiveDuration(Player)
        } else {
            Device.p_Timer.GetActiveDuration()
        }
        if (RemainingSeconds > 0.0 or Device.p_Visibility = on_screen_timer_visibility.Always) {
            Show()
            TimerText := ToTimerMMSS(DefaultingCeil(RemainingSeconds))
            Text := Device.p_TextTemplate.ReplaceAll("%Timer%", TimerText)
            TextWidget.SetText(Text)
        } else {
            Hide()
        }
