using { /Fortnite.com/Devices }
using { /Verse.org/Simulation }
using { /UnrealEngine.com/Temporary/Diagnostics }

using { HavocVerse.Debug }
using { HavocVerse.Logging }
using { HavocVerse.Util }

pet_display_config := class<concrete>:
    @editable PetId: string = ""

    @editable Purchasable: logic = false
    @editable CostResource: string = ""
    @editable CostAmount: int = 0

    @editable OnButton: []button_device = array{}
    @editable OnTrigger: []trigger_device = array{}
    @editable UnownedErrorHUDMessage: []hud_message_device = array{}

pet_frontend_device := class(creative_device, aware_device):
    @editable p_PetController: pet_controller_device = pet_controller_device{}
    @editable p_ResourceManager: resource_manager_device = resource_manager_device{}

    @editable p_PetDisplays: []pet_display_config = array{}

    DeviceMeta<override>()<transacts>:aware_device_meta=
        aware_device_meta{
            DeviceName := "pet_frontend"
        }

    OnBegin<override>()<suspends>:void=
        AwareInitDevice(Self, OnInit)

    OnInit<private>()<suspends>:void=
        for (DisplayConfig:p_PetDisplays) {
            Closure := pet_display_closure{
                Device := Self
                Display := DisplayConfig
            }
            for (Button:DisplayConfig.OnButton) {
                Button.InteractedWithEvent.Subscribe(Closure.OnButtonInteractedWith)
            }
            for (Trigger:DisplayConfig.OnTrigger) {
                Trigger.TriggeredEvent.Subscribe(Closure.OnTriggered)
            }
        }

    HandlePetDisplayInteraction<internal>(Player:player, DisplayConfig:pet_display_config):void=
        PetId := DisplayConfig.PetId
        if (not p_PetController.GetPetConfig[PetId]) {
            Log("Handling display interaction for unexisting pet \"{PetId}\"", ?Level := log_level.Error)
            return
        }
        if (not p_PetController.IsPetUnlocked[Player, PetId]) {
            if (DisplayConfig.Purchasable?) {
                if (DisplayConfig.CostResource.Length > 0) {
                    Success := p_ResourceManager.Take(Player, DisplayConfig.CostResource, DisplayConfig.CostAmount,
                        ?NotifyNotEnough := true, ?SendFeedback := true)
                    if (not (Success?)) {
                        return
                    }
                }
                p_PetController.UnlockPet(Player, PetId)
            } else {
                for (HUDMessage:DisplayConfig.UnownedErrorHUDMessage) {
                    HUDMessage.Show(Player)
                }
                return
            }
        }
        if (p_PetController.IsPetActive[Player, PetId]) {
            p_PetController.DeactivatePet(Player, PetId)
        } else {
            p_PetController.ActivatePet(Player, PetId)
        }

pet_display_closure := class:
    Device: pet_frontend_device
    Display: pet_display_config

    OnButtonInteractedWith(Agent:agent):void=
        if (Player := player[Agent]) {
            Device.HandlePetDisplayInteraction(Player, Display)
        }

    OnTriggered(Agent:?agent):void=
        if (Player := player[Agent?]) {
            Device.HandlePetDisplayInteraction(Player, Display)
        }
