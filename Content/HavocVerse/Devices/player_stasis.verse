using { /Fortnite.com/Devices }
using { /Fortnite.com/Characters }
using { /Verse.org/Simulation }
using { /UnrealEngine.com/Temporary/Diagnostics }

using { HavocVerse.Debug }
using { HavocVerse.Util }

player_stasis_device := class(creative_device, aware_device, aware_debug_device, custom_device):
    @editable p_UserActions: []player_stasis_user_action_config = array{}

    var m_KnownVulnerableStatus: [player]logic = map{}

    DeviceMeta<override>()<transacts>:aware_device_meta=
        aware_device_meta{
            DeviceName := "player_stasis"
        }

    DebugStats<override>()<transacts>:aware_debug_stats=
        aware_debug_stats{
            Length := map{
                "knownVulnerableStatus" => m_KnownVulnerableStatus.Length
            }
        }

    OnBegin<override>()<suspends>:void=
        AwareInitDevice(Self, OnInit)
    
    OnInit<private>()<suspends>:void=
        for (UserActionConfig:p_UserActions) {
            UserAction := player_stasis_user_action{
                Device := Self
                Config := UserActionConfig
            }
            UserActionConfig.OnTrigger.TriggeredEvent.Subscribe(UserAction.OnTriggered)
        }

        GetPlayspace().PlayerRemovedEvent().Subscribe(OnPlayerRemoved)
    
    OnPlayerRemoved<private>(Player:player):void=
        set m_KnownVulnerableStatus = m_KnownVulnerableStatus.RemoveByKey(Player)(0)

    CustomTrigger<override>(Action:string, Agent:?agent):logic=
        if (FortCharacter := Agent?.GetFortCharacter[]) {
            if (Action = "put_in_stasis") {
                FortCharacter.PutInStasis(stasis_args{})
                return true
            } else if (Action = "release_from_stasis") {
                FortCharacter.ReleaseFromStasis()
                return true
            }
        }
        false

player_stasis_user_action_kind := enum:
    PutInStasis
    RemoveFromStasis

player_stasis_user_action_config := class<concrete>:
    @editable Action: player_stasis_user_action_kind = player_stasis_user_action_kind.PutInStasis
    @editable OnTrigger: trigger_device = trigger_device{}

    @editable Invulnerable: logic = true
    @editable AllowTurning: logic = false
    @editable AllowFalling: logic = false
    @editable AllowEmotes: logic = false

player_stasis_user_action := class:
    Device: player_stasis_device
    Config: player_stasis_user_action_config

    OnTriggered(Agent:?agent):void=
        if (Player := player[Agent?], FortCharacter := Player.GetFortCharacter[]) {
            case(Config.Action):
                player_stasis_user_action_kind.PutInStasis =>
                    args := stasis_args{
                        AllowTurning := Config.AllowTurning
                        AllowFalling := Config.AllowFalling
                        AllowEmotes := Config.AllowEmotes
                    }
                    FortCharacter.PutInStasis(args)
                    if (Config.Invulnerable?) {
                        IsVulnerable := if (FortCharacter.IsVulnerable[]). true else false
                        if:
                            set Device.m_KnownVulnerableStatus[Player] = IsVulnerable
                        FortCharacter.SetVulnerability(false)
                    }
                player_stasis_user_action_kind.RemoveFromStasis =>
                    FortCharacter.ReleaseFromStasis()
                    if (Config.Invulnerable?) {
                        RemoveRes := Device.m_KnownVulnerableStatus.RemoveByKey(Player)
                        set Device.m_KnownVulnerableStatus = RemoveRes(0)
                        if (KnownStatus := RemoveRes(1)?) {
                            FortCharacter.SetVulnerability(KnownStatus)
                        }
                    }
        }
