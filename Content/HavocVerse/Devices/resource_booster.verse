using { /Fortnite.com/Devices }
using { /Fortnite.com/UI }
using { /Verse.org/Simulation }
using { /UnrealEngine.com/Temporary/Diagnostics }
using { /UnrealEngine.com/Temporary/SpatialMath }
using { /UnrealEngine.com/Temporary/UI }

using { HavocVerse.Debug }
using { HavocVerse.Devices.GameSpecific }
using { HavocVerse.UI }
using { HavocVerse.Util }
using { HavocVerse.Storage }
using { HavocVerse.Logging }

resource_booster_sharing_policy := enum:
    Shared

resource_booster_device<public> := class(creative_device, aware_device, custom_device, resetable, persistence_recordable, persistence_resetable):
    @editable p_DevId<public>: string = "N/A"

    @editable p_AccessControl: []access_control_device = array{}
    @editable p_SharingPolicy: resource_booster_sharing_policy = resource_booster_sharing_policy.Shared
    @editable p_PersistenceKeepAcrossTycoonRebirths: logic = false

    @editable p_AutoGenerationManager<public>: auto_generation_manager = auto_generation_manager{}
    @editable p_MultiplierChannel<public>: string = "booster1"
    @editable p_ResourceMultipliers: []resource_booster_multiplier_config = array{}

    @editable p_InWorldTimer: []timer_device = array{}

    @editable p_HUDEnabled: logic = false
    @editable p_HUDManagerConfig: hud_manager_base_config = hud_manager_base_config{
        RefreshRate := 1.0
    }
    @editable p_HUDTextTemplate: string = "For %Timer%"
    @editable p_HUDTextPlacement: ui_canvas_placement_config = ui_canvas_placement_config{}
    @editable p_HUDTextStyle: ui_text_style_config = ui_text_style_config{}

    @editable p_HUDBackgroundImage: string = ""
    @editable p_HUDBackgroundSize: vector2 = vector2{X := 100.0, Y := 50.0}
    @editable p_HUDBackgroundPlacement: ui_canvas_placement_config = ui_canvas_placement_config{}

    @editable p_AllowPersistence: logic = true
    @editable p_PersistenceId: string = "res_booster1"
    @editable p_PersistenceBits: int = 16

    @editable p_UserActions: []resource_booster_user_action_config = array{}

    var m_HudManager: ?resource_booster_hud_manager = false
    var m_ExpiresAt: float = 0.0

    var m_ShrukenTimer: logic = false
    var m_InitialTimerScale: vector3 = vector3{X := 1.0, Y := 1.0, Z := 1.0}

    var m_PersistenceControl: persistence_recordable_control = persistence_recordable_control{}

    DeviceMeta<override>()<transacts>:aware_device_meta=
        aware_device_meta{
            DeviceName := "resource_booster"
            DevId := p_DevId
        }

    OnBegin<override>()<suspends>:void=
        AwareInitDevice(Self, OnInit)

    OnInit<private>()<suspends>:void=
        for (UserActionConfig:p_UserActions) {
            UserAction := resource_booster_user_action{
                Device := Self
                Config := UserActionConfig
            }
            UserActionConfig.OnTrigger.TriggeredEvent.Subscribe(UserAction.OnTriggered)
        }

        if (AccessControl := p_AccessControl[0]) {
            AccessControl.RegisterUsingDevice(Self)
        }
        if (Timer := p_InWorldTimer[0]) {
            set m_InitialTimerScale = Timer.GetTransform().Scale
        }

        if (p_HUDEnabled?) {
            HudManager := resource_booster_hud_manager{
                AccessControl := option{p_AccessControl[0]}
                RefreshRate := p_HUDManagerConfig.RefreshRate
                ZOrder := p_HUDManagerConfig.ZOrder
                Device := Self
            }
            set m_HudManager = Some(HudManager)
            HudManager.Init(GetPlayspace())
        }
        spawn:
            StartUpdateLoop()

    Reset<override>(Agent:?agent):void=
        if (p_PersistenceKeepAcrossTycoonRebirths?) {
            return
        }
        set m_ExpiresAt = 0.0
        Update()

    StartUpdateLoop<private>()<suspends>:void=
        loop {
            Sleep(1.0)
            Update()
        }

    EnsureMultipliersState<private>():void=
        RemainingTime := GetRemainingSharedTime()
        BoosterActive := if (RemainingTime > 0.0). true else false
        for (MultiplierConfig:p_ResourceMultipliers) {
            if (MultiplierConfig.m_Active = true and BoosterActive = false) {
                p_AutoGenerationManager.GetResourceMultiplierInstance(MultiplierConfig.ResourceId).SetMultiplier(p_MultiplierChannel, None())
                set MultiplierConfig.m_Active = false
            } else if (MultiplierConfig.m_Active = false and BoosterActive = true) {
                p_AutoGenerationManager.GetResourceMultiplierInstance(MultiplierConfig.ResourceId).SetMultiplier(p_MultiplierChannel, MultiplierConfig.Multiplier)
                set MultiplierConfig.m_Active = true
            }
        }

    InWorldUpdate<private>():void=
        RemainingTime := GetRemainingSharedTime()
        if (Timer := p_InWorldTimer[0]) {
            if (RemainingTime > 0.0) {
                if (m_ShrukenTimer?, Timer.TeleportTo[GetScaledTimerTransform(m_InitialTimerScale)]) {
                    Timer.Enable()
                    set m_ShrukenTimer = false
                }
                Timer.Start()
                Timer.SetActiveDuration(RemainingTime)
            } else {
                if (Timer.GetActiveDuration() > 0.0) {
                    Timer.Complete()
                }
                if (not (m_ShrukenTimer?), Timer.TeleportTo[GetScaledTimerTransform(MakeUniformVec3(0.0001))]) {
                    Timer.Disable()
                    set m_ShrukenTimer = true
                }
            }
        }

    GetScaledTimerTransform<private>(Scale:vector3)<transacts>:transform=
        if (Timer := p_InWorldTimer[0]) {
            Transform := Timer.GetTransform()
            transform{
                Translation := Transform.Translation
                Rotation := Transform.Rotation
                Scale := Scale
            }
        } else {
            transform{}
        }

    Update<private>():void=
        EnsureMultipliersState()
        if (HudManager := m_HudManager?) {
            HudManager.Refresh()
        }
        InWorldUpdate()

    GetRemainingSharedTime<public>()<transacts>:float=
        Max(m_ExpiresAt - GetSimulationElapsedTime(), 0.0)

    GetRemainingTime<public>(Player:player)<transacts>:float=
        GetRemainingSharedTime()

    ModifySharedTime<public>(ExtraTime:float):void=
        set m_ExpiresAt = Max(m_ExpiresAt, GetSimulationElapsedTime()) + ExtraTime
        Update()
        if (HudManager := m_HudManager?) {
            HudManager.RecheckUnmanagedPlayers()
        }

    ModifyTime<public>(Player:player, ExtraTime:float):void=
        ModifySharedTime(ExtraTime)

    ResetSharedTime<public>():void=
        set m_ExpiresAt = 0.0
        Update()

    ResetTime<public>(Player:player):void=
        ResetSharedTime()

    CustomTrigger<override>(Input:string, Agent:?agent):logic=
        SplitRes := Input.Split(':')
        if (SplitRes[0] = "modify_time", ExtraTime := ParseFloat[SplitRes[1]]) {
            if (Player := player[Agent?]) {
                ModifyTime(Player, ExtraTime)
            } else {
                ModifySharedTime(ExtraTime)
            }
            return true
        } else if (SplitRes[0] = "reset") {
            if (Player := player[Agent?]) {
                ResetTime(Player)
            } else {
                ResetSharedTime()
            }
            return true
        }
        false

    GetPersistenceControl<override>()<transacts><decides>:persistence_recordable_control=
        m_PersistenceControl

    PersistenceReserve<override>(StorageMem:storage_memory)<transacts><decides>:resource_booster_persistence_state=
        p_AllowPersistence?
        resource_booster_persistence_state{
            RemainingTimeChunk := StorageMem.Reserve[p_PersistenceBits, p_PersistenceId + "_remaining"]
        }

    PersistenceSave<override>(i_State:persistence_recordable_state, Agent:agent, SaveCause:persistence_save_cause):void=
        if (State := resource_booster_persistence_state[i_State]) {
            RemainingTime := DefaultingFloor(GetRemainingSharedTime())
            State.RemainingTimeChunk.WriteInt(Agent, RemainingTime)
        }

    PersistenceLoad<override>(i_State:persistence_recordable_state, Agent:agent, Query:persistence_load_query):void=
        if (State := resource_booster_persistence_state[i_State]) {
            StoredRemainingTime := ToFloat(State.RemainingTimeChunk.ReadInt(Agent))
            if (Query.Overwrite? or StoredRemainingTime > GetRemainingSharedTime()) {
                set m_ExpiresAt = GetSimulationElapsedTime() + StoredRemainingTime
                Update()
                if (HudManager := m_HudManager?) {
                    HudManager.RecheckUnmanagedPlayers()
                }
            }
        }

    PersistenceReset<override>(i_State:persistence_recordable_state, Agent:agent, Query:persistence_reset_query):void=
        if (p_PersistenceKeepAcrossTycoonRebirths?, tycoon_manager_rebirth_persistence_reset_query[Query]) {
            Log("Skipping reset of {Self} state because it is initialized by a rebirth", ?Level := log_level.Debug)
        } else {
            i_State.Reset(Agent)
            PersistenceLoad(i_State, Agent, persistence_load_query{
                Overwrite := true
                FirstLoad := false
            })
        }

resource_booster_persistence_state<public> := class(persistence_recordable_state):
    RemainingTimeChunk<public>: storage_memory_chunk

    Reset<override>(Agent:agent):void=
        RemainingTimeChunk.Reset(Agent)

resource_booster_multiplier_config := class<concrete>:
    @editable ResourceId: string = ""
    @editable Multiplier: float = 1.0

    var m_Active: logic = false
    
resource_booster_user_action_kind := enum:
    ModifyTime
    Reset

resource_booster_user_action_config := class<concrete>:
    @editable Action: resource_booster_user_action_kind = resource_booster_user_action_kind.ModifyTime
    @editable OnTrigger: trigger_device = trigger_device{}
    @editable Value: float = 90.0

resource_booster_user_action := class:
    Device: resource_booster_device
    Config: resource_booster_user_action_config

    OnTriggered(Agent:?agent):void=
        if (Player := player[Agent?]) {
            case(Config.Action):
                resource_booster_user_action_kind.ModifyTime => Device.ModifyTime(Player, Config.Value)
                resource_booster_user_action_kind.Reset => Device.ResetTime(Player)
        }

resource_booster_hud_manager := class(hud_manager_base):
    Device: resource_booster_device

    CreateHUD<override>(Player:player):?hud_manager_player=
        if (Device.GetRemainingTime(Player) <= 0.0) {
            return None()
        }
        TextWidget := Device.p_HUDTextStyle.ToTextBlock("")

        var CanvasSlots: []canvas_slot = array{}
        BackgroundWidget := CreateOptionalImage(Device.p_HUDBackgroundImage, Device.p_HUDBackgroundSize)
        if (l_BackgroundWidget := BackgroundWidget?) {
            set CanvasSlots += array{Device.p_HUDBackgroundPlacement.ToCanvasSlot(l_BackgroundWidget)}
        } else {
            return None()
        }
        Canvas := canvas{
            Slots := CanvasSlots + array{
                Device.p_HUDTextPlacement.ToCanvasSlot(TextWidget)
            }
        }
        Element := resource_booster_hud_element{
            Widget := Canvas
            Device := Device
            Player := Player
            TextWidget := MakeCachedTextBlock(TextWidget)
        }
        Some(MakeSingleElementHUDPlayer(Player, Canvas, Element))

resource_booster_hud_element := class(hud_element):
    Device: resource_booster_device
    Player: player

    TextWidget: cached_text_block

    Update<override>():void=
        RemainingSeconds := Device.GetRemainingTime(Player)
        if (RemainingSeconds > 0.0) {
            Show()
            TimerText := ToTimerMMSS(DefaultingCeil(RemainingSeconds))
            Text := Device.p_HUDTextTemplate.ReplaceAll("%Timer%", TimerText)
            TextWidget.SetText(Text)
        } else {
            Hide()
        }
