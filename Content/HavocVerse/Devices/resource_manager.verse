using { /Fortnite.com/Devices }
using { /Verse.org/Simulation }
using { /Verse.org/Random }
using { /UnrealEngine.com/Temporary/Diagnostics }

using { HavocVerse.Debug }
using { HavocVerse.Devices.GameSpecific }
using { HavocVerse.Resource }
using { HavocVerse.Storage }
using { HavocVerse.Logging }
using { HavocVerse.Util }

resource_feedback_kind<public> := enum:
    Give
    Take
    RateIncrease
    RateDecrease

resource_feedback_config := class<concrete>:
    @editable FeedbackFor: resource_feedback_kind = resource_feedback_kind.Give
    @editable HUDMessenger: []hud_message_device = array{}
    @editable Formatter: number_formatter_config = number_formatter_config{
        Template := "+%Val%"
    }
    @editable DisplayTime: float = 1.2

    ToResourceFeedback()<transacts><decides>:resource_feedback=
        resource_feedback{
            Kind := FeedbackFor
            HUDMessenger := HUDMessenger[0]
            Formatter := Formatter.CreateFormatter()
            DisplayTime := DisplayTime
        }

resource_feedback := class:
    Kind: resource_feedback_kind
    HUDMessenger: hud_message_device
    Formatter: number_formatter
    DisplayTime: float

base_resource_config := class<concrete>:
    @editable ResourceId: string = ""

    @editable NotEnoughHUDMessage: []hud_message_device = array{}
    @editable ValueFeedbacks: []resource_feedback_config = array{}

    @editable AllowPersistence: logic = false
    @editable PersistenceReserveBits: int = 31

proxied_score_resource_config := class<concrete>:
    @editable ResourceId: string = ""
    @editable ScoreManager: score_manager_device = score_manager_device{}

resource_config := class<concrete>(base_resource_config):
    @editable CondButton: []conditional_button_device = array{}

virtual_resource_sharing_policy := enum:
    Global
    PerPlayer
    PerTeam

virtual_resource_config := class<concrete>(base_resource_config):
    @editable SharingPolicy: virtual_resource_sharing_policy = virtual_resource_sharing_policy.PerPlayer

resource_manager_entry := class:
    Res: resource
    NotEnoughHUDMessage: ?hud_message_device
    Feedbacks: [resource_feedback_kind]resource_feedback
    PersistenceBits: ?int

resource_take_event<public> := class:
    Agent<public>: ?agent
    ResourceId<public>: string
    Amount<public>: int

# Device that exposes resource management functionalities to Verse for actual items
# and provides the same abstraction for virtual resources.
resource_manager_device<public> := class(creative_device, aware_device, aware_debug_device, custom_device, persistence_recordable, persistence_resetable):
    @editable p_ResourceConfigs: []resource_config = array{}
    # Reminder: can't put this in resource configs as custom Verse devices
    #           are bugged when nested in custom type arrays
    @editable p_BulkGranters: []bulk_resource_granter_device = array{}

    @editable p_VirtualResources: []virtual_resource_config = array{}

    @editable p_ProxiedScoreManagers: []proxied_score_resource_config = array{}

    var m_Setup: logic = false
    var m_Resources: [string]resource_manager_entry = map{}

    var m_BackRefAutoGenerationManagers: []auto_generation_manager = array{}

    var m_PersistenceControl: persistence_recordable_control = persistence_recordable_control{}

    var m_PostTakeListeners: []type{__(:resource_take_event):void} = array{}

    DeviceMeta<override>()<transacts>:aware_device_meta=
        aware_device_meta{
            DeviceName := "resource_manager"
        }

    DebugStats<override>()<transacts>:aware_debug_stats=
        var TotalBalanceKeys: int = 0
        for (__->Entry:m_Resources, BoundVirtualRes := bound_virtual_resource[Entry.Res]) {
            set TotalBalanceKeys += BoundVirtualRes.m_Balances.Length
        }
        aware_debug_stats{
            Length := map{
                "resources" => m_Resources.Length
                "resources.balanceKeys:sum" => TotalBalanceKeys
                "backRefAutoGenerationManagers" => m_BackRefAutoGenerationManagers.Length
                "listeners::postTake" => m_PostTakeListeners.Length
            }
        }

    OnBegin<override>()<suspends>:void=
        AwareInitDevice(Self, Setup)

    # Ensures the resource manager is setup. Calls to this function are idempotent.
    Setup<public>():void=
        if (m_Setup?) {
            return
        }
        set m_Setup = true

        # Physical resources
        var BulkGranters: [string]bulk_resource_granter_device = map{}
        for (
            BulkGranter:p_BulkGranters
            Prev := option{BulkGranters[BulkGranter.p_ResourceId]}
            set BulkGranters[BulkGranter.p_ResourceId] = BulkGranter
        ) {
            if (Prev?) {
                Log("Bulk granter for resource {BulkGranter.p_ResourceId} has duplicate entries in the resource manager", ?Level := log_level.Warning)
            }
        }

        for (
            ResourceConfig:p_ResourceConfigs
            Prev := option{m_Resources[ResourceConfig.ResourceId]}
        ) {
            if (Prev?) {
                Log("Resource {ResourceConfig.ResourceId} has duplicate declarations in the resource manager", ?Level := log_level.Warning)
            }
            if (ResourceConfig.CondButton.Length > 1) {
                Log("Resource {ResourceConfig.ResourceId} has multiple conditional buttons, only the first one will be used", ?Level := log_level.Error)
            }
            Entry := CreateResourceEntry(ResourceConfig, physical_resource{
                ResourceId := ResourceConfig.ResourceId
                BulkGranter := option{BulkGranters[ResourceConfig.ResourceId]}
                CondButton := option{ResourceConfig.CondButton[0]}
            })
            if (set m_Resources[ResourceConfig.ResourceId] = Entry) {}
        }

        for (__->BulkGranter:BulkGranters) {
            if (not m_Resources[BulkGranter.p_ResourceId]) {
                Log("Resource manager has a bulk granter for resource {BulkGranter.p_ResourceId} without matching resource declaration", ?Level := log_level.Error)
            }
        }

        # Virtual resources
        for (
            ResourceConfig:p_VirtualResources
            Prev := option{m_Resources[ResourceConfig.ResourceId]}
        ) {
            if (physical_resource[Prev?]) {
                Log("Resource {ResourceConfig.ResourceId} is both declared as a virtual currency and has a physical item granter", ?Level := log_level.Error)
            } else if (Prev?) {
                Log("Virtual resource {ResourceConfig.ResourceId} has duplicate declarations in the resource manager", ?Level := log_level.Error)
            } else {
                Entry := case(ResourceConfig.SharingPolicy):
                    virtual_resource_sharing_policy.Global => CreateResourceEntry(ResourceConfig, global_virtual_resource{
                        ResourceId := ResourceConfig.ResourceId
                    })
                    virtual_resource_sharing_policy.PerPlayer => CreateResourceEntry(ResourceConfig, per_player_virtual_resource{
                        ResourceId := ResourceConfig.ResourceId
                    })
                    virtual_resource_sharing_policy.PerTeam => CreateResourceEntry(ResourceConfig, per_team_virtual_resource{
                        ResourceId := ResourceConfig.ResourceId
                        TeamCollection := GetPlayspace().GetTeamCollection()
                    })
                if (set m_Resources[ResourceConfig.ResourceId] = Entry) {}
            }
        }

        # Proxied resources
        for (
            ProxiedScoreConfig:p_ProxiedScoreManagers
            ResourceId := ProxiedScoreConfig.ResourceId
            Prev := option{m_Resources[ResourceId]}
        ) {
            if (Prev?) {
                Log("Proxied score resource {ResourceId} has duplicate declarations in the resource manager", ?Level := log_level.Error)
            } else {
                Entry := CreateProxiedResourceEntry(proxied_score_resource{
                    ResourceId := ResourceId
                    ScoreManager := ProxiedScoreConfig.ScoreManager
                })
                set m_Resources = AddToMap(m_Resources, ResourceId, Entry)
            }
        }

        GetPlayspace().PlayerRemovedEvent().Subscribe(OnPlayerRemoved)

    OnPlayerRemoved<private>(Player:player):void=
        spawn:
            SchedulePlayerCleanup(Player)

    SchedulePlayerCleanup<private>(Player:player)<suspends>:void=
        Sleep(0.0)
        for (
            __->Entry:m_Resources
            PerPlayerRes := per_player_virtual_resource[Entry.Res]
            PerPlayerRes.m_Balances[Player]
        ) {
            set PerPlayerRes.m_Balances = PerPlayerRes.m_Balances.RemoveByKey(Player)(0)
        }

    RegisterAutoGenerationManager<public>(AutoGenerationManager:auto_generation_manager):void=
        Setup()
        set m_BackRefAutoGenerationManagers += array{AutoGenerationManager}

    # Returns the auto generation rate of a given resource for an agent.
    GetAutoGenerationRate<public>(Agent:agent, ResourceId:string):float=
        var Rate: float = 0.0
        for (AutoGenerationManager:m_BackRefAutoGenerationManagers) {
            set Rate += AutoGenerationManager.GetGenerationRate(ResourceId, Some(Agent))
        }
        Rate

    # Returns the global total multiplier for an agent across auto generation managers.
    GetTotalMultiplier<public>(Agent:agent):float=
        var TheoreticalMultiplier: float = 1.0
        for (AutoGenerationManager:m_BackRefAutoGenerationManagers, AutoGenerationManager.IsGeneratingFor[Agent]) {
            set TheoreticalMultiplier *= AutoGenerationManager.GetGlobalMultiplierInstance().GetTotalMultiplier()
        }
        TheoreticalMultiplier

    # Returns the effective multiplier of a given resource for an agent.
    # This handles weighting the different auto generation managers.
    GetEffectiveMultiplier<public>(Agent:agent, ResourceId:string):float=
        var Entries: []tuple(float,float) = array{}
        for (AutoGenerationManager:m_BackRefAutoGenerationManagers, AutoGenerationManager.IsGeneratingFor[Agent]) {
            Rate := AutoGenerationManager.GetGenerationRate(ResourceId, Some(Agent))
            if (Rate > 0.0) {
                Multiplier := AutoGenerationManager.GetEffectiveMultiplier(ResourceId)
                set Entries += array{(Rate, Multiplier)}
            }
        }

        var MaxRate: float = 1.0
        for (Entry:Entries, Entry(0) > MaxRate) {
            set MaxRate = Max(Entry(0), MaxRate)
        }

        var MultipliersSum: float = 0.0
        var TotalWeight: float = 0.0
        for (Entry:Entries) {
            Weight := Entry(0) / MaxRate
            set TotalWeight += Weight
            set MultipliersSum += Entry(1) * Weight
        }

        if (TotalWeight > 0.0) {
            MultipliersSum / TotalWeight
        } else {
            var TheoreticalMultiplier: float = 1.0
            for (AutoGenerationManager:m_BackRefAutoGenerationManagers, AutoGenerationManager.IsGeneratingFor[Agent]) {
                set TheoreticalMultiplier *= AutoGenerationManager.GetEffectiveMultiplier(ResourceId)
            }
            TheoreticalMultiplier
        }

    CreateResourceEntry<private>(ResourceConfig:base_resource_config, Res:resource):resource_manager_entry=
        if (ResourceConfig.NotEnoughHUDMessage.Length > 1) {
            Log("Resource '{ResourceConfig.ResourceId}' has multiple not enough HUD messages, only the first one will be used", ?Level := log_level.Warning)
        }
        var Feedbacks: [resource_feedback_kind]resource_feedback = map{}
        for (FeedbackConfig:ResourceConfig.ValueFeedbacks, Feedback := FeedbackConfig.ToResourceFeedback[]) {
            if (set Feedbacks[Feedback.Kind] = Feedback) {}
        }
        resource_manager_entry{
            Res := Res
            NotEnoughHUDMessage := option{ResourceConfig.NotEnoughHUDMessage[0]}
            Feedbacks := Feedbacks
            PersistenceBits := if (ResourceConfig.AllowPersistence?) {
                Some(ResourceConfig.PersistenceReserveBits)
            } else {
                false
            }
        }

    CreateProxiedResourceEntry<private>(Res:resource):resource_manager_entry=
        resource_manager_entry{
            Res := Res
            NotEnoughHUDMessage := None()
            Feedbacks := map{}
            PersistenceBits := false
        }

    # Returns a resource by resource id.
    GetResource<public>(ResourceKind:string)<decides><transacts>:resource=
        m_Resources[ResourceKind].Res

    GetResources<public>()<decides><transacts>:[]resource=
        for (__->Entry:m_Resources) {
            Entry.Res
        }

    # Returns a virtual resource by resource id.
    # Virtual resources allow more flexibility compared to physical currencies.
    GetVirtualResource<public>(ResourceKind:string)<decides><transacts>:virtual_resource=
        virtual_resource[m_Resources[ResourceKind].Res]

    # Checks whether a resource is proxied.
    IsProxiedResource<public>(ResourceKind:string)<decides><transacts>:void=
        proxied_score_resource[m_Resources[ResourceKind].Res]

    # Checks whether a resource can be shared between at least several agents.
    # Fails if not.
    IsShareableResource<public>(ResourceKind:string)<decides><transacts>:void=
        VirtualResource := GetVirtualResource[ResourceKind]
        global_virtual_resource[VirtualResource] or per_team_virtual_resource[VirtualResource]

    # Returns the amount of a given resource held by an agent.
    # Fails if the resource is not supported, or if the amount is negative.
    GetAmount<public>(Agent:agent, ResourceKind:string):int=
        GetAmount(Some(Agent), ResourceKind)

    # Returns the amount of a given resource.
    # Returns -1 if the resource doesn't exist.
    GetAmount<public>(Agent:?agent, ResourceKind:string):int=
        if (Res := m_Resources[ResourceKind].Res) {
            Res.GetAmount(Agent)
        } else {
            Log("Resource '{ResourceKind}' is not supported by the resource manager (GetAmount)", ?Level := log_level.Error)
            -1
        }

    # Returns the formatted amount of a given resource held by an agent.
    # Returns -1 if the resource doesn't exist.
    GetFormattedAmount<public>(Agent:agent, ResourceKind:string):string=
        GetFormattedAmount(Some(Agent), ResourceKind)

    # Returns the formatted amount of a given resource.
    # Fails if the resource is not supported, or if the amount is negative.
    GetFormattedAmount<public>(Agent:?agent, ResourceKind:string):string=
        if (Res := m_Resources[ResourceKind].Res) {
            Amount := Res.GetAmount(Agent)
            FormatResourceAmount(Res, Amount)
        } else {
            Log("Resource '{ResourceKind}' is not supported by the resource manager (GetFormattedAmount)", ?Level := log_level.Error)
            "-1"
        }

    # Grants an amount of a given resource to an agent.
    # Fails if the resource is not supported, or if the amount is negative.
    Give<public>(Agent:agent, ResourceKind:string, Amount:int, ?SendFeedback:logic = false):logic=
        Give(Some(Agent), ResourceKind, Amount, ?SendFeedback := SendFeedback)

    # Grants an amount of a given resource.
    # Fails if the resource is not supported, or if the amount is negative.
    Give<public>(Agent:?agent, ResourceKind:string, Amount:int, ?SendFeedback:logic = false):logic=
        if (Res := m_Resources[ResourceKind].Res) {
            if (l_Agent := Agent?) {
                Log(MsgResourceOperation(l_Agent, "Give", ResourceKind, Amount), ?Level := log_level.Verbose)
            }
            if (Amount = 0) {
                return true
            }
            Success := Res.Give(Agent, Amount)
            if (Success?, SendFeedback?) {
                SendFeedbackNotification(resource_feedback_kind.Give, Agent, ResourceKind, ToFloat(Amount))
            }
            Success
        } else {
            Log("Resource '{ResourceKind}' is not supported by the resource manager (Give)", ?Level := log_level.Error)
            false
        }

    # Takes an amount of a given resource from an agent.
    # Fails if the resource is not supported, or if the amount is negative.
    Take<public>(Agent:agent, ResourceKind:string, Amount:int, ?SendFeedback:logic = false, ?NotifyNotEnough:logic = false):logic=
        Take(Some(Agent), ResourceKind, Amount, ?SendFeedback := SendFeedback, ?NotifyNotEnough := NotifyNotEnough)

    # Takes an amount of a given resource.
    # Fails if the resource is not supported, or if the amount is negative.
    Take<public>(Agent:?agent, ResourceKind:string, Amount:int, ?SendFeedback:logic = false, ?NotifyNotEnough:logic = false):logic=
        if (Res := m_Resources[ResourceKind].Res) {
            if (l_Agent := Agent?) {
                Log(MsgResourceOperation(l_Agent, "Take", ResourceKind, Amount), ?Level := log_level.Verbose)
            }
            if (Amount = 0) {
                return true
            }
            Success := Res.Take(Agent, Amount)
            if (not (Success?), NotifyNotEnough?) {
                if (l_Agent := Agent?) {
                    SendNotEnoughNotification(l_Agent, ResourceKind)
                } else {
                    for (Player:GetPlayspace().GetPlayers()) {
                        SendNotEnoughNotification(Player, ResourceKind)
                    }
                }
            }
            if (
                not (Success?)
                (not (NotifyNotEnough?)) or (not (m_Resources[ResourceKind].NotEnoughHUDMessage?))
                l_Agent := Agent?
            ) {
                Log(MsgSilentNotEnough(l_Agent, ResourceKind, Amount), ?Level := log_level.Debug)
            }
            if (Success?, SendFeedback?) {
                SendFeedbackNotification(resource_feedback_kind.Take, Agent, ResourceKind, ToFloat(Amount))
            }
            if (Success?, m_PostTakeListeners.Length > 0) {
                Event := resource_take_event{
                    Agent := Agent
                    ResourceId := ResourceKind
                    Amount := Amount
                }
                for (Listener:m_PostTakeListeners) {
                    Listener(Event)
                }
            }
            Success
        } else {
            Log("Resource '{ResourceKind}' is not supported by the resource manager (Take)", ?Level := log_level.Error)
            false
        }

    MsgResourceOperation<private><localizes>(Agent:agent, Operation:string, ResourceKind:string, Amount:int):message=
        "Resource manager '{Operation}' {Amount} of resource {ResourceKind} for '{Agent}'"
    MsgSilentNotEnough<private><localizes>(Agent:agent, ResourceKind:string, Amount:int):message=
        "'{Agent}' silent fail of Take operation, not enough {ResourceKind} (required {Amount})"

    ## Group distribution {{{

    # TODO: Support for access-control based resource sharing.
    #       This will make possible co-op gamemodes.

    GetAmount<public>(AccessControl:access_control_device, ResourceKind:string):int=
        GetGroupTotalAmount(AccessControl.ListAllowedAgents(), ResourceKind)

    Give<public>(AccessControl:access_control_device, ResourceKind:string, Amount:int, ?SendFeedback:logic = false):logic=
        GiveToGroup(AccessControl.ListAllowedAgents(), ResourceKind, Amount, ?SendFeedback := SendFeedback)

    Take<public>(AccessControl:access_control_device, ResourceKind:string, Amount:int, ?SendFeedback:logic = false, ?NotifyNotEnough:logic = false):logic=
        # FIXME
        if (Agent := AccessControl.ListAllowedAgents()[0]) {
            Take(Agent, ResourceKind, Amount, ?SendFeedback := SendFeedback, ?NotifyNotEnough := NotifyNotEnough)
        } else {
            false
        }

    # Returns the total amount of a given resource held by a group of agents.
    GetGroupTotalAmount<public>(Agents:[]agent, ResourceKind:string):int=
        var Amount: int = 0
        for (Agent:Agents) {
            set Amount += GetAmount(Agent, ResourceKind)
        }
        Amount

    # Grants an amount of a given resource to several agents, distributed according to a distribution policy.
    # Fails if the resource doesn't exist or none of the amount could make its way to someone.
    GiveToGroup<public>(
        Agents:[]agent,
        ResourceKind:string,
        Amount:int,
        ?DistributionPolicy:resource_distribution_policy = resource_distribution_policy.FullAmountToEach,
        ?SendFeedback:logic = false
    ):logic=
        var AnySuccess: logic = false
        case(DistributionPolicy):
            resource_distribution_policy.FirstPlayer => 
                if (Agent := Agents[0]) {
                    set AnySuccess = Give(Agent, ResourceKind, Amount, ?SendFeedback := SendFeedback)
                }
            resource_distribution_policy.FullAmountToEach => 
                for (Agent:Agents) {
                    Success := Give(Agent, ResourceKind, Amount, ?SendFeedback := SendFeedback)
                    if (Success?, not (AnySuccess?)) {
                        set AnySuccess = true
                    }
                }
            resource_distribution_policy.Split => 
                GiveToEach := DivDefaultingFloor(Amount, Agents.Length)
                Remaining := Amount - GiveToEach
                RemainingGiveTo := if (Remaining > 0) {
                    GetRandomInt(0, Agents.Length-1)
                } else {
                    0
                }
                for (Index->Agent:Agents) {
                    AgentAmount := if (Index = RemainingGiveTo) {
                        GiveToEach + Remaining
                    } else {
                        GiveToEach
                    }
                    Success := Give(Agent, ResourceKind, AgentAmount, ?SendFeedback := SendFeedback)
                    if (Success?, not (AnySuccess?)) {
                        set AnySuccess = true
                    }
                }
        AnySuccess
    ## }}}

    ## Feedback {{{

    # Sends a notification to an agent if they don't have enough of a given resource.
    # Returns false if the resource does not exist or is not configured with a HUD message device for that purpose.
    SendNotEnoughNotification<public>(Agent:agent, ResourceKind:string):logic=
        if (MessageDevice := m_Resources[ResourceKind].NotEnoughHUDMessage?) {
            MessageDevice.Show(Agent)
            true
        } else {
            false
        }

    # Calls `SendFeedbackNotification` with an optional agent, default to all affected players.
    SendFeedbackNotification<public>(FeedbackKind:resource_feedback_kind, Agent:?agent, ResourceKind:string, Value:float):void=
        if (l_Agent := Agent?) {
            SendFeedbackNotification(FeedbackKind, l_Agent, ResourceKind, Value)
        } else {
            for (Player:GetPlayspace().GetPlayers()) {
                SendFeedbackNotification(FeedbackKind, Player, ResourceKind, Value)
            }
        }

    # Sends a notification to an agent that a value related to the resource has changed (see resource_feedback_kind).
    # Returns false if the resource does not exist or is not configured with a HUD message device for that purpose.
    SendFeedbackNotification<public>(FeedbackKind:resource_feedback_kind, Agent:agent, ResourceKind:string, Value:float):logic=
        if (Entry := m_Resources[ResourceKind], Feedback := Entry.Feedbacks[FeedbackKind]) {
            FormattedAmount := Feedback.Formatter.Format(Value)
            Feedback.HUDMessenger.Show(Agent, ToMessage(FormattedAmount), ?DisplayTime := Feedback.DisplayTime)
            true
        } else {
            false
        }
    ## }}}

    # Subscribe to post 'Take' events.
    SubscribeToResourceTake<public>(Listener:type{__(:resource_take_event):void}):void=
        set m_PostTakeListeners += array{Listener}

    CustomTrigger<override>(Input:string, Agent:?agent):logic=
        SplitRes := Input.Split(':')
        if (SplitRes[0] = "give", ResourceId := SplitRes[1], Amount := ParseInt[SplitRes[2]]) {
            Give(Agent, ResourceId, Amount)
            true
        } else if (SplitRes[0] = "set", ResourceId := SplitRes[1], TargetAmount := ParseInt[SplitRes[2]]) {
            CurrentAmount := GetAmount(Agent, ResourceId)
            if (TargetAmount > CurrentAmount) {
                Give(Agent, ResourceId, TargetAmount - CurrentAmount)
            } else if (TargetAmount < CurrentAmount) {
                Take(Agent, ResourceId, CurrentAmount - TargetAmount)
            }
            true
        } else {
            false
        }

    ## Persistence {{{
    GetPersistenceControl<override>()<transacts><decides>:persistence_recordable_control=
        m_PersistenceControl
    
    PersistenceReserve<override>(StorageMem:storage_memory)<transacts><decides>:resource_manager_persistence_state=
        var ResourceChunks: [string]storage_memory_chunk = map{}
        for (ResourceId->Entry:m_Resources, ReserveBits := Entry.PersistenceBits?) {
            if (MemChunk := StorageMem.Reserve[ReserveBits, "resource_{ResourceId}"]) {
                if (set ResourceChunks[ResourceId] = MemChunk) {}
            }
        }
        resource_manager_persistence_state{
            ResourceChunks := ResourceChunks
        }

    PersistenceSave<override>(i_State:persistence_recordable_state, Agent:agent, SaveCause:persistence_save_cause):void=
        if (State := resource_manager_persistence_state[i_State]) {
            for (ResourceId->MemChunk:State.ResourceChunks) {
                Amount := GetAmount(Agent, ResourceId)
                MemChunk.WriteInt(Agent, Amount)
            }
        }

    PersistenceLoad<override>(i_State:persistence_recordable_state, Agent:agent, Query:persistence_load_query):void=
        if (State := resource_manager_persistence_state[i_State]) {
            Log(MsgLoadPersistence(Agent, ToString(Query), State.ResourceChunks.Length), ?Level := log_level.Debug)
            for (ResourceId->MemChunk:State.ResourceChunks) {
                SavedAmount := MemChunk.ReadInt(Agent)
                Log(MsgRecoverResource(Agent, ResourceId, SavedAmount), ?Level := log_level.Debug)
                CurrentAmount := GetAmount(Agent, ResourceId)
                if (IsShareableResource[ResourceId] and not (Query.Overwrite?)) {
                    if (SavedAmount > CurrentAmount) {
                        Give(Agent, ResourceId, SavedAmount - CurrentAmount)
                    }
                } else if (SavedAmount > CurrentAmount) {
                    ToGive := SavedAmount - CurrentAmount
                    Give(Agent, ResourceId, ToGive)
                } else if (SavedAmount < CurrentAmount) {
                    Take(Agent, ResourceId, CurrentAmount - SavedAmount)
                }
            }
        } else {
            Log("Resource manager was given invalid persistence state to load from", ?Level := log_level.Error)
        }

    PersistenceReset<override>(i_State:persistence_recordable_state, Agent:agent, Query:persistence_reset_query):void=
        if (TycoonManager := tycoon_manager_device[Query.SourceDevice], State := resource_manager_persistence_state[i_State]) {
            for (ResourceId->Chunk:State.ResourceChunks) {
                if (not TycoonManager.p_RebirthRetainResources.Find[ResourceId]) {
                    Chunk.Reset(Agent)
                } else {
                    Log("Retaining resource \"{ResourceId}\" during tycoon rebirth for '{Agent}'", ?Level := log_level.Debug)
                }
            }
        } else {
            i_State.Reset(Agent)
        }
        PersistenceLoad(i_State, Agent, persistence_load_query{
            Overwrite := true
            FirstLoad := false
        })
    MsgLoadPersistence<private><localizes>(Agent:agent, QueryStr:string, ChunkCount:int):message=
        "Loading resource manager persistence ({QueryStr}) for '{Agent}' ({ChunkCount} resources)"
    MsgRecoverResource<private><localizes>(Agent:agent, ResourceId:string, Amount:int):message=
        "Recovering {Amount} of resource {ResourceId} for '{Agent}'"

resource_manager_persistence_state<public> := class(persistence_recordable_state):
    ResourceChunks<public>: [string]storage_memory_chunk

    Reset<override>(Agent:agent):void=
        for (__->Chunk:ResourceChunks) {
            Chunk.Reset(Agent)
        }
## }}}
