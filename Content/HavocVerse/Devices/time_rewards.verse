using { /Fortnite.com/Devices }
using { /Fortnite.com/UI }
using { /Verse.org/Simulation }
using { /Verse.org/Colors }
using { /UnrealEngine.com/Temporary/Diagnostics }
using { /UnrealEngine.com/Temporary/UI }
using { /UnrealEngine.com/Temporary/SpatialMath }

using { HavocVerse.Debug }
using { HavocVerse.Storage }
using { HavocVerse.UI }
using { HavocVerse.Util }
using { HavocVerse.Util.Task }
using { HavocVerse.Logging }

time_rewards_timer_sharing := enum:
    Global
    PerPlayer

time_rewards_reward_config := class<concrete>:
    @editable Delay: float = 1.0
    @editable Duration: float = 90.0
    @editable OnStartTriggers: []trigger_device = array{}
    @editable OnRewardTriggers: []trigger_device = array{}

time_rewards_device := class(creative_device, aware_device, aware_debug_device, resetable, persistence_recordable):
    @editable p_AccessControl: []access_control_device = array{}

    @editable p_TimerSharing: time_rewards_timer_sharing = time_rewards_timer_sharing.PerPlayer
    @editable p_Timeline: []time_rewards_reward_config = array{}

    @editable p_HUDEnabled: logic = false
    @editable p_HUDManagerConfig: hud_manager_base_config = hud_manager_base_config{
        RefreshRate := 1.0
    }
    @editable p_HUDConfig: hud_time_rewards_config = hud_time_rewards_config{}
    @editable p_HUDBackgroundPlacement: ui_canvas_placement_config = ui_canvas_placement_config{}
    @editable p_HUDTextPlacement: ui_canvas_placement_config = ui_canvas_placement_config{}

    var m_HudManager: ?time_rewards_hud_manager = false
    var m_Players: [player]time_rewards_player = map{}

    var m_GlobalRewardIndex: int = 0
    var m_GlobalStartedAt: float = -1.0

    var m_PersistenceControl: persistence_recordable_control = persistence_recordable_control{}

    DeviceMeta<override>()<transacts>:aware_device_meta=
        aware_device_meta{
            DeviceName := "time_rewards"
        }

    DebugStats<override>()<transacts>:aware_debug_stats=
        aware_debug_stats{
            Length := map{
                "players" => m_Players.Length
            }
        }

    OnBegin<override>()<suspends>:void=
        AwareInitDevice(Self, OnInit)

    OnInit<private>()<suspends>:void=
        HudManager := time_rewards_hud_manager{
            AccessControl := option{p_AccessControl[0]}
            RefreshRate := p_HUDManagerConfig.RefreshRate
            ZOrder := p_HUDManagerConfig.ZOrder
            Device := Self
        }
        set m_HudManager = Some(HudManager)
        if (p_TimerSharing = time_rewards_timer_sharing.Global, Reward := p_Timeline[0]) {
            set m_GlobalStartedAt = GetSimulationElapsedTime()
            for (Trigger:Reward.OnStartTriggers) {
                Trigger.Trigger()
            }
        }
        HudManager.Init(GetPlayspace())
        spawn:
            StartEventLoop()
    
    Reset<override>(Agent:?agent):void=
        if (p_TimerSharing = time_rewards_timer_sharing.PerPlayer, Agent?) {
            if (RewardPlayer := m_Players[player[Agent?]]) {
                set RewardPlayer.m_RewardIndex = 0
                set RewardPlayer.m_StartedAt = -1.0
            }
        } else {
            set m_GlobalRewardIndex = 0
            set m_GlobalStartedAt = GetSimulationElapsedTime()
            if (Reward := p_Timeline[0]) {
                for (Trigger:Reward.OnStartTriggers) {
                    Trigger.Trigger()
                }
            }
        }

    StartEventLoop<private>()<suspends>:void=
        loop {
            Sleep(1.0)
            Now := GetSimulationElapsedTime()
            if (p_TimerSharing = time_rewards_timer_sharing.Global) {
                StartedAt := m_GlobalStartedAt
                RewardIndex := m_GlobalRewardIndex
                if (
                    Reward := p_Timeline[m_GlobalRewardIndex]
                    StartedAt >= 0.0
                    Now >= m_GlobalStartedAt + Reward.Delay + Reward.Duration
                ) {
                    for (Trigger:Reward.OnRewardTriggers) {
                        Trigger.Trigger()
                    }
                    set m_GlobalRewardIndex = RewardIndex+1
                    if (NextReward := p_Timeline[m_GlobalRewardIndex]) {
                        set m_GlobalStartedAt = Now
                        for (Trigger:NextReward.OnStartTriggers) {
                            Trigger.Trigger()
                        }
                    } else {
                        set m_GlobalStartedAt = -1.0
                    }
                }
            } else if (p_TimerSharing = time_rewards_timer_sharing.PerPlayer) {
                for (
                    __->RewardPlayer:m_Players
                    Reward := p_Timeline[RewardPlayer.m_RewardIndex]
                    RewardPlayer.m_StartedAt >= 0.0
                    Now >= RewardPlayer.m_StartedAt + Reward.Delay + Reward.Duration
                ) {
                    for (Trigger:Reward.OnRewardTriggers) {
                        Trigger.Trigger(RewardPlayer.Player)
                    }
                    set RewardPlayer.m_RewardIndex = RewardPlayer.m_RewardIndex+1
                    if (NextReward := p_Timeline[RewardPlayer.m_RewardIndex]) {
                        set RewardPlayer.m_StartedAt = Now
                        for (Trigger:NextReward.OnStartTriggers) {
                            Trigger.Trigger(RewardPlayer.Player)
                        }
                    } else {
                        set RewardPlayer.m_StartedAt = -1.0
                    }
                }
            }
            for (__->RewardPlayer:m_Players) {
                if (Element := RewardPlayer.m_HUDElement?) {
                    Element.Update(Now)
                }
            }
        }

    GetPersistenceControl<override>()<transacts><decides>:persistence_recordable_control=
        m_PersistenceControl

    PersistenceReserve<override>(StorageMem:storage_memory)<transacts><decides>:time_rewards_persistence_state=
        time_rewards_persistence_state{
            RewardIndexChunk := StorageMem.Reserve[6, "time_rewards_index"]
            ElapsedTimeChunk := StorageMem.Reserve[16, "time_rewards_elapsed"]
        }

    PersistenceSave<override>(i_State:persistence_recordable_state, Agent:agent, SaveCause:persistence_save_cause):void=
        if (State := time_rewards_persistence_state[i_State]) {
            Res := if (p_TimerSharing = time_rewards_timer_sharing.Global) {
                StartedAt := m_GlobalStartedAt
                RewardIndex := m_GlobalRewardIndex
                (StartedAt, RewardIndex)
            } else if (p_TimerSharing = time_rewards_timer_sharing.PerPlayer, RewardPlayer := m_Players[Agent]) {
                (RewardPlayer.m_StartedAt, RewardPlayer.m_RewardIndex)
            } else {
                return
            }
            State.RewardIndexChunk.WriteInt(Agent, Res(1))
            ElapsedTime := if (Res(0) >= 0.0) {
                Now := GetSimulationElapsedTime()
                DefaultingFloor(Now-Res(0))
            } else {
                0
            }
            State.ElapsedTimeChunk.WriteInt(Agent, ElapsedTime)
        }

    PersistenceLoad<override>(i_State:persistence_recordable_state, Agent:agent, Query:persistence_load_query):void=
        if (p_TimerSharing <> time_rewards_timer_sharing.PerPlayer) {
            return
        }
        if (State := time_rewards_persistence_state[i_State], Player := player[Agent]) {
            RewardPlayer := if (l_RewardPlayer := m_Players[Player]) {
                l_RewardPlayer
            } else {
                l_RewardPlayer := time_rewards_player{
                    Player := Player
                }
                if (set m_Players[Player] = l_RewardPlayer) {}
                l_RewardPlayer
            }
            set RewardPlayer.m_RewardIndex = State.RewardIndexChunk.ReadInt(Agent)
            if (p_Timeline[RewardPlayer.m_RewardIndex]) {
                ElapsedTime := State.ElapsedTimeChunk.ReadInt(Agent)
                Log("Recovered time rewards, reward #{RewardPlayer.m_RewardIndex} with {ElapsedTime}s elapsed", ?Level := log_level.Debug)
                set RewardPlayer.m_StartedAt = GetSimulationElapsedTime() - ToFloat(ElapsedTime)
            } else {
                Log("Recovered time rewards, reward #{RewardPlayer.m_RewardIndex}", ?Level := log_level.Debug)
                set RewardPlayer.m_StartedAt = 1.0
            }
        } else {
            Log("Time rewards device was given invalid persistence state to load from", ?Level := log_level.Error)
        }

time_rewards_persistence_state := class(persistence_recordable_state):
    RewardIndexChunk: storage_memory_chunk
    ElapsedTimeChunk: storage_memory_chunk

    Reset<override>(Agent:agent):void=
        RewardIndexChunk.Reset(Agent)
        ElapsedTimeChunk.Reset(Agent)

time_rewards_player := class:
    Player: player

    var m_RewardIndex: int = 0
    var m_StartedAt: float = -1.0

    var m_HUDElement: ?hud_time_rewards_instance = false

time_rewards_hud_manager := class(hud_manager_base):
    Device: time_rewards_device

    CreateHUD<override>(Player:player):?hud_manager_player=
        RewardPlayer := if (l_RewardPlayer := Device.m_Players[Player]) {
            l_RewardPlayer
        } else {
            l_RewardPlayer := time_rewards_player{
                Player := Player
            }
            if (set Device.m_Players[Player] = l_RewardPlayer) {}
            if (Device.p_TimerSharing = time_rewards_timer_sharing.PerPlayer, Reward := Device.p_Timeline[0]) {
                set l_RewardPlayer.m_StartedAt = GetSimulationElapsedTime()
                for (Trigger:Reward.OnStartTriggers) {
                    Trigger.Trigger(Player)
                }
            }
            l_RewardPlayer
        }
        if (not (Device.p_HUDEnabled?)) {
            return false
        }
        Element := CreateHUDTimeRewardsCountdown(Device.p_HUDConfig, RewardPlayer)
        set RewardPlayer.m_HUDElement = Some(Element)
        option{MakeSingleElementHUDPlayer(Player, canvas[Element.Widget], Element)}

    CreateHUDTimeRewardsCountdown(
        Config: hud_time_rewards_config,
        RewardPlayer: time_rewards_player
    ):hud_time_rewards_instance=
        BackgroundWidget := if (
            Config.BackgroundTexture.Length > 0
            Texture := AwareDefaultAssetsProvider[].GetTextureAsset[Config.BackgroundTexture]
        ) {
            texture_block{
                DefaultImage := Texture
                DefaultDesiredSize := Config.BackgroundSize
            }
        } else {
            color_block{
                DefaultOpacity := 0.0
                DefaultDesiredSize := Config.BackgroundSize
            }
        }
        TextWidget := Config.TextStyle.ToTextBlock("")

        Widget := canvas{
            Slots := array{
                Device.p_HUDBackgroundPlacement.ToCanvasSlot(BackgroundWidget)
                Device.p_HUDTextPlacement.ToCanvasSlot(TextWidget)
            }
        }
        hud_time_rewards_instance{
            Widget := Widget
            Device := Device
            RewardPlayer := RewardPlayer
            TextWidget := MakeCachedTextBlock(TextWidget)
            TextTemplate := Config.TextTemplate
        }

hud_time_rewards_config<public> := struct<concrete>:
    @editable BackgroundSize<public>: vector2 = vector2{X := 100.0, Y := 50.0}
    @editable BackgroundTexture<public>: string = ""
    @editable TextStyle<public>: ui_text_style_config = ui_text_style_config{}
    @editable TextTemplate<public>: string = "In %Timer%"

hud_time_rewards_instance := class(hud_element):
    Device: time_rewards_device
    RewardPlayer: time_rewards_player
    TextTemplate: string

    TextWidget: cached_text_block

    Update<override>():void=
        Update(GetSimulationElapsedTime())

    Update(Now:float):void=
        State := case(Device.p_TimerSharing):
            time_rewards_timer_sharing.Global => (Device.m_GlobalRewardIndex, Device.m_GlobalStartedAt)
            time_rewards_timer_sharing.PerPlayer => (RewardPlayer.m_RewardIndex, RewardPlayer.m_StartedAt)
        StartedAt := State(1)
        if (
            StartedAt >= 0.0
            Reward := Device.p_Timeline[State(0)]
            Now >= StartedAt + Reward.Delay
        ) {
            Show()
            BonusAt := StartedAt + Reward.Delay + Reward.Duration
            Text := if (RemainingSeconds := Ceil[BonusAt - Now], RemainingSeconds > 0) {
                ToTimerMMSS(RemainingSeconds)
            } else {
                "Now"
            }
            TextWidget.SetText(TextTemplate.ReplaceAll("%Timer%", Text))
        } else {
            Hide()
        }
