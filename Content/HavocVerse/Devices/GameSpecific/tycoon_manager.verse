using { /Fortnite.com/Devices }
using { /Fortnite.com/Characters }
using { /Verse.org/Simulation }
using { /UnrealEngine.com/Temporary/Diagnostics }
using { /UnrealEngine.com/Temporary/SpatialMath }
using { /UnrealEngine.com/Temporary/UI }

using { HavocVerse.Debug }
using { HavocVerse.Devices }
using { HavocVerse.Storage }
using { HavocVerse.JSON }
using { HavocVerse.UI }
using { HavocVerse.Util }
using { HavocVerse.Util.Task }
using { HavocVerse.Logging }

tycoon_manager_rebirth_reset_trigger_config := class<concrete>:
    @editable Trigger: trigger_device = trigger_device{}
    @editable ForLevel: int = -1

    IsApplicable<public>(RebirthLevel:int)<transacts><decides>:void=
        ForLevel < 0 or ForLevel = RebirthLevel

tycoon_manager_progress_line<public> := class<concrete>:
    @editable Name<public>: string = "N/A"
    @editable ClassSlot: int = -1
    @editable AccessControl<public>: ?access_control_device = false
    @editable FrontCheckpoint: ?player_checkpoint_device = false
    @editable FrontTeleporter: ?teleporter_device = false
    @editable AutoGenerationManager: ?auto_generation_manager = false
    @editable ClaimBuyplate: ?buy_plate_device = false
    @editable ExtraBuyplates: []creative_device = array{}
    @editable ExtraRollbackDevices: []creative_device = array{}
    @editable ExtraRollbackTriggers: []trigger_device = array{}
    @editable BeaconLines: []tycoon_manager_progress_line_beacon_line = array{}

    @editable RebirthResetTriggers: []tycoon_manager_rebirth_reset_trigger_config = array{}
    @editable RebirthBillboard: []billboard_device = array{}
    @editable RebirthBillboardTemplate: string = "Rebirths: %CurrentLevel%"
    @editable RebirthBillboardHideOnZero: logic = false

    var m_Index: int = -1
    var m_AutoGenerationTickets: [string]auto_generation_ticket = map{}
    var m_PersistenceRecorder: ?persistence_recorder_device = false
    var m_RebirthRecordableHandle: ?persistence_recorder_recordable_handle = false
    var m_BuyPlatesBackrefCache: [string]buy_plate_device = map{}

    var m_RebirthLevel: int = 0
    var m_SetResourceMultipliers: [string]float = map{}

tycoon_manager_progress_line_beacon_line := class<concrete>:
    @editable BeaconLineId: string = "beaconline1"
    @editable MapIndicator: map_indicator_device = map_indicator_device{}
    @editable MapIndicatorUnmetCost: ?map_indicator_device = false

    var m_CurrentEntry: int = 0
    var m_LockedTargetBuyPlate: ?string = false

    var m_MapIndicatorState: ?tycoon_manager_beacon_line_buy_plate_state = false
    var m_MapIndicatorEnabled: ?tycoon_manager_beacon_line_map_indicator_variant = false
    var m_MapIndicatorPulseActive: logic = false
    var m_MapIndicatorPulseToAgents: [agent]void = map{}

    GetActiveMapIndicator()<decides><transacts>:map_indicator_device=
        case(m_MapIndicatorEnabled?):
            tycoon_manager_beacon_line_map_indicator_variant.Primary => MapIndicator
            tycoon_manager_beacon_line_map_indicator_variant.UnmetCost => MapIndicatorUnmetCost?

    DisablePulse(ProgressLine:tycoon_manager_progress_line):void=
        if (not m_MapIndicatorPulseActive?) {
            return
        }
        set m_MapIndicatorPulseActive = false
        if (l_MapIndicator := GetActiveMapIndicator[]) {
            for (Agent:ProgressLine.AccessControl?.ListAllowedAgents()) {
                if. set m_MapIndicatorPulseToAgents[Agent] = void
            }
            for (Agent->__:m_MapIndicatorPulseToAgents) {
                l_MapIndicator.DeactivateObjectivePulse(Agent)
            }
        }
        set m_MapIndicatorPulseToAgents = map{}

tycoon_manager_beacon_line_map_indicator_variant := enum:
    Primary
    UnmetCost

ToString(Variant:tycoon_manager_beacon_line_map_indicator_variant)<transacts>:string=
    case(Variant):
        tycoon_manager_beacon_line_map_indicator_variant.Primary => "Primary"
        tycoon_manager_beacon_line_map_indicator_variant.UnmetCost => "UnmetCost"

tycoon_manager_beacon_line_buy_plate_state := class<final>:
    BuyPlate: buy_plate_device
    UnmetCost: logic

tycoon_manager_player_state := class<final>:
    ProgressLine: tycoon_manager_progress_line

tycoon_manager_rebirth_pet_config<public> := class<concrete>:
    @editable PetId<public>: string = ""
    @editable RebirthLevel<public>: int = 1

tycoon_manager_beacon_line_area_config := class<concrete>:
    @editable Volume: volume_device = volume_device{}
    @editable OuterVolume: ?volume_device = false

    var m_InVolume: [agent]void = map{}

tycoon_claim_or_teleport_result<public> := enum:
    Claimed
    Teleported
    Failed

tycoon_rebirth_event<public> := class<final>:
    Player<public>: player
    NewRebirthLevel<public>: int
    Virtual<public>: logic

tycoon_manager_device<public> := class(creative_device, aware_device, aware_debug_device, pet_condition_provider):
    @editable p_ResourceManager<public>: resource_manager_device = resource_manager_device{}
    @editable p_AnalyticsController<public>: analytics_controller_device = analytics_controller_device{}
    @editable p_ClassManager<public>: class_manager_device = class_manager_device{}
    @editable p_ProgressLines: []tycoon_manager_progress_line = array{}
    @editable p_TimeRewardsDevice: []time_rewards_device = array{}
    @editable p_PetController: []pet_controller_device = array{}

    @editable var p_TycoonJsonConfig: string = ""
    @editable var p_BuyplatesJsonConfig: string = ""
    @editable p_RoutingIdSingleMatch: logic = true
    @editable p_AlwaysSetCostOverride: logic = true

    @editable p_ClaimFailHUDMessage: []hud_message_device = array{}
    @editable p_MiscPlayerHeight<public>: float = 180.0

    @editable p_BeaconLinesHideAreas: []tycoon_manager_beacon_line_area_config = array{}
    @editable p_BeaconLinesMapIndicatorOffset: vector3 = vector3{}

    @editable p_EnableQuitRollback: logic = true
    @editable p_QuitRollbackGranularity: quit_rollback_device_granularity = quit_rollback_device_granularity.Player

    @editable p_EnablePersistence: logic = false
    @editable p_PersistenceSaveSlots: []shared_storage_chunk_device = array{}
    @editable p_PersistenceSavePoints: []exposed_save_point_device = array{}
    @editable p_PersistenceAutoSaveInternal: float = 3.0
    @editable p_PersistenceBuyPlatesReserveBits: int = -1
    @editable p_PersistenceIncludeResourceManager: logic = true
    @editable p_PersistenceExtraDevices: []creative_device = array{}
    @editable p_PersistenceSelectMenuEnabled: logic = true
    @editable p_PersistenceUseDialogSelectMenu: []persistence_recorder_popup_dialog_menu_config = array{}
    @editable p_PersistenceSelectMenuConfig: default_persistence_recorder_load_select_menu_config = default_persistence_recorder_load_select_menu_config{}
    @editable p_PersistenceBuyPlateRecoveryChunkSize: int = -1
    @editable p_PersistenceBuyPlateRecoveryChunkDelay: float = 0.1

    @editable p_RebirthRetainResources<public>: []string = array{}
    @editable p_RebirthPets<public>: []tycoon_manager_rebirth_pet_config = array{}
    @editable p_RebirthLevelReserveBits: int = 4
    @editable p_RebirthNextLevelOnTrigger: []trigger_device = array{}
    @editable p_RebirthNextLevelDialog: []popup_dialog_device = array{}
    @editable p_RebirthNextLevelDialogPerLevel: logic = false
    @editable p_RebirthNextLevelDialogPerLevelDefaultLast: logic = true
    @editable p_RebirthNextLevelDialogConfirmButtonIndex: int = 0
    @editable p_RebirthPlaceholdersMultiplierFormatter: number_formatter_config = number_formatter_config{
        Template := "%Val%x"
    }
    @editable p_RebirthResetTriggers: []tycoon_manager_rebirth_reset_trigger_config = array{}
    @editable p_RebirthMultipliersChannel: string = "tycoon_rebirth"
    @editable p_RebirthHUDManagerConfig: hud_manager_base_config = hud_manager_base_config{
        RefreshRate := -1.0
    }
    @editable p_RebirthHUDImageSize: vector2 = vector2{X := 100.0, Y := 50.0}
    @editable p_RebirthHUDImagePlacement: ui_canvas_placement_config = ui_canvas_placement_config{}

    @editable p_BuyPlateCostMetTrackingRefreshRate: float = 0.5

    PlayerSaveSlotSelectedEvent<public>: event(tuple(tycoon_manager_progress_line,player_save_slot_selected_event)) = event(tuple(tycoon_manager_progress_line,player_save_slot_selected_event)){}
    PlayerJoinProgressLineEvent<public>: event(tuple(player,tycoon_manager_progress_line)) = event(tuple(player,tycoon_manager_progress_line)){}

    var m_TycoonConfig<private>: tycoon_manager_config = tycoon_manager_config{}
    var m_BuyplatesConfigs<private>: [string]tycoon_manager_buyplates_entry = map{}
    var m_Initialized<private>: logic = false

    var m_Players<private>: [player]tycoon_manager_player_state = map{}
    var m_RebirthHudManager<internal>: ?hud_manager_base = false
    var m_RebirthDialogs<private>: [int]tuple(string,dialog_device_wrapper) = map{}
    var m_RebirthPets<private>: [string]int = map{}

    var m_PersistenceControl<private>: persistence_recordable_control = persistence_recordable_control{}

    var m_RebirthListeners<private>: []type{__(:tycoon_rebirth_event):void} = array{}

    DeviceMeta<override>()<transacts>:aware_device_meta=
        aware_device_meta{
            DeviceName := "tycoon_manager"
        }

    DebugStats<override>()<transacts>:aware_debug_stats=
        aware_debug_stats{
            Length := map{
                "buyplatesConfigs" => m_BuyplatesConfigs.Length
                "players" => m_Players.Length
                "rebirthDialogs" => m_RebirthDialogs.Length
                "listeners::rebirth" => m_RebirthListeners.Length
            }
        }

    OnBegin<override>()<suspends>:void=
        AwareInitDevice(Self, OnInit)


    OnInit<private>()<suspends>:void=
        set m_TycoonConfig = tycoon_manager_config{  
            BeaconLines := map{
              "beaconline1" => tycoon_manager_beacon_line_config{InOrderBuyPlates := array{tycoon_manager_beacon_line_config_entry{AvailableOptions := map{"bp_050"=>void,"bp_051"=>void,"bp_052"=>void,"bp_500"=>void,"bp_400"=>void,"bp_300"=>void,"bp_100"=>void,"bp_200"=>void,"bp_600"=>void,"bp_001"=>void,"bp_401"=>void,"bp_501"=>void,"bp_101"=>void,"bp_301"=>void,"bp_601"=>void,"bp_007"=>void,"bp_402"=>void,"bp_509"=>void,"bp_002"=>void,"bp_302"=>void,"bp_602"=>void,"bp_201"=>void,"bp_003"=>void,"bp_403"=>void,"bp_603"=>void,"bp_004"=>void,"bp_303"=>void,"bp_510"=>void,"bp_202"=>void,"bp_404"=>void,"bp_304"=>void,"bp_604"=>void,"bp_005"=>void,"bp_502"=>void,"bp_203"=>void,"bp_605"=>void,"bp_006"=>void,"bp_204"=>void,"bp_405"=>void,"bp_008"=>void,"bp_305"=>void,"bp_503"=>void,"bp_009"=>void,"bp_205"=>void,"bp_606"=>void,"bp_306"=>void,"bp_010"=>void,"bp_406"=>void,"bp_504"=>void,"bp_607"=>void,"bp_011"=>void,"bp_102"=>void,"bp_012"=>void,"bp_307"=>void,"bp_013"=>void,"bp_206"=>void,"bp_505"=>void,"bp_014"=>void,"bp_700"=>void,"bp_207"=>void,"bp_506"=>void,"bp_308"=>void,"bp_015"=>void,"bp_701"=>void,"bp_507"=>void,"bp_208"=>void,"bp_016"=>void,"bp_407"=>void,"bp_309"=>void,"bp_702"=>void,"bp_508"=>void,"bp_017"=>void,"bp_209"=>void,"bp_703"=>void,"bp_018"=>void,"bp_511"=>void,"bp_408"=>void,"bp_704"=>void,"bp_019"=>void,"bp_210"=>void,"bp_310"=>void,"bp_020"=>void,"bp_705"=>void,"bp_021"=>void,"bp_512"=>void,"bp_311"=>void,"bp_103"=>void,"bp_514"=>void,"bp_513"=>void,"bp_608"=>void,"bp_022"=>void,"bp_312"=>void,"bp_409"=>void,"bp_023"=>void,"bp_609"=>void,"bp_211"=>void,"bp_024"=>void,"bp_610"=>void,"bp_515"=>void,"bp_025"=>void,"bp_212"=>void,"bp_611"=>void,"bp_410"=>void,"bp_313"=>void,"bp_516"=>void,"bp_026"=>void,"bp_612"=>void,"bp_027"=>void,"bp_213"=>void,"bp_314"=>void,"bp_613"=>void,"bp_028"=>void,"bp_029"=>void,"bp_517"=>void,"bp_214"=>void,"bp_411"=>void,"bp_315"=>void,"bp_030"=>void,"bp_518"=>void,"bp_614"=>void,"bp_215"=>void,"bp_031"=>void,"bp_615"=>void,"bp_032"=>void,"bp_519"=>void,"bp_412"=>void,"bp_033"=>void,"bp_616"=>void,"bp_316"=>void,"bp_520"=>void,"bp_034"=>void,"bp_617"=>void,"bp_104"=>void,"bp_521"=>void,"bp_317"=>void,"bp_618"=>void,"bp_413"=>void,"bp_035"=>void,"bp_216"=>void,"bp_619"=>void,"bp_036"=>void,"bp_318"=>void,"bp_620"=>void,"bp_037"=>void,"bp_522"=>void,"bp_217"=>void,"bp_414"=>void,"bp_621"=>void,"bp_038"=>void,"bp_319"=>void,"bp_523"=>void,"bp_622"=>void,"bp_039"=>void,"bp_218"=>void,"bp_623"=>void,"bp_524"=>void,"bp_415"=>void,"bp_624"=>void,"bp_219"=>void,"bp_525"=>void,"bp_320"=>void,"bp_625"=>void,"bp_626"=>void,"bp_526"=>void,"bp_416"=>void,"bp_321"=>void,"bp_627"=>void,"bp_220"=>void,"bp_628"=>void,"bp_322"=>void,"bp_527"=>void,"bp_417"=>void,"bp_629"=>void,"bp_250"=>void,"bp_630"=>void,"bp_528"=>void,"bp_323"=>void,"bp_251"=>void,"bp_418"=>void,"bp_529"=>void,"bp_324"=>void,"bp_252"=>void,"bp_530"=>void,"bp_419"=>void,"bp_325"=>void,"bp_531"=>void,"bp_253"=>void,"bp_420"=>void,"bp_421"=>void,"bp_422"=>void,"bp_631"=>void}}}}
            }  
            Rebirth := option{tycoon_manager_rebirth_config{
              MaxLevel := 16000
              DefaultToLast := true
              LevelConfigs := map{
                0 => tycoon_manager_rebirth_level_config{HudImageTexture := option{""}, ResourceMultipliers := map{}}
                1 => tycoon_manager_rebirth_level_config{HudImageTexture := option{"UI/Rebirths/rebirth_01"}, ResourceMultipliers := map{"scrap"=>2.0, "food"=>2.0, "meds"=>2.0, "garage"=>2.0, "armory"=>2.0, "walls"=>2.0}}
                2 => tycoon_manager_rebirth_level_config{HudImageTexture := option{"UI/Rebirths/rebirth_02"}, ResourceMultipliers := map{"scrap"=>3.0, "food"=>3.0, "meds"=>3.0, "garage"=>3.0, "armory"=>3.0, "walls"=>3.0}}
                3 => tycoon_manager_rebirth_level_config{HudImageTexture := option{"UI/Rebirths/rebirth_03"}, ResourceMultipliers := map{"scrap"=>4.0, "food"=>4.0, "meds"=>4.0, "garage"=>4.0, "armory"=>4.0, "walls"=>4.0}}
                4 => tycoon_manager_rebirth_level_config{HudImageTexture := option{"UI/Rebirths/rebirth_04"}, ResourceMultipliers := map{"scrap"=>5.0, "food"=>5.0, "meds"=>5.0, "garage"=>5.0, "armory"=>5.0, "walls"=>5.0}}
                5 => tycoon_manager_rebirth_level_config{HudImageTexture := option{"UI/Rebirths/rebirth_05"}, ResourceMultipliers := map{"scrap"=>5.0, "food"=>5.0, "meds"=>5.0, "garage"=>5.0, "armory"=>5.0, "walls"=>5.0}}
                6 => tycoon_manager_rebirth_level_config{HudImageTexture := option{"UI/Rebirths/rebirth_06"}, ResourceMultipliers := map{"scrap"=>5.0, "food"=>5.0, "meds"=>5.0, "garage"=>5.0, "armory"=>5.0, "walls"=>5.0}}
                7 => tycoon_manager_rebirth_level_config{HudImageTexture := option{"UI/Rebirths/rebirth_07"}, ResourceMultipliers := map{"scrap"=>5.0, "food"=>5.0, "meds"=>5.0, "garage"=>5.0, "armory"=>5.0, "walls"=>5.0}}
                8 => tycoon_manager_rebirth_level_config{HudImageTexture := option{"UI/Rebirths/rebirth_08"}, ResourceMultipliers := map{"scrap"=>5.0, "food"=>5.0, "meds"=>5.0, "garage"=>5.0, "armory"=>5.0, "walls"=>5.0}}
                9 => tycoon_manager_rebirth_level_config{HudImageTexture := option{"UI/Rebirths/rebirth_09"}, ResourceMultipliers := map{"scrap"=>5.0, "food"=>5.0, "meds"=>5.0, "garage"=>5.0, "armory"=>5.0, "walls"=>5.0}}
                10 => tycoon_manager_rebirth_level_config{HudImageTexture := option{"UI/Rebirths/rebirth_10"}, ResourceMultipliers := map{"scrap"=>5.0, "food"=>5.0, "meds"=>5.0, "garage"=>5.0, "armory"=>5.0, "walls"=>5.0}}
                11 => tycoon_manager_rebirth_level_config{HudImageTexture := option{"UI/Rebirths/rebirth_11"}, ResourceMultipliers := map{"scrap"=>5.0, "food"=>5.0, "meds"=>5.0, "garage"=>5.0, "armory"=>5.0, "walls"=>5.0}}
                12 => tycoon_manager_rebirth_level_config{HudImageTexture := option{"UI/Rebirths/rebirth_12"}, ResourceMultipliers := map{"scrap"=>5.0, "food"=>5.0, "meds"=>5.0, "garage"=>5.0, "armory"=>5.0, "walls"=>5.0}}
                13 => tycoon_manager_rebirth_level_config{HudImageTexture := option{"UI/Rebirths/rebirth_13"}, ResourceMultipliers := map{"scrap"=>5.0, "food"=>5.0, "meds"=>5.0, "garage"=>5.0, "armory"=>5.0, "walls"=>5.0}}
                14 => tycoon_manager_rebirth_level_config{HudImageTexture := option{"UI/Rebirths/rebirth_14"}, ResourceMultipliers := map{"scrap"=>5.0, "food"=>5.0, "meds"=>5.0, "garage"=>5.0, "armory"=>5.0, "walls"=>5.0}}
                15 => tycoon_manager_rebirth_level_config{HudImageTexture := option{"UI/Rebirths/rebirth_15"}, ResourceMultipliers := map{"scrap"=>5.0, "food"=>5.0, "meds"=>5.0, "garage"=>5.0, "armory"=>5.0, "walls"=>5.0}}
                16 => tycoon_manager_rebirth_level_config{HudImageTexture := option{"UI/Rebirths/rebirth_16"}, ResourceMultipliers := map{"scrap"=>5.0, "food"=>5.0, "meds"=>5.0, "garage"=>5.0, "armory"=>5.0, "walls"=>5.0}}
                17 => tycoon_manager_rebirth_level_config{HudImageTexture := option{"UI/Rebirths/rebirth_17"}, ResourceMultipliers := map{"scrap"=>5.0, "food"=>5.0, "meds"=>5.0, "garage"=>5.0, "armory"=>5.0, "walls"=>5.0}}
                18 => tycoon_manager_rebirth_level_config{HudImageTexture := option{"UI/Rebirths/rebirth_18"}, ResourceMultipliers := map{"scrap"=>5.0, "food"=>5.0, "meds"=>5.0, "garage"=>5.0, "armory"=>5.0, "walls"=>5.0}}
                19 => tycoon_manager_rebirth_level_config{HudImageTexture := option{"UI/Rebirths/rebirth_19"}, ResourceMultipliers := map{"scrap"=>5.0, "food"=>5.0, "meds"=>5.0, "garage"=>5.0, "armory"=>5.0, "walls"=>5.0}}
                20 => tycoon_manager_rebirth_level_config{HudImageTexture := option{"UI/Rebirths/rebirth_20"}, ResourceMultipliers := map{"scrap"=>5.0, "food"=>5.0, "meds"=>5.0, "garage"=>5.0, "armory"=>5.0, "walls"=>5.0}}
                21 => tycoon_manager_rebirth_level_config{HudImageTexture := option{"UI/Rebirths/rebirth_21"}, ResourceMultipliers := map{"scrap"=>5.0, "food"=>5.0, "meds"=>5.0, "garage"=>5.0, "armory"=>5.0, "walls"=>5.0}}
                22 => tycoon_manager_rebirth_level_config{HudImageTexture := option{"UI/Rebirths/rebirth_22"}, ResourceMultipliers := map{"scrap"=>5.0, "food"=>5.0, "meds"=>5.0, "garage"=>5.0, "armory"=>5.0, "walls"=>5.0}}
                23 => tycoon_manager_rebirth_level_config{HudImageTexture := option{"UI/Rebirths/rebirth_23"}, ResourceMultipliers := map{"scrap"=>5.0, "food"=>5.0, "meds"=>5.0, "garage"=>5.0, "armory"=>5.0, "walls"=>5.0}}
                24 => tycoon_manager_rebirth_level_config{HudImageTexture := option{"UI/Rebirths/rebirth_24"}, ResourceMultipliers := map{"scrap"=>5.0, "food"=>5.0, "meds"=>5.0, "garage"=>5.0, "armory"=>5.0, "walls"=>5.0}}
                25 => tycoon_manager_rebirth_level_config{HudImageTexture := option{"UI/Rebirths/rebirth_25"}, ResourceMultipliers := map{"scrap"=>5.0, "food"=>5.0, "meds"=>5.0, "garage"=>5.0, "armory"=>5.0, "walls"=>5.0}}
                26 => tycoon_manager_rebirth_level_config{HudImageTexture := option{"UI/Rebirths/rebirth_26"}, ResourceMultipliers := map{"scrap"=>5.0, "food"=>5.0, "meds"=>5.0, "garage"=>5.0, "armory"=>5.0, "walls"=>5.0}}
                50 => tycoon_manager_rebirth_level_config{HudImageTexture := option{"UI/Rebirths/rebirth_50"}, ResourceMultipliers := map{"scrap"=>5.0, "food"=>5.0, "meds"=>5.0, "garage"=>5.0, "armory"=>5.0, "walls"=>5.0}}
                100 => tycoon_manager_rebirth_level_config{HudImageTexture := option{"UI/Rebirths/rebirth_100"}, ResourceMultipliers := map{"scrap"=>5.0, "food"=>5.0, "meds"=>5.0, "garage"=>5.0, "armory"=>5.0, "walls"=>5.0}}
                200 => tycoon_manager_rebirth_level_config{HudImageTexture := option{"UI/Rebirths/rebirth_200"}, ResourceMultipliers := map{"scrap"=>5.0, "food"=>5.0, "meds"=>5.0, "garage"=>5.0, "armory"=>5.0, "walls"=>5.0}}
                300 => tycoon_manager_rebirth_level_config{HudImageTexture := option{"UI/Rebirths/rebirth_300"}, ResourceMultipliers := map{"scrap"=>5.0, "food"=>5.0, "meds"=>5.0, "garage"=>5.0, "armory"=>5.0, "walls"=>5.0}}
                400 => tycoon_manager_rebirth_level_config{HudImageTexture := option{"UI/Rebirths/rebirth_400"}, ResourceMultipliers := map{"scrap"=>5.0, "food"=>5.0, "meds"=>5.0, "garage"=>5.0, "armory"=>5.0, "walls"=>5.0}}
                500 => tycoon_manager_rebirth_level_config{HudImageTexture := option{"UI/Rebirths/rebirth_500"}, ResourceMultipliers := map{"scrap"=>5.0, "food"=>5.0, "meds"=>5.0, "garage"=>5.0, "armory"=>5.0, "walls"=>5.0}}
              }
            }}
          }

        if (
            RebirthConfig := m_TycoonConfig.Rebirth?
            MaxPersistenceRebirthLevel := FromBits(array{true}.Repeat(p_RebirthLevelReserveBits))
            RebirthConfig.MaxLevel > MaxPersistenceRebirthLevel
        ) {
            Log("Tycoon managed has max rebirth level to {RebirthConfig.MaxLevel} but only {p_RebirthLevelReserveBits} bits (max {MaxPersistenceRebirthLevel}) are allocated for rebirth persistence", ?Level := log_level.Error)
        }

        set m_BuyplatesConfigs = map{
            "bp_claim" => tycoon_manager_buyplates_entry{
              BuyplateId := "bp_claim"
              Cost := false
              Generates := option{tycoon_manager_buyplates_entry_resource{ResourceId := "scrap", Amount := 1, CheckDontTake := false}}
              Unlocks := array{
                tycoon_manager_unlock_entry{TargetRoutingId := "bp_001", Payload := false}
                tycoon_manager_unlock_entry{TargetRoutingId := "bp_100", Payload := false}
                tycoon_manager_unlock_entry{TargetRoutingId := "bp_050", Payload := false}
                tycoon_manager_unlock_entry{TargetRoutingId := "bp_051", Payload := false}
                tycoon_manager_unlock_entry{TargetRoutingId := "bp_500", Payload := false}
                tycoon_manager_unlock_entry{TargetRoutingId := "bp_052", Payload := false}
              }
              m_StorageIndex := 0
            }
            "bp_050" => tycoon_manager_buyplates_entry{
              BuyplateId := "bp_050"
              Cost := false
              Generates := false
              Unlocks := array{
                tycoon_manager_unlock_entry{TargetRoutingId := "bp_300", Payload := false}
                tycoon_manager_unlock_entry{TargetRoutingId := "basecrops", Payload := option{"enable_field:22"}}
              }
              m_StorageIndex := 1
            }
            "bp_051" => tycoon_manager_buyplates_entry{
              BuyplateId := "bp_051"
              Cost := false
              Generates := false
              Unlocks := array{
                tycoon_manager_unlock_entry{TargetRoutingId := "bp_400", Payload := false}
              }
              m_StorageIndex := 2
            }
            "bp_052" => tycoon_manager_buyplates_entry{
              BuyplateId := "bp_052"
              Cost := false
              Generates := false
              Unlocks := array{
                tycoon_manager_unlock_entry{TargetRoutingId := "bp_600", Payload := false}
              }
              m_StorageIndex := 3
            }
            "bp_001" => tycoon_manager_buyplates_entry{
              BuyplateId := "bp_001"
              Cost := option{tycoon_manager_buyplates_entry_resource{ResourceId := "scrap", Amount := 30, CheckDontTake := false}}
              Generates := option{tycoon_manager_buyplates_entry_resource{ResourceId := "scrap", Amount := 5, CheckDontTake := false}}
              Unlocks := array{
                tycoon_manager_unlock_entry{TargetRoutingId := "bp_007", Payload := false}
              }
              m_StorageIndex := 4
            }
            "bp_007" => tycoon_manager_buyplates_entry{
              BuyplateId := "bp_007"
              Cost := option{tycoon_manager_buyplates_entry_resource{ResourceId := "scrap", Amount := 360, CheckDontTake := false}}
              Generates := option{tycoon_manager_buyplates_entry_resource{ResourceId := "scrap", Amount := 9, CheckDontTake := false}}
              Unlocks := array{
                tycoon_manager_unlock_entry{TargetRoutingId := "bp_002", Payload := false}
                tycoon_manager_unlock_entry{TargetRoutingId := "basecrops", Payload := option{"enable_field:01"}}
                tycoon_manager_unlock_entry{TargetRoutingId := "basecrops", Payload := option{"enable_field:02"}}
                tycoon_manager_unlock_entry{TargetRoutingId := "basecrops", Payload := option{"enable_field:03"}}
                tycoon_manager_unlock_entry{TargetRoutingId := "basecrops", Payload := option{"enable_field:04"}}
                tycoon_manager_unlock_entry{TargetRoutingId := "basecrops", Payload := option{"enable_field:05"}}
                tycoon_manager_unlock_entry{TargetRoutingId := "basecrops", Payload := option{"enable_field:06"}}
                tycoon_manager_unlock_entry{TargetRoutingId := "basecrops", Payload := option{"enable_field:07"}}
                tycoon_manager_unlock_entry{TargetRoutingId := "basecrops", Payload := option{"enable_field:08"}}
                tycoon_manager_unlock_entry{TargetRoutingId := "basecrops", Payload := option{"enable_field:09"}}
              }
              m_StorageIndex := 5
            }
            "bp_002" => tycoon_manager_buyplates_entry{
              BuyplateId := "bp_002"
              Cost := option{tycoon_manager_buyplates_entry_resource{ResourceId := "scrap", Amount := 1350, CheckDontTake := false}}
              Generates := option{tycoon_manager_buyplates_entry_resource{ResourceId := "scrap", Amount := 15, CheckDontTake := false}}
              Unlocks := array{
                tycoon_manager_unlock_entry{TargetRoutingId := "bp_003", Payload := false}
              }
              m_StorageIndex := 6
            }
            "bp_003" => tycoon_manager_buyplates_entry{
              BuyplateId := "bp_003"
              Cost := option{tycoon_manager_buyplates_entry_resource{ResourceId := "scrap", Amount := 2700, CheckDontTake := false}}
              Generates := option{tycoon_manager_buyplates_entry_resource{ResourceId := "scrap", Amount := 21, CheckDontTake := false}}
              Unlocks := array{
                tycoon_manager_unlock_entry{TargetRoutingId := "bp_004", Payload := false}
              }
              m_StorageIndex := 7
            }
            "bp_004" => tycoon_manager_buyplates_entry{
              BuyplateId := "bp_004"
              Cost := option{tycoon_manager_buyplates_entry_resource{ResourceId := "scrap", Amount := 4590, CheckDontTake := false}}
              Generates := option{tycoon_manager_buyplates_entry_resource{ResourceId := "scrap", Amount := 26, CheckDontTake := false}}
              Unlocks := array{
                tycoon_manager_unlock_entry{TargetRoutingId := "bp_005", Payload := false}
              }
              m_StorageIndex := 8
            }
            "bp_005" => tycoon_manager_buyplates_entry{
              BuyplateId := "bp_005"
              Cost := option{tycoon_manager_buyplates_entry_resource{ResourceId := "scrap", Amount := 10400, CheckDontTake := false}}
              Generates := option{tycoon_manager_buyplates_entry_resource{ResourceId := "scrap", Amount := 33, CheckDontTake := false}}
              Unlocks := array{
                tycoon_manager_unlock_entry{TargetRoutingId := "bp_006", Payload := false}
              }
              m_StorageIndex := 9
            }
            "bp_006" => tycoon_manager_buyplates_entry{
              BuyplateId := "bp_006"
              Cost := option{tycoon_manager_buyplates_entry_resource{ResourceId := "scrap", Amount := 14900, CheckDontTake := false}}
              Generates := option{tycoon_manager_buyplates_entry_resource{ResourceId := "scrap", Amount := 37, CheckDontTake := false}}
              Unlocks := array{
                tycoon_manager_unlock_entry{TargetRoutingId := "bp_008", Payload := false}
              }
              m_StorageIndex := 10
            }
            "bp_008" => tycoon_manager_buyplates_entry{
              BuyplateId := "bp_008"
              Cost := option{tycoon_manager_buyplates_entry_resource{ResourceId := "scrap", Amount := 19800, CheckDontTake := false}}
              Generates := option{tycoon_manager_buyplates_entry_resource{ResourceId := "scrap", Amount := 41, CheckDontTake := false}}
              Unlocks := array{
                tycoon_manager_unlock_entry{TargetRoutingId := "bp_009", Payload := false}
              }
              m_StorageIndex := 11
            }
            "bp_009" => tycoon_manager_buyplates_entry{
              BuyplateId := "bp_009"
              Cost := option{tycoon_manager_buyplates_entry_resource{ResourceId := "scrap", Amount := 25400, CheckDontTake := false}}
              Generates := option{tycoon_manager_buyplates_entry_resource{ResourceId := "scrap", Amount := 45, CheckDontTake := false}}
              Unlocks := array{
                tycoon_manager_unlock_entry{TargetRoutingId := "bp_010", Payload := false}
              }
              m_StorageIndex := 12
            }
            "bp_010" => tycoon_manager_buyplates_entry{
              BuyplateId := "bp_010"
              Cost := option{tycoon_manager_buyplates_entry_resource{ResourceId := "scrap", Amount := 31500, CheckDontTake := false}}
              Generates := option{tycoon_manager_buyplates_entry_resource{ResourceId := "scrap", Amount := 50, CheckDontTake := false}}
              Unlocks := array{
                tycoon_manager_unlock_entry{TargetRoutingId := "bp_011", Payload := false}
              }
              m_StorageIndex := 13
            }
            "bp_011" => tycoon_manager_buyplates_entry{
              BuyplateId := "bp_011"
              Cost := option{tycoon_manager_buyplates_entry_resource{ResourceId := "scrap", Amount := 38200, CheckDontTake := false}}
              Generates := option{tycoon_manager_buyplates_entry_resource{ResourceId := "scrap", Amount := 55, CheckDontTake := false}}
              Unlocks := array{
                tycoon_manager_unlock_entry{TargetRoutingId := "bp_012", Payload := false}
              }
              m_StorageIndex := 14
            }
            "bp_012" => tycoon_manager_buyplates_entry{
              BuyplateId := "bp_012"
              Cost := option{tycoon_manager_buyplates_entry_resource{ResourceId := "scrap", Amount := 45600, CheckDontTake := false}}
              Generates := option{tycoon_manager_buyplates_entry_resource{ResourceId := "scrap", Amount := 61, CheckDontTake := false}}
              Unlocks := array{
                tycoon_manager_unlock_entry{TargetRoutingId := "bp_013", Payload := false}
              }
              m_StorageIndex := 15
            }
            "bp_013" => tycoon_manager_buyplates_entry{
              BuyplateId := "bp_013"
              Cost := option{tycoon_manager_buyplates_entry_resource{ResourceId := "scrap", Amount := 53900, CheckDontTake := false}}
              Generates := option{tycoon_manager_buyplates_entry_resource{ResourceId := "scrap", Amount := 67, CheckDontTake := false}}
              Unlocks := array{
                tycoon_manager_unlock_entry{TargetRoutingId := "bp_014", Payload := false}
              }
              m_StorageIndex := 16
            }
            "bp_014" => tycoon_manager_buyplates_entry{
              BuyplateId := "bp_014"
              Cost := option{tycoon_manager_buyplates_entry_resource{ResourceId := "scrap", Amount := 62900, CheckDontTake := false}}
              Generates := option{tycoon_manager_buyplates_entry_resource{ResourceId := "scrap", Amount := 74, CheckDontTake := false}}
              Unlocks := array{
                tycoon_manager_unlock_entry{TargetRoutingId := "bp_015", Payload := false}
              }
              m_StorageIndex := 17
            }
            "bp_015" => tycoon_manager_buyplates_entry{
              BuyplateId := "bp_015"
              Cost := option{tycoon_manager_buyplates_entry_resource{ResourceId := "scrap", Amount := 81000, CheckDontTake := false}}
              Generates := option{tycoon_manager_buyplates_entry_resource{ResourceId := "scrap", Amount := 81, CheckDontTake := false}}
              Unlocks := array{
                tycoon_manager_unlock_entry{TargetRoutingId := "bp_016", Payload := false}
              }
              m_StorageIndex := 18
            }
            "bp_016" => tycoon_manager_buyplates_entry{
              BuyplateId := "bp_016"
              Cost := option{tycoon_manager_buyplates_entry_resource{ResourceId := "scrap", Amount := 93200, CheckDontTake := false}}
              Generates := option{tycoon_manager_buyplates_entry_resource{ResourceId := "scrap", Amount := 89, CheckDontTake := false}}
              Unlocks := array{
                tycoon_manager_unlock_entry{TargetRoutingId := "bp_017", Payload := false}
              }
              m_StorageIndex := 19
            }
            "bp_017" => tycoon_manager_buyplates_entry{
              BuyplateId := "bp_017"
              Cost := option{tycoon_manager_buyplates_entry_resource{ResourceId := "scrap", Amount := 107000, CheckDontTake := false}}
              Generates := option{tycoon_manager_buyplates_entry_resource{ResourceId := "scrap", Amount := 94, CheckDontTake := false}}
              Unlocks := array{
                tycoon_manager_unlock_entry{TargetRoutingId := "bp_018", Payload := false}
              }
              m_StorageIndex := 20
            }
            "bp_018" => tycoon_manager_buyplates_entry{
              BuyplateId := "bp_018"
              Cost := option{tycoon_manager_buyplates_entry_resource{ResourceId := "scrap", Amount := 121000, CheckDontTake := false}}
              Generates := option{tycoon_manager_buyplates_entry_resource{ResourceId := "scrap", Amount := 99, CheckDontTake := false}}
              Unlocks := array{
                tycoon_manager_unlock_entry{TargetRoutingId := "bp_019", Payload := false}
              }
              m_StorageIndex := 21
            }
            "bp_019" => tycoon_manager_buyplates_entry{
              BuyplateId := "bp_019"
              Cost := option{tycoon_manager_buyplates_entry_resource{ResourceId := "scrap", Amount := 135000, CheckDontTake := false}}
              Generates := option{tycoon_manager_buyplates_entry_resource{ResourceId := "scrap", Amount := 105, CheckDontTake := false}}
              Unlocks := array{
                tycoon_manager_unlock_entry{TargetRoutingId := "bp_020", Payload := false}
              }
              m_StorageIndex := 22
            }
            "bp_020" => tycoon_manager_buyplates_entry{
              BuyplateId := "bp_020"
              Cost := option{tycoon_manager_buyplates_entry_resource{ResourceId := "scrap", Amount := 151000, CheckDontTake := false}}
              Generates := option{tycoon_manager_buyplates_entry_resource{ResourceId := "scrap", Amount := 112, CheckDontTake := false}}
              Unlocks := array{
                tycoon_manager_unlock_entry{TargetRoutingId := "bp_021", Payload := false}
              }
              m_StorageIndex := 23
            }
            "bp_021" => tycoon_manager_buyplates_entry{
              BuyplateId := "bp_021"
              Cost := option{tycoon_manager_buyplates_entry_resource{ResourceId := "scrap", Amount := 168000, CheckDontTake := false}}
              Generates := option{tycoon_manager_buyplates_entry_resource{ResourceId := "scrap", Amount := 119, CheckDontTake := false}}
              Unlocks := array{
                tycoon_manager_unlock_entry{TargetRoutingId := "bp_022", Payload := false}
              }
              m_StorageIndex := 24
            }
            "bp_022" => tycoon_manager_buyplates_entry{
              BuyplateId := "bp_022"
              Cost := option{tycoon_manager_buyplates_entry_resource{ResourceId := "scrap", Amount := 204000, CheckDontTake := false}}
              Generates := option{tycoon_manager_buyplates_entry_resource{ResourceId := "scrap", Amount := 127, CheckDontTake := false}}
              Unlocks := array{
                tycoon_manager_unlock_entry{TargetRoutingId := "bp_023", Payload := false}
              }
              m_StorageIndex := 25
            }
            "bp_023" => tycoon_manager_buyplates_entry{
              BuyplateId := "bp_023"
              Cost := option{tycoon_manager_buyplates_entry_resource{ResourceId := "scrap", Amount := 225000, CheckDontTake := false}}
              Generates := option{tycoon_manager_buyplates_entry_resource{ResourceId := "scrap", Amount := 136, CheckDontTake := false}}
              Unlocks := array{
                tycoon_manager_unlock_entry{TargetRoutingId := "bp_024", Payload := false}
              }
              m_StorageIndex := 26
            }
            "bp_024" => tycoon_manager_buyplates_entry{
              BuyplateId := "bp_024"
              Cost := option{tycoon_manager_buyplates_entry_resource{ResourceId := "scrap", Amount := 248000, CheckDontTake := false}}
              Generates := option{tycoon_manager_buyplates_entry_resource{ResourceId := "scrap", Amount := 145, CheckDontTake := false}}
              Unlocks := array{
                tycoon_manager_unlock_entry{TargetRoutingId := "bp_025", Payload := false}
              }
              m_StorageIndex := 27
            }
            "bp_025" => tycoon_manager_buyplates_entry{
              BuyplateId := "bp_025"
              Cost := option{tycoon_manager_buyplates_entry_resource{ResourceId := "scrap", Amount := 272000, CheckDontTake := false}}
              Generates := option{tycoon_manager_buyplates_entry_resource{ResourceId := "scrap", Amount := 155, CheckDontTake := false}}
              Unlocks := array{
                tycoon_manager_unlock_entry{TargetRoutingId := "bp_026", Payload := false}
                tycoon_manager_unlock_entry{TargetRoutingId := "basecrops", Payload := option{"enable_field:17"}}
                tycoon_manager_unlock_entry{TargetRoutingId := "basecrops", Payload := option{"enable_field:18"}}
                tycoon_manager_unlock_entry{TargetRoutingId := "basecrops", Payload := option{"enable_field:19"}}
                tycoon_manager_unlock_entry{TargetRoutingId := "basecrops", Payload := option{"enable_field:20"}}
                tycoon_manager_unlock_entry{TargetRoutingId := "basecrops", Payload := option{"enable_field:21"}}
              }
              m_StorageIndex := 28
            }
            "bp_026" => tycoon_manager_buyplates_entry{
              BuyplateId := "bp_026"
              Cost := option{tycoon_manager_buyplates_entry_resource{ResourceId := "scrap", Amount := 297000, CheckDontTake := false}}
              Generates := option{tycoon_manager_buyplates_entry_resource{ResourceId := "scrap", Amount := 165, CheckDontTake := false}}
              Unlocks := array{
                tycoon_manager_unlock_entry{TargetRoutingId := "bp_027", Payload := false}
              }
              m_StorageIndex := 29
            }
            "bp_027" => tycoon_manager_buyplates_entry{
              BuyplateId := "bp_027"
              Cost := option{tycoon_manager_buyplates_entry_resource{ResourceId := "scrap", Amount := 325000, CheckDontTake := false}}
              Generates := option{tycoon_manager_buyplates_entry_resource{ResourceId := "scrap", Amount := 176, CheckDontTake := false}}
              Unlocks := array{
                tycoon_manager_unlock_entry{TargetRoutingId := "bp_028", Payload := false}
              }
              m_StorageIndex := 30
            }
            "bp_028" => tycoon_manager_buyplates_entry{
              BuyplateId := "bp_028"
              Cost := option{tycoon_manager_buyplates_entry_resource{ResourceId := "scrap", Amount := 354000, CheckDontTake := false}}
              Generates := option{tycoon_manager_buyplates_entry_resource{ResourceId := "scrap", Amount := 188, CheckDontTake := false}}
              Unlocks := array{
                tycoon_manager_unlock_entry{TargetRoutingId := "bp_029", Payload := false}
              }
              m_StorageIndex := 31
            }
            "bp_029" => tycoon_manager_buyplates_entry{
              BuyplateId := "bp_029"
              Cost := option{tycoon_manager_buyplates_entry_resource{ResourceId := "scrap", Amount := 385000, CheckDontTake := false}}
              Generates := option{tycoon_manager_buyplates_entry_resource{ResourceId := "scrap", Amount := 201, CheckDontTake := false}}
              Unlocks := array{
                tycoon_manager_unlock_entry{TargetRoutingId := "bp_030", Payload := false}
              }
              m_StorageIndex := 32
            }
            "bp_030" => tycoon_manager_buyplates_entry{
              BuyplateId := "bp_030"
              Cost := option{tycoon_manager_buyplates_entry_resource{ResourceId := "scrap", Amount := 418000, CheckDontTake := false}}
              Generates := option{tycoon_manager_buyplates_entry_resource{ResourceId := "scrap", Amount := 214, CheckDontTake := false}}
              Unlocks := array{
                tycoon_manager_unlock_entry{TargetRoutingId := "bp_031", Payload := false}
              }
              m_StorageIndex := 33
            }
            "bp_031" => tycoon_manager_buyplates_entry{
              BuyplateId := "bp_031"
              Cost := option{tycoon_manager_buyplates_entry_resource{ResourceId := "scrap", Amount := 453000, CheckDontTake := false}}
              Generates := option{tycoon_manager_buyplates_entry_resource{ResourceId := "scrap", Amount := 224, CheckDontTake := false}}
              Unlocks := array{
                tycoon_manager_unlock_entry{TargetRoutingId := "bp_032", Payload := false}
              }
              m_StorageIndex := 34
            }
            "bp_032" => tycoon_manager_buyplates_entry{
              BuyplateId := "bp_032"
              Cost := option{tycoon_manager_buyplates_entry_resource{ResourceId := "scrap", Amount := 490000, CheckDontTake := false}}
              Generates := option{tycoon_manager_buyplates_entry_resource{ResourceId := "scrap", Amount := 236, CheckDontTake := false}}
              Unlocks := array{
                tycoon_manager_unlock_entry{TargetRoutingId := "bp_033", Payload := false}
              }
              m_StorageIndex := 35
            }
            "bp_033" => tycoon_manager_buyplates_entry{
              BuyplateId := "bp_033"
              Cost := option{tycoon_manager_buyplates_entry_resource{ResourceId := "scrap", Amount := 529000, CheckDontTake := false}}
              Generates := option{tycoon_manager_buyplates_entry_resource{ResourceId := "scrap", Amount := 250, CheckDontTake := false}}
              Unlocks := array{
                tycoon_manager_unlock_entry{TargetRoutingId := "bp_034", Payload := false}
              }
              m_StorageIndex := 36
            }
            "bp_034" => tycoon_manager_buyplates_entry{
              BuyplateId := "bp_034"
              Cost := option{tycoon_manager_buyplates_entry_resource{ResourceId := "scrap", Amount := 570000, CheckDontTake := false}}
              Generates := option{tycoon_manager_buyplates_entry_resource{ResourceId := "scrap", Amount := 261, CheckDontTake := false}}
              Unlocks := array{
                tycoon_manager_unlock_entry{TargetRoutingId := "bp_035", Payload := false}
              }
              m_StorageIndex := 37
            }
            "bp_035" => tycoon_manager_buyplates_entry{
              BuyplateId := "bp_035"
              Cost := option{tycoon_manager_buyplates_entry_resource{ResourceId := "scrap", Amount := 669000, CheckDontTake := false}}
              Generates := option{tycoon_manager_buyplates_entry_resource{ResourceId := "scrap", Amount := 283, CheckDontTake := false}}
              Unlocks := array{
                tycoon_manager_unlock_entry{TargetRoutingId := "bp_036", Payload := false}
              }
              m_StorageIndex := 38
            }
            "bp_036" => tycoon_manager_buyplates_entry{
              BuyplateId := "bp_036"
              Cost := option{tycoon_manager_buyplates_entry_resource{ResourceId := "scrap", Amount := 720000, CheckDontTake := false}}
              Generates := option{tycoon_manager_buyplates_entry_resource{ResourceId := "scrap", Amount := 304, CheckDontTake := false}}
              Unlocks := array{
                tycoon_manager_unlock_entry{TargetRoutingId := "bp_037", Payload := false}
              }
              m_StorageIndex := 39
            }
            "bp_037" => tycoon_manager_buyplates_entry{
              BuyplateId := "bp_037"
              Cost := option{tycoon_manager_buyplates_entry_resource{ResourceId := "scrap", Amount := 775000, CheckDontTake := false}}
              Generates := option{tycoon_manager_buyplates_entry_resource{ResourceId := "scrap", Amount := 332, CheckDontTake := false}}
              Unlocks := array{
                tycoon_manager_unlock_entry{TargetRoutingId := "bp_038", Payload := false}
              }
              m_StorageIndex := 40
            }
            "bp_038" => tycoon_manager_buyplates_entry{
              BuyplateId := "bp_038"
              Cost := option{tycoon_manager_buyplates_entry_resource{ResourceId := "scrap", Amount := 834000, CheckDontTake := false}}
              Generates := option{tycoon_manager_buyplates_entry_resource{ResourceId := "scrap", Amount := 364, CheckDontTake := false}}
              Unlocks := array{
                tycoon_manager_unlock_entry{TargetRoutingId := "bp_039", Payload := false}
              }
              m_StorageIndex := 41
            }
            "bp_039" => tycoon_manager_buyplates_entry{
              BuyplateId := "bp_039"
              Cost := option{tycoon_manager_buyplates_entry_resource{ResourceId := "scrap", Amount := 900000, CheckDontTake := false}}
              Generates := false
              Unlocks := array{
              }
              m_StorageIndex := 42
            }
            "bp_100" => tycoon_manager_buyplates_entry{
              BuyplateId := "bp_100"
              Cost := option{tycoon_manager_buyplates_entry_resource{ResourceId := "scrap", Amount := 15, CheckDontTake := false}}
              Generates := option{tycoon_manager_buyplates_entry_resource{ResourceId := "walls", Amount := 1, CheckDontTake := false}}
              Unlocks := array{
                tycoon_manager_unlock_entry{TargetRoutingId := "bp_200", Payload := false}
              }
              m_StorageIndex := 43
            }
            "bp_200" => tycoon_manager_buyplates_entry{
              BuyplateId := "bp_200"
              Cost := option{tycoon_manager_buyplates_entry_resource{ResourceId := "walls", Amount := 20, CheckDontTake := false}}
              Generates := option{tycoon_manager_buyplates_entry_resource{ResourceId := "walls", Amount := 5, CheckDontTake := false}}
              Unlocks := array{
                tycoon_manager_unlock_entry{TargetRoutingId := "bp_101", Payload := false}
              }
              m_StorageIndex := 44
            }
            "bp_101" => tycoon_manager_buyplates_entry{
              BuyplateId := "bp_101"
              Cost := option{tycoon_manager_buyplates_entry_resource{ResourceId := "walls", Amount := 360, CheckDontTake := false}}
              Generates := option{tycoon_manager_buyplates_entry_resource{ResourceId := "walls", Amount := 9, CheckDontTake := false}}
              Unlocks := array{
                tycoon_manager_unlock_entry{TargetRoutingId := "bp_201", Payload := false}
              }
              m_StorageIndex := 45
            }
            "bp_201" => tycoon_manager_buyplates_entry{
              BuyplateId := "bp_201"
              Cost := option{tycoon_manager_buyplates_entry_resource{ResourceId := "walls", Amount := 1350, CheckDontTake := false}}
              Generates := option{tycoon_manager_buyplates_entry_resource{ResourceId := "walls", Amount := 15, CheckDontTake := false}}
              Unlocks := array{
                tycoon_manager_unlock_entry{TargetRoutingId := "bp_202", Payload := false}
              }
              m_StorageIndex := 46
            }
            "bp_202" => tycoon_manager_buyplates_entry{
              BuyplateId := "bp_202"
              Cost := option{tycoon_manager_buyplates_entry_resource{ResourceId := "walls", Amount := 2700, CheckDontTake := false}}
              Generates := option{tycoon_manager_buyplates_entry_resource{ResourceId := "walls", Amount := 20, CheckDontTake := false}}
              Unlocks := array{
                tycoon_manager_unlock_entry{TargetRoutingId := "bp_203", Payload := false}
              }
              m_StorageIndex := 47
            }
            "bp_203" => tycoon_manager_buyplates_entry{
              BuyplateId := "bp_203"
              Cost := option{tycoon_manager_buyplates_entry_resource{ResourceId := "walls", Amount := 4500, CheckDontTake := false}}
              Generates := option{tycoon_manager_buyplates_entry_resource{ResourceId := "walls", Amount := 23, CheckDontTake := false}}
              Unlocks := array{
                tycoon_manager_unlock_entry{TargetRoutingId := "bp_204", Payload := false}
              }
              m_StorageIndex := 48
            }
            "bp_204" => tycoon_manager_buyplates_entry{
              BuyplateId := "bp_204"
              Cost := option{tycoon_manager_buyplates_entry_resource{ResourceId := "walls", Amount := 8800, CheckDontTake := false}}
              Generates := option{tycoon_manager_buyplates_entry_resource{ResourceId := "walls", Amount := 25, CheckDontTake := false}}
              Unlocks := array{
                tycoon_manager_unlock_entry{TargetRoutingId := "bp_205", Payload := false}
              }
              m_StorageIndex := 49
            }
            "bp_205" => tycoon_manager_buyplates_entry{
              BuyplateId := "bp_205"
              Cost := option{tycoon_manager_buyplates_entry_resource{ResourceId := "walls", Amount := 11800, CheckDontTake := false}}
              Generates := option{tycoon_manager_buyplates_entry_resource{ResourceId := "walls", Amount := 35, CheckDontTake := false}}
              Unlocks := array{
                tycoon_manager_unlock_entry{TargetRoutingId := "bp_102", Payload := false}
              }
              m_StorageIndex := 50
            }
            "bp_102" => tycoon_manager_buyplates_entry{
              BuyplateId := "bp_102"
              Cost := option{tycoon_manager_buyplates_entry_resource{ResourceId := "walls", Amount := 16000, CheckDontTake := false}}
              Generates := option{tycoon_manager_buyplates_entry_resource{ResourceId := "walls", Amount := 40, CheckDontTake := false}}
              Unlocks := array{
                tycoon_manager_unlock_entry{TargetRoutingId := "bp_206", Payload := false}
              }
              m_StorageIndex := 51
            }
            "bp_206" => tycoon_manager_buyplates_entry{
              BuyplateId := "bp_206"
              Cost := option{tycoon_manager_buyplates_entry_resource{ResourceId := "walls", Amount := 20800, CheckDontTake := false}}
              Generates := option{tycoon_manager_buyplates_entry_resource{ResourceId := "walls", Amount := 49, CheckDontTake := false}}
              Unlocks := array{
                tycoon_manager_unlock_entry{TargetRoutingId := "bp_207", Payload := false}
              }
              m_StorageIndex := 52
            }
            "bp_207" => tycoon_manager_buyplates_entry{
              BuyplateId := "bp_207"
              Cost := option{tycoon_manager_buyplates_entry_resource{ResourceId := "walls", Amount := 26600, CheckDontTake := false}}
              Generates := option{tycoon_manager_buyplates_entry_resource{ResourceId := "walls", Amount := 58, CheckDontTake := false}}
              Unlocks := array{
                tycoon_manager_unlock_entry{TargetRoutingId := "bp_208", Payload := false}
              }
              m_StorageIndex := 53
            }
            "bp_208" => tycoon_manager_buyplates_entry{
              BuyplateId := "bp_208"
              Cost := option{tycoon_manager_buyplates_entry_resource{ResourceId := "walls", Amount := 50400, CheckDontTake := false}}
              Generates := option{tycoon_manager_buyplates_entry_resource{ResourceId := "walls", Amount := 70, CheckDontTake := false}}
              Unlocks := array{
                tycoon_manager_unlock_entry{TargetRoutingId := "bp_209", Payload := false}
              }
              m_StorageIndex := 54
            }
            "bp_209" => tycoon_manager_buyplates_entry{
              BuyplateId := "bp_209"
              Cost := option{tycoon_manager_buyplates_entry_resource{ResourceId := "walls", Amount := 63000, CheckDontTake := false}}
              Generates := option{tycoon_manager_buyplates_entry_resource{ResourceId := "walls", Amount := 84, CheckDontTake := false}}
              Unlocks := array{
                tycoon_manager_unlock_entry{TargetRoutingId := "bp_210", Payload := false}
              }
              m_StorageIndex := 55
            }
            "bp_210" => tycoon_manager_buyplates_entry{
              BuyplateId := "bp_210"
              Cost := option{tycoon_manager_buyplates_entry_resource{ResourceId := "walls", Amount := 78100, CheckDontTake := false}}
              Generates := option{tycoon_manager_buyplates_entry_resource{ResourceId := "walls", Amount := 98, CheckDontTake := false}}
              Unlocks := array{
                tycoon_manager_unlock_entry{TargetRoutingId := "bp_103", Payload := false}
              }
              m_StorageIndex := 56
            }
            "bp_103" => tycoon_manager_buyplates_entry{
              BuyplateId := "bp_103"
              Cost := option{tycoon_manager_buyplates_entry_resource{ResourceId := "walls", Amount := 95800, CheckDontTake := false}}
              Generates := option{tycoon_manager_buyplates_entry_resource{ResourceId := "walls", Amount := 115, CheckDontTake := false}}
              Unlocks := array{
                tycoon_manager_unlock_entry{TargetRoutingId := "bp_211", Payload := false}
              }
              m_StorageIndex := 57
            }
            "bp_211" => tycoon_manager_buyplates_entry{
              BuyplateId := "bp_211"
              Cost := option{tycoon_manager_buyplates_entry_resource{ResourceId := "walls", Amount := 116000, CheckDontTake := false}}
              Generates := option{tycoon_manager_buyplates_entry_resource{ResourceId := "walls", Amount := 128, CheckDontTake := false}}
              Unlocks := array{
                tycoon_manager_unlock_entry{TargetRoutingId := "bp_212", Payload := false}
              }
              m_StorageIndex := 58
            }
            "bp_212" => tycoon_manager_buyplates_entry{
              BuyplateId := "bp_212"
              Cost := option{tycoon_manager_buyplates_entry_resource{ResourceId := "walls", Amount := 163000, CheckDontTake := false}}
              Generates := option{tycoon_manager_buyplates_entry_resource{ResourceId := "walls", Amount := 143, CheckDontTake := false}}
              Unlocks := array{
                tycoon_manager_unlock_entry{TargetRoutingId := "bp_213", Payload := false}
              }
              m_StorageIndex := 59
            }
            "bp_213" => tycoon_manager_buyplates_entry{
              BuyplateId := "bp_213"
              Cost := option{tycoon_manager_buyplates_entry_resource{ResourceId := "walls", Amount := 193000, CheckDontTake := false}}
              Generates := option{tycoon_manager_buyplates_entry_resource{ResourceId := "walls", Amount := 160, CheckDontTake := false}}
              Unlocks := array{
                tycoon_manager_unlock_entry{TargetRoutingId := "bp_214", Payload := false}
              }
              m_StorageIndex := 60
            }
            "bp_214" => tycoon_manager_buyplates_entry{
              BuyplateId := "bp_214"
              Cost := option{tycoon_manager_buyplates_entry_resource{ResourceId := "walls", Amount := 226000, CheckDontTake := false}}
              Generates := option{tycoon_manager_buyplates_entry_resource{ResourceId := "walls", Amount := 179, CheckDontTake := false}}
              Unlocks := array{
                tycoon_manager_unlock_entry{TargetRoutingId := "bp_215", Payload := false}
              }
              m_StorageIndex := 61
            }
            "bp_215" => tycoon_manager_buyplates_entry{
              BuyplateId := "bp_215"
              Cost := option{tycoon_manager_buyplates_entry_resource{ResourceId := "walls", Amount := 264000, CheckDontTake := false}}
              Generates := option{tycoon_manager_buyplates_entry_resource{ResourceId := "walls", Amount := 200, CheckDontTake := false}}
              Unlocks := array{
                tycoon_manager_unlock_entry{TargetRoutingId := "bp_104", Payload := false}
              }
              m_StorageIndex := 62
            }
            "bp_104" => tycoon_manager_buyplates_entry{
              BuyplateId := "bp_104"
              Cost := option{tycoon_manager_buyplates_entry_resource{ResourceId := "walls", Amount := 306000, CheckDontTake := false}}
              Generates := option{tycoon_manager_buyplates_entry_resource{ResourceId := "walls", Amount := 224, CheckDontTake := false}}
              Unlocks := array{
                tycoon_manager_unlock_entry{TargetRoutingId := "bp_216", Payload := false}
              }
              m_StorageIndex := 63
            }
            "bp_216" => tycoon_manager_buyplates_entry{
              BuyplateId := "bp_216"
              Cost := option{tycoon_manager_buyplates_entry_resource{ResourceId := "walls", Amount := 353000, CheckDontTake := false}}
              Generates := option{tycoon_manager_buyplates_entry_resource{ResourceId := "walls", Amount := 264, CheckDontTake := false}}
              Unlocks := array{
                tycoon_manager_unlock_entry{TargetRoutingId := "bp_217", Payload := false}
              }
              m_StorageIndex := 64
            }
            "bp_217" => tycoon_manager_buyplates_entry{
              BuyplateId := "bp_217"
              Cost := option{tycoon_manager_buyplates_entry_resource{ResourceId := "walls", Amount := 408000, CheckDontTake := false}}
              Generates := option{tycoon_manager_buyplates_entry_resource{ResourceId := "walls", Amount := 314, CheckDontTake := false}}
              Unlocks := array{
                tycoon_manager_unlock_entry{TargetRoutingId := "bp_218", Payload := false}
              }
              m_StorageIndex := 65
            }
            "bp_218" => tycoon_manager_buyplates_entry{
              BuyplateId := "bp_218"
              Cost := option{tycoon_manager_buyplates_entry_resource{ResourceId := "walls", Amount := 474000, CheckDontTake := false}}
              Generates := option{tycoon_manager_buyplates_entry_resource{ResourceId := "walls", Amount := 346, CheckDontTake := false}}
              Unlocks := array{
                tycoon_manager_unlock_entry{TargetRoutingId := "bp_219", Payload := false}
              }
              m_StorageIndex := 66
            }
            "bp_219" => tycoon_manager_buyplates_entry{
              BuyplateId := "bp_219"
              Cost := option{tycoon_manager_buyplates_entry_resource{ResourceId := "walls", Amount := 547000, CheckDontTake := false}}
              Generates := option{tycoon_manager_buyplates_entry_resource{ResourceId := "walls", Amount := 381, CheckDontTake := false}}
              Unlocks := array{
                tycoon_manager_unlock_entry{TargetRoutingId := "bp_220", Payload := false}
              }
              m_StorageIndex := 67
            }
            "bp_220" => tycoon_manager_buyplates_entry{
              BuyplateId := "bp_220"
              Cost := option{tycoon_manager_buyplates_entry_resource{ResourceId := "walls", Amount := 627000, CheckDontTake := false}}
              Generates := option{tycoon_manager_buyplates_entry_resource{ResourceId := "walls", Amount := 428, CheckDontTake := false}}
              Unlocks := array{
                tycoon_manager_unlock_entry{TargetRoutingId := "bp_250", Payload := false}
              }
              m_StorageIndex := 68
            }
            "bp_250" => tycoon_manager_buyplates_entry{
              BuyplateId := "bp_250"
              Cost := option{tycoon_manager_buyplates_entry_resource{ResourceId := "walls", Amount := 717000, CheckDontTake := false}}
              Generates := option{tycoon_manager_buyplates_entry_resource{ResourceId := "walls", Amount := 476, CheckDontTake := false}}
              Unlocks := array{
                tycoon_manager_unlock_entry{TargetRoutingId := "bp_251", Payload := false}
              }
              m_StorageIndex := 69
            }
            "bp_251" => tycoon_manager_buyplates_entry{
              BuyplateId := "bp_251"
              Cost := option{tycoon_manager_buyplates_entry_resource{ResourceId := "walls", Amount := 817000, CheckDontTake := false}}
              Generates := option{tycoon_manager_buyplates_entry_resource{ResourceId := "walls", Amount := 530, CheckDontTake := false}}
              Unlocks := array{
                tycoon_manager_unlock_entry{TargetRoutingId := "bp_252", Payload := false}
              }
              m_StorageIndex := 70
            }
            "bp_252" => tycoon_manager_buyplates_entry{
              BuyplateId := "bp_252"
              Cost := option{tycoon_manager_buyplates_entry_resource{ResourceId := "walls", Amount := 928000, CheckDontTake := false}}
              Generates := option{tycoon_manager_buyplates_entry_resource{ResourceId := "walls", Amount := 580, CheckDontTake := false}}
              Unlocks := array{
                tycoon_manager_unlock_entry{TargetRoutingId := "bp_253", Payload := false}
              }
              m_StorageIndex := 71
            }
            "bp_253" => tycoon_manager_buyplates_entry{
              BuyplateId := "bp_253"
              Cost := option{tycoon_manager_buyplates_entry_resource{ResourceId := "walls", Amount := 1000000, CheckDontTake := false}}
              Generates := false
              Unlocks := array{
              }
              m_StorageIndex := 72
            }
            "bp_300" => tycoon_manager_buyplates_entry{
              BuyplateId := "bp_300"
              Cost := option{tycoon_manager_buyplates_entry_resource{ResourceId := "food", Amount := 15, CheckDontTake := false}}
              Generates := option{tycoon_manager_buyplates_entry_resource{ResourceId := "food", Amount := 5, CheckDontTake := false}}
              Unlocks := array{
                tycoon_manager_unlock_entry{TargetRoutingId := "bp_301", Payload := false}
              }
              m_StorageIndex := 73
            }
            "bp_301" => tycoon_manager_buyplates_entry{
              BuyplateId := "bp_301"
              Cost := option{tycoon_manager_buyplates_entry_resource{ResourceId := "food", Amount := 150, CheckDontTake := false}}
              Generates := option{tycoon_manager_buyplates_entry_resource{ResourceId := "food", Amount := 10, CheckDontTake := false}}
              Unlocks := array{
                tycoon_manager_unlock_entry{TargetRoutingId := "bp_302", Payload := false}
              }
              m_StorageIndex := 74
            }
            "bp_302" => tycoon_manager_buyplates_entry{
              BuyplateId := "bp_302"
              Cost := option{tycoon_manager_buyplates_entry_resource{ResourceId := "food", Amount := 1350, CheckDontTake := false}}
              Generates := option{tycoon_manager_buyplates_entry_resource{ResourceId := "food", Amount := 15, CheckDontTake := false}}
              Unlocks := array{
                tycoon_manager_unlock_entry{TargetRoutingId := "bp_303", Payload := false}
              }
              m_StorageIndex := 75
            }
            "bp_303" => tycoon_manager_buyplates_entry{
              BuyplateId := "bp_303"
              Cost := option{tycoon_manager_buyplates_entry_resource{ResourceId := "food", Amount := 4050, CheckDontTake := false}}
              Generates := option{tycoon_manager_buyplates_entry_resource{ResourceId := "food", Amount := 25, CheckDontTake := false}}
              Unlocks := array{
                tycoon_manager_unlock_entry{TargetRoutingId := "bp_304", Payload := false}
              }
              m_StorageIndex := 76
            }
            "bp_304" => tycoon_manager_buyplates_entry{
              BuyplateId := "bp_304"
              Cost := option{tycoon_manager_buyplates_entry_resource{ResourceId := "food", Amount := 7430, CheckDontTake := false}}
              Generates := option{tycoon_manager_buyplates_entry_resource{ResourceId := "food", Amount := 35, CheckDontTake := false}}
              Unlocks := array{
                tycoon_manager_unlock_entry{TargetRoutingId := "bp_305", Payload := false}
              }
              m_StorageIndex := 77
            }
            "bp_305" => tycoon_manager_buyplates_entry{
              BuyplateId := "bp_305"
              Cost := option{tycoon_manager_buyplates_entry_resource{ResourceId := "food", Amount := 16200, CheckDontTake := false}}
              Generates := option{tycoon_manager_buyplates_entry_resource{ResourceId := "food", Amount := 46, CheckDontTake := false}}
              Unlocks := array{
                tycoon_manager_unlock_entry{TargetRoutingId := "bp_306", Payload := false}
              }
              m_StorageIndex := 78
            }
            "bp_306" => tycoon_manager_buyplates_entry{
              BuyplateId := "bp_306"
              Cost := option{tycoon_manager_buyplates_entry_resource{ResourceId := "food", Amount := 24500, CheckDontTake := false}}
              Generates := option{tycoon_manager_buyplates_entry_resource{ResourceId := "food", Amount := 60, CheckDontTake := false}}
              Unlocks := array{
                tycoon_manager_unlock_entry{TargetRoutingId := "bp_307", Payload := false}
              }
              m_StorageIndex := 79
            }
            "bp_307" => tycoon_manager_buyplates_entry{
              BuyplateId := "bp_307"
              Cost := option{tycoon_manager_buyplates_entry_resource{ResourceId := "food", Amount := 44100, CheckDontTake := false}}
              Generates := option{tycoon_manager_buyplates_entry_resource{ResourceId := "food", Amount := 75, CheckDontTake := false}}
              Unlocks := array{
                tycoon_manager_unlock_entry{TargetRoutingId := "bp_308", Payload := false}
              }
              m_StorageIndex := 80
            }
            "bp_308" => tycoon_manager_buyplates_entry{
              BuyplateId := "bp_308"
              Cost := option{tycoon_manager_buyplates_entry_resource{ResourceId := "food", Amount := 61000, CheckDontTake := false}}
              Generates := option{tycoon_manager_buyplates_entry_resource{ResourceId := "food", Amount := 84, CheckDontTake := false}}
              Unlocks := array{
                tycoon_manager_unlock_entry{TargetRoutingId := "bp_309", Payload := false}
              }
              m_StorageIndex := 81
            }
            "bp_309" => tycoon_manager_buyplates_entry{
              BuyplateId := "bp_309"
              Cost := option{tycoon_manager_buyplates_entry_resource{ResourceId := "food", Amount := 79900, CheckDontTake := false}}
              Generates := option{tycoon_manager_buyplates_entry_resource{ResourceId := "food", Amount := 95, CheckDontTake := false}}
              Unlocks := array{
                tycoon_manager_unlock_entry{TargetRoutingId := "bp_310", Payload := false}
              }
              m_StorageIndex := 82
            }
            "bp_310" => tycoon_manager_buyplates_entry{
              BuyplateId := "bp_310"
              Cost := option{tycoon_manager_buyplates_entry_resource{ResourceId := "food", Amount := 122000, CheckDontTake := false}}
              Generates := option{tycoon_manager_buyplates_entry_resource{ResourceId := "food", Amount := 107, CheckDontTake := false}}
              Unlocks := array{
                tycoon_manager_unlock_entry{TargetRoutingId := "bp_311", Payload := false}
              }
              m_StorageIndex := 83
            }
            "bp_311" => tycoon_manager_buyplates_entry{
              BuyplateId := "bp_311"
              Cost := option{tycoon_manager_buyplates_entry_resource{ResourceId := "food", Amount := 150000, CheckDontTake := false}}
              Generates := option{tycoon_manager_buyplates_entry_resource{ResourceId := "food", Amount := 120, CheckDontTake := false}}
              Unlocks := array{
                tycoon_manager_unlock_entry{TargetRoutingId := "bp_312", Payload := false}
              }
              m_StorageIndex := 84
            }
            "bp_312" => tycoon_manager_buyplates_entry{
              BuyplateId := "bp_312"
              Cost := option{tycoon_manager_buyplates_entry_resource{ResourceId := "food", Amount := 183000, CheckDontTake := false}}
              Generates := option{tycoon_manager_buyplates_entry_resource{ResourceId := "food", Amount := 137, CheckDontTake := false}}
              Unlocks := array{
                tycoon_manager_unlock_entry{TargetRoutingId := "bp_313", Payload := false}
              }
              m_StorageIndex := 85
            }
            "bp_313" => tycoon_manager_buyplates_entry{
              BuyplateId := "bp_313"
              Cost := option{tycoon_manager_buyplates_entry_resource{ResourceId := "food", Amount := 256000, CheckDontTake := false}}
              Generates := option{tycoon_manager_buyplates_entry_resource{ResourceId := "food", Amount := 155, CheckDontTake := false}}
              Unlocks := array{
                tycoon_manager_unlock_entry{TargetRoutingId := "bp_314", Payload := false}
              }
              m_StorageIndex := 86
            }
            "bp_314" => tycoon_manager_buyplates_entry{
              BuyplateId := "bp_314"
              Cost := option{tycoon_manager_buyplates_entry_resource{ResourceId := "food", Amount := 305000, CheckDontTake := false}}
              Generates := option{tycoon_manager_buyplates_entry_resource{ResourceId := "food", Amount := 176, CheckDontTake := false}}
              Unlocks := array{
                tycoon_manager_unlock_entry{TargetRoutingId := "bp_315", Payload := false}
              }
              m_StorageIndex := 87
            }
            "bp_315" => tycoon_manager_buyplates_entry{
              BuyplateId := "bp_315"
              Cost := option{tycoon_manager_buyplates_entry_resource{ResourceId := "food", Amount := 412000, CheckDontTake := false}}
              Generates := option{tycoon_manager_buyplates_entry_resource{ResourceId := "food", Amount := 189, CheckDontTake := false}}
              Unlocks := array{
                tycoon_manager_unlock_entry{TargetRoutingId := "bp_316", Payload := false}
              }
              m_StorageIndex := 88
            }
            "bp_316" => tycoon_manager_buyplates_entry{
              BuyplateId := "bp_316"
              Cost := option{tycoon_manager_buyplates_entry_resource{ResourceId := "food", Amount := 480000, CheckDontTake := false}}
              Generates := option{tycoon_manager_buyplates_entry_resource{ResourceId := "food", Amount := 226, CheckDontTake := false}}
              Unlocks := array{
                tycoon_manager_unlock_entry{TargetRoutingId := "bp_317", Payload := false}
              }
              m_StorageIndex := 89
            }
            "bp_317" => tycoon_manager_buyplates_entry{
              BuyplateId := "bp_317"
              Cost := option{tycoon_manager_buyplates_entry_resource{ResourceId := "food", Amount := 608000, CheckDontTake := false}}
              Generates := option{tycoon_manager_buyplates_entry_resource{ResourceId := "food", Amount := 254, CheckDontTake := false}}
              Unlocks := array{
                tycoon_manager_unlock_entry{TargetRoutingId := "bp_318", Payload := false}
              }
              m_StorageIndex := 90
            }
            "bp_318" => tycoon_manager_buyplates_entry{
              BuyplateId := "bp_318"
              Cost := option{tycoon_manager_buyplates_entry_resource{ResourceId := "food", Amount := 707000, CheckDontTake := false}}
              Generates := option{tycoon_manager_buyplates_entry_resource{ResourceId := "food", Amount := 280, CheckDontTake := false}}
              Unlocks := array{
                tycoon_manager_unlock_entry{TargetRoutingId := "bp_319", Payload := false}
              }
              m_StorageIndex := 91
            }
            "bp_319" => tycoon_manager_buyplates_entry{
              BuyplateId := "bp_319"
              Cost := option{tycoon_manager_buyplates_entry_resource{ResourceId := "food", Amount := 817000, CheckDontTake := false}}
              Generates := option{tycoon_manager_buyplates_entry_resource{ResourceId := "food", Amount := 334, CheckDontTake := false}}
              Unlocks := array{
                tycoon_manager_unlock_entry{TargetRoutingId := "bp_320", Payload := false}
                tycoon_manager_unlock_entry{TargetRoutingId := "basecrops", Payload := option{"enable_field:10"}}
                tycoon_manager_unlock_entry{TargetRoutingId := "basecrops", Payload := option{"enable_field:11"}}
                tycoon_manager_unlock_entry{TargetRoutingId := "basecrops", Payload := option{"enable_field:12"}}
                tycoon_manager_unlock_entry{TargetRoutingId := "basecrops", Payload := option{"enable_field:13"}}
                tycoon_manager_unlock_entry{TargetRoutingId := "basecrops", Payload := option{"enable_field:14"}}
                tycoon_manager_unlock_entry{TargetRoutingId := "basecrops", Payload := option{"enable_field:15"}}
                tycoon_manager_unlock_entry{TargetRoutingId := "basecrops", Payload := option{"enable_field:16"}}
              }
              m_StorageIndex := 92
            }
            "bp_320" => tycoon_manager_buyplates_entry{
              BuyplateId := "bp_320"
              Cost := option{tycoon_manager_buyplates_entry_resource{ResourceId := "food", Amount := 1020000, CheckDontTake := false}}
              Generates := option{tycoon_manager_buyplates_entry_resource{ResourceId := "food", Amount := 411, CheckDontTake := false}}
              Unlocks := array{
                tycoon_manager_unlock_entry{TargetRoutingId := "bp_321", Payload := false}
              }
              m_StorageIndex := 93
            }
            "bp_321" => tycoon_manager_buyplates_entry{
              BuyplateId := "bp_321"
              Cost := option{tycoon_manager_buyplates_entry_resource{ResourceId := "food", Amount := 1190000, CheckDontTake := false}}
              Generates := option{tycoon_manager_buyplates_entry_resource{ResourceId := "food", Amount := 468, CheckDontTake := false}}
              Unlocks := array{
                tycoon_manager_unlock_entry{TargetRoutingId := "bp_322", Payload := false}
              }
              m_StorageIndex := 94
            }
            "bp_322" => tycoon_manager_buyplates_entry{
              BuyplateId := "bp_322"
              Cost := option{tycoon_manager_buyplates_entry_resource{ResourceId := "food", Amount := 1390000, CheckDontTake := false}}
              Generates := option{tycoon_manager_buyplates_entry_resource{ResourceId := "food", Amount := 521, CheckDontTake := false}}
              Unlocks := array{
                tycoon_manager_unlock_entry{TargetRoutingId := "bp_323", Payload := false}
              }
              m_StorageIndex := 95
            }
            "bp_323" => tycoon_manager_buyplates_entry{
              BuyplateId := "bp_323"
              Cost := option{tycoon_manager_buyplates_entry_resource{ResourceId := "food", Amount := 1610000, CheckDontTake := false}}
              Generates := option{tycoon_manager_buyplates_entry_resource{ResourceId := "food", Amount := 572, CheckDontTake := false}}
              Unlocks := array{
                tycoon_manager_unlock_entry{TargetRoutingId := "bp_324", Payload := false}
              }
              m_StorageIndex := 96
            }
            "bp_324" => tycoon_manager_buyplates_entry{
              BuyplateId := "bp_324"
              Cost := option{tycoon_manager_buyplates_entry_resource{ResourceId := "food", Amount := 1850000, CheckDontTake := false}}
              Generates := option{tycoon_manager_buyplates_entry_resource{ResourceId := "food", Amount := 600, CheckDontTake := false}}
              Unlocks := array{
                tycoon_manager_unlock_entry{TargetRoutingId := "bp_325", Payload := false}
              }
              m_StorageIndex := 97
            }
            "bp_325" => tycoon_manager_buyplates_entry{
              BuyplateId := "bp_325"
              Cost := option{tycoon_manager_buyplates_entry_resource{ResourceId := "food", Amount := 2000000, CheckDontTake := false}}
              Generates := false
              Unlocks := array{
              }
              m_StorageIndex := 98
            }
            "bp_400" => tycoon_manager_buyplates_entry{
              BuyplateId := "bp_400"
              Cost := option{tycoon_manager_buyplates_entry_resource{ResourceId := "garage", Amount := 5, CheckDontTake := false}}
              Generates := option{tycoon_manager_buyplates_entry_resource{ResourceId := "garage", Amount := 5, CheckDontTake := false}}
              Unlocks := array{
                tycoon_manager_unlock_entry{TargetRoutingId := "bp_401", Payload := false}
              }
              m_StorageIndex := 99
            }
            "bp_401" => tycoon_manager_buyplates_entry{
              BuyplateId := "bp_401"
              Cost := option{tycoon_manager_buyplates_entry_resource{ResourceId := "garage", Amount := 75, CheckDontTake := false}}
              Generates := option{tycoon_manager_buyplates_entry_resource{ResourceId := "garage", Amount := 10, CheckDontTake := false}}
              Unlocks := array{
                tycoon_manager_unlock_entry{TargetRoutingId := "bp_402", Payload := false}
              }
              m_StorageIndex := 100
            }
            "bp_402" => tycoon_manager_buyplates_entry{
              BuyplateId := "bp_402"
              Cost := option{tycoon_manager_buyplates_entry_resource{ResourceId := "garage", Amount := 675, CheckDontTake := false}}
              Generates := option{tycoon_manager_buyplates_entry_resource{ResourceId := "garage", Amount := 20, CheckDontTake := false}}
              Unlocks := array{
                tycoon_manager_unlock_entry{TargetRoutingId := "bp_403", Payload := false}
              }
              m_StorageIndex := 101
            }
            "bp_403" => tycoon_manager_buyplates_entry{
              BuyplateId := "bp_403"
              Cost := option{tycoon_manager_buyplates_entry_resource{ResourceId := "garage", Amount := 2360, CheckDontTake := false}}
              Generates := option{tycoon_manager_buyplates_entry_resource{ResourceId := "garage", Amount := 30, CheckDontTake := false}}
              Unlocks := array{
                tycoon_manager_unlock_entry{TargetRoutingId := "bp_404", Payload := false}
              }
              m_StorageIndex := 102
            }
            "bp_404" => tycoon_manager_buyplates_entry{
              BuyplateId := "bp_404"
              Cost := option{tycoon_manager_buyplates_entry_resource{ResourceId := "garage", Amount := 5850, CheckDontTake := false}}
              Generates := option{tycoon_manager_buyplates_entry_resource{ResourceId := "garage", Amount := 45, CheckDontTake := false}}
              Unlocks := array{
                tycoon_manager_unlock_entry{TargetRoutingId := "bp_405", Payload := false}
                tycoon_manager_unlock_entry{TargetRoutingId := "garage_device", Payload := option{"enable"}}
              }
              m_StorageIndex := 103
            }
            "bp_405" => tycoon_manager_buyplates_entry{
              BuyplateId := "bp_405"
              Cost := option{tycoon_manager_buyplates_entry_resource{ResourceId := "garage", Amount := 9900, CheckDontTake := false}}
              Generates := option{tycoon_manager_buyplates_entry_resource{ResourceId := "garage", Amount := 65, CheckDontTake := false}}
              Unlocks := array{
                tycoon_manager_unlock_entry{TargetRoutingId := "bp_406", Payload := false}
              }
              m_StorageIndex := 104
            }
            "bp_406" => tycoon_manager_buyplates_entry{
              BuyplateId := "bp_406"
              Cost := option{tycoon_manager_buyplates_entry_resource{ResourceId := "garage", Amount := 15800, CheckDontTake := false}}
              Generates := option{tycoon_manager_buyplates_entry_resource{ResourceId := "garage", Amount := 90, CheckDontTake := false}}
              Unlocks := array{
                tycoon_manager_unlock_entry{TargetRoutingId := "bp_407", Payload := false}
              }
              m_StorageIndex := 105
            }
            "bp_407" => tycoon_manager_buyplates_entry{
              BuyplateId := "bp_407"
              Cost := option{tycoon_manager_buyplates_entry_resource{ResourceId := "garage", Amount := 47700, CheckDontTake := false}}
              Generates := option{tycoon_manager_buyplates_entry_resource{ResourceId := "garage", Amount := 105, CheckDontTake := false}}
              Unlocks := array{
                tycoon_manager_unlock_entry{TargetRoutingId := "bp_408", Payload := false}
              }
              m_StorageIndex := 106
            }
            "bp_408" => tycoon_manager_buyplates_entry{
              BuyplateId := "bp_408"
              Cost := option{tycoon_manager_buyplates_entry_resource{ResourceId := "garage", Amount := 66600, CheckDontTake := false}}
              Generates := option{tycoon_manager_buyplates_entry_resource{ResourceId := "garage", Amount := 128, CheckDontTake := false}}
              Unlocks := array{
                tycoon_manager_unlock_entry{TargetRoutingId := "bp_409", Payload := false}
              }
              m_StorageIndex := 107
            }
            "bp_409" => tycoon_manager_buyplates_entry{
              BuyplateId := "bp_409"
              Cost := option{tycoon_manager_buyplates_entry_resource{ResourceId := "garage", Amount := 112000, CheckDontTake := false}}
              Generates := option{tycoon_manager_buyplates_entry_resource{ResourceId := "garage", Amount := 143, CheckDontTake := false}}
              Unlocks := array{
                tycoon_manager_unlock_entry{TargetRoutingId := "bp_410", Payload := false}
              }
              m_StorageIndex := 108
            }
            "bp_410" => tycoon_manager_buyplates_entry{
              BuyplateId := "bp_410"
              Cost := option{tycoon_manager_buyplates_entry_resource{ResourceId := "garage", Amount := 144000, CheckDontTake := false}}
              Generates := option{tycoon_manager_buyplates_entry_resource{ResourceId := "garage", Amount := 170, CheckDontTake := false}}
              Unlocks := array{
                tycoon_manager_unlock_entry{TargetRoutingId := "bp_411", Payload := false}
              }
              m_StorageIndex := 109
            }
            "bp_411" => tycoon_manager_buyplates_entry{
              BuyplateId := "bp_411"
              Cost := option{tycoon_manager_buyplates_entry_resource{ResourceId := "garage", Amount := 219000, CheckDontTake := false}}
              Generates := option{tycoon_manager_buyplates_entry_resource{ResourceId := "garage", Amount := 199, CheckDontTake := false}}
              Unlocks := array{
                tycoon_manager_unlock_entry{TargetRoutingId := "bp_412", Payload := false}
              }
              m_StorageIndex := 110
            }
            "bp_412" => tycoon_manager_buyplates_entry{
              BuyplateId := "bp_412"
              Cost := option{tycoon_manager_buyplates_entry_resource{ResourceId := "garage", Amount := 273000, CheckDontTake := false}}
              Generates := option{tycoon_manager_buyplates_entry_resource{ResourceId := "garage", Amount := 221, CheckDontTake := false}}
              Unlocks := array{
                tycoon_manager_unlock_entry{TargetRoutingId := "bp_413", Payload := false}
              }
              m_StorageIndex := 111
            }
            "bp_413" => tycoon_manager_buyplates_entry{
              BuyplateId := "bp_413"
              Cost := option{tycoon_manager_buyplates_entry_resource{ResourceId := "garage", Amount := 332000, CheckDontTake := false}}
              Generates := option{tycoon_manager_buyplates_entry_resource{ResourceId := "garage", Amount := 245, CheckDontTake := false}}
              Unlocks := array{
                tycoon_manager_unlock_entry{TargetRoutingId := "bp_414", Payload := false}
              }
              m_StorageIndex := 112
            }
            "bp_414" => tycoon_manager_buyplates_entry{
              BuyplateId := "bp_414"
              Cost := option{tycoon_manager_buyplates_entry_resource{ResourceId := "garage", Amount := 465000, CheckDontTake := false}}
              Generates := option{tycoon_manager_buyplates_entry_resource{ResourceId := "garage", Amount := 282, CheckDontTake := false}}
              Unlocks := array{
                tycoon_manager_unlock_entry{TargetRoutingId := "bp_415", Payload := false}
              }
              m_StorageIndex := 113
            }
            "bp_415" => tycoon_manager_buyplates_entry{
              BuyplateId := "bp_415"
              Cost := option{tycoon_manager_buyplates_entry_resource{ResourceId := "garage", Amount := 554000, CheckDontTake := false}}
              Generates := option{tycoon_manager_buyplates_entry_resource{ResourceId := "garage", Amount := 313, CheckDontTake := false}}
              Unlocks := array{
                tycoon_manager_unlock_entry{TargetRoutingId := "bp_416", Payload := false}
              }
              m_StorageIndex := 114
            }
            "bp_416" => tycoon_manager_buyplates_entry{
              BuyplateId := "bp_416"
              Cost := option{tycoon_manager_buyplates_entry_resource{ResourceId := "garage", Amount := 683000, CheckDontTake := false}}
              Generates := option{tycoon_manager_buyplates_entry_resource{ResourceId := "garage", Amount := 339, CheckDontTake := false}}
              Unlocks := array{
                tycoon_manager_unlock_entry{TargetRoutingId := "bp_417", Payload := false}
              }
              m_StorageIndex := 115
            }
            "bp_417" => tycoon_manager_buyplates_entry{
              BuyplateId := "bp_417"
              Cost := option{tycoon_manager_buyplates_entry_resource{ResourceId := "garage", Amount := 795000, CheckDontTake := false}}
              Generates := option{tycoon_manager_buyplates_entry_resource{ResourceId := "garage", Amount := 393, CheckDontTake := false}}
              Unlocks := array{
                tycoon_manager_unlock_entry{TargetRoutingId := "bp_418", Payload := false}
              }
              m_StorageIndex := 116
            }
            "bp_418" => tycoon_manager_buyplates_entry{
              BuyplateId := "bp_418"
              Cost := option{tycoon_manager_buyplates_entry_resource{ResourceId := "garage", Amount := 1009000, CheckDontTake := false}}
              Generates := option{tycoon_manager_buyplates_entry_resource{ResourceId := "garage", Amount := 440, CheckDontTake := false}}
              Unlocks := array{
                tycoon_manager_unlock_entry{TargetRoutingId := "bp_419", Payload := false}
              }
              m_StorageIndex := 117
            }
            "bp_419" => tycoon_manager_buyplates_entry{
              BuyplateId := "bp_419"
              Cost := option{tycoon_manager_buyplates_entry_resource{ResourceId := "garage", Amount := 1170000, CheckDontTake := false}}
              Generates := option{tycoon_manager_buyplates_entry_resource{ResourceId := "garage", Amount := 516, CheckDontTake := false}}
              Unlocks := array{
                tycoon_manager_unlock_entry{TargetRoutingId := "bp_420", Payload := false}
              }
              m_StorageIndex := 118
            }
            "bp_420" => tycoon_manager_buyplates_entry{
              BuyplateId := "bp_420"
              Cost := option{tycoon_manager_buyplates_entry_resource{ResourceId := "garage", Amount := 1430000, CheckDontTake := false}}
              Generates := option{tycoon_manager_buyplates_entry_resource{ResourceId := "garage", Amount := 591, CheckDontTake := false}}
              Unlocks := array{
                tycoon_manager_unlock_entry{TargetRoutingId := "bp_421", Payload := false}
              }
              m_StorageIndex := 119
            }
            "bp_421" => tycoon_manager_buyplates_entry{
              BuyplateId := "bp_421"
              Cost := option{tycoon_manager_buyplates_entry_resource{ResourceId := "garage", Amount := 1650000, CheckDontTake := false}}
              Generates := option{tycoon_manager_buyplates_entry_resource{ResourceId := "garage", Amount := 650, CheckDontTake := false}}
              Unlocks := array{
                tycoon_manager_unlock_entry{TargetRoutingId := "bp_422", Payload := false}
              }
              m_StorageIndex := 120
            }
            "bp_422" => tycoon_manager_buyplates_entry{
              BuyplateId := "bp_422"
              Cost := option{tycoon_manager_buyplates_entry_resource{ResourceId := "garage", Amount := 2000000, CheckDontTake := false}}
              Generates := false
              Unlocks := array{
              }
              m_StorageIndex := 121
            }
            "bp_500" => tycoon_manager_buyplates_entry{
              BuyplateId := "bp_500"
              Cost := option{tycoon_manager_buyplates_entry_resource{ResourceId := "armory", Amount := 5, CheckDontTake := false}}
              Generates := option{tycoon_manager_buyplates_entry_resource{ResourceId := "armory", Amount := 5, CheckDontTake := false}}
              Unlocks := array{
                tycoon_manager_unlock_entry{TargetRoutingId := "bp_501", Payload := false}
              }
              m_StorageIndex := 122
            }
            "bp_501" => tycoon_manager_buyplates_entry{
              BuyplateId := "bp_501"
              Cost := option{tycoon_manager_buyplates_entry_resource{ResourceId := "armory", Amount := 100, CheckDontTake := false}}
              Generates := option{tycoon_manager_buyplates_entry_resource{ResourceId := "armory", Amount := 10, CheckDontTake := false}}
              Unlocks := array{
                tycoon_manager_unlock_entry{TargetRoutingId := "bp_509", Payload := false}
              }
              m_StorageIndex := 123
            }
            "bp_509" => tycoon_manager_buyplates_entry{
              BuyplateId := "bp_509"
              Cost := option{tycoon_manager_buyplates_entry_resource{ResourceId := "armory", Amount := 900, CheckDontTake := false}}
              Generates := option{tycoon_manager_buyplates_entry_resource{ResourceId := "armory", Amount := 20, CheckDontTake := false}}
              Unlocks := array{
                tycoon_manager_unlock_entry{TargetRoutingId := "bp_510", Payload := false}
              }
              m_StorageIndex := 124
            }
            "bp_510" => tycoon_manager_buyplates_entry{
              BuyplateId := "bp_510"
              Cost := option{tycoon_manager_buyplates_entry_resource{ResourceId := "armory", Amount := 4200, CheckDontTake := false}}
              Generates := option{tycoon_manager_buyplates_entry_resource{ResourceId := "armory", Amount := 35, CheckDontTake := false}}
              Unlocks := array{
                tycoon_manager_unlock_entry{TargetRoutingId := "bp_502", Payload := false}
              }
              m_StorageIndex := 125
            }
            "bp_502" => tycoon_manager_buyplates_entry{
              BuyplateId := "bp_502"
              Cost := option{tycoon_manager_buyplates_entry_resource{ResourceId := "armory", Amount := 8400, CheckDontTake := false}}
              Generates := option{tycoon_manager_buyplates_entry_resource{ResourceId := "armory", Amount := 40, CheckDontTake := false}}
              Unlocks := array{
                tycoon_manager_unlock_entry{TargetRoutingId := "bp_503", Payload := false}
              }
              m_StorageIndex := 126
            }
            "bp_503" => tycoon_manager_buyplates_entry{
              BuyplateId := "bp_503"
              Cost := option{tycoon_manager_buyplates_entry_resource{ResourceId := "armory", Amount := 16500, CheckDontTake := false}}
              Generates := option{tycoon_manager_buyplates_entry_resource{ResourceId := "armory", Amount := 45, CheckDontTake := false}}
              Unlocks := array{
                tycoon_manager_unlock_entry{TargetRoutingId := "bp_504", Payload := false}
              }
              m_StorageIndex := 127
            }
            "bp_504" => tycoon_manager_buyplates_entry{
              BuyplateId := "bp_504"
              Cost := option{tycoon_manager_buyplates_entry_resource{ResourceId := "armory", Amount := 23300, CheckDontTake := false}}
              Generates := option{tycoon_manager_buyplates_entry_resource{ResourceId := "armory", Amount := 50, CheckDontTake := false}}
              Unlocks := array{
                tycoon_manager_unlock_entry{TargetRoutingId := "bp_505", Payload := false}
              }
              m_StorageIndex := 128
            }
            "bp_505" => tycoon_manager_buyplates_entry{
              BuyplateId := "bp_505"
              Cost := option{tycoon_manager_buyplates_entry_resource{ResourceId := "armory", Amount := 36900, CheckDontTake := false}}
              Generates := option{tycoon_manager_buyplates_entry_resource{ResourceId := "armory", Amount := 55, CheckDontTake := false}}
              Unlocks := array{
                tycoon_manager_unlock_entry{TargetRoutingId := "bp_506", Payload := false}
              }
              m_StorageIndex := 129
            }
            "bp_506" => tycoon_manager_buyplates_entry{
              BuyplateId := "bp_506"
              Cost := option{tycoon_manager_buyplates_entry_resource{ResourceId := "armory", Amount := 46800, CheckDontTake := false}}
              Generates := option{tycoon_manager_buyplates_entry_resource{ResourceId := "armory", Amount := 60, CheckDontTake := false}}
              Unlocks := array{
                tycoon_manager_unlock_entry{TargetRoutingId := "bp_507", Payload := false}
              }
              m_StorageIndex := 130
            }
            "bp_507" => tycoon_manager_buyplates_entry{
              BuyplateId := "bp_507"
              Cost := option{tycoon_manager_buyplates_entry_resource{ResourceId := "armory", Amount := 57600, CheckDontTake := false}}
              Generates := option{tycoon_manager_buyplates_entry_resource{ResourceId := "armory", Amount := 66, CheckDontTake := false}}
              Unlocks := array{
                tycoon_manager_unlock_entry{TargetRoutingId := "bp_508", Payload := false}
              }
              m_StorageIndex := 131
            }
            "bp_508" => tycoon_manager_buyplates_entry{
              BuyplateId := "bp_508"
              Cost := option{tycoon_manager_buyplates_entry_resource{ResourceId := "armory", Amount := 69500, CheckDontTake := false}}
              Generates := option{tycoon_manager_buyplates_entry_resource{ResourceId := "armory", Amount := 72, CheckDontTake := false}}
              Unlocks := array{
                tycoon_manager_unlock_entry{TargetRoutingId := "bp_511", Payload := false}
              }
              m_StorageIndex := 132
            }
            "bp_511" => tycoon_manager_buyplates_entry{
              BuyplateId := "bp_511"
              Cost := option{tycoon_manager_buyplates_entry_resource{ResourceId := "armory", Amount := 82400, CheckDontTake := false}}
              Generates := option{tycoon_manager_buyplates_entry_resource{ResourceId := "armory", Amount := 79, CheckDontTake := false}}
              Unlocks := array{
                tycoon_manager_unlock_entry{TargetRoutingId := "bp_512", Payload := false}
              }
              m_StorageIndex := 133
            }
            "bp_512" => tycoon_manager_buyplates_entry{
              BuyplateId := "bp_512"
              Cost := option{tycoon_manager_buyplates_entry_resource{ResourceId := "armory", Amount := 129000, CheckDontTake := false}}
              Generates := option{tycoon_manager_buyplates_entry_resource{ResourceId := "armory", Amount := 87, CheckDontTake := false}}
              Unlocks := array{
                tycoon_manager_unlock_entry{TargetRoutingId := "bp_513", Payload := false}
              }
              m_StorageIndex := 134
            }
            "bp_513" => tycoon_manager_buyplates_entry{
              BuyplateId := "bp_513"
              Cost := option{tycoon_manager_buyplates_entry_resource{ResourceId := "armory", Amount := 150000, CheckDontTake := false}}
              Generates := option{tycoon_manager_buyplates_entry_resource{ResourceId := "armory", Amount := 96, CheckDontTake := false}}
              Unlocks := array{
                tycoon_manager_unlock_entry{TargetRoutingId := "bp_514", Payload := false}
              }
              m_StorageIndex := 135
            }
            "bp_514" => tycoon_manager_buyplates_entry{
              BuyplateId := "bp_514"
              Cost := option{tycoon_manager_buyplates_entry_resource{ResourceId := "armory", Amount := 173000, CheckDontTake := false}}
              Generates := option{tycoon_manager_buyplates_entry_resource{ResourceId := "armory", Amount := 105, CheckDontTake := false}}
              Unlocks := array{
                tycoon_manager_unlock_entry{TargetRoutingId := "bp_515", Payload := false}
              }
              m_StorageIndex := 136
            }
            "bp_515" => tycoon_manager_buyplates_entry{
              BuyplateId := "bp_515"
              Cost := option{tycoon_manager_buyplates_entry_resource{ResourceId := "armory", Amount := 198000, CheckDontTake := false}}
              Generates := option{tycoon_manager_buyplates_entry_resource{ResourceId := "armory", Amount := 116, CheckDontTake := false}}
              Unlocks := array{
                tycoon_manager_unlock_entry{TargetRoutingId := "bp_516", Payload := false}
              }
              m_StorageIndex := 137
            }
            "bp_516" => tycoon_manager_buyplates_entry{
              BuyplateId := "bp_516"
              Cost := option{tycoon_manager_buyplates_entry_resource{ResourceId := "armory", Amount := 226000, CheckDontTake := false}}
              Generates := option{tycoon_manager_buyplates_entry_resource{ResourceId := "armory", Amount := 128, CheckDontTake := false}}
              Unlocks := array{
                tycoon_manager_unlock_entry{TargetRoutingId := "bp_517", Payload := false}
              }
              m_StorageIndex := 138
            }
            "bp_517" => tycoon_manager_buyplates_entry{
              BuyplateId := "bp_517"
              Cost := option{tycoon_manager_buyplates_entry_resource{ResourceId := "armory", Amount := 305000, CheckDontTake := false}}
              Generates := option{tycoon_manager_buyplates_entry_resource{ResourceId := "armory", Amount := 141, CheckDontTake := false}}
              Unlocks := array{
                tycoon_manager_unlock_entry{TargetRoutingId := "bp_518", Payload := false}
              }
              m_StorageIndex := 139
            }
            "bp_518" => tycoon_manager_buyplates_entry{
              BuyplateId := "bp_518"
              Cost := option{tycoon_manager_buyplates_entry_resource{ResourceId := "armory", Amount := 345000, CheckDontTake := false}}
              Generates := option{tycoon_manager_buyplates_entry_resource{ResourceId := "armory", Amount := 155, CheckDontTake := false}}
              Unlocks := array{
                tycoon_manager_unlock_entry{TargetRoutingId := "bp_519", Payload := false}
              }
              m_StorageIndex := 140
            }
            "bp_519" => tycoon_manager_buyplates_entry{
              BuyplateId := "bp_519"
              Cost := option{tycoon_manager_buyplates_entry_resource{ResourceId := "armory", Amount := 389000, CheckDontTake := false}}
              Generates := option{tycoon_manager_buyplates_entry_resource{ResourceId := "armory", Amount := 171, CheckDontTake := false}}
              Unlocks := array{
                tycoon_manager_unlock_entry{TargetRoutingId := "bp_520", Payload := false}
              }
              m_StorageIndex := 141
            }
            "bp_520" => tycoon_manager_buyplates_entry{
              BuyplateId := "bp_520"
              Cost := option{tycoon_manager_buyplates_entry_resource{ResourceId := "armory", Amount := 438000, CheckDontTake := false}}
              Generates := option{tycoon_manager_buyplates_entry_resource{ResourceId := "armory", Amount := 188, CheckDontTake := false}}
              Unlocks := array{
                tycoon_manager_unlock_entry{TargetRoutingId := "bp_521", Payload := false}
              }
              m_StorageIndex := 142
            }
            "bp_521" => tycoon_manager_buyplates_entry{
              BuyplateId := "bp_521"
              Cost := option{tycoon_manager_buyplates_entry_resource{ResourceId := "armory", Amount := 491000, CheckDontTake := false}}
              Generates := option{tycoon_manager_buyplates_entry_resource{ResourceId := "armory", Amount := 207, CheckDontTake := false}}
              Unlocks := array{
                tycoon_manager_unlock_entry{TargetRoutingId := "bp_522", Payload := false}
              }
              m_StorageIndex := 143
            }
            "bp_522" => tycoon_manager_buyplates_entry{
              BuyplateId := "bp_522"
              Cost := option{tycoon_manager_buyplates_entry_resource{ResourceId := "armory", Amount := 637000, CheckDontTake := false}}
              Generates := option{tycoon_manager_buyplates_entry_resource{ResourceId := "armory", Amount := 228, CheckDontTake := false}}
              Unlocks := array{
                tycoon_manager_unlock_entry{TargetRoutingId := "bp_523", Payload := false}
              }
              m_StorageIndex := 144
            }
            "bp_523" => tycoon_manager_buyplates_entry{
              BuyplateId := "bp_523"
              Cost := option{tycoon_manager_buyplates_entry_resource{ResourceId := "armory", Amount := 712000, CheckDontTake := false}}
              Generates := option{tycoon_manager_buyplates_entry_resource{ResourceId := "armory", Amount := 251, CheckDontTake := false}}
              Unlocks := array{
                tycoon_manager_unlock_entry{TargetRoutingId := "bp_524", Payload := false}
              }
              m_StorageIndex := 145
            }
            "bp_524" => tycoon_manager_buyplates_entry{
              BuyplateId := "bp_524"
              Cost := option{tycoon_manager_buyplates_entry_resource{ResourceId := "armory", Amount := 795000, CheckDontTake := false}}
              Generates := option{tycoon_manager_buyplates_entry_resource{ResourceId := "armory", Amount := 276, CheckDontTake := false}}
              Unlocks := array{
                tycoon_manager_unlock_entry{TargetRoutingId := "bp_525", Payload := false}
              }
              m_StorageIndex := 146
            }
            "bp_525" => tycoon_manager_buyplates_entry{
              BuyplateId := "bp_525"
              Cost := option{tycoon_manager_buyplates_entry_resource{ResourceId := "armory", Amount := 886000, CheckDontTake := false}}
              Generates := option{tycoon_manager_buyplates_entry_resource{ResourceId := "armory", Amount := 302, CheckDontTake := false}}
              Unlocks := array{
                tycoon_manager_unlock_entry{TargetRoutingId := "bp_526", Payload := false}
              }
              m_StorageIndex := 147
            }
            "bp_526" => tycoon_manager_buyplates_entry{
              BuyplateId := "bp_526"
              Cost := option{tycoon_manager_buyplates_entry_resource{ResourceId := "armory", Amount := 986000, CheckDontTake := false}}
              Generates := option{tycoon_manager_buyplates_entry_resource{ResourceId := "armory", Amount := 331, CheckDontTake := false}}
              Unlocks := array{
                tycoon_manager_unlock_entry{TargetRoutingId := "bp_527", Payload := false}
              }
              m_StorageIndex := 148
            }
            "bp_527" => tycoon_manager_buyplates_entry{
              BuyplateId := "bp_527"
              Cost := option{tycoon_manager_buyplates_entry_resource{ResourceId := "armory", Amount := 1240000, CheckDontTake := false}}
              Generates := option{tycoon_manager_buyplates_entry_resource{ResourceId := "armory", Amount := 363, CheckDontTake := false}}
              Unlocks := array{
                tycoon_manager_unlock_entry{TargetRoutingId := "bp_528", Payload := false}
              }
              m_StorageIndex := 149
            }
            "bp_528" => tycoon_manager_buyplates_entry{
              BuyplateId := "bp_528"
              Cost := option{tycoon_manager_buyplates_entry_resource{ResourceId := "armory", Amount := 1380000, CheckDontTake := false}}
              Generates := option{tycoon_manager_buyplates_entry_resource{ResourceId := "armory", Amount := 399, CheckDontTake := false}}
              Unlocks := array{
                tycoon_manager_unlock_entry{TargetRoutingId := "bp_529", Payload := false}
              }
              m_StorageIndex := 150
            }
            "bp_529" => tycoon_manager_buyplates_entry{
              BuyplateId := "bp_529"
              Cost := option{tycoon_manager_buyplates_entry_resource{ResourceId := "armory", Amount := 1530000, CheckDontTake := false}}
              Generates := option{tycoon_manager_buyplates_entry_resource{ResourceId := "armory", Amount := 438, CheckDontTake := false}}
              Unlocks := array{
                tycoon_manager_unlock_entry{TargetRoutingId := "bp_530", Payload := false}
              }
              m_StorageIndex := 151
            }
            "bp_530" => tycoon_manager_buyplates_entry{
              BuyplateId := "bp_530"
              Cost := option{tycoon_manager_buyplates_entry_resource{ResourceId := "armory", Amount := 1690000, CheckDontTake := false}}
              Generates := option{tycoon_manager_buyplates_entry_resource{ResourceId := "armory", Amount := 481, CheckDontTake := false}}
              Unlocks := array{
                tycoon_manager_unlock_entry{TargetRoutingId := "bp_531", Payload := false}
              }
              m_StorageIndex := 152
            }
            "bp_531" => tycoon_manager_buyplates_entry{
              BuyplateId := "bp_531"
              Cost := option{tycoon_manager_buyplates_entry_resource{ResourceId := "armory", Amount := 2000000, CheckDontTake := false}}
              Generates := false
              Unlocks := array{
              }
              m_StorageIndex := 153
            }
            "bp_600" => tycoon_manager_buyplates_entry{
              BuyplateId := "bp_600"
              Cost := option{tycoon_manager_buyplates_entry_resource{ResourceId := "meds", Amount := 5, CheckDontTake := false}}
              Generates := option{tycoon_manager_buyplates_entry_resource{ResourceId := "meds", Amount := 10, CheckDontTake := false}}
              Unlocks := array{
                tycoon_manager_unlock_entry{TargetRoutingId := "bp_601", Payload := false}
                tycoon_manager_unlock_entry{TargetRoutingId := "bp_700", Payload := false}
              }
              m_StorageIndex := 154
            }
            "bp_601" => tycoon_manager_buyplates_entry{
              BuyplateId := "bp_601"
              Cost := option{tycoon_manager_buyplates_entry_resource{ResourceId := "meds", Amount := 150, CheckDontTake := false}}
              Generates := option{tycoon_manager_buyplates_entry_resource{ResourceId := "meds", Amount := 15, CheckDontTake := false}}
              Unlocks := array{
                tycoon_manager_unlock_entry{TargetRoutingId := "bp_602", Payload := false}
              }
              m_StorageIndex := 155
            }
            "bp_602" => tycoon_manager_buyplates_entry{
              BuyplateId := "bp_602"
              Cost := option{tycoon_manager_buyplates_entry_resource{ResourceId := "meds", Amount := 1130, CheckDontTake := false}}
              Generates := option{tycoon_manager_buyplates_entry_resource{ResourceId := "meds", Amount := 22, CheckDontTake := false}}
              Unlocks := array{
                tycoon_manager_unlock_entry{TargetRoutingId := "bp_603", Payload := false}
              }
              m_StorageIndex := 156
            }
            "bp_603" => tycoon_manager_buyplates_entry{
              BuyplateId := "bp_603"
              Cost := option{tycoon_manager_buyplates_entry_resource{ResourceId := "meds", Amount := 2120, CheckDontTake := false}}
              Generates := option{tycoon_manager_buyplates_entry_resource{ResourceId := "meds", Amount := 28, CheckDontTake := false}}
              Unlocks := array{
                tycoon_manager_unlock_entry{TargetRoutingId := "bp_604", Payload := false}
              }
              m_StorageIndex := 157
            }
            "bp_604" => tycoon_manager_buyplates_entry{
              BuyplateId := "bp_604"
              Cost := option{tycoon_manager_buyplates_entry_resource{ResourceId := "meds", Amount := 5060, CheckDontTake := false}}
              Generates := option{tycoon_manager_buyplates_entry_resource{ResourceId := "meds", Amount := 35, CheckDontTake := false}}
              Unlocks := array{
                tycoon_manager_unlock_entry{TargetRoutingId := "bp_605", Payload := false}
              }
              m_StorageIndex := 158
            }
            "bp_605" => tycoon_manager_buyplates_entry{
              BuyplateId := "bp_605"
              Cost := option{tycoon_manager_buyplates_entry_resource{ResourceId := "meds", Amount := 7430, CheckDontTake := false}}
              Generates := option{tycoon_manager_buyplates_entry_resource{ResourceId := "meds", Amount := 43, CheckDontTake := false}}
              Unlocks := array{
                tycoon_manager_unlock_entry{TargetRoutingId := "bp_606", Payload := false}
              }
              m_StorageIndex := 159
            }
            "bp_606" => tycoon_manager_buyplates_entry{
              BuyplateId := "bp_606"
              Cost := option{tycoon_manager_buyplates_entry_resource{ResourceId := "meds", Amount := 13800, CheckDontTake := false}}
              Generates := option{tycoon_manager_buyplates_entry_resource{ResourceId := "meds", Amount := 47, CheckDontTake := false}}
              Unlocks := array{
                tycoon_manager_unlock_entry{TargetRoutingId := "bp_607", Payload := false}
              }
              m_StorageIndex := 160
            }
            "bp_607" => tycoon_manager_buyplates_entry{
              BuyplateId := "bp_607"
              Cost := option{tycoon_manager_buyplates_entry_resource{ResourceId := "meds", Amount := 18000, CheckDontTake := false}}
              Generates := option{tycoon_manager_buyplates_entry_resource{ResourceId := "meds", Amount := 51, CheckDontTake := false}}
              Unlocks := array{
                tycoon_manager_unlock_entry{TargetRoutingId := "bp_608", Payload := false}
              }
              m_StorageIndex := 161
            }
            "bp_700" => tycoon_manager_buyplates_entry{
              BuyplateId := "bp_700"
              Cost := option{tycoon_manager_buyplates_entry_resource{ResourceId := "meds", Amount := 33900, CheckDontTake := false}}
              Generates := option{tycoon_manager_buyplates_entry_resource{ResourceId := "meds", Amount := 55, CheckDontTake := false}}
              Unlocks := array{
                tycoon_manager_unlock_entry{TargetRoutingId := "bp_701", Payload := false}
              }
              m_StorageIndex := 162
            }
            "bp_701" => tycoon_manager_buyplates_entry{
              BuyplateId := "bp_701"
              Cost := option{tycoon_manager_buyplates_entry_resource{ResourceId := "meds", Amount := 41300, CheckDontTake := false}}
              Generates := option{tycoon_manager_buyplates_entry_resource{ResourceId := "meds", Amount := 59, CheckDontTake := false}}
              Unlocks := array{
                tycoon_manager_unlock_entry{TargetRoutingId := "bp_702", Payload := false}
              }
              m_StorageIndex := 163
            }
            "bp_702" => tycoon_manager_buyplates_entry{
              BuyplateId := "bp_702"
              Cost := option{tycoon_manager_buyplates_entry_resource{ResourceId := "meds", Amount := 49300, CheckDontTake := false}}
              Generates := option{tycoon_manager_buyplates_entry_resource{ResourceId := "meds", Amount := 63, CheckDontTake := false}}
              Unlocks := array{
                tycoon_manager_unlock_entry{TargetRoutingId := "bp_703", Payload := false}
              }
              m_StorageIndex := 164
            }
            "bp_703" => tycoon_manager_buyplates_entry{
              BuyplateId := "bp_703"
              Cost := option{tycoon_manager_buyplates_entry_resource{ResourceId := "meds", Amount := 57800, CheckDontTake := false}}
              Generates := option{tycoon_manager_buyplates_entry_resource{ResourceId := "meds", Amount := 68, CheckDontTake := false}}
              Unlocks := array{
                tycoon_manager_unlock_entry{TargetRoutingId := "bp_704", Payload := false}
              }
              m_StorageIndex := 165
            }
            "bp_704" => tycoon_manager_buyplates_entry{
              BuyplateId := "bp_704"
              Cost := option{tycoon_manager_buyplates_entry_resource{ResourceId := "meds", Amount := 67000, CheckDontTake := false}}
              Generates := option{tycoon_manager_buyplates_entry_resource{ResourceId := "meds", Amount := 73, CheckDontTake := false}}
              Unlocks := array{
                tycoon_manager_unlock_entry{TargetRoutingId := "bp_705", Payload := false}
              }
              m_StorageIndex := 166
            }
            "bp_705" => tycoon_manager_buyplates_entry{
              BuyplateId := "bp_705"
              Cost := option{tycoon_manager_buyplates_entry_resource{ResourceId := "meds", Amount := 76800, CheckDontTake := false}}
              Generates := option{tycoon_manager_buyplates_entry_resource{ResourceId := "meds", Amount := 78, CheckDontTake := false}}
              Unlocks := array{
              }
              m_StorageIndex := 167
            }
            "bp_608" => tycoon_manager_buyplates_entry{
              BuyplateId := "bp_608"
              Cost := option{tycoon_manager_buyplates_entry_resource{ResourceId := "meds", Amount := 107000, CheckDontTake := false}}
              Generates := option{tycoon_manager_buyplates_entry_resource{ResourceId := "meds", Amount := 83, CheckDontTake := false}}
              Unlocks := array{
                tycoon_manager_unlock_entry{TargetRoutingId := "bp_609", Payload := false}
              }
              m_StorageIndex := 168
            }
            "bp_609" => tycoon_manager_buyplates_entry{
              BuyplateId := "bp_609"
              Cost := option{tycoon_manager_buyplates_entry_resource{ResourceId := "meds", Amount := 120000, CheckDontTake := false}}
              Generates := option{tycoon_manager_buyplates_entry_resource{ResourceId := "meds", Amount := 88, CheckDontTake := false}}
              Unlocks := array{
                tycoon_manager_unlock_entry{TargetRoutingId := "bp_610", Payload := false}
              }
              m_StorageIndex := 169
            }
            "bp_610" => tycoon_manager_buyplates_entry{
              BuyplateId := "bp_610"
              Cost := option{tycoon_manager_buyplates_entry_resource{ResourceId := "meds", Amount := 135000, CheckDontTake := false}}
              Generates := option{tycoon_manager_buyplates_entry_resource{ResourceId := "meds", Amount := 94, CheckDontTake := false}}
              Unlocks := array{
                tycoon_manager_unlock_entry{TargetRoutingId := "bp_611", Payload := false}
              }
              m_StorageIndex := 170
            }
            "bp_611" => tycoon_manager_buyplates_entry{
              BuyplateId := "bp_611"
              Cost := option{tycoon_manager_buyplates_entry_resource{ResourceId := "meds", Amount := 150000, CheckDontTake := false}}
              Generates := option{tycoon_manager_buyplates_entry_resource{ResourceId := "meds", Amount := 100, CheckDontTake := false}}
              Unlocks := array{
                tycoon_manager_unlock_entry{TargetRoutingId := "bp_612", Payload := false}
              }
              m_StorageIndex := 171
            }
            "bp_612" => tycoon_manager_buyplates_entry{
              BuyplateId := "bp_612"
              Cost := option{tycoon_manager_buyplates_entry_resource{ResourceId := "meds", Amount := 167000, CheckDontTake := false}}
              Generates := option{tycoon_manager_buyplates_entry_resource{ResourceId := "meds", Amount := 107, CheckDontTake := false}}
              Unlocks := array{
                tycoon_manager_unlock_entry{TargetRoutingId := "bp_613", Payload := false}
              }
              m_StorageIndex := 172
            }
            "bp_613" => tycoon_manager_buyplates_entry{
              BuyplateId := "bp_613"
              Cost := option{tycoon_manager_buyplates_entry_resource{ResourceId := "meds", Amount := 185000, CheckDontTake := false}}
              Generates := option{tycoon_manager_buyplates_entry_resource{ResourceId := "meds", Amount := 114, CheckDontTake := false}}
              Unlocks := array{
                tycoon_manager_unlock_entry{TargetRoutingId := "bp_614", Payload := false}
              }
              m_StorageIndex := 173
            }
            "bp_614" => tycoon_manager_buyplates_entry{
              BuyplateId := "bp_614"
              Cost := option{tycoon_manager_buyplates_entry_resource{ResourceId := "meds", Amount := 222000, CheckDontTake := false}}
              Generates := option{tycoon_manager_buyplates_entry_resource{ResourceId := "meds", Amount := 121, CheckDontTake := false}}
              Unlocks := array{
                tycoon_manager_unlock_entry{TargetRoutingId := "bp_615", Payload := false}
              }
              m_StorageIndex := 174
            }
            "bp_615" => tycoon_manager_buyplates_entry{
              BuyplateId := "bp_615"
              Cost := option{tycoon_manager_buyplates_entry_resource{ResourceId := "meds", Amount := 244000, CheckDontTake := false}}
              Generates := option{tycoon_manager_buyplates_entry_resource{ResourceId := "meds", Amount := 129, CheckDontTake := false}}
              Unlocks := array{
                tycoon_manager_unlock_entry{TargetRoutingId := "bp_616", Payload := false}
              }
              m_StorageIndex := 175
            }
            "bp_616" => tycoon_manager_buyplates_entry{
              BuyplateId := "bp_616"
              Cost := option{tycoon_manager_buyplates_entry_resource{ResourceId := "meds", Amount := 267000, CheckDontTake := false}}
              Generates := option{tycoon_manager_buyplates_entry_resource{ResourceId := "meds", Amount := 138, CheckDontTake := false}}
              Unlocks := array{
                tycoon_manager_unlock_entry{TargetRoutingId := "bp_617", Payload := false}
              }
              m_StorageIndex := 176
            }
            "bp_617" => tycoon_manager_buyplates_entry{
              BuyplateId := "bp_617"
              Cost := option{tycoon_manager_buyplates_entry_resource{ResourceId := "meds", Amount := 292000, CheckDontTake := false}}
              Generates := option{tycoon_manager_buyplates_entry_resource{ResourceId := "meds", Amount := 145, CheckDontTake := false}}
              Unlocks := array{
                tycoon_manager_unlock_entry{TargetRoutingId := "bp_618", Payload := false}
              }
              m_StorageIndex := 177
            }
            "bp_618" => tycoon_manager_buyplates_entry{
              BuyplateId := "bp_618"
              Cost := option{tycoon_manager_buyplates_entry_resource{ResourceId := "meds", Amount := 318000, CheckDontTake := false}}
              Generates := option{tycoon_manager_buyplates_entry_resource{ResourceId := "meds", Amount := 155, CheckDontTake := false}}
              Unlocks := array{
                tycoon_manager_unlock_entry{TargetRoutingId := "bp_619", Payload := false}
              }
              m_StorageIndex := 178
            }
            "bp_619" => tycoon_manager_buyplates_entry{
              BuyplateId := "bp_619"
              Cost := option{tycoon_manager_buyplates_entry_resource{ResourceId := "meds", Amount := 346000, CheckDontTake := false}}
              Generates := option{tycoon_manager_buyplates_entry_resource{ResourceId := "meds", Amount := 165, CheckDontTake := false}}
              Unlocks := array{
                tycoon_manager_unlock_entry{TargetRoutingId := "bp_620", Payload := false}
              }
              m_StorageIndex := 179
            }
            "bp_620" => tycoon_manager_buyplates_entry{
              BuyplateId := "bp_620"
              Cost := option{tycoon_manager_buyplates_entry_resource{ResourceId := "meds", Amount := 375000, CheckDontTake := false}}
              Generates := option{tycoon_manager_buyplates_entry_resource{ResourceId := "meds", Amount := 176, CheckDontTake := false}}
              Unlocks := array{
                tycoon_manager_unlock_entry{TargetRoutingId := "bp_621", Payload := false}
              }
              m_StorageIndex := 180
            }
            "bp_621" => tycoon_manager_buyplates_entry{
              BuyplateId := "bp_621"
              Cost := option{tycoon_manager_buyplates_entry_resource{ResourceId := "meds", Amount := 407000, CheckDontTake := false}}
              Generates := option{tycoon_manager_buyplates_entry_resource{ResourceId := "meds", Amount := 188, CheckDontTake := false}}
              Unlocks := array{
                tycoon_manager_unlock_entry{TargetRoutingId := "bp_622", Payload := false}
              }
              m_StorageIndex := 181
            }
            "bp_622" => tycoon_manager_buyplates_entry{
              BuyplateId := "bp_622"
              Cost := option{tycoon_manager_buyplates_entry_resource{ResourceId := "meds", Amount := 441000, CheckDontTake := false}}
              Generates := option{tycoon_manager_buyplates_entry_resource{ResourceId := "meds", Amount := 205, CheckDontTake := false}}
              Unlocks := array{
                tycoon_manager_unlock_entry{TargetRoutingId := "bp_623", Payload := false}
              }
              m_StorageIndex := 182
            }
            "bp_623" => tycoon_manager_buyplates_entry{
              BuyplateId := "bp_623"
              Cost := option{tycoon_manager_buyplates_entry_resource{ResourceId := "meds", Amount := 478000, CheckDontTake := false}}
              Generates := option{tycoon_manager_buyplates_entry_resource{ResourceId := "meds", Amount := 219, CheckDontTake := false}}
              Unlocks := array{
                tycoon_manager_unlock_entry{TargetRoutingId := "bp_624", Payload := false}
              }
              m_StorageIndex := 183
            }
            "bp_624" => tycoon_manager_buyplates_entry{
              BuyplateId := "bp_624"
              Cost := option{tycoon_manager_buyplates_entry_resource{ResourceId := "meds", Amount := 517000, CheckDontTake := false}}
              Generates := option{tycoon_manager_buyplates_entry_resource{ResourceId := "meds", Amount := 234, CheckDontTake := false}}
              Unlocks := array{
                tycoon_manager_unlock_entry{TargetRoutingId := "bp_625", Payload := false}
              }
              m_StorageIndex := 184
            }
            "bp_625" => tycoon_manager_buyplates_entry{
              BuyplateId := "bp_625"
              Cost := option{tycoon_manager_buyplates_entry_resource{ResourceId := "meds", Amount := 559000, CheckDontTake := false}}
              Generates := option{tycoon_manager_buyplates_entry_resource{ResourceId := "meds", Amount := 250, CheckDontTake := false}}
              Unlocks := array{
                tycoon_manager_unlock_entry{TargetRoutingId := "bp_626", Payload := false}
              }
              m_StorageIndex := 185
            }
            "bp_626" => tycoon_manager_buyplates_entry{
              BuyplateId := "bp_626"
              Cost := option{tycoon_manager_buyplates_entry_resource{ResourceId := "meds", Amount := 604000, CheckDontTake := false}}
              Generates := option{tycoon_manager_buyplates_entry_resource{ResourceId := "meds", Amount := 270, CheckDontTake := false}}
              Unlocks := array{
                tycoon_manager_unlock_entry{TargetRoutingId := "bp_627", Payload := false}
              }
              m_StorageIndex := 186
            }
            "bp_627" => tycoon_manager_buyplates_entry{
              BuyplateId := "bp_627"
              Cost := option{tycoon_manager_buyplates_entry_resource{ResourceId := "meds", Amount := 653000, CheckDontTake := false}}
              Generates := option{tycoon_manager_buyplates_entry_resource{ResourceId := "meds", Amount := 297, CheckDontTake := false}}
              Unlocks := array{
                tycoon_manager_unlock_entry{TargetRoutingId := "bp_628", Payload := false}
              }
              m_StorageIndex := 187
            }
            "bp_628" => tycoon_manager_buyplates_entry{
              BuyplateId := "bp_628"
              Cost := option{tycoon_manager_buyplates_entry_resource{ResourceId := "meds", Amount := 707000, CheckDontTake := false}}
              Generates := option{tycoon_manager_buyplates_entry_resource{ResourceId := "meds", Amount := 321, CheckDontTake := false}}
              Unlocks := array{
                tycoon_manager_unlock_entry{TargetRoutingId := "bp_629", Payload := false}
              }
              m_StorageIndex := 188
            }
            "bp_629" => tycoon_manager_buyplates_entry{
              BuyplateId := "bp_629"
              Cost := option{tycoon_manager_buyplates_entry_resource{ResourceId := "meds", Amount := 764000, CheckDontTake := false}}
              Generates := option{tycoon_manager_buyplates_entry_resource{ResourceId := "meds", Amount := 354, CheckDontTake := false}}
              Unlocks := array{
                tycoon_manager_unlock_entry{TargetRoutingId := "bp_630", Payload := false}
              }
              m_StorageIndex := 189
            }
            "bp_630" => tycoon_manager_buyplates_entry{
              BuyplateId := "bp_630"
              Cost := option{tycoon_manager_buyplates_entry_resource{ResourceId := "meds", Amount := 828000, CheckDontTake := false}}
              Generates := option{tycoon_manager_buyplates_entry_resource{ResourceId := "meds", Amount := 400, CheckDontTake := false}}
              Unlocks := array{
                tycoon_manager_unlock_entry{TargetRoutingId := "bp_631", Payload := false}
              }
              m_StorageIndex := 190
            }
            "bp_631" => tycoon_manager_buyplates_entry{
              BuyplateId := "bp_631"
              Cost := option{tycoon_manager_buyplates_entry_resource{ResourceId := "meds", Amount := 1000000, CheckDontTake := false}}
              Generates := false
              Unlocks := array{
              }
              m_StorageIndex := 191
            }
          }

        for (Index->ProgressLine:p_ProgressLines) {
            set ProgressLine.m_Index = Index
            if (AccessControl := ProgressLine.AccessControl?, AccessControl.p_Mode = access_control_mode.Whitelist) {
                JoinClosure := tycoon_manager_player_closure{
                    Device := Self
                    ProgressLine := ProgressLine
                }
                # since access control is in whitelist mode, and these events are backed by ordered arrays,
                # we can make the assumption that OnPlayerAdded is processed before persistence load
                AccessControl.SubscribeToPlayerJoinEvents(JoinClosure.OnPlayerAdded, JoinClosure.OnPlayerRemoved)

                if (ClaimBuyplate := ProgressLine.ClaimBuyplate?) {
                    Middleware := tycoon_manager_claim_buyplate_middleware{
                        Device := Self
                        ProgressLine := ProgressLine
                    }
                    ClaimBuyplate.AddMiddleware(Middleware)
                }
                if (p_BuyPlateCostMetTrackingRefreshRate > 0.0) {
                    spawn:
                        StartCostMetTrackingLoop(ProgressLine, AccessControl)
                    Closure := tycoon_manager_buyplate_cost_met_reflection_closure{
                        Device := Self
                        ProgressLine := ProgressLine
                    }
                    p_ResourceManager.SubscribeToResourceTake(Closure.OnResourceTake)
                }

                if (p_EnablePersistence?) {
                    PersistenceRecorder := persistence_recorder_device{
                        p_DevId := "progress line {ProgressLine.Name}"
                        p_PlayerTrackingMode := persistence_recorder_player_tracking_mode.AccessControlWhitelist
                        p_AccessControl := AccessControl
                        p_StorageSaveSlots := p_PersistenceSaveSlots
                        p_SavePoints := p_PersistenceSavePoints
                        p_AutoSaveInterval := p_PersistenceAutoSaveInternal
                        p_UseDefaultSelectMenu := p_PersistenceSelectMenuEnabled
                        p_DefaultSelectMenuConfig := p_PersistenceSelectMenuConfig
                        p_UseDialogSelectMenu := p_PersistenceUseDialogSelectMenu
                        p_SkipSelectMenuOnFirstPlay := true
                        p_CustomBuyPlatesReserveBits := p_PersistenceBuyPlatesReserveBits
                        p_BuyPlateRecoveryChunkSize := p_PersistenceBuyPlateRecoveryChunkSize
                        p_BuyPlateRecoveryChunkDelay := p_PersistenceBuyPlateRecoveryChunkDelay
                    }
                    if (m_TycoonConfig.Rebirth?) {
                        RebirthDummyDevice := tycoon_manager_rebirth_dummy_device{
                            TycoonManager := Self
                            ProgressLine := ProgressLine
                        }
                        Handle := PersistenceRecorder.RegisterRecordable(RebirthDummyDevice)
                        set ProgressLine.m_RebirthRecordableHandle = Some(Handle)
                    }
                    if (p_PersistenceIncludeResourceManager?) {
                        PersistenceRecorder.RegisterDeviceToRecord(p_ResourceManager)
                    }
                    for (TimeRewardsDevice:p_TimeRewardsDevice) {
                        PersistenceRecorder.RegisterDeviceToRecord(TimeRewardsDevice)
                    }
                    for (Device:p_PersistenceExtraDevices, persistence_recordable_base[Device]) {
                        PersistenceRecorder.RegisterDeviceToRecord(Device)
                    }
                    spawn:
                        PersistenceRecorder.OnBegin()
                    set ProgressLine.m_PersistenceRecorder = Some(PersistenceRecorder)

                    Closure := tycoon_manager_player_persistence_load_closure{
                        Callback := OnPlayerSaveSlotSelected
                        ProgressLine := ProgressLine
                    }
                    SubscribeToEvent(PersistenceRecorder.PlayerSaveSlotSelectedEvent, Closure.Call)
                }

                DeviceRegisterClosure := tycoon_manager_device_register_closure{
                    Device := Self
                    ProgressLine := ProgressLine
                }
                Subscribed := AccessControl.SubscribeToDevicesBackrefs(DeviceRegisterClosure.Call)
                if (Subscribed?) {
                    if (ClaimBuyplate := ProgressLine.ClaimBuyplate?) {
                        OnDeviceRegistered(ProgressLine, ClaimBuyplate)
                    }
                    for (BuyPlate:ProgressLine.ExtraBuyplates) {
                        OnDeviceRegistered(ProgressLine, BuyPlate)
                    }
                } else {
                    Log("Could not subscribe to access control device of progress line \"{ProgressLine.Name}\" because backrefs building is not enabled", ?Level := log_level.Error)
                }

                ResetBeaconLines(ProgressLine)
            } else if (ProgressLine.AccessControl?) {
                Log("Progress line \"{ProgressLine.Name}\" has an access control, but its mode is not Whitelist", ?Level := log_level.Error)
            } else {
                Log("Progress line \"{ProgressLine.Name}\" is missing an access control device", ?Level := log_level.Error)
            }
        }

        GetPlayspace().PlayerRemovedEvent().Subscribe(OnPlayerQuit)
        for (BeaconLineHideArea:p_BeaconLinesHideAreas) {
            Closure := tycoon_manager_beacon_line_area_closure{
                Device := Self
                Config := BeaconLineHideArea
            }

            BeaconLineHideArea.Volume.AgentEntersEvent.Subscribe(Closure.OnAgentEntered)
            OuterVolume := BeaconLineHideArea.OuterVolume? or BeaconLineHideArea.Volume
            OuterVolume.AgentExitsEvent.Subscribe(Closure.OnAgentExited)
        }

        if (p_EnablePersistence?) {
            for (Index->Device:p_PersistenceExtraDevices, not persistence_recordable_base[Device]) {
                Log("Tycoon manager is configured with an extra persistence device that does not support persistence (index {Index})", ?Level := log_level.Error)
            }
        }

        if (RebirthConfig := m_TycoonConfig.Rebirth?) {
            for (Trigger:p_RebirthNextLevelOnTrigger) {
                Trigger.TriggeredEvent.Subscribe(OnRebirthNextLevelTriggerCalled)
            }

            HudManager := tycoon_manager_rebirth_hud_manager{
                RefreshRate := p_RebirthHUDManagerConfig.RefreshRate
                ZOrder := p_RebirthHUDManagerConfig.ZOrder
                Device := Self
                Config := RebirthConfig
            }
            set m_RebirthHudManager = Some(HudManager)
            spawn:
                HudManager.Run(GetPlayspace())
        }

        for (
            PetRebirthConfig:p_RebirthPets
            set m_RebirthPets[PetRebirthConfig.PetId] = PetRebirthConfig.RebirthLevel
        ) {}
        for (PetController:p_PetController) {
            for (PetId->__:m_RebirthPets) {
                PetController.SetPetConditionProvider(PetId, Self)
            }
        }

        AwareSubscribeToPostInit(OnPostInit)

    OnPostInit<private>():void=
        set m_Initialized = true
        for (ProgressLine:p_ProgressLines) {
            for (BeaconLine:ProgressLine.BeaconLines) {
                set BeaconLine.m_CurrentEntry = 0
                set BeaconLine.m_LockedTargetBuyPlate = None()
                UpdateBeaconLine(ProgressLine, BeaconLine)
            }
        }

    OnPlayerQuit<private>(Player:player):void=
        for (BeaconLineArea:p_BeaconLinesHideAreas) {
            set BeaconLineArea.m_InVolume = BeaconLineArea.m_InVolume.RemoveByKey(Player)(0)
        }
        for (ProgressLine:p_ProgressLines) {
            for (BeaconLine:ProgressLine.BeaconLines) {
                set BeaconLine.m_MapIndicatorPulseToAgents = BeaconLine.m_MapIndicatorPulseToAgents.RemoveByKey(Player)(0)
            }
        }
        if (PlayerState := m_Players[Player]) {
            OnPlayerRemoved(PlayerState.ProgressLine, Player)
            set m_Players = m_Players.RemoveByKey(Player)(0)
        }

    OnPlayerAdded<internal>(ProgressLine:tycoon_manager_progress_line, Player:player):void=
        PlayerState := tycoon_manager_player_state{
            ProgressLine := ProgressLine
        }
        if. set m_Players[Player] = PlayerState
        if (ProgressLine.ClassSlot > 0) {
            p_ClassManager.SetClassSlot(Player, ProgressLine.ClassSlot)
        }
        SyncPlayerBeaconLines(Player)
        if (AccessControl := ProgressLine.AccessControl?) {
            UpdateCostMetTracking(ProgressLine, AccessControl)
        }
        PlayerJoinProgressLineEvent.Signal(Player, ProgressLine)

    OnPlayerRemoved<internal>(ProgressLine:tycoon_manager_progress_line, Player:player):void=
        SyncPlayerBeaconLines(Player)
        DoQuitRollback(ProgressLine, Player)
        if (ProgressLine.ClassSlot > 0) {
            p_ClassManager.ResetClass(Player)
        }
        if (PlayerState := m_Players[Player], PlayerState.ProgressLine.m_Index = ProgressLine.m_Index) {
            set m_Players = m_Players.RemoveByKey(Player)(0)
        }

    SubscribeToRebirthEvent<public>(Listener:type{__(:tycoon_rebirth_event):void}):void=
        set m_RebirthListeners += array{Listener}

    IsPlaying<public>(Player:player)<transacts><decides>:void=
        m_Players[Player]

    ForceExpelPlayer<public>(Player:player):logic=
        if (
            PlayerState := m_Players[Player]
            AccessControl := PlayerState.ProgressLine.AccessControl?
            AccessControl.IsAllowed[Player]
        ) {
            AccessControl.RemoveFromWhitelist(Player, false)
            true
        } else {
            false
        }

    TeleportHome<public>(Player:player):logic=
        if (ProgressLine := GetPlayerState[Player].ProgressLine, Teleporter := ProgressLine.FrontTeleporter?) {
            Teleporter.Teleport(Player)
            true
        } else if (TeleportHome0[Player]) {
            true
        } else {
            false
        }

    # Attempts to teleport the player to the front checkpoint of their progress line.
    # Fails if the player is not part of a progress line, if no front checkpoint device was linked, or if .TeleportTo[] fails.
    TeleportHome0<private>(Player:player)<decides><transacts>:void=
        ProgressLine := GetPlayerState[Player].ProgressLine
        Checkpoint := ProgressLine.FrontCheckpoint?
        FortCharacter := Player.GetFortCharacter[]
        TargetTransform := Checkpoint.GetTransform()
        TargetPosition := TargetTransform.Translation.AddZ(p_MiscPlayerHeight)
        FortCharacter.StallingTeleportTo[TargetPosition, TargetTransform.Rotation]

    ClaimOrTeleportHome<public>(Player:player):tycoon_claim_or_teleport_result=
        if (IsPlaying[Player]) {
            TPSuccess := TeleportHome(Player)
            if (TPSuccess?) {
                return tycoon_claim_or_teleport_result.Teleported
            } else {
                Log("Couldn't teleport '{Player}' to their progress line, check editor logs for reason", ?Level := log_level.Error)
            }
        } else {
            var InUseProgressLines: [int]void = map{}
            for (__->PlayerState:m_Players) {
                if. set InUseProgressLines[PlayerState.ProgressLine.m_Index] = void
            }
            for (ProgressLine:p_ProgressLines, not InUseProgressLines[ProgressLine.m_Index]) {
                if (ClaimBuyplate := ProgressLine.ClaimBuyplate?) {
                    ClaimBuyplate.Unlock(Player, ?UserTriggered := false, ?FirstUnlock := false, ?Force := true)
                } else {
                    ClaimProgressLineFromBuyplate(ProgressLine, Player)
                }
                TPSuccess := TeleportHome(Player)
                if (not TPSuccess?) {
                    Log("Couldn't teleport '{Player}' to their progress line after claiming, check editor logs for reason", ?Level := log_level.Error)
                }
                return tycoon_claim_or_teleport_result.Claimed
            }
            Log("Couldn't auto claim a progress line for '{Player}': all progress lines are in use", ?Level := log_level.Error)
        }
        tycoon_claim_or_teleport_result.Failed

    ClaimProgressLineFromBuyplate<internal>(ProgressLine:tycoon_manager_progress_line, Player:agent):void=
        if (AccessControl := ProgressLine.AccessControl?) {
            AccessControl.AddToWhitelist(Player)
        } else {
            Log("Couldn't claim progress line #{ProgressLine.m_Index} for '{Player}': missing access control", ?Level := log_level.Error)
        }

    DoQuitRollback<private>(ProgressLine:tycoon_manager_progress_line, Player:player):void=
        if (not (p_EnableQuitRollback?) or ProgressLine.AccessControl?.ListAllowedAgents().Length > 0) {
            return
        }
        set ProgressLine.m_RebirthLevel = 0
        RollbackAgent: ?agent = case(p_QuitRollbackGranularity):
            quit_rollback_device_granularity.Player => Some(Player)
            quit_rollback_device_granularity.Omitted => false
        DoRollback(ProgressLine, RollbackAgent)
        if (ClaimBuyplate := ProgressLine.ClaimBuyplate?) {
            ClaimBuyplate.Reset(RollbackAgent)
        }
        EnsureRebirthEffects(ProgressLine)

    DoRollback<private>(ProgressLine:tycoon_manager_progress_line, RollbackAgent:?agent):void=
        for (Ticket:ProgressLine.m_AutoGenerationTickets) {
            Ticket.Disable()
        }
        var RollbackDeviceCount: int = 0
        if (DeviceBackrefs := ProgressLine.AccessControl?.GetUsingDevices[]) {
            for (Device:DeviceBackrefs, Resetable := resetable[Device]) {
                Resetable.Reset(RollbackAgent)
                set RollbackDeviceCount += 1
            }
        }
        for (BuyPlateDevice:ProgressLine.ExtraBuyplates, BuyPlate := buy_plate_device[BuyPlateDevice]) {
            BuyPlate.Reset(RollbackAgent)
            set RollbackDeviceCount += 1
        }
        for (Device:ProgressLine.ExtraRollbackDevices, Resetable := resetable[Device]) {
            Resetable.Reset(RollbackAgent)
            set RollbackDeviceCount += 1
        }
        ResetBeaconLines(ProgressLine)
        if (l_Agent := RollbackAgent?) {
            for (Trigger:ProgressLine.ExtraRollbackTriggers) {
                Trigger.Trigger(l_Agent)
            }
        } else {
            for (Trigger:ProgressLine.ExtraRollbackTriggers) {
                Trigger.Trigger()
            }
        }
        Log("Rolled back progress line \"{ProgressLine.Name}\", {RollbackDeviceCount} affected devices", ?Level := log_level.Debug)

    StartCostMetTrackingLoop<private>(
        ProgressLine:tycoon_manager_progress_line,
        AccessControl:access_control_device
    )<suspends>:void=
        loop {
            UpdateCostMetTracking(ProgressLine, AccessControl)
            Sleep(p_BuyPlateCostMetTrackingRefreshRate)
        }

    UpdateCostMetTracking<internal>(
        ProgressLine:tycoon_manager_progress_line,
        AccessControl:access_control_device
    ):void=
        if (Agent := AccessControl.ListAllowedAgents()[0]) {
            var NewStateBy: []string = array{}
            for (Device:AccessControl.GetUsingDevices[], BuyPlate := buy_plate_device[Device]) {
                NewState := BuyPlate.UpdateCostMetTracking(Agent)
                if (NewState?) {
                    set NewStateBy += array{BuyPlate.GetRoutingId()}
                }
            }
            for (
                BeaconLine:ProgressLine.BeaconLines
                BeaconLineConfig := m_TycoonConfig.BeaconLines[BeaconLine.BeaconLineId]
                LineEntry := BeaconLineConfig.InOrderBuyPlates[BeaconLine.m_CurrentEntry]
                LineEntry.ContainsAnyBuyPlateOf[NewStateBy]
            ) {
                UpdateBeaconLine(ProgressLine, BeaconLine)
            }
        }

    ResetBeaconLines<private>(ProgressLine:tycoon_manager_progress_line):void=
        for (BeaconLine:ProgressLine.BeaconLines) {
            set BeaconLine.m_CurrentEntry = 0
            set BeaconLine.m_LockedTargetBuyPlate = None()
            UpdateBeaconLine(ProgressLine, BeaconLine)
        }

    IsBeaconLineEntryComplete<private>(
        ProgressLine:tycoon_manager_progress_line,
        LineEntry:tycoon_manager_beacon_line_config_entry
    )<decides><transacts>:void=
        var AllUnlocked: logic = true
        for (BuyPlateId->__:LineEntry.AvailableOptions, AllUnlocked?) {
            if (BuyPlate := FindBuyPlateById[ProgressLine, BuyPlateId]) {
                if (not BuyPlate.IsUnlocked[]) {
                    set AllUnlocked = false
                }
            } else {
                set AllUnlocked = false
            }
        }
        AllUnlocked?

    BestRankingBeaconLineEntryBuyPlate<private>(
        ProgressLine:tycoon_manager_progress_line,
        LineEntry:tycoon_manager_beacon_line_config_entry
    ):?buy_plate_device=
        Agent := option{ProgressLine.AccessControl?.ListAllowedAgents()[0]}

        var AnyCostMet: logic = false
        EligibleEntries := for (
            BuyPlateId->__:LineEntry.AvailableOptions
            BuyPlate := FindBuyPlateById[ProgressLine, BuyPlateId]
            BuyPlate.IsEnabled[]
            not BuyPlate.IsUnlocked[]
        ) {
            CostMet := if (l_Agent := Agent?) {
                BuyPlate.DoesAgentMeetsCostCached(l_Agent)
            } else {
                false
            }
            if (CostMet?, not AnyCostMet?) {
                set AnyCostMet = true
            }
            (BuyPlate, CostMet)
        }

        if (
            AnyCostMet?
            Player := player[Agent?]
            FortPlayer := Player.GetFortCharacter[]
            AgentPosition := FortPlayer.GetTransform().Translation
        ) {
            var TopRanking: ?tuple(buy_plate_device,float) = false
            for (Entry:EligibleEntries, Entry(1)?) {
                BuyPlate := Entry(0)
                Dist := Distance(AgentPosition, BuyPlate.GetPosition())
                if ((not TopRanking?) or (Dist < TopRanking?(1))) {
                    set TopRanking = Some(Entry(0), Dist)
                }
            }
            option{TopRanking?(0)}
        } else {
            option{EligibleEntries[0](0)}
        }

    UpdateBeaconLine<private>(
        ProgressLine:tycoon_manager_progress_line,
        BeaconLine:tycoon_manager_progress_line_beacon_line
    ):void=
        var NewEntry: logic = false
        loop {
            if (
                BeaconLineConfig := m_TycoonConfig.BeaconLines[BeaconLine.BeaconLineId]
                LineEntry := BeaconLineConfig.InOrderBuyPlates[BeaconLine.m_CurrentEntry]
                IsBeaconLineEntryComplete[ProgressLine, LineEntry]
            ) {
                # handle unordered unlocking
                set BeaconLine.m_CurrentEntry += 1
                set BeaconLine.m_LockedTargetBuyPlate = None()
                set NewEntry = true
            } else {
                break
            }
        }

        var MapIndicatorPointTo: ?tycoon_manager_beacon_line_buy_plate_state = false
        if (
            BeaconLineConfig := m_TycoonConfig.BeaconLines[BeaconLine.BeaconLineId]
            LineEntry := BeaconLineConfig.InOrderBuyPlates[BeaconLine.m_CurrentEntry]
        ) {
            if (NewEntry?) {
                for (BuyPlateId->__:LineEntry.AvailableOptions, not FindBuyPlateById[ProgressLine, BuyPlateId]) {
                    Log("Buy plate \"{BuyPlateId}\" is part of the now active entry in beacon line \"{BeaconLine.BeaconLineId}\" but is not registered", ?Level := log_level.Warning)
                }
            }

            TargetBuyplate := if (
                LockedTargetBuyPlateId := BeaconLine.m_LockedTargetBuyPlate?
                LineEntry.ContainsBuyPlate[LockedTargetBuyPlateId]
                BuyPlate := FindBuyPlateById[ProgressLine, LockedTargetBuyPlateId]
                BuyPlate.IsEnabled[]
                not BuyPlate.IsUnlocked[]
            ) {
                Some(BuyPlate)
            } else {
                TopRanking := BestRankingBeaconLineEntryBuyPlate(ProgressLine, LineEntry)
                TopRankingStr := if (BuyPlate := TopRanking?) {
                    ToString(BuyPlate)
                } else {
                    "None"
                }
                Log(MsgRankedBeaconLineBuyPlates(BeaconLine.BeaconLineId, TopRankingStr), ?Level := log_level.Verbose)
                TopRanking
            }

            if (BuyPlate := TargetBuyplate?) {
                UnmetCost := if (Agent := ProgressLine.AccessControl?.ListAllowedAgents()[0]) {
                    InvertLogic(BuyPlate.DoesAgentMeetsCostCached(Agent))
                } else {
                    false
                }
                if (not UnmetCost?) {
                    set BeaconLine.m_LockedTargetBuyPlate = Some(BuyPlate.GetRoutingId())
                }

                if (
                    BuyPlate = BeaconLine.m_MapIndicatorState?.BuyPlate
                    UnmetCost = BeaconLine.m_MapIndicatorState?.UnmetCost
                ) {
                    return
                }
                set MapIndicatorPointTo = Some(tycoon_manager_beacon_line_buy_plate_state{
                    BuyPlate := BuyPlate
                    UnmetCost := UnmetCost
                })
            }
        }

        if (
            BuyPlateState := MapIndicatorPointTo?
            Position := BuyPlateState.BuyPlate.GetPosition(?ExtraOffset := p_BeaconLinesMapIndicatorOffset)
            Transform := transform{
                Translation := Position
                Rotation := IdentityRotation()
                Scale := MakeUniformVec3(1.0)
            }
            MapIndicatorVariant := if (BuyPlateState.UnmetCost?, BeaconLine.MapIndicatorUnmetCost?) {
                tycoon_manager_beacon_line_map_indicator_variant.UnmetCost
            } else {
                tycoon_manager_beacon_line_map_indicator_variant.Primary
            }
            MapIndicator := case(MapIndicatorVariant):
                tycoon_manager_beacon_line_map_indicator_variant.Primary => BeaconLine.MapIndicator
                tycoon_manager_beacon_line_map_indicator_variant.UnmetCost => BeaconLine.MapIndicatorUnmetCost?
            MapIndicator.TeleportTo[Transform]
        ) {
            set BeaconLine.m_MapIndicatorState = Some(BuyPlateState)

            if (PrevVariant := BeaconLine.m_MapIndicatorEnabled?, PrevVariant <> MapIndicatorVariant) {
                BeaconLine.DisablePulse(ProgressLine)
                if (PrevMapIndicator := BeaconLine.GetActiveMapIndicator[]) {
                    PrevMapIndicator.Disable()
                }
                set BeaconLine.m_MapIndicatorEnabled = false
            }
            if (not BeaconLine.m_MapIndicatorEnabled?) {
                set BeaconLine.m_MapIndicatorEnabled = Some(MapIndicatorVariant)
                MapIndicator.Enable()
                Log(MsgBeaconLineMapIndicatorVariantEnabled(BeaconLine.BeaconLineId, ToString(MapIndicatorVariant)),
                    ?Level := log_level.Verbose)
            }
            if (not BeaconLine.m_MapIndicatorPulseActive?) {
                set BeaconLine.m_MapIndicatorPulseActive = true
                for (
                    Agent:ProgressLine.AccessControl?.ListAllowedAgents()
                    not IsWithinBeaconLineHideArea[Agent]
                ) {
                    spawn:
                        ActivateObjectivePulseWithWarmup(Agent, BeaconLine)
                }
            }
        } else {
            set BeaconLine.m_MapIndicatorState = false

            BeaconLine.DisablePulse(ProgressLine)
            if (MapIndicator := BeaconLine.GetActiveMapIndicator[], MapIndicatorVariant := BeaconLine.m_MapIndicatorEnabled?) {
                MapIndicator.Disable()
                set BeaconLine.m_MapIndicatorEnabled = false
                Log(MsgBeaconLineMapIndicatorVariantDisabled(BeaconLine.BeaconLineId, ToString(MapIndicatorVariant)),
                    ?Level := log_level.Verbose)
            }
        }

    MsgRankedBeaconLineBuyPlates<private><localizes>(BeaconLineId:string, BuyPlateStr:string):message=
        "Beacon line \"{BeaconLineId}\" elected {BuyPlateStr} as best ranking buy plate"
    MsgBeaconLineMapIndicatorVariantEnabled<private><localizes>(BeaconLineId:string, VariantStr:string):message=
        "Beacon line \"{BeaconLineId}\" enabled map indicator variant {VariantStr}"
    MsgBeaconLineMapIndicatorVariantDisabled<private><localizes>(BeaconLineId:string, VariantStr:string):message=
        "Beacon line \"{BeaconLineId}\" disabled map indicator"

    IsWithinBeaconLineHideArea<private>(Agent:agent)<transacts><decides>:void=
        var Within: logic = false
        for (BeaconLineArea:p_BeaconLinesHideAreas, BeaconLineArea.m_InVolume[Agent]) {
            set Within = true
        }
        Within?

    SyncPlayerBeaconLines<internal>(Player:player):void=
        if (
            PlayerState := m_Players[Player]
            ProgressLine := PlayerState.ProgressLine
            AccessControl := ProgressLine.AccessControl?
            AccessControl.IsAllowed[Player]
            not IsWithinBeaconLineHideArea[Player]
        ) {
            for (
                BeaconLine:ProgressLine.BeaconLines
                BeaconLine.m_MapIndicatorPulseActive?
                not BeaconLine.m_MapIndicatorPulseToAgents[Player]
            ) {
                spawn:
                    ActivateObjectivePulseWithWarmup(Player, BeaconLine)
            }
            Log(MsgSyncPlayerBeaconLines(Player, "active"), ?Level := log_level.Verbose)
        } else if (
            PlayerState := m_Players[Player]
            ProgressLine := PlayerState.ProgressLine
        ) {
            for (
                BeaconLine:ProgressLine.BeaconLines
                BeaconLine.m_MapIndicatorPulseToAgents[Player]
                MapIndicator := BeaconLine.GetActiveMapIndicator[]
            ) {
                MapIndicator.DeactivateObjectivePulse(Player)
                set BeaconLine.m_MapIndicatorPulseToAgents = BeaconLine.m_MapIndicatorPulseToAgents.RemoveByKey(Player)(0)
            }
            Log(MsgSyncPlayerBeaconLines(Player, "inactive"), ?Level := log_level.Verbose)
        }

    MsgSyncPlayerBeaconLines<private><localizes>(Agent:agent, StateStr:string):message=
        "Syncing beacon lines for '{Agent}'. New State: {StateStr}"

    ActivateObjectivePulseWithWarmup<private>(
        Agent:agent,
        BeaconLine:tycoon_manager_progress_line_beacon_line
    )<suspends>:void=
        if. set BeaconLine.m_MapIndicatorPulseToAgents[Agent] = void
        var It: int = 0
        loop {
            if (
                BeaconLine.m_MapIndicatorPulseActive?
                It < 10
                MapIndicator := BeaconLine.GetActiveMapIndicator[]
            ) {
                MapIndicator.ActivateObjectivePulse(Agent)
                Sleep(0.1)
            } else {
                break
            }
            set It += 1
        }

    FindBuyPlateById<private>(ProgressLine:tycoon_manager_progress_line, BuyPlateId:string)<transacts><decides>:buy_plate_device=
        ProgressLine.m_BuyPlatesBackrefCache[BuyPlateId] or FindBuyPlateById0[ProgressLine, BuyPlateId]

    FindBuyPlateById0<private>(ProgressLine:tycoon_manager_progress_line, BuyPlateId:string)<transacts><decides>:buy_plate_device=
        var Result: ?buy_plate_device = false
        if (DeviceBackrefs := ProgressLine.AccessControl?.GetUsingDevices[]) {
            for (
                Device:DeviceBackrefs,
                not (Result?),
                BuyPlate := buy_plate_device[Device],
                BuyPlate.GetRoutingId() = BuyPlateId
            ) {
                set Result = Some(BuyPlate)
            }
        }
        Result?

    OnDeviceRegistered<internal>(ProgressLine:tycoon_manager_progress_line, Device:creative_device):void=
        if (BuyPlate := buy_plate_device[Device]) {
            BuyPlateId := BuyPlate.GetRoutingId()
            if. set ProgressLine.m_BuyPlatesBackrefCache[BuyPlateId] = BuyPlate
            if (Config := m_BuyplatesConfigs[BuyPlate.GetRoutingId()]) {
                Closure := tycoon_manager_configured_buy_plate_closure{
                    Device := Self
                    ProgressLine := ProgressLine
                    Config := Config
                }
                if. BuyPlate.SubscribeToEvent["unlock", Closure.OnUnlock]
                if. BuyPlate.SubscribeToEvent["enable", Closure.OnEnable]

                if (Cost := Config.Cost?) {
                    CostOverride := tycoon_manager_buy_plate_cost_override{
                        Device := Self
                        Cost := Cost
                    }
                    BuyPlate.SetCostOverride(CostOverride)
                } else if (BuyPlate.p_ExpectsCostOverride? or p_AlwaysSetCostOverride?) {
                    BuyPlate.SetCostOverride(buy_plate_free_cost_override{})
                }
            } else {
                Closure := tycoon_manager_any_buy_plate_closure{
                    Device := Self
                    ProgressLine := ProgressLine
                    BuyPlateId := BuyPlate.GetRoutingId()
                }
                if. BuyPlate.SubscribeToEvent["unlock", Closure.OnUnlock]
                if. BuyPlate.SubscribeToEvent["enable", Closure.OnEnable]
            }
            if (m_Initialized?) {
                for (
                    BeaconLineId->BeaconLineConfig:m_TycoonConfig.BeaconLines
                    LineEntry := BeaconLineConfig.InOrderBuyPlates[0]
                    LineEntry.ContainsBuyPlate[BuyPlateId]
                ) {
                    for (BeaconLine:ProgressLine.BeaconLines, BeaconLine.BeaconLineId = BeaconLineId) {
                        UpdateBeaconLine(ProgressLine, BeaconLine)
                    }
                }
            }
        }
        if (PersistenceRecorder := ProgressLine.m_PersistenceRecorder?) {
            if (BuyPlate := buy_plate_device[Device], BuyPlateId := BuyPlate.GetRoutingId()) {
                FixedIndex := if (Index := m_BuyplatesConfigs[BuyPlateId].m_StorageIndex) {
                    Index
                } else {
                    Log("Tycoon manager is to persist buyplate \"{BuyPlateId}\" but its ordering cannot be ensured as it is not present in the config", ?Level := log_level.Warning)
                    -1
                }
                PersistenceRecorder.RegisterBuyPlateToRecord(BuyPlate, ?FixedIndex := FixedIndex)
            } else {
                PersistenceRecorder.RegisterDeviceToRecord(Device, ?Silent := true)
            }
        }

    OnAnyBuyPlateEnabled<internal>(ProgressLine:tycoon_manager_progress_line, BuyplateId:string):void=
        for (
            BeaconLine:ProgressLine.BeaconLines
            BeaconLineConfig := m_TycoonConfig.BeaconLines[BeaconLine.BeaconLineId]
            LineEntry := BeaconLineConfig.InOrderBuyPlates[BeaconLine.m_CurrentEntry]
            LineEntry.ContainsBuyPlate[BuyplateId]
        ) {
            UpdateBeaconLine(ProgressLine, BeaconLine)
        }
        if (AccessControl := ProgressLine.AccessControl?) {
            UpdateCostMetTracking(ProgressLine, AccessControl)
        }

    OnAnyBuyPlateUnlocked<internal>(ProgressLine:tycoon_manager_progress_line, BuyplateId:string, Agent:?agent):void=
        for (
            BeaconLine:ProgressLine.BeaconLines
            BeaconLineConfig := m_TycoonConfig.BeaconLines[BeaconLine.BeaconLineId]
        ) {
            if (
                (LineEntry := BeaconLineConfig.InOrderBuyPlates[BeaconLine.m_CurrentEntry],
                 LineEntry.ContainsBuyPlate[BuyplateId])
            ) {
                set BeaconLine.m_LockedTargetBuyPlate = None()
                UpdateBeaconLine(ProgressLine, BeaconLine)
            } else if (
                BeaconLine.m_LockedTargetBuyPlate?
                BeaconLineConfig.ContainsBuyPlate[BuyplateId]
            ) {
                set BeaconLine.m_LockedTargetBuyPlate = None()
                UpdateBeaconLine(ProgressLine, BeaconLine)
            }
        }

    OnConfiguredBuyPlateUnlocked<internal>(ProgressLine:tycoon_manager_progress_line, Config:tycoon_manager_buyplates_entry, Agent:?agent):void=
        if (Generation := Config.Generates?) {
            if (Ticket := ProgressLine.m_AutoGenerationTickets[Generation.ResourceId]) {
                Ticket.ModifyRate(ToFloat(Generation.Amount))
            } else if (
                Ticket := ProgressLine.AutoGenerationManager?.CreateTicket[Generation.ResourceId, false]
                set ProgressLine.m_AutoGenerationTickets[Generation.ResourceId] = Ticket
            ) {
                Ticket.SetRate(ToFloat(Generation.Amount))
            } else {
                Log("Failed to create auto generation ticket for resource \"{Generation.ResourceId}\" for generation of progress line \"{ProgressLine.Name}\". Is auto generation manager set?", ?Level := log_level.Error)
            }
        }
        if (Config.Unlocks.Length > 0, DeviceBackrefs := ProgressLine.AccessControl?.GetUsingDevices[]) {
            var Unlocked: [string]logic = map{}
            for (
                Device:DeviceBackrefs
                DeviceRoutingId := aware_device[Device].DeviceMeta().GetRoutingId()
            ) {
                for (
                    Unlock:Config.Unlocks
                    DeviceRoutingId = Unlock.TargetRoutingId
                    (not Unlocked[DeviceRoutingId]) or (not p_RoutingIdSingleMatch?) or (Unlock.Payload?)
                    set Unlocked[DeviceRoutingId] = true
                ) {
                    TriggerData := Unlock.Payload? or "enable"
                    Device.AttemptCustomTrigger(TriggerData, Agent, ?Context := "progress line \"{ProgressLine.Name}\"")
                }
            }
            for (Unlock:Config.Unlocks, not Unlocked[Unlock.TargetRoutingId]) {
                Log("Failed to process unlock \"{Unlock}\" after \"{Config.BuyplateId}\" unlocked for progress line \"{ProgressLine.Name}\". No matching device found in backrefs.", ?Level := log_level.Error)
            }
        }

    OnPlayerSaveSlotSelected<private>(ProgressLine:tycoon_manager_progress_line, Event:player_save_slot_selected_event):void=
        PlayerSaveSlotSelectedEvent.Signal(ProgressLine, Event)

        EventName := if ((Event.AlreadyPlayed?) and (Event.Reset?)) {
            "claim_reset"
        } else if (Event.AlreadyPlayed?) {
            "claim_continue"
        } else {
            "claim_first"
        }
        p_AnalyticsController.Emit(Event.Player, EventName)

    OnRebirthNextLevelTriggerCalled<private>(Agent:?agent):void=
        Player := if (l_Player := player[Agent?]) {
            l_Player
        } else {
            Log("Tycoon manager rebirth next level OnTrigger does nothing without a player instigator", ?Level := log_level.Warning)
            return
        }
        PromptRebirth(Player)

    PromptRebirth<public>(Player:player):void=
        if (RebirthConfig := m_TycoonConfig.Rebirth?, PlayerState := GetPlayerState[Player]) {
            RebirthLevel := PlayerState.ProgressLine.m_RebirthLevel
            if (RebirthLevel >= RebirthConfig.MaxLevel) {
                Log(MsgRebirthMaxLevelOnTrigger(Player), ?Level := log_level.Debug)
                return
            }
            DialogIndex := if (p_RebirthNextLevelDialogPerLevel?) {
                if (p_RebirthNextLevelDialogPerLevelDefaultLast?) {
                    var Index: int = RebirthLevel
                    loop {
                        if (Index <= 0 or p_RebirthNextLevelDialog[Index]) {
                            break
                        }
                        set Index -= 1
                    }
                    Index
                } else {
                    RebirthLevel
                }
            } else {
                0
            }
            if (DialogDevice := p_RebirthNextLevelDialog[DialogIndex]) {
                StateRes := if (l_StateRes := m_RebirthDialogs[DialogIndex]) {
                    l_StateRes
                } else {
                    InitialTemplate := DialogDevice.GetDescriptionText()
                    Wrapper := dialog_device_wrapper{
                        DialogDevice := DialogDevice
                    }
                    l_StateRes := (InitialTemplate, Wrapper)
                    if (set m_RebirthDialogs[DialogIndex] = l_StateRes) {}
                    l_StateRes
                }
                var Template: string = StateRes(0)
                set Template = Template
                    .ReplaceAll("%CurrentLevel%", ToString(RebirthLevel))
                    .ReplaceAll("%NextLevel%", ToString(RebirthLevel + 1))
                if (NextLevelConfig := RebirthConfig.GetLevelConfig[RebirthLevel + 1]) {
                    NumberFormatter := p_RebirthPlaceholdersMultiplierFormatter.CreateFormatter()
                    for (ResourceId->Multiplier:NextLevelConfig.ResourceMultipliers) {
                        set Template = Template.ReplaceAll("%Multiplier:{ResourceId}%", NumberFormatter.Format(Multiplier))
                    }
                }
                Title := DialogDevice.GetTitleText()
                DialogDevice.SetTitleText(ToMessage(Title)) 
                DialogDevice.SetDescriptionText(ToMessage(Template))
                Closure := tycoon_manager_rebirth_dialog_closure{
                    Device := Self
                    Player := Player
                    ProgressLine := PlayerState.ProgressLine
                    ExpectedCurrentLevel := RebirthLevel
                }
                StateRes(1).Show(Player, Some(Closure.Call))
            } else {
                RebirthPlayer(Player)
            }
        } else {
            Log("Tycoon manager rebirth next level OnTrigger does nothing when rebirth is not configured", ?Level := log_level.Warning)
        }

    GetPlayerState<internal>(Player:player)<transacts><decides>:tycoon_manager_player_state=
        m_Players[Player]

    GetRebirthLevel<public>(Player:player)<transacts><decides>:int=
        m_Players[Player].ProgressLine.m_RebirthLevel

    RebirthPlayer<internal>(Player:player):void=
        if (RebirthConfig := m_TycoonConfig.Rebirth?, PlayerState := m_Players[Player]) {
            ProgressLine := PlayerState.ProgressLine
            if (ProgressLine.m_RebirthLevel >= RebirthConfig.MaxLevel) {
                Log(MsgRebirthMaxLevel(Player), ?Level := log_level.Debug)
                return
            }
            set ProgressLine.m_RebirthLevel += 1
            RebirthLevel := ProgressLine.m_RebirthLevel
            Log(MsgRebirth(Player, RebirthLevel, "manual"), ?Level := log_level.Debug)
            if (PersistenceRecorder := PlayerState.ProgressLine.m_PersistenceRecorder?) {
                # save rebirth level
                if (Handle := PlayerState.ProgressLine.m_RebirthRecordableHandle?) {
                    Handle.Save(Player)
                } else {
                    PersistenceRecorder.SaveAgent(Player)
                }
                # reset persistence state
                PersistenceRecorder.ResetAgent(Player, tycoon_manager_rebirth_persistence_reset_query{
                    SourceDevice := Self
                    ResetFilter := Some(RebirthPersistenceResetFilter)
                })
            }
            DoRollback(PlayerState.ProgressLine, Some(Player))
            if (ClaimBuyplate := ProgressLine.ClaimBuyplate?) {
                ClaimBuyplate.Unlock(Player, ?UserTriggered := false, ?FirstUnlock := false, ?Force := true)
            }
            for (TriggerConfig:p_RebirthResetTriggers, TriggerConfig.IsApplicable[RebirthLevel]) {
                TriggerConfig.Trigger.Trigger(Player)
            }
            for (TriggerConfig:ProgressLine.RebirthResetTriggers, TriggerConfig.IsApplicable[RebirthLevel]) {
                TriggerConfig.Trigger.Trigger(Player)
            }
            if (m_RebirthListeners.Length > 0) {
                Event := tycoon_rebirth_event{
                    Player := Player
                    NewRebirthLevel := RebirthLevel
                    Virtual := false
                }
                for (Listener:m_RebirthListeners) {
                    Listener(Event)
                }
            }
            if (PersistenceRecorder := PlayerState.ProgressLine.m_PersistenceRecorder?) {
                # save persistence in case triggers caused relevant changes
                PersistenceRecorder.SaveAgent(Player)
            }
            EnsureRebirthEffects(PlayerState.ProgressLine)
            if (AccessControl := ProgressLine.AccessControl?) {
                UpdateCostMetTracking(ProgressLine, AccessControl)
            }
        }

    ForceVirtualRebirth<public>(Player:player):void=
        if (RebirthConfig := m_TycoonConfig.Rebirth?, PlayerState := m_Players[Player]) {
            ProgressLine := PlayerState.ProgressLine
            if (ProgressLine.m_RebirthLevel >= RebirthConfig.MaxLevel) {
                Log(MsgRebirthMaxLevel(Player), ?Level := log_level.Debug)
                return
            }
            set ProgressLine.m_RebirthLevel += 1
            RebirthLevel := ProgressLine.m_RebirthLevel
            Log(MsgRebirth(Player, RebirthLevel, "forced virtual"), ?Level := log_level.Debug)
            if (PersistenceRecorder := PlayerState.ProgressLine.m_PersistenceRecorder?) {
                # save rebirth level
                if (Handle := PlayerState.ProgressLine.m_RebirthRecordableHandle?) {
                    Handle.Save(Player)
                } else {
                    PersistenceRecorder.SaveAgent(Player)
                }
            }
            if (m_RebirthListeners.Length > 0) {
                Event := tycoon_rebirth_event{
                    Player := Player
                    NewRebirthLevel := RebirthLevel
                    Virtual := true
                }
                for (Listener:m_RebirthListeners) {
                    Listener(Event)
                }
            }
            EnsureRebirthEffects(PlayerState.ProgressLine)
        }

    RebirthPersistenceResetFilter<private>(Recordable:persistence_recordable_base)<transacts><decides>:void=
        not pet_controller_device[Recordable]

    EnsureRebirthEffects<internal>(ProgressLine:tycoon_manager_progress_line):void=
        if (HudManager := m_RebirthHudManager?) {
            HudManager.Refresh()
        }
        if (RebirthConfig := m_TycoonConfig.Rebirth?, AutoGenerationManager := ProgressLine.AutoGenerationManager?) {
            ResourceMultipliers := if (LevelConfig := RebirthConfig.GetLevelConfig[ProgressLine.m_RebirthLevel]) {
                LevelConfig.ResourceMultipliers
            } else {
                map{}
            }
            for (ResourceId->Multiplier:ResourceMultipliers) {
                AutoGenerationManager.GetResourceMultiplierInstance(ResourceId).SetMultiplier(p_RebirthMultipliersChannel, Multiplier)
                if (set ProgressLine.m_SetResourceMultipliers[ResourceId] = Multiplier) {}
            }
            # remove multipliers that are no longer wanted
            for (ResourceId->__:ProgressLine.m_SetResourceMultipliers, not ResourceMultipliers[ResourceId]) {
                AutoGenerationManager.GetResourceMultiplierInstance(ResourceId).SetMultiplier(p_RebirthMultipliersChannel, None())
            }
        }
        if (Billboard := ProgressLine.RebirthBillboard[0]) {
            if (ProgressLine.m_RebirthLevel > 0 or not (ProgressLine.RebirthBillboardHideOnZero?)) {
                Text := ProgressLine.RebirthBillboardTemplate
                    .ReplaceAll("%CurrentLevel%", ToString(ProgressLine.m_RebirthLevel))
                Billboard.ShowText()
                Billboard.SetText(ToMessage(Text))
            } else {
                Billboard.HideText()
                Billboard.SetText(ToMessage(""))
            }
        }

    MsgRebirth<private><localizes>(Player:player, NewLevel:int, Context:string):message=
        "Rebirthing '{Player}' to level {NewLevel} ({Context})"
    MsgRebirthMaxLevel<private><localizes>(Player:player):message=
        "Ignoring next rebirth level for '{Player}' as they reached max level"
    MsgRebirthMaxLevelOnTrigger<private><localizes>(Player:player):message=
        "Ignoring next rebirth level OnTrigger for '{Player}' as they reached max level"

    IsPetUnlocked<override>(Player:player, PetId:string)<transacts><decides>:void=
        RebirthLevel := m_Players[Player].ProgressLine.m_RebirthLevel
        RebirthLevel >= m_RebirthPets[PetId]

tycoon_manager_beacon_line_area_closure := class<final>:
    Device: tycoon_manager_device
    Config: tycoon_manager_beacon_line_area_config

    OnAgentEntered<internal>(Agent:agent):void=
        if. set Config.m_InVolume[Agent] = void
        if (Player := player[Agent]) {
            Device.SyncPlayerBeaconLines(Player)
        }

    OnAgentExited<internal>(Agent:agent):void=
        set Config.m_InVolume = Config.m_InVolume.RemoveByKey(Agent)(0)
        if (Player := player[Agent]) {
            Device.SyncPlayerBeaconLines(Player)
        }

tycoon_manager_player_closure := class:
    Device: tycoon_manager_device
    ProgressLine: tycoon_manager_progress_line

    OnPlayerAdded(Player:player):void=
        Device.OnPlayerAdded(ProgressLine, Player)

    OnPlayerRemoved(Player:player, GameQuit:logic):void=
        Device.OnPlayerRemoved(ProgressLine, Player)

## Persistence {{{
tycoon_manager_player_persistence_load_closure := class:
    Callback: type{__(:tycoon_manager_progress_line,:player_save_slot_selected_event):void}
    ProgressLine: tycoon_manager_progress_line

    Call(Event:player_save_slot_selected_event):void=
        Callback(ProgressLine, Event)

tycoon_manager_rebirth_dummy_device := class(persistence_recordable, persistence_resetable):
    TycoonManager: tycoon_manager_device
    ProgressLine: tycoon_manager_progress_line

    GetPersistenceControl<override>()<transacts><decides>:persistence_recordable_control=
        None()?

    PersistenceReserve<override>(StorageMem:storage_memory)<transacts><decides>:tycoon_manager_rebirth_persistence_state=
        TycoonManager.p_EnablePersistence?
        tycoon_manager_rebirth_persistence_state{
            RebirthLevelChunk := StorageMem.Reserve[TycoonManager.p_RebirthLevelReserveBits, "tycoon_rebirth_level"]
        }

    PersistenceSave<override>(i_State:persistence_recordable_state, Agent:agent, SaveCause:persistence_save_cause):void=
        if (SaveCause = persistence_save_cause.AutoSave) {
            return
        }
        if (
            State := tycoon_manager_rebirth_persistence_state[i_State]
            PlayerState := TycoonManager.GetPlayerState[player[Agent]]
            PlayerState.ProgressLine.m_Index = ProgressLine.m_Index
        ) {
            State.RebirthLevelChunk.WriteInt(Agent, ProgressLine.m_RebirthLevel)
        }

    PersistenceLoad<override>(i_State:persistence_recordable_state, Agent:agent, Query:persistence_load_query):void=
        if (
            State := tycoon_manager_rebirth_persistence_state[i_State]
            PlayerState := TycoonManager.GetPlayerState[player[Agent]]
            PlayerState.ProgressLine.m_Index = ProgressLine.m_Index
        ) {
            SavedLevel := State.RebirthLevelChunk.ReadInt(Agent)
            PrevRebirthLevel := ProgressLine.m_RebirthLevel
            set ProgressLine.m_RebirthLevel = if (Query.Overwrite?) {
                SavedLevel
            } else {
                Max(ProgressLine.m_RebirthLevel, SavedLevel)
            }
            if (PrevRebirthLevel <> ProgressLine.m_RebirthLevel) {
                TycoonManager.EnsureRebirthEffects(ProgressLine)
            }
        }

    PersistenceReset<override>(i_State:persistence_recordable_state, Agent:agent, Query:persistence_reset_query):void=
        if (tycoon_manager_rebirth_persistence_reset_query[Query]) {
            Log("Skipping reset of rebirth persistence state because it is initialized by a rebirth", ?Level := log_level.Debug)
        } else {
            i_State.Reset(Agent)
            PersistenceLoad(i_State, Agent, persistence_load_query{
                Overwrite := true
                FirstLoad := false
            })
        }

tycoon_manager_rebirth_persistence_reset_query<public> := class(persistence_reset_query):

tycoon_manager_rebirth_persistence_state<public> := class(persistence_recordable_state):
    RebirthLevelChunk<public>: storage_memory_chunk

    Reset<override>(Agent:agent):void=
        RebirthLevelChunk.Reset(Agent)
## }}}

## Buy plate features {{{
tycoon_manager_buyplate_cost_met_reflection_closure := class:
    Device: tycoon_manager_device
    ProgressLine: tycoon_manager_progress_line

    OnResourceTake(Event:resource_take_event):void=
        if (AccessControl := ProgressLine.AccessControl?, AccessControl.IsAllowed[Event.Agent?]) {
            Device.UpdateCostMetTracking(ProgressLine, AccessControl)
        }

tycoon_manager_claim_buyplate_middleware := class(buy_plate_middleware):
    Device: tycoon_manager_device
    ProgressLine: tycoon_manager_progress_line

    MayTryUnlockBuyplate<override>(Agent:agent, BuyPlate:buy_plate_device):logic=
        if (Player := player[Agent], Device.GetPlayerState[Player]) {
            for (HUDMessage:Device.p_ClaimFailHUDMessage) {
                HUDMessage.Show(Agent)
            }
            false
        } else {
            true
        }

    OnUnlockBuyplate<override>(Agent:agent, BuyPlate:buy_plate_device):void=
        if (Player := player[Agent]) {
            Device.ClaimProgressLineFromBuyplate(ProgressLine, Player)
        }

tycoon_manager_buy_plate_cost_override := class(buy_plate_cost_override):
    Device: tycoon_manager_device
    Cost: tycoon_manager_buyplates_entry_resource

    TryUnlock<override>(Agent:agent, Silent:logic):logic=
        if (Cost.Amount <= 0) {
            return true
        }
        Notify := InvertLogic(Silent)
        if (Cost.CheckDontTake?) {
            Amount := Device.p_ResourceManager.GetAmount(Agent, Cost.ResourceId)
            if (Amount >= Cost.Amount) {
                return true
            } else if (Notify?) {
                Device.p_ResourceManager.SendNotEnoughNotification(Agent, Cost.ResourceId)
            }
        } else {
            Success := Device.p_ResourceManager.Take(Some(Agent), Cost.ResourceId, Cost.Amount,
                ?NotifyNotEnough := Notify, ?SendFeedback := Notify)
            return Success
        }
        false

    HasEnough<override>(Agent:agent):logic=
        if (Cost.Amount <= 0) {
            return true
        }
        Amount := Device.p_ResourceManager.GetAmount(Agent, Cost.ResourceId)
        if (Amount >= Cost.Amount) {
            true
        } else {
            false
        }
## }}}

tycoon_manager_device_register_closure := class:
    Device: tycoon_manager_device
    ProgressLine: tycoon_manager_progress_line

    Call(RegisteredDevice:creative_device):void=
        Device.OnDeviceRegistered(ProgressLine, RegisteredDevice)

## Buy plates closures {{{
tycoon_manager_any_buy_plate_closure := class:
    Device: tycoon_manager_device
    ProgressLine: tycoon_manager_progress_line
    BuyPlateId: string

    OnUnlock(Agent:?agent):void=
        Device.OnAnyBuyPlateUnlocked(ProgressLine, BuyPlateId, Agent)

    OnEnable(Agent:?agent):void=
        Device.OnAnyBuyPlateEnabled(ProgressLine, BuyPlateId)

tycoon_manager_configured_buy_plate_closure := class:
    Device: tycoon_manager_device
    ProgressLine: tycoon_manager_progress_line
    Config: tycoon_manager_buyplates_entry

    OnUnlock(Agent:?agent):void=
        Device.OnConfiguredBuyPlateUnlocked(ProgressLine, Config, Agent)
        Device.OnAnyBuyPlateUnlocked(ProgressLine, Config.BuyplateId, Agent)

    OnEnable(Agent:?agent):void=
        Device.OnAnyBuyPlateEnabled(ProgressLine, Config.BuyplateId)
## }}}
## Rebirth UI {{{
tycoon_manager_rebirth_dialog_closure := class:
    Device: tycoon_manager_device
    Player: player

    ProgressLine: tycoon_manager_progress_line
    ExpectedCurrentLevel: int

    Call(Event:dialog_response_event):void=
        Log("Rebirth dialog event {Event}", ?Level := log_level.Debug)
        if (
            Event.ResponseType = dialog_response_type.Clicked
            Event.ButtonIndex? = Device.p_RebirthNextLevelDialogConfirmButtonIndex
            player[Event.Agent] = Player
            PlayerState := Device.GetPlayerState[Player]
            PlayerState.ProgressLine.m_Index = ProgressLine.m_Index
            PlayerState.ProgressLine.m_RebirthLevel = ExpectedCurrentLevel
        ) {
            Device.RebirthPlayer(Player)
        }

tycoon_manager_rebirth_hud_manager := class(hud_manager_base):
    Device: tycoon_manager_device
    Config: tycoon_manager_rebirth_config

    CreateHUD<override>(Player:player):?hud_manager_player=
        Image := CreateDefaultingTextureBlock(Device.p_RebirthHUDImageSize, None())
        Canvas := canvas{
            Slots := array{
                Device.p_RebirthHUDImagePlacement.ToCanvasSlot(Image.Widget())
            }
        }
        Element := tycoon_manager_rebirth_hud_image{
            Widget := Canvas
            Manager := Self
            Player := Player
            Image:= Image
        }
        Some(MakeSingleElementHUDPlayer(Player, Canvas, Element))

tycoon_manager_rebirth_hud_image := class(hud_element):
    Manager: tycoon_manager_rebirth_hud_manager
    Player: player

    Image: defaulting_texture_block

    var m_LastLevel: ?int = false

    Update<override>():void=
        Level := option{Manager.Device.GetRebirthLevel[Player]}
        if (m_LastLevel = Level) {
            return
        }
        set m_LastLevel = Level
        if (
            l_Level := Level?
            TextureName := Manager.Config.GetLevelConfig[l_Level].HudImageTexture?
            Texture := AwareDefaultAssetsProvider[].GetTextureAsset[TextureName]
        ) {
            Show()
            Image.SetTexture(Texture)
            Log("Rebirth HUD image switching to level {l_Level}", ?Level := log_level.Debug)
        } else {
            Hide()
            Log("Rebirth HUD image switching to no level", ?Level := log_level.Debug)
        }
## }}}
