using { /Verse.org/Simulation }
using { /Fortnite.com/Devices }
using { /Fortnite.com/Playspaces }
using { /UnrealEngine.com/Temporary/Diagnostics }

using { HavocVerse.Storage }
using { HavocVerse.Logging }
using { HavocVerse.Util }

AwareNetAwaitHavocIntegration: logic = true

var g_AwareNetInstance<internal>: weak_map(session,aware_net_global_instance) = map{}

AwareNet_DebugStats<public>()<transacts>:aware_debug_stats=
    var Result: [string]int = map{}
    if (AwareNetInstance := g_AwareNetInstance[GetSession()]) {
        set Result = ConcatenateMaps(Result, map{
            "postponeInit" => AwareNetInstance.m_PostponeInit.Length
            "listeners::deviceInit" => AwareNetInstance.m_DeviceInitListeners.Length
            "listeners::postInit" => AwareNetInstance.m_PostInitListeners.Length
        })
        if (AwareNet := AwareNetInstance.m_AwareNet?) {
            set Result = ConcatenateMaps(Result, map{
                "members" => AwareNet.m_Members.Length
                "initQueue" => AwareNet.m_InitQueue.Length
                "sequentialInitQueue" => AwareNet.m_SequentialInitQueue.Length
            })
        }
    }
    aware_debug_stats{
        Length := Result
    }

aware_integration<public> := interface:
    GetSimulationElapsedTimeAtJoin<public>(Player:player)<transacts>:float

    SubscribeToPlayerSpawnEvent<public>(Listener:type{__(:player):void}):void

    GlobalDebugStats<public>()<transacts>:[string]aware_debug_stats

aware_net_global_instance<internal> := class:
    var m_AwareNet: ?aware_net = false

    var m_Integration: ?aware_integration = false
    var m_DefaultAssetsProvider: ?assets_provider = false
    var m_DefaultTagsProvider: ?tags_provider = false

    var m_PostponeInit: []tuple(creative_device, type{__()<suspends>:void}) = array{}

    var m_DeviceInitListeners: []tuple(?aware_device_filter, type{__(:creative_device):void}) = array{}
    var m_PostInitListeners: []type{__():void} = array{}

GetAwareNetInstance<internal>():?aware_net_global_instance=
    if (l_AwareNetInstance := g_AwareNetInstance[GetSession()]) {
        Some(l_AwareNetInstance)
    } else if (
        l_AwareNetInstance := aware_net_global_instance{}
        set g_AwareNetInstance[GetSession()] = l_AwareNetInstance
    ) {
        Some(l_AwareNetInstance)
    } else {
        None()
    }

AwareInitIntegration<public>(
    Config:aware_net_config, Integration:?aware_integration,
    DefaultAssetsProvider:assets_provider, DefaultTagsProvider:tags_provider
)<suspends>:void=
    AwareNetInstanceRes := GetAwareNetInstance()
    if (AwareNetInstance := AwareNetInstanceRes?, not (AwareNetInstance.m_AwareNet?)) {
        AwareNet := aware_net{
            Instance := AwareNetInstance
            Config := Config
        }
        set AwareNetInstance.m_AwareNet = Some(AwareNet)
        set AwareNetInstance.m_Integration = Integration
        set AwareNetInstance.m_DefaultAssetsProvider = Some(DefaultAssetsProvider)
        set AwareNetInstance.m_DefaultTagsProvider = Some(DefaultTagsProvider)
        Postponed := AwareNetInstance.m_PostponeInit
        set AwareNetInstance.m_PostponeInit = array{}
        for (Entry:Postponed) {
            spawn:
                AwareNet.RegisterDevice(Entry(0), Entry(1))
        }
    } else {
        Log("AwareNet could not setup integration for session (attempted more than once?)")
    }

# Initialize a device to AwareNet
AwareInitDevice<public>(Device:creative_device, InitFn:type{__()<suspends>:void})<suspends>:void=
    AwareNetInstanceRes := GetAwareNetInstance()
    if (AwareNetInstance := AwareNetInstanceRes?) {
        if (AwareNet := AwareNetInstance.m_AwareNet?) {
            AwareNet.RegisterDevice(Device, InitFn)
        } else if (
            not (AwareNetAwaitHavocIntegration?)
            AwareNet := aware_net{
                Instance := AwareNetInstance
                Config := aware_net_config{}
            }
            set AwareNetInstance.m_AwareNet = Some(AwareNet)
        ) {
            Log("AwareNet setup with default config as waiting for integration device is not desired", ?Level := log_level.Warning)
            AwareNet.RegisterDevice(Device, InitFn)
        } else {
            if (AwareNetInstance.m_PostponeInit.Length = 0) {
                Log("AwareNet is postponing devices init until integration device initializes")
            }
            set AwareNetInstance.m_PostponeInit += array{(Device, InitFn)}
        }
    } else {
        Log("AwareNet instance could not be setup for session (while registering {Device})", ?Level := log_level.Error)
        InitFn()
        return
    }

# List all creative devices known by AwareNet
AwareListDevices<public>()<transacts>:[]creative_device=
    if (AwareNet := g_AwareNetInstance[GetSession()].m_AwareNet?) {
        for (NetMember:AwareNet.m_Members) {
            NetMember.Device
        }
    } else {
        array{}
    }

AwareGetDebugDevice<public>()<transacts><decides>:debug_device=
    var Device: ?debug_device = None()
    for (AwareDevice:AwareListDevices(), not (Device?), DebugDevice := debug_device[AwareDevice]) {
        set Device = Some(DebugDevice)
    }
    Device?

AwareGetIntegration<public>()<transacts><decides>:aware_integration=
    g_AwareNetInstance[GetSession()].m_Integration?

AwareGetPlayspace<public>()<transacts><decides>:fort_playspace=
    Integration := g_AwareNetInstance[GetSession()].m_Integration?
    creative_device[Integration].GetPlayspace()

AwareDefaultAssetsProvider<public>()<transacts><decides>:assets_provider=
    g_AwareNetInstance[GetSession()].m_DefaultAssetsProvider?

AwareDefaultTagsProvider<public>()<transacts><decides>:tags_provider=
    g_AwareNetInstance[GetSession()].m_DefaultTagsProvider?

aware_device_filter<public> := class<concrete>:
    @editable DeviceName<public>: string = ""

    IsApplicable<public>(Device:creative_device)<transacts><decides>:void=
        if (AwareDevice := aware_device[Device]) {
            DeviceName = AwareDevice.DeviceMeta().DeviceName
        } else {
            false?
        }

    IsApplicable<internal>(NetMember:aware_net_member)<transacts><decides>:void=
        if (l_Meta := NetMember.Meta?) {
            DeviceName = l_Meta.DeviceName
        } else {
            false?
        }

# Subscribe to device initialization events
AwareSubscribeToDeviceInit<public>(Listener:type{__(:creative_device):void}, Filter:?aware_device_filter):void=
    AwareNetInstanceRes := GetAwareNetInstance()
    if (AwareNetInstance := AwareNetInstanceRes?) {
        set AwareNetInstance.m_DeviceInitListeners += array{(Filter, Listener)}
        if (AwareNet := AwareNetInstance.m_AwareNet?) {
            for (NetMember:AwareNet.m_Members, NetMember.m_PostInitialized?) {
                Listener(NetMember.Device)
            }
        }
    }

# Subscribe to overall post init event
AwareSubscribeToPostInit<public>(Listener:type{__():void}):void=
    AwareNetInstanceRes := GetAwareNetInstance()
    if (AwareNetInstance := AwareNetInstanceRes?) {
        if (AwareNet := AwareNetInstance.m_AwareNet?, AwareNet.m_FinishedInit?) {
            Listener()
        } else {
            set AwareNetInstance.m_PostInitListeners += array{Listener}
        }
    }

AwareEnsureDelaySincePlayerJoined<public>(Player:player, Delay:float)<suspends>:void=
    if (Integration := AwareGetIntegration[]) {
        JoinedAt := Integration.GetSimulationElapsedTimeAtJoin(Player)
        WaitFor := JoinedAt + Delay - GetSimulationElapsedTime()
        if (WaitFor > 0.0) {
            Sleep(WaitFor)
        }
    } else {
        Sleep(Delay)
    }

aware_net_config<public> := class<concrete>:
    @editable PostInitDebounceTime<public>: float = 1.0
    @editable ForcePersistenceRecordableSequential<public>: logic = true
    @editable InitDebug<public>: logic = false
    @editable ForceSequentialDebug<public>: logic = false
    @editable ForceSequentialDebugInterval<public>: float = 0.0
    @editable InitFilteringOutDebug<public>: []string = array{}
    @editable EnforceInitOrder<public>: []aware_net_config_init_order_entry = array{
        aware_net_config_init_order_entry{
            DeviceName := "debug_device"
            ExpectedDeviceCount := 1
        }
    }

aware_net_config_init_order_entry<public> := class<concrete>:
    @editable DeviceName: string = ""
    @editable ExpectedDeviceCount: int = 0

aware_net_init_step<internal> := class:
    Index: int
    DeviceName: string
    TotalExpected: int

    var m_ScheduledInit: int = 0
    var m_FinishedInit: int = 0

aware_net<internal> := class:
    Instance: aware_net_global_instance
    Config: aware_net_config

    var m_FinishedInit: logic = false
    var m_LastInitAt: float = 0.0

    var m_Members: []aware_net_member = array{}
    var m_InitQueue: []aware_net_member = array{}

    var m_InitStepNextIndex: int = 0
    var m_InitStep: ?aware_net_init_step = false

    var m_SequentialInitQueue<internal>: []aware_net_member = array{}
    var m_SequentialInitQueueRunning<private>: logic = false

    RegisterDevice<internal>(Device:creative_device, InitFn:type{__()<suspends>:void})<suspends>:void=
        NetMember := aware_net_member{
            AwareNet := Self
            Device := Device
            Meta := if (AwareDevice := aware_device[Device]) {
                Some(AwareDevice.DeviceMeta())
            } else {
                None()
            }
            InitFn := InitFn
        }
        set m_Members += array{NetMember}

        InitStep := GetInitStep()
        if (l_InitStep := InitStep?) {
            if (DeviceName := NetMember.Meta?.DeviceName, DeviceName = l_InitStep.DeviceName) {
                set l_InitStep.m_ScheduledInit += 1
                ScheduleMemberInit(NetMember)
            } else {
                set m_InitQueue += array{NetMember}
            }
        } else {
            if (UseSequentialInit[NetMember]) {
                QueueSequentialDeviceInit(NetMember)
            } else {
                NetMember.Init(false)
            }
        }

    ScheduleMemberInit<private>(NetMember:aware_net_member)<suspends>:void=
        if (UseSequentialInit[NetMember]) {
            QueueSequentialDeviceInit(NetMember)
        } else {
            spawn:
                NetMember.Init(false)
        }

    OnPostMemberInit<internal>(NetMember:aware_net_member)<suspends>:void=
        set NetMember.m_PostInitialized = true
        set m_LastInitAt = GetSimulationElapsedTime()
        for (ListenerEntry:Instance.m_DeviceInitListeners) {
            if (Filter := ListenerEntry(0)?) {
                if (Filter.IsApplicable[NetMember]) {
                    ListenerEntry(1)(NetMember.Device)
                }
            } else {
                ListenerEntry(1)(NetMember.Device)
            }
        }
        InitStep := GetInitStep()
        if (l_InitStep := InitStep?, DeviceName := NetMember.Meta?.DeviceName, DeviceName = l_InitStep.DeviceName) {
            set l_InitStep.m_FinishedInit += 1
            if (l_InitStep.m_ScheduledInit >= l_InitStep.TotalExpected and l_InitStep.m_FinishedInit >= l_InitStep.m_ScheduledInit) {
                set m_InitStep = None()
                LogLevel := if (Config.InitDebug?). log_level.Debug else log_level.Verbose
                Log("AwareNet finished init step {l_InitStep.Index} (expected min. {l_InitStep.TotalExpected} {l_InitStep.DeviceName}, processed {l_InitStep.m_FinishedInit})", ?Level := LogLevel)
                AnyRemaining := NextInitStepCatchup()
                if (not (AnyRemaining?)) {
                    SchedulePostInitDebounce()
                }
            }
        } else {
            SchedulePostInitDebounce()
        }

    SchedulePostInitDebounce<private>():void=
        if (m_FinishedInit?) {
            return
        }
        if (m_InitQueue.Length > 0 or m_SequentialInitQueue.Length > 0) {
            return
        }
        CurrentLastInitAt := m_LastInitAt
        spawn:
            PostInitDebounceTask(CurrentLastInitAt)

    PostInitDebounceTask<private>(PrevLastInitAt:float)<suspends>:void=
        Sleep(Config.PostInitDebounceTime)
        if (m_LastInitAt = PrevLastInitAt, not (m_FinishedInit?)) {
            set m_FinishedInit = true
            LogLevel := if (Config.InitDebug?). log_level.Debug else log_level.Verbose
            Log("AwareNet inferred overall post init", ?Level := LogLevel)
            for (Listener:Instance.m_PostInitListeners) {
                Listener()
            }
            set Instance.m_PostInitListeners = array{}
        }

    NextInitStepCatchup<private>()<suspends>:logic=
        InitStep := GetInitStep()
        if (l_InitStep := InitStep?) {
            AnyRemaining := if (m_InitQueue.Length > 0). true else false
            var FilteredInitQueue: []aware_net_member = array{}
            var ScheduleInitOf: []aware_net_member = array{}
            for (NetMember:m_InitQueue) {
                if (NetMember.Meta?.DeviceName = l_InitStep.DeviceName) {
                    set ScheduleInitOf += array{NetMember}
                } else {
                    set FilteredInitQueue += array{NetMember}
                }
            }
            set m_InitQueue = FilteredInitQueue
            set l_InitStep.m_ScheduledInit += ScheduleInitOf.Length
            for (NetMember:ScheduleInitOf) {
                ScheduleMemberInit(NetMember)
            }
            AnyRemaining
        } else {
            InitQueue := m_InitQueue
            set m_InitQueue = array{}
            for (NetMember:InitQueue) {
                ScheduleMemberInit(NetMember)
            }
            if (InitQueue.Length > 0). true else false
        }

    GetInitStep<private>():?aware_net_init_step=
        if (l_InitStep := m_InitStep?) {
            Some(l_InitStep)
        } else if (m_InitStepNextIndex >= 0) {
            var Res: ?aware_net_init_step = None()
            loop {
                if (InitOrderEntry := Config.EnforceInitOrder[m_InitStepNextIndex]) {
                    if (InitOrderEntry.ExpectedDeviceCount > 0) {
                        l_InitStep := aware_net_init_step{
                            Index := m_InitStepNextIndex
                            DeviceName := InitOrderEntry.DeviceName
                            TotalExpected := InitOrderEntry.ExpectedDeviceCount
                        }
                        set m_InitStepNextIndex += 1
                        set m_InitStep = Some(l_InitStep)
                        set Res = Some(l_InitStep)
                        LogLevel := if (Config.InitDebug?). log_level.Debug else log_level.Verbose
                        Log("AwareNet moving on to init step {l_InitStep.Index}, expecting min. {l_InitStep.TotalExpected} {l_InitStep.DeviceName}", ?Level := LogLevel)
                        break
                    } else {
                        Log("AwareNet skipped init step {m_InitStepNextIndex} as it expects no device. Fix if misconfigured or remove if not necessary.", ?Level := log_level.Error)
                        set m_InitStepNextIndex += 1
                    }
                } else {
                    break
                }
            }
            Res
        } else {
            None()
        }

    UseSequentialInit<private>(NetMember:aware_net_member)<transacts><decides>:void=
        Config.ForceSequentialDebug? or
            (NetMember.Meta?.PreferSequentialLoad?) or
            (Config.ForcePersistenceRecordableSequential? and persistence_recordable[NetMember.Device])

    QueueSequentialDeviceInit<private>(NetMember:aware_net_member):void=
        set m_SequentialInitQueue += array{NetMember}
        if (not (m_SequentialInitQueueRunning?)) {
            set m_SequentialInitQueueRunning = true
            spawn:
                SequentialInitQueueLoop()
        }

    SequentialInitQueueLoop<private>()<suspends>:void=
        loop {
            if (PopRes := m_SequentialInitQueue.Pop[0]) {
                set m_SequentialInitQueue = PopRes(0)
                PopRes(1).Init(true)
            } else {
                set m_SequentialInitQueueRunning = false
                break
            }
        }

aware_net_member<internal> := class:
    AwareNet: aware_net
    Device: creative_device
    Meta: ?aware_device_meta

    InitFn: type{__()<suspends>:void}

    var m_Initialized: logic = false
    var m_PostInitialized: logic = false

    Init<internal>(Sequential:logic)<suspends>:logic=
        if (m_Initialized?) {
            Log("Attempted to initialize {AsString()} more than once.", ?Level := log_level.Error)
            return false
        }
        set m_Initialized = true
        LogLevel := if (AwareNet.Config.InitDebug?). log_level.Debug else log_level.Verbose
        if (DebugInitShouldFilterOut[]) {
            Log("Skipping init of {AsString()}", ?Level := LogLevel)
            return false
        }
        Log("Initializing {AsString()} (sequential: {LogicToString(Sequential)})", ?Level := LogLevel)
        if (Sequential?) {
            if (AwareNet.Config.ForceSequentialDebug?, AwareNet.Config.ForceSequentialDebugInterval > 0.0) {
                Sleep(AwareNet.Config.ForceSequentialDebugInterval)
            }

            var Success: logic = false
            race:
                block:
                    InitFn()
                    set Success = true
                SequentialInitTimeoutProtection()
            if (Success?) {
                Log("Done initializing {AsString()}", ?Level := LogLevel)
            }
            AwareNet.OnPostMemberInit(Self)
        } else {
            spawn:
                ScheduleNonCompletablePostInit()
            InitFn()
        }
        true

    SequentialInitTimeoutProtection<private>()<suspends>:void=
        Sleep(2.0)
        Log("Device {AsString()} was taking more than 2s to initialize. Bailed out.", ?Level := log_level.Error)

    ScheduleNonCompletablePostInit<private>()<suspends>:void=
        Sleep(0.0)
        AwareNet.OnPostMemberInit(Self)

    DebugInitShouldFilterOut<private>()<transacts><decides>:void=
        AwareNet.Config.InitDebug?
        DeviceStr := AsString()
        var FilterOut: logic = false
        for (Filter:AwareNet.Config.InitFilteringOutDebug, not (FilterOut?), DeviceStr.StartsWith[Filter]) {
            set FilterOut = true
        }
        FilterOut?

    AsString<public>()<transacts>:string=
        if (l_Meta := Meta?) {
            ToString(l_Meta)
        } else {
            "<unknown>"
        }
