using { /Fortnite.com/Characters }
using { /Fortnite.com/Devices }
using { /Fortnite.com/UI }
using { /Verse.org/Simulation }
using { /Verse.org/Colors }
using { /Verse.org/Random }
using { /UnrealEngine.com/Temporary/Diagnostics }
using { /UnrealEngine.com/Temporary/UI }
using { /UnrealEngine.com/Temporary/SpatialMath }

using { HavocVerse.Devices }
using { HavocVerse.Storage }
using { HavocVerse.Persistence }
using { HavocVerse.UI }
using { HavocVerse.Util }
using { HavocVerse.Logging }

var g_DebugDevice<public>: weak_map(session,debug_device) = map{}

ToString<public>(Agent:agent)<transacts>:string=
    if (
        Player := player[Agent]
        DebugDevice := g_DebugDevice[GetSession()]
        AgentId := DebugDevice.m_AgentIds[Player](0)
    ) {
        "<AGENT:{AgentId}>"
    } else {
        "<unknown agent>"
    }

debug_device_persistent_player_data<public> := class<final><persistable>:
    Version: int = 0

    v0_TypedCode: logic = false
    v0_FeedEnabled: logic = false
    v0_RemoteOpen: logic = false

debug_log_message := class<final>:
    var OriginalMessage: ?message
    var CheckRich: logic
    StrMessage: string
    Level: log_level
    SentAt: float
    Discriminator: int

debug_player := class:
    Player: player

    var m_InVolume: logic = false

    var m_CrouchStartTime: float = 0.0
    var m_CrouchCombo: int = 0

    var m_TypedCode: logic = false
    var m_FailedCodeAttempts: int = 0

    var m_Persisted: logic = false
    var m_RemoteOpen: logic = false
    var m_FeedEnabled: logic = false
    var m_FeedDisplayDuration: float = 5.0

    var m_OpenedMenu: ?debug_menu_instance = false

debug_device_custom_action_config := class<concrete>:
    @editable Label: string = ""
    @editable Trigger: trigger_device = trigger_device{}

debug_device_tick_monitoring := class:
    var LastTickAt: float
    var TickHistory: []float = array{}

debug_action<public> := class:
    Name<public>: string
    Executor<public>: debug_action_executor

debug_action_executor<public> := interface:
    Execute<public>(Player:player):void

debug_action_executor_callback<public> := class(debug_action_executor):
    Callback<public>: type{__(:player):void}

    Execute<override>(Player:player):void=
        Callback(Player)

MakeDebugActionExecutor<public>(Callback:type{__(:player):void}):debug_action_executor_callback=
    debug_action_executor_callback{
        Callback := Callback
    }

debug_draw_args<public> := class:
    Color<public>: color
    Thickness<public>: float
    DrawDurationPolicy<public>: debug_draw_duration_policy
    Duration<public>: float

MakeDebugDrawArgs<public><constructor>(CopyOf:debug_draw_args)<transacts> := debug_draw_args:
    Color := CopyOf.Color
    Thickness := CopyOf.Thickness
    DrawDurationPolicy := CopyOf.DrawDurationPolicy
    Duration := CopyOf.Duration

debug_device_secret_volume_mode<internal> := enum:
    Crouch
    Jump

debug_device<public> := class(creative_device, aware_device, aware_debug_device):
    Logger<public>: log = log{Channel := havoc_log_chan}
    DebugDraw<public>: debug_draw = debug_draw{}

    @editable p_EnableCodePrompt: logic = true
    @editable p_SecretCode: string = "038293"
    @editable p_OpenButton: []button_device = array{}
    @editable p_SecretVolume: []volume_device = array{}
    @editable p_SecretVolumeMode: debug_device_secret_volume_mode = debug_device_secret_volume_mode.Crouch
    @editable p_RemoteOpenInputTrigger: ?input_trigger_device = false
    @editable p_RemoteOpenEnabledbyDefault: logic = false

    @editable p_EnablePersistence: logic = true
    @editable p_CustomActions: []debug_device_custom_action_config = array{}

    @editable p_LogStackLimit: int = 800
    @editable p_VerboseLogStackLimit: int = 100
    @editable p_RaisePassthroughLogsToNormal: logic = false
    @editable p_RaisePassthroughLogsFrom: log_level = log_level.Verbose

    @editable p_FeedEnabledByDefault: logic = false
    @editable p_FeedMessageLimit: int = 50

    @editable p_MenuConfig: debug_menu_config = debug_menu_config{}

    @editable p_TickMonitoring: logic = true
    @editable p_TickMonitoringHistoryLength: int = 100
    @editable p_TickMonitoringLogs: logic = false

    var m_LogLevel<public>: log_level = log_level.Debug
    var m_LogStack<private>: []debug_log_message = array{}
    var m_VerboseLogStack: []debug_log_message = array{}

    var m_HudManager: ?debug_hud_manager = false
    var m_Players: [player]debug_player = map{}
    var m_AgentIds: [agent]tuple(string,message) = map{}

    var m_MenuGeneralSharedOptions: []debug_menu_option = array{}
    var m_MenuGeneralIndividualOptions: []debug_menu_option = array{}
    var m_DebugDrawActions: []debug_action = array{}

    var m_TickMonitoring: ?debug_device_tick_monitoring = false
    var m_InitAt: float = 0.0

    var m_DebugDrawDurationPolicy: debug_draw_duration_policy := debug_draw_duration_policy.FiniteDuration
    var m_DebugDrawThickness: float = 2.0
    var m_DebugDrawDuration: float = 15.0

    DeviceMeta<override>()<transacts>:aware_device_meta=
        aware_device_meta{
            DeviceName := "debug_device"
        }

    DebugStats<override>()<transacts>:aware_debug_stats=
        aware_debug_stats{
            Length := map{
                "logStack" => m_LogStack.Length
                "verboseLogStack" => m_VerboseLogStack.Length
                "players" => m_Players.Length
                "agentIds" => m_AgentIds.Length
                "menuGeneralSharedOptions" => m_MenuGeneralSharedOptions.Length
                "menuGeneralIndividualOptions" => m_MenuGeneralIndividualOptions.Length
                "debugDrawActions" => m_DebugDrawActions.Length
            }
        }

    OnBegin<override>()<suspends>:void=
        set m_InitAt = GetSimulationElapsedTime()
        AwareInitDevice(Self, OnInit)

    OnInit<private>()<suspends>:void=
        if (g_DebugDevice[GetSession()]) {
            LogMessage("More than a single debug device.", ?Level := log_level.Warning)
        } else if (set g_DebugDevice[GetSession()] = Self) {
            if (LogEntries := g_GlobalLogUnboundStack[GetSession()]) {
                if:
                    set g_GlobalLogUnboundStack[GetSession()] = array{}
                for (LogEntry:LogEntries) {
                    Log(None(), LogEntry(0), LogEntry(1), LogEntry(2))
                }
            }

            LogMessage("Set debug device as default logger.", ?Level := log_level.Normal)
        } else {
            LogMessage("Could not register debug device as default logger.", ?Level := log_level.Error)
        }

        for (Button:p_OpenButton) {
            Button.InteractedWithEvent.Subscribe(OnOpenButtonInteractedWith)
        }
        for (SecretVolume:p_SecretVolume) {
            SecretVolume.AgentEntersEvent.Subscribe(OnSecretVolumeEntered)
            SecretVolume.AgentExitsEvent.Subscribe(OnSecretVolumeExited)
        }

        GetPlayspace().PlayerAddedEvent().Subscribe(OnPlayerAdded)
        GetPlayspace().PlayerRemovedEvent().Subscribe(OnPlayerRemoved)

        set m_MenuGeneralSharedOptions += array{
            debug_menu_option{
                Label := "Log Level"
                Provider := debug_device_log_level_choice{
                    Device := Self
                }
            }
        }

        if (InputTrigger := p_RemoteOpenInputTrigger?) {
            InputTrigger.PressedEvent.Subscribe(OnOpenButtonInteractedWith)
            set m_MenuGeneralIndividualOptions += array{
                debug_menu_option{
                    Label := "Remote Open"
                    Provider := debug_device_remote_open_checkbox{
                        Device := Self
                    }
                }
            }
        }
        set m_MenuGeneralIndividualOptions += array{
            debug_menu_option{
                Label := "Show Log Feed"
                Provider := debug_device_log_feed_checkbox{
                    Device := Self
                }
            }
            debug_menu_option{
                Label := "Log Feed Duration"
                Provider := debug_device_log_feed_duration_slider{
                    Device := Self
                }
            }
            debug_menu_option{
                Label := "Vulnerable"
                Provider := debug_device_vulnerable_checkbox{
                    Device := Self
                }
            }
        }

        set m_DebugDrawActions = array{
            debug_action{
                Name := "Buyp hl (all)"
                Executor := MakeDebugActionExecutor(DbgBuyPlateHighlightAll)
            }
            debug_action{
                Name := "Buyp hl (enabled)"
                Executor := MakeDebugActionExecutor(DbgBuyPlateHighlightEnabled)
            }
            debug_action{
                Name := "Buyp Groups"
                Executor := MakeDebugActionExecutor(DbgBuyPlateGroupsBounds)
            }
        }

        HudManager := debug_hud_manager{
            AccessControl := None()
            Device := Self
        }
        set m_HudManager = Some(HudManager)
        for (Player:GetPlayspace().GetPlayers()) {
            OnPlayerAdded(Player)
        }

        spawn:
            HudManager.Run(GetPlayspace())

        if (p_TickMonitoring?) {
            spawn:
                StartTickMonitor()
        }
        spawn:
            StartUptimeProvider()

    AddGeneralDebugOption<public>(Shared:logic, Option:debug_menu_option):void=
        if (Shared?) {
            set m_MenuGeneralSharedOptions += array{Option}
        } else {
            set m_MenuGeneralIndividualOptions += array{Option}
        }

    AddDebugAction<public>(Action:debug_action):void=
        set m_DebugDrawActions += array{Action}

    DebugDrawArgs<public>()<transacts>:debug_draw_args=
        debug_draw_args{
            Color := NamedColors.Blue
            Thickness := m_DebugDrawThickness
            DrawDurationPolicy := m_DebugDrawDurationPolicy
            Duration := m_DebugDrawDuration
        }

    StartTickMonitor<private>()<suspends>:void=
        loop {
            Sleep(0.0)
            OnTick()
        }

    StartUptimeProvider<private>()<suspends>:void=
        loop {
            Sleep(1.0)

            var AnyOpenMenu: logic = false
            for (__->DebugPlayer:m_Players, not AnyOpenMenu?, DebugPlayer.m_OpenedMenu?) {
                set AnyOpenMenu = true
            }

            if (AnyOpenMenu?) {
                UptimeText := FormatUptime()
                for (__->DebugPlayer:m_Players, DebugMenuInstance := DebugPlayer.m_OpenedMenu?) {
                    DebugMenuInstance.UpdateUptime(UptimeText)
                }
            }
        }

    FormatUptime<internal>():string=
        UptimeSecs := GetSimulationElapsedTime() - m_InitAt
        ToHumanDuration(UptimeSecs, human_duration_parts{
            Days := false
            Hours := if (UptimeSecs >= TimeUnits.Hour). true else false
            Minutes := true
            Seconds := true
        }, human_duration_opts{
            Delimiter := " "
            DaysFmt := ("d", "d")
            HoursFmt := ("hr", "hrs")
            MinutesFmt := ("min", "mins")
            SecondsFmt := ("s", "s")
        })

    OnTick<private>():void=
        Now := GetSimulationElapsedTime()
        if (TickMonitoring := m_TickMonitoring?) {
            TickDuration := Now - TickMonitoring.LastTickAt
            set TickMonitoring.LastTickAt = Now

            if (
                TickMonitoring.TickHistory.Length >= p_TickMonitoringHistoryLength
                SlicedHist := TickMonitoring.TickHistory.Slice[(TickMonitoring.TickHistory.Length - p_TickMonitoringHistoryLength) + 1]
            ) {
                set TickMonitoring.TickHistory = SlicedHist + array{TickDuration}
            } else {
                set TickMonitoring.TickHistory += array{TickDuration}
            }

            if (FirstTick := TickMonitoring.TickHistory[0], Rest := TickMonitoring.TickHistory.Slice[1]) {
                var TotalTicks: float = FirstTick
                var MinTick: float = FirstTick
                var MaxTick: float = FirstTick
                for (Tick:Rest) {
                    set TotalTicks += Tick
                    if (Tick < MinTick) {
                        set MinTick = Tick
                    }
                    if (Tick > MaxTick) {
                        set MaxTick = Tick
                    }
                }
                AvgTick := TotalTicks / ToFloat(Rest.Length + 1)

                if (p_TickMonitoringLogs?) {
                    LogMessage("TickM: {TickDuration} last, {AvgTick} avg, {MinTick} min, {MaxTick} max", ?Level := log_level.Verbose)
                }

                for (__->DebugPlayer:m_Players, DebugMenuInstance := DebugPlayer.m_OpenedMenu?) {
                    DebugMenuInstance.UpdateTickMonitoring(TickDuration, AvgTick, MinTick, MaxTick)
                }
            } else {
                LogMessage("Verse tick took {TickDuration}s", ?Level := log_level.Verbose)
            }
        } else {
            TickMonitoring := debug_device_tick_monitoring{
                LastTickAt := Now
            }
            set m_TickMonitoring = Some(TickMonitoring)
            TickMonitoring
        }

    OnPlayerAdded<private>(Player:player):void=
        GrantAgentId(Player)
        spawn:
            HandlePlayerAdded(Player)

    OnPlayerRemoved<private>(Player:player):void=
        set m_Players = m_Players.RemoveByKey(Player)(0)
        set m_AgentIds = m_AgentIds.RemoveByKey(Player)(0)

    GrantAgentId<private>(Agent:agent):void=
        loop {
            AgentId := ToString(GetRandomInt(0, 10000))
            var NotDuplicate: logic = true
            for (Entry:m_AgentIds, NotDuplicate?, Entry(0) = AgentId) {
                set NotDuplicate = false
            }
            if (NotDuplicate?) {
                AgentMessage := ToMessage(Agent)
                if. set m_AgentIds[Agent] = (AgentId, AgentMessage)
                break
            }
        }

    HandlePlayerAdded(Player:player)<suspends>:void=
        AwareEnsureDelaySincePlayerJoined(Player, 0.5)
        if (not Player.IsActive[]) {
            return
        }
        if (FortCharacter := Player.GetFortCharacter[]) {
            case(p_SecretVolumeMode):
                debug_device_secret_volume_mode.Crouch => FortCharacter.CrouchedEvent().Subscribe(OnCrouched)
                debug_device_secret_volume_mode.Jump => FortCharacter.JumpedEvent().Subscribe(OnJumped)
        }
        if (p_EnablePersistence?, PlayerData := PlayerDataGetDebugDevice[Player]) {
            GetDebugPlayer(Player)
        } else if (not (p_EnableCodePrompt?), p_RemoteOpenEnabledbyDefault?) {
            GetDebugPlayer(Player)
        }

    GetDebugPlayer<internal>(Player:player):debug_player=
        if (l_DebugPlayer := m_Players[Player]) {
            l_DebugPlayer
        } else {
            DebugPlayer := debug_player{
                Player := Player
                m_TypedCode := if (p_EnableCodePrompt?). false else true
                m_FeedEnabled := p_FeedEnabledByDefault
                m_RemoteOpen := p_RemoteOpenEnabledbyDefault
            }
            if (p_EnablePersistence?, PlayerData := PlayerDataGetDebugDevice[Player]) {
                LogMessage(MsgPlayerRecovered(Player), ?Level := log_level.Debug)
                set DebugPlayer.m_TypedCode = PlayerData.v0_TypedCode
                set DebugPlayer.m_FeedEnabled = PlayerData.v0_FeedEnabled
                set DebugPlayer.m_RemoteOpen = PlayerData.v0_RemoteOpen
            }
            if (set m_Players[Player] = DebugPlayer, HudManager := m_HudManager?) {
                HudManager.RecheckUnmanagedPlayers()
            }
            if (DebugPlayer.m_TypedCode?, DebugPlayer.m_RemoteOpen?, InputTrigger := p_RemoteOpenInputTrigger?) {
                InputTrigger.Register(Player)
            }
            DebugPlayer
        }

    MsgPlayerRecovered<private><localizes>(Player:player):message=
        "Recovered debug device preferences for '{Player}'"

    PersistDebugPlayer<internal>(DebugPlayer:debug_player):void=
        if (p_EnablePersistence?) {
            set DebugPlayer.m_Persisted = true
            NewPlayerData := debug_device_persistent_player_data{
                v0_TypedCode := DebugPlayer.m_TypedCode
                v0_FeedEnabled := DebugPlayer.m_FeedEnabled
                v0_RemoteOpen := DebugPlayer.m_RemoteOpen
            }
            PlayerDataSetDebugDevice(DebugPlayer.Player, Some(NewPlayerData))
        }

    OnOpenButtonInteractedWith(Agent:agent):void=
        if (Player := player[Agent]) {
            OpenMenu(Player, GetDebugPlayer(Player))
        }

    OnSecretVolumeEntered(Agent:agent):void=
        if (Player := player[Agent]) {
            DebugPlayer := GetDebugPlayer(Player)
            set DebugPlayer.m_InVolume = true
        }

    OnSecretVolumeExited(Agent:agent):void=
        if (Player := player[Agent]) {
            DebugPlayer := GetDebugPlayer(Player)
            set DebugPlayer.m_InVolume = false
            if (p_SecretVolumeMode = debug_device_secret_volume_mode.Crouch) {
                set DebugPlayer.m_CrouchCombo = 0
            }
        }

    OnCrouched<private>(FortCharacter:fort_character, State:logic):void=
        OnSecretVolumeAction(FortCharacter)

    OnJumped<private>(FortCharacter:fort_character):void=
        OnSecretVolumeAction(FortCharacter)

    OnSecretVolumeAction<private>(FortCharacter:fort_character):void=
        Player := if (l_Player := player[FortCharacter.GetAgent[]]) {
            l_Player
        } else {
            return
        }
        DebugPlayer := GetDebugPlayer(Player)
        if (not (DebugPlayer.m_InVolume?)) {
            return
        }
        Now := GetSimulationElapsedTime()
        TimeWindow := case(p_SecretVolumeMode):
            debug_device_secret_volume_mode.Crouch => 1.0
            debug_device_secret_volume_mode.Jump => 3.0
        if (DebugPlayer.m_CrouchStartTime + TimeWindow <= Now) {
            set DebugPlayer.m_CrouchCombo = 0
        }
        if (DebugPlayer.m_CrouchCombo = 0) {
            set DebugPlayer.m_CrouchStartTime = Now
        }
        set DebugPlayer.m_CrouchCombo += 1
        TargetCombo := case(p_SecretVolumeMode):
            debug_device_secret_volume_mode.Crouch => 4
            debug_device_secret_volume_mode.Jump => 3
        if (DebugPlayer.m_CrouchCombo >= TargetCombo) {
            set DebugPlayer.m_CrouchCombo = 0
            if (DebugPlayer.m_FailedCodeAttempts >= 3) {
                LogMessage("Too many failed code attempts, ignoring.", ?Level := log_level.Debug)
                return
            }
            OpenMenu(Player, DebugPlayer)
        }

    OpenMenu<private>(Player:player, DebugPlayer:debug_player):void=
        if (DebugPlayer.m_OpenedMenu?) {
            return
        }
        if (AssetsProvider := AwareDefaultAssetsProvider[]) {
            Menu := debug_menu{
                AssetsProvider := AssetsProvider
                Device := Self
                Config := p_MenuConfig
                DebugPlayer := DebugPlayer
                SudoInstigator := None()
            }
            if (not (p_EnableCodePrompt?) or DebugPlayer.m_TypedCode?) {
                Menu.Open(Player)
            } else {
                Menu.OpenCodePrompt(Player)
            }
        }

    LogMessage<public>(Message:string, ?Level:log_level = log_level.Normal)<transacts>:void=
        if (Level = log_level.Verbose and m_LogLevel <> log_level.Verbose) {
            return
        }
        Print(Message, Level)
        if (GetLogLevelIndex(m_LogLevel) > GetLogLevelIndex(Level)) {
            return
        }
        Log(None(), Message, Level, GetSimulationElapsedTime())

    LogMessage<public>(Message:message, ?Level:log_level = log_level.Normal)<transacts>:void=
        if (Level = log_level.Verbose and m_LogLevel <> log_level.Verbose) {
            return
        }
        StrMessage := Localize(Message)
            .ReplaceAll("'Anonymous[256]'", "''")
            .ReplaceAll("Anonymous[256]", "<redacted>")
        Print(StrMessage, Level)
        if (GetLogLevelIndex(m_LogLevel) > GetLogLevelIndex(Level)) {
            return
        }
        Log(Some(Message), StrMessage, Level, GetSimulationElapsedTime())

    Log<private>(OriginalMessage:?message, StrMessage:string, Level:log_level, Time:float)<transacts>:void=
        var LogStack: []debug_log_message = if (Level = log_level.Verbose) {
            m_VerboseLogStack
        } else {
            m_LogStack
        }
        LogStackLimit := if (Level = log_level.Verbose) {
            p_VerboseLogStackLimit
        } else {
            p_LogStackLimit
        }

        Now := GetSimulationElapsedTime()
        Message := debug_log_message{
            OriginalMessage := OriginalMessage
            CheckRich := if (OriginalMessage?) {
                false
            } else {
                true
            }
            StrMessage := StrMessage
            Level := Level
            SentAt := Now
            Discriminator := if (Prev := LogStack[LogStack.Length-1], Prev.SentAt = Now) {
                Prev.Discriminator + 1
            } else {
                0
            }
        }

        set LogStack += array{Message}
        if (LogStack.Length >= LogStackLimit, l_Slice := LogStack.Slice[LogStack.Length - LogStackLimit]) {
            set LogStack = l_Slice
        }
        if (Level = log_level.Verbose) {
            set m_VerboseLogStack = LogStack
        } else {
            set m_LogStack = LogStack
        }

    GetLogStack<internal>():[]debug_log_message=
        for (Entry:(m_LogStack + m_VerboseLogStack)) {
            if (Entry.CheckRich?) {
                set Entry.CheckRich = false
                if (Entry.StrMessage.Contains["<AGENT:"]) {
                    RichMessage := ReplaceIntoMessageWithFn(Entry.StrMessage, Self.AgentIdToMessage,
                        ?OpenDelimiter := "<AGENT:", ?CloseDelimiter := ">",
                        ?CommonEscapes := false)
                    set Entry.OriginalMessage = Some(RichMessage)
                }
            }
            Entry
        }
    
    AgentIdToMessage<public>(AgentId:string):?message=
        for (Agent->Entry:m_AgentIds, Entry(0) = AgentId) {
            return Some(Entry(1))
        }
        None()

    Print<private>(StrMessage:string, Level:log_level)<transacts>:void=
        if (
            p_RaisePassthroughLogsToNormal?
            LevelIndex := GetLogLevelIndex(Level)
            LevelIndex < GetLogLevelIndex(log_level.Normal)
            LevelIndex >= GetLogLevelIndex(p_RaisePassthroughLogsFrom)
        ) {
            Prefix := case(Level):
                log_level.Verbose => "[V] "
                log_level.Debug => "[D] "
                _ => ""
            Logger.Print("{Prefix}{StrMessage}", ?Level := log_level.Normal)
        } else {
            Logger.Print(StrMessage, ?Level := Level)
        }

    DbgBuyPlateHighlightAll<private>(Player:player):void=
        for (AwareNetDevice:AwareListDevices(), BuyPlateManager := buy_plate_manager_device[AwareNetDevice]) {
            BuyPlateManager.DebugDrawFakeVolumes(DebugDraw, DebugDrawArgs(), false)
        }

    DbgBuyPlateHighlightEnabled<private>(Player:player):void=
        for (AwareNetDevice:AwareListDevices(), BuyPlateManager := buy_plate_manager_device[AwareNetDevice]) {
            BuyPlateManager.DebugDrawFakeVolumes(DebugDraw, DebugDrawArgs(), true)
        }

    DbgBuyPlateGroupsBounds<private>(Player:player):void=
        for (AwareNetDevice:AwareListDevices(), BuyPlateManager := buy_plate_manager_device[AwareNetDevice]) {
            BuyPlateManager.DebugDrawAutoVolumesBoundaries(DebugDraw, DebugDrawArgs())
        }

debug_hud_manager := class(hud_manager_base):
    Device: debug_device

    CreateHUD<override>(Player:player):?hud_manager_player=
        DebugPlayer := if (l_DebugPlayer := Device.m_Players[Player]) {
            l_DebugPlayer
        } else {
            return None()
        }

        TextWidget := text_block{
            DefaultText := ToMessage("")
            DefaultJustification := text_justification.Left
            DefaultTextColor := NamedColors.White
            DefaultShadowOffset := option{vector2{
                X := 2.0
                Y := 2.0
            }}
            DefaultShadowColor := NamedColors.Black
        }
        TextWidget.SetShadowOpacity(0.65)

        Canvas := canvas{
            Slots := array{
                canvas_slot{
                    Offsets := margin{
                        Left := 10.0
                        Top := 10.0
                        Right := 10.0
                        Bottom := 10.0
                    }
                    SizeToContent := false
                    Alignment := vector2{X := 0.0, Y := 0.0}
                    ZOrder := 900
                    Widget := TextWidget
                }
            }
        }
        Element := debug_hud_feed{
            Widget := Canvas
            Device := Device
            DebugPlayer := DebugPlayer
            TextWidget := TextWidget
        }
        Some(MakeSingleElementHUDPlayer(Player, Canvas, Element))

debug_hud_feed := class(hud_element):
    Device: debug_device
    DebugPlayer: debug_player

    TextWidget: text_block

    Update<override>():void=
        if (DebugPlayer.m_FeedEnabled?) {
            Show()
            Update(GetSimulationElapsedTime())
        } else {
            Hide()
            TextWidget.SetText(ToMessage(""))
        }

    Update<private>(Now:float):void=
        LogStack := Device.GetLogStack()
        SliceStart := Max(LogStack.Length - Device.p_FeedMessageLimit, 0)
        if (LastMessages := LogStack.Slice[SliceStart]) {
            var Content: message = ToMessage("")
            CurrentLevelIndex := GetLogLevelIndex(Device.m_LogLevel)
            for (
                Index->LogMessage:LastMessages
                LogMessage.SentAt + DebugPlayer.m_FeedDisplayDuration > Now
                CurrentLevelIndex <= GetLogLevelIndex(LogMessage.Level)
            ) {
                if (Index > 0) {
                    set Content = ConcatMessage(Content, "\n")
                }
                Prefix := case(LogMessage.Level):
                    log_level.Verbose => "[V] "
                    log_level.Debug => "[D] "
                    log_level.Normal => ""
                    log_level.Warning => "[W] "
                    log_level.Error => "[E] "
                if (Message := LogMessage.OriginalMessage?) {
                    set Content = ConcatMessage(Content, Prefix, Message) 
                } else {
                    set Content = ConcatMessage(Content, "{Prefix}{LogMessage.StrMessage}")
                }
            }
            TextWidget.SetText(Content)
        }

CmpLogMessageByTime(A:debug_log_message, B:debug_log_message)<computes>:int=
    Cmp := CmpFloat(A.SentAt, B.SentAt)
    if (Cmp = 0) {
        CmpInt(A.Discriminator, B.Discriminator)
    } else {
        Cmp
    }

debug_device_log_feed_checkbox := class(debug_menu_checkbox_provider):
    Device: debug_device

    GetState<override>(Player:player)<transacts><decides>:void=
        Device.m_Players[Player].m_FeedEnabled?

    SetState<override>(Player:player, Value:logic):void=
        DebugPlayer := Device.GetDebugPlayer(Player)
        set DebugPlayer.m_FeedEnabled = Value
        Device.PersistDebugPlayer(DebugPlayer)

debug_device_log_feed_duration_slider := class(debug_menu_slider_provider):
    Device: debug_device

    GetValue<override>(Player:player)<transacts><decides>:float=
        Device.m_Players[Player].m_FeedDisplayDuration

    SetValue<override>(Player:player, Value:float):void=
        DebugPlayer := Device.GetDebugPlayer(Player)
        set DebugPlayer.m_FeedDisplayDuration = Value

    GetBounds<override>()<transacts>:tuple(float,float,float)=
        (1.0, 60.0, 5.0)

debug_device_log_level_choice := class(debug_menu_choices_provider):
    Device: debug_device

    GetSelected<override>(Player:player)<transacts>:int=
        GetLogLevelIndex(Device.m_LogLevel)

    SetSelected<override>(Player:player, Index:int)<transacts><decides>:void=
        LogLevel := GetLogLevelFromIndex[Index]
        set Device.m_LogLevel = LogLevel

    FormatOption<override>(Index:int)<transacts><decides>:string=
        LogLevel := GetLogLevelFromIndex[Index]
        case(LogLevel):
            log_level.Verbose => "Verbose"
            log_level.Debug => "Debug"
            log_level.Normal => "Normal"
            log_level.Warning => "Warning"
            log_level.Error => "Error"

debug_device_vulnerable_checkbox := class(debug_menu_checkbox_provider, debug_menu_checkbox_labels):
    Device: debug_device

    GetEnabledLabel<override>()<transacts>:string = "Yes"
    GetDisabledLabel<override>()<transacts>:string = "No"

    GetState<override>(Player:player)<transacts><decides>:void=
        Player.GetFortCharacter[].IsVulnerable[]

    SetState<override>(Player:player, Value:logic):void=
        if (FortCharacter := Player.GetFortCharacter[]) {
            FortCharacter.SetVulnerability(Value)
        }

debug_device_remote_open_checkbox := class(debug_menu_checkbox_provider):
    Device: debug_device

    GetState<override>(Player:player)<transacts><decides>:void=
        Device.m_Players[Player].m_RemoteOpen?

    SetState<override>(Player:player, Value:logic):void=
        if (InputTrigger := Device.p_RemoteOpenInputTrigger?) {
            DebugPlayer := Device.GetDebugPlayer(Player)
            set DebugPlayer.m_RemoteOpen = Value
            if (Value?) {
                InputTrigger.Register(Player)
            } else {
                InputTrigger.Unregister(Player)
            }
            Device.PersistDebugPlayer(DebugPlayer)
        }

debug_device_debug_draw_duration_policy_choice := class(debug_menu_choices_provider):
    Device: debug_device

    GetSelected<override>(Player:player)<transacts>:int=
        case(Device.m_DebugDrawDurationPolicy):
            debug_draw_duration_policy.SingleFrame => 0
            debug_draw_duration_policy.FiniteDuration => 0
            debug_draw_duration_policy.Persistent => 1

    SetSelected<override>(Player:player, Index:int)<transacts><decides>:void=
        NewPolicy := case(Index):
            0 => debug_draw_duration_policy.FiniteDuration
            1 => debug_draw_duration_policy.Persistent
        set Device.m_DebugDrawDurationPolicy = NewPolicy

    FormatOption<override>(Index:int)<transacts><decides>:string=
        case(Index):
            0 => "Ephemeral"
            1 => "Persistent"

debug_device_debug_draw_thickness_slider := class(debug_menu_slider_provider):
    Device: debug_device

    GetValue<override>(Player:player)<transacts><decides>:float=
        Device.m_DebugDrawThickness

    SetValue<override>(Player:player, Value:float):void=
        set Device.m_DebugDrawThickness = Value

    GetBounds<override>()<transacts>:tuple(float,float,float)=
        (0.5, 5.0, 0.5)
