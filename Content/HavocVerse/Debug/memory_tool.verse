using { /Fortnite.com/Devices }
using { /Fortnite.com/UI }
using { /Verse.org/Simulation }
using { /Verse.org/Colors }
using { /UnrealEngine.com/Temporary/Diagnostics }
using { /UnrealEngine.com/Temporary/SpatialMath }
using { /UnrealEngine.com/Temporary/UI }

using { HavocVerse.Debug }
using { HavocVerse.Devices }
using { HavocVerse.UI }
using { HavocVerse.Logging }
using { HavocVerse.Util }
using { HavocVerse.Util.Task }

memory_tool_menu := class<final>(ui_menu):
    CreateInstance<override>(Player:player):tuple(ui_menu_instance,canvas)=
        Instance := ui_menu_instance:
            Menu := Self
            Player := Player

        Background := color_block:
            DefaultColor := NamedColors.Black
            DefaultOpacity := 0.85
        Content := memory_tool_menu_content{}

        ## Top Row {{{
        CloseButton := button_regular:
            DefaultText := ToMessage("X")
        CloseButton.OnClick().Subscribe(Instance.OnCloseClick)
        CloseButtonWidget := ApplySizeHint(CloseButton, vector2{X := 60.0, Y := 50.0})

        TitleWidget := CreateTextBlock("Memory Tool")
        TopRowStart := stack_box:
            Orientation := orientation.Horizontal
            Slots := array{
                stack_box_slot:
                    Widget := CloseButtonWidget
                    HorizontalAlignment := horizontal_alignment.Left
                    VerticalAlignment := vertical_alignment.Center
                stack_box_slot:
                    Widget := TitleWidget
                    HorizontalAlignment := horizontal_alignment.Left
                    VerticalAlignment := vertical_alignment.Center
                    Padding := margin{
                        Left := 20.0
                    }
            }

        TabButtonSlots := for (Index->Tab:array{"Overview", "History"}) {
            Closure := memory_tool_tab_button_closure:
                Content := Content
                Tab := Tab
            TabButton := button_regular:
                DefaultText := ToMessage(Tab)
            TabButton.OnClick().Subscribe(Closure.OnClick)
            stack_box_slot:
                Widget := ApplySizeHint(TabButton, vector2{X := 200.0, Y := 50.0})
                HorizontalAlignment := horizontal_alignment.Left
                VerticalAlignment := vertical_alignment.Top
                Padding := margin{
                    Left := if (Index = 0) { 0.0 } else { 10.0 }
                }
        }
        TopRowEnd := stack_box:
            Orientation := orientation.Horizontal
            Slots := TabButtonSlots

        TopRow := canvas:
            Slots := array{
                canvas_slot:
                    Widget := TopRowStart
                    SizeToContent := true
                    Alignment := vector2{X := 0.0, Y := 0.0}
                canvas_slot:
                    Widget := TopRowEnd
                    SizeToContent := true
                    Anchors := anchors{
                        Minimum := vector2{X := 1.0, Y := 0.0}
                        Maximum := vector2{X := 1.0, Y := 0.0}
                    }
                    Alignment := vector2{X := 1.0, Y := 0.0}
            }
        TopRowWidget := overlay:
            Slots := array{
                overlay_slot:
                    Widget := color_block:
                        DefaultColor := NamedColors.MediumSeaGreen
                    HorizontalAlignment := horizontal_alignment.Fill
                    VerticalAlignment := vertical_alignment.Fill
                overlay_slot:
                    Widget := TopRow
                    HorizontalAlignment := horizontal_alignment.Fill
                    VerticalAlignment := vertical_alignment.Top
                    Padding := margin{
                        Left := 10.0
                        Right := 10.0
                        Top := 10.0
                        Bottom := 10.0
                    }
            }
        ## }}}

        MainStack := stack_box:
            Orientation := orientation.Vertical
            Slots := array{
                stack_box_slot:
                    Widget := TopRowWidget
                    VerticalAlignment := vertical_alignment.Top
                    HorizontalAlignment := horizontal_alignment.Fill
                stack_box_slot:
                    Widget := Instance.AddComponent(Content)
                    VerticalAlignment := vertical_alignment.Fill
                    HorizontalAlignment := horizontal_alignment.Fill
            }

        Canvas := canvas:
            Slots := array{
                canvas_slot:
                    Widget := Background
                    SizeToContent := false
                    Anchors := anchors{
                        Minimum := vector2{X := 0.0, Y := 0.0}
                        Maximum := vector2{X := 1.0, Y := 1.0}
                    }
                    Alignment := vector2{X := 0.0, Y := 0.0}
                canvas_slot:
                    Widget := MainStack
                    SizeToContent := false
                    Anchors := anchors{
                        Minimum := vector2{X := 0.0, Y := 0.0}
                        Maximum := vector2{X := 1.0, Y := 1.0}
                    }
                    Alignment := vector2{X := 0.0, Y := 0.0}
            }
        (Instance, Canvas)

memory_tool_tab_button_closure<internal> := class<final>:
    Content: memory_tool_menu_content
    Tab: string

    OnClick(__:widget_message):void=
        Content.SetTab(Tab)

memory_tool_stats_grouping_mode := enum:
    NoGrouping
    DeviceKind

memory_tool_stats_sorting_mode := enum:
    Alphabetical
    HigherCount

memory_tool_menu_content := class<final>(ui_component):
    var m_ActiveTab: string = "Overview"

    var m_GroupingMode: memory_tool_stats_grouping_mode = memory_tool_stats_grouping_mode.DeviceKind
    var m_SortingMode: memory_tool_stats_sorting_mode = memory_tool_stats_sorting_mode.Alphabetical
    var m_SortingOrderAsc: logic = true

    Create<override>():widget=
        case(m_ActiveTab):
            "Overview" => AddChild((
                Component := memory_tool_menu_overview{Parent := Self}
                Component.UpdateEntries()
                Component
            ))
            _ => overlay{}

    SetTab(Tab:string):void=
        if (m_ActiveTab <> Tab) {
            set m_ActiveTab = Tab
            ForceUpdate()
        }
    
memory_tool_menu_overview := class<final>(ui_component):
    Parent: memory_tool_menu_content

    GroupingCycleButton: button_regular = button_regular{}
    SortingCycleButton: button_regular = button_regular{}
    SortingOrderButton: button_regular = button_regular{}

    PerRowEntries: int = 3
    PageTextWidget: text_block = text_block{DefaultTextColor := NamedColors.Black}

    var m_Entries: []memory_tool_stats_entry = array{}
    var m_ScrollPosition: int = 0

    var m_Content: ?ui_component = false

    CreateControls<private>():widget=
        ScrollActions := stack_box:
            Orientation := orientation.Horizontal
            Slots := array{
                stack_box_slot:
                    Widget := PageTextWidget
                    VerticalAlignment := vertical_alignment.Center
                    HorizontalAlignment := horizontal_alignment.Left
                }
        for (Index->Text:array{"Up", "Down"}) {
            Button := button_regular:
                DefaultText := ToMessage(Text)
            case(Index):
                0 => BindSubscription of Button.OnClick().Subscribe(OnScrollUpClick)
                1 => BindSubscription of Button.OnClick().Subscribe(OnScrollDownClick)
                _ => {}
            ScrollActions.AddWidget(stack_box_slot{
                Widget := ApplySizeHint(Button, vector2{X := 160.0, Y := 40.0})
                HorizontalAlignment := horizontal_alignment.Left
                Padding := margin{
                    Left := if (Index > 0). 10.0 else 20.0
                }
            })
        }

        BindSubscription of GroupingCycleButton.OnClick().Subscribe(OnGroupingCycleClick)
        UpdateGroupingCycleButton()
        BindSubscription of SortingCycleButton.OnClick().Subscribe(OnSortingCycleClick)
        UpdateSortingCycleButton()
        BindSubscription of SortingOrderButton.OnClick().Subscribe(OnSortingOrderClick)
        UpdateSortingOrderButton()

        RowStart := stack_box:
            Orientation := orientation.Horizontal
            Slots := array{
                stack_box_slot:
                    Widget := ApplySizeHint(GroupingCycleButton, vector2{X := 350.0, Y := 40.0})
                    HorizontalAlignment := horizontal_alignment.Left
                    VerticalAlignment := vertical_alignment.Center
                stack_box_slot:
                    Widget := ApplySizeHint(SortingCycleButton, vector2{X := 300.0, Y := 40.0})
                    HorizontalAlignment := horizontal_alignment.Left
                    VerticalAlignment := vertical_alignment.Center
                    Padding := margin{
                        Left := 10.0
                    }
                stack_box_slot:
                    Widget := ApplySizeHint(SortingOrderButton, vector2{X := 150.0, Y := 40.0})
                    HorizontalAlignment := horizontal_alignment.Left
                    VerticalAlignment := vertical_alignment.Center
                    Padding := margin{
                        Left := 10.0
                    }
            }
        Row := canvas:
            Slots := array{
                canvas_slot:
                    Widget := RowStart
                    SizeToContent := true
                    Alignment := vector2{X := 0.0, Y := 0.0}
                canvas_slot:
                    Widget := ScrollActions
                    SizeToContent := true
                    Anchors := anchors{
                        Minimum := vector2{X := 1.0, Y := 0.0}
                        Maximum := vector2{X := 1.0, Y := 0.0}
                    }
                    Alignment := vector2{X := 1.0, Y := 0.0}
            }

        overlay:
            Slots := array{
                overlay_slot:
                    Widget := color_block:
                        DefaultColor := NamedColors.Silver
                    HorizontalAlignment := horizontal_alignment.Fill
                    VerticalAlignment := vertical_alignment.Fill
                overlay_slot:
                    Widget := Row
                    HorizontalAlignment := horizontal_alignment.Fill
                    VerticalAlignment := vertical_alignment.Top
                    Padding := margin{
                        Left := 10.0
                        Right := 10.0
                        Top := 10.0
                        Bottom := 10.0
                    }
            }

    Create<override>():widget=
        Content := m_Content? or (
            Component := memory_tool_menu_overview_content{Parent := Self}
            set m_Content = Some(Component)
            Component
        )
        Stack := stack_box:
            Orientation := orientation.Vertical
            Slots := array{
                stack_box_slot:
                    Widget := CreateControls()
                    HorizontalAlignment := horizontal_alignment.Fill
                    VerticalAlignment := vertical_alignment.Top
                stack_box_slot:
                    Widget := AddChild(Content)
                    HorizontalAlignment := horizontal_alignment.Fill
                    VerticalAlignment := vertical_alignment.Top
                    Padding := margin{
                        Top := 10.0
                        Left := 16.0
                        Right := 16.0
                    }
            }

    UpdateContent<private>():void=
        if (Content := m_Content?) {
            Content.ForceUpdate()
        }

    UpdateEntries<internal>():void=
        GroupingMode := Parent.m_GroupingMode
        var Entries: []memory_tool_stats_entry = if (GroupingMode = memory_tool_stats_grouping_mode.NoGrouping) {
            for (
                Device:AwareListDevices()
                DebugDevice := aware_debug_device[Device]
            ) {
                DeviceStats := DebugDevice.DebugStats()
                memory_tool_stats_entry{
                    EntryName := ToString(Device)
                    Stats := DeviceStats
                }
            }
        } else {
            var ByGroup: [comparable]memory_tool_stats_entry = map{}
            for (
                Device:AwareListDevices()
                AwareDevice := aware_device[Device]
                DebugDevice := aware_debug_device[Device]
            ) {
                Meta := AwareDevice.DeviceMeta()
                DerivedGroup := case(GroupingMode):
                    memory_tool_stats_grouping_mode.DeviceKind => Meta.DeviceName
                    _ => ""
                DeviceStats := DebugDevice.DebugStats()
                if (MergeWith := ByGroup[DerivedGroup]) {
                    set MergeWith.SourceCount += 1
                    set MergeWith.Stats = MergeWith.Stats + DeviceStats
                } else {
                    NewEntry := memory_tool_stats_entry{
                        EntryName := DerivedGroup
                        Stats := DeviceStats
                    }
                    if. set ByGroup[DerivedGroup] = NewEntry
                }
            }
            for (__->Entry:ByGroup) {
                set Entry.EntryName = "{Entry.EntryName} ({Entry.SourceCount})"
                Entry
            }
        }
        if (Integration := AwareGetIntegration[]) {
            GlobalEntries := for (Name->Stats:Integration.GlobalDebugStats()) {
                memory_tool_stats_entry{
                    EntryName := Name
                    Stats := Stats
                }
            }
            set Entries = GlobalEntries + Entries
        }

        SortOpts := quick_sort_opts{
            Order := if (Parent.m_SortingOrderAsc?). sort_order.Ascending else sort_order.Descending
        }
        set Entries = case(Parent.m_SortingMode):
            memory_tool_stats_sorting_mode.Alphabetical => Entries.QuickSortWithOpts(CmpMemoryEntryByName, SortOpts)
            memory_tool_stats_sorting_mode.HigherCount => Entries.QuickSortWithOpts(CmpMemoryEntryByTotalCount, SortOpts)

        PerPage := 2 * PerRowEntries
        OnScreen := 4 * PerRowEntries
        PageCount := DivDefaultingCeil(Entries.Length, PerPage)
        PageIndex := if (m_ScrollPosition >= PageCount) {
            PageCount-1
        } else {
            m_ScrollPosition
        }

        StartIndex := PageIndex * PerPage
        EndIndex := Min(StartIndex + OnScreen, Entries.Length)
        RelevantEntries := Entries.Slice[StartIndex, EndIndex] or array{}
        set m_Entries = RelevantEntries

        Text := "{PageIndex+1}/{PageCount}"
        PageTextWidget.SetText(ToMessage(Text))
        UpdateContent()

    UpdateGroupingCycleButton<private>():void=
        GroupingModeStr := case(Parent.m_GroupingMode):
            memory_tool_stats_grouping_mode.NoGrouping => "None"
            memory_tool_stats_grouping_mode.DeviceKind => "Kind"
        GroupingCycleButton.SetText(ToMessage("Group: {GroupingModeStr}"))

    OnGroupingCycleClick<private>(__:widget_message):void=
        NextMode := case(Parent.m_GroupingMode):
            memory_tool_stats_grouping_mode.NoGrouping => memory_tool_stats_grouping_mode.DeviceKind
            memory_tool_stats_grouping_mode.DeviceKind => memory_tool_stats_grouping_mode.NoGrouping
        set Parent.m_GroupingMode = NextMode
        UpdateGroupingCycleButton()
        UpdateEntries()

    UpdateSortingCycleButton<private>():void=
        SortingModeStr := case(Parent.m_SortingMode):
            memory_tool_stats_sorting_mode.Alphabetical => "A-z"
            memory_tool_stats_sorting_mode.HigherCount => "Total"
        SortingCycleButton.SetText(ToMessage("Sort: {SortingModeStr}"))

    OnSortingCycleClick<private>(__:widget_message):void=
        NextMode := case(Parent.m_SortingMode):
            memory_tool_stats_sorting_mode.Alphabetical => memory_tool_stats_sorting_mode.HigherCount
            memory_tool_stats_sorting_mode.HigherCount => memory_tool_stats_sorting_mode.Alphabetical
        set Parent.m_SortingMode = NextMode
        UpdateSortingCycleButton()
        UpdateEntries()

    UpdateSortingOrderButton<private>():void=
        SortingOrderStr := if (Parent.m_SortingOrderAsc?). "Asc" else "Desc"
        SortingOrderButton.SetText(ToMessage("{SortingOrderStr}"))

    OnSortingOrderClick<private>(__:widget_message):void=
        if (Parent.m_SortingOrderAsc?) {
            set Parent.m_SortingOrderAsc = false
        } else {
            set Parent.m_SortingOrderAsc = true
        }
        UpdateSortingOrderButton()
        UpdateEntries()

    OnScrollUpClick<private>(__:widget_message):void = ScrollBy(-1)
    OnScrollDownClick<private>(__:widget_message):void = ScrollBy(1)

    ScrollBy<private>(Amount:int):void=
        set m_ScrollPosition = Max(0, m_ScrollPosition + Amount)
        UpdateEntries()

memory_tool_stats_entry := class<final>:
    var EntryName: string
    var Stats: aware_debug_stats
    var SourceCount: int = 1

CmpMemoryEntryByName<internal>(A:memory_tool_stats_entry, B:memory_tool_stats_entry)<transacts>:int=
    CmpStr(A.EntryName, B.EntryName)

CmpMemoryEntryByTotalCount<internal>(A:memory_tool_stats_entry, B:memory_tool_stats_entry)<transacts>:int=
    TotalA := A.Stats.Length.Values().Fold(0, iAdd)
    TotalB := B.Stats.Length.Values().Fold(0, iAdd)
    CmpInt(TotalA, TotalB)

memory_tool_menu_overview_content := class<final>(ui_component):
    Parent: memory_tool_menu_overview

    Create<override>():widget=
        RowSlots := for (RowIndex->RowEntries:Parent.m_Entries.Chunks(Parent.PerRowEntries)) {
            ColSlots := for (ColIndex->Entry:RowEntries) {
                StatsDetailWidget := stack_box:
                    Orientation := orientation.Vertical
                for (Key->Value:Entry.Stats.Length) {
                    LineWidget := stack_box:
                        Orientation := orientation.Horizontal
                        Slots := array{
                            stack_box_slot:
                                Widget := CreateTextBlock("{Key}: ", ?TextColor := NamedColors.Silver)
                                HorizontalAlignment := horizontal_alignment.Left
                                VerticalAlignment := vertical_alignment.Top
                            stack_box_slot:
                                Widget := CreateTextBlock(ToString(Value), ?TextColor := NamedColors.CornflowerBlue)
                                HorizontalAlignment := horizontal_alignment.Left
                                VerticalAlignment := vertical_alignment.Top
                        }
                    StatsDetailWidget.AddWidget(stack_box_slot{
                        Widget := LineWidget
                        HorizontalAlignment := horizontal_alignment.Left
                        VerticalAlignment := vertical_alignment.Top
                    })
                }

                EntryName := CreateTextBlock(Entry.EntryName)
                EntryWidget := stack_box:
                    Orientation := orientation.Vertical
                    Slots := array{
                        stack_box_slot:
                            Widget := EntryName
                            HorizontalAlignment := horizontal_alignment.Left
                            VerticalAlignment := vertical_alignment.Top
                        stack_box_slot:
                            Widget := StatsDetailWidget
                            HorizontalAlignment := horizontal_alignment.Left
                            VerticalAlignment := vertical_alignment.Top
                            Padding := margin{
                                Left := 20.0
                            }
                    }

                XAnchor := case(ColIndex):
                    0 => 0.0
                    1 => 0.5
                    _ => 1.0
                canvas_slot:
                    Widget := overlay:
                        Slots := array{
                            overlay_slot:
                                Widget := color_block:
                                    DefaultDesiredSize := vector2{X := 620.0}
                                    DefaultOpacity := 0.0
                                HorizontalAlignment := horizontal_alignment.Fill
                            overlay_slot:
                                Widget := EntryWidget
                                HorizontalAlignment := horizontal_alignment.Left
                        }
                    SizeToContent := true
                    Anchors := anchors{
                        Minimum := vector2{X := XAnchor, Y := 0.0}
                        Maximum := vector2{X := XAnchor, Y := 0.0}
                    }
                    Alignment := vector2{X := XAnchor, Y := 0.0}
            }

            stack_box_slot{
                Widget := canvas{Slots := ColSlots}
                HorizontalAlignment := horizontal_alignment.Fill
                VerticalAlignment := vertical_alignment.Top
                Padding := margin{
                    Top := if (RowIndex = 0) { 0.0 } else { 10.0 }
                }
            }
        }

        stack_box:
            Orientation := orientation.Vertical
            Slots := RowSlots
