using { /UnrealEngine.com/Temporary/Diagnostics }

using { HavocVerse.Logging }

rt_dbg_loop_state<internal> := enum:
	Running
	Completed
	Bailed

rt_dbg_loop<public> := class<final>:
	Context<public>: string
	MaxIt<public>: int

	var m_It<internal>: int = 0
	var m_State<internal>: rt_dbg_loop_state = rt_dbg_loop_state.Running

	It<public>()<transacts>:logic=
		if (m_State <> rt_dbg_loop_state.Running) {
			Log("UNEXPECTED! Loop marked as {m_State} is still being iterated on (ctx: {Context}, it: {m_It}/{MaxIt})",
				?Level := log_level.Error)
			return true
		}
		set m_It += 1
		if (m_It > MaxIt) {
			Log("SEVERE! Bailing out of loop because of suspected infinite loop (ctx: {Context}, it: {m_It}/{MaxIt})",
				?Level := log_level.Error)
			return true
		}
		false

RTStartLoop<public>(Context:string, ?MaxIt:int = 1000)<transacts>:rt_dbg_loop=
	rt_dbg_loop{
		Context := Context
		MaxIt := MaxIt
	}

RTEndLoop<public>(Instance:rt_dbg_loop)<transacts>:void=
	if (Instance.m_State = rt_dbg_loop_state.Running) {
		set Instance.m_State = rt_dbg_loop_state.Completed
	}

ToString<internal>(State:rt_dbg_loop_state)<reads>:string=
	case(State):
		rt_dbg_loop_state.Running => "Running"
		rt_dbg_loop_state.Completed => "Completed"
		rt_dbg_loop_state.Bailed => "Bailed"
