using { /Fortnite.com/Devices }
using { /Fortnite.com/Devices/CreativeAnimation }
using { /UnrealEngine.com/Temporary/Diagnostics }
using { /UnrealEngine.com/Temporary/SpatialMath }
using { /Verse.org/Simulation }

using { HavocVerse.Debug }
using { HavocVerse.Devices }
using { HavocVerse.Math }
using { HavocVerse.Util }
using { HavocVerse.Logging }

apoc_farming_crop_config := class<final><concrete>:
    @editable Id: string = ""
    @editable Stages: []apoc_farming_crop_stage_config = array{}

apoc_farming_crop_stage_config := class<final><concrete>:
    @editable GrowthDuration: float = 1.0
    @editable Prop: creative_prop_asset = DefaultCreativePropAsset

apoc_farming_field_generate_config := class<final><concrete>:
    @editable ResourceId: string = ""
    @editable Amount: int_range = int_range{}

apoc_farming_field_model_config := class<final><concrete>:
    @editable Id: string = ""
    @editable CropKind: string = ""

    @editable StartEnabled: logic = false
    @editable Generate: []apoc_farming_field_generate_config = array{}
    @editable HarvestTriggers: []trigger_device = array{}

apoc_farming_device := class<concrete>(creative_device, aware_device):
    @editable p_Crops: []apoc_farming_crop_config = array{}
    @editable p_FieldModels: []apoc_farming_field_model_config = array{}

    @editable p_ResourceManager: resource_manager_device = resource_manager_device{}
    @editable p_HarvestAutoReplant: logic = true
    @editable p_HarvestTriggersPostEffect: logic = true

    @editable p_InfoPropPlantMe: ?creative_prop_asset = false
    @editable p_InfoPropReadyToHarvest: ?creative_prop_asset = false

    @editable p_ButtonTextPlant: string = "Plant field"
    @editable p_ButtonTextHarvest: string = "Harvest field"

    @editable p_PlantEffectProp: ?creative_prop_asset = false
    @editable p_PlantEffectOffset: vector3 = vector3{Z := 50.0}
    @editable p_PlantEffectDuration: float = 0.5

    @editable p_HarvestEffectProp: ?creative_prop_asset = false
    @editable p_HarvestEffectOffset: vector3 = vector3{Z := 50.0}
    @editable p_HarvestEffectDuration: float = 0.5

    DeviceMeta<override>()<transacts>:aware_device_meta=
        aware_device_meta{
            DeviceName := "apoc_farming"
        }

    GetCropById<public>(CropId:string)<decides><transacts>:apoc_farming_crop_config=
        var Found: ?apoc_farming_crop_config = None()
        for (Crop:p_Crops, not (Found?), Crop.Id = CropId) {
            set Found = Some(Crop)
        }
        Found?

    GetFieldModelById<public>(FieldId:string)<decides><transacts>:apoc_farming_field_model_config=
        var Found: ?apoc_farming_field_model_config = None()
        for (FieldModel:p_FieldModels, not (Found?), FieldModel.Id = FieldId) {
            set Found = Some(FieldModel)
        }
        Found?

apoc_farming_field_config := class<final><concrete>:
    @editable FieldId: string = ""
    @editable SproutMarkers: []creative_prop = array{}
    @editable Buttons: []button_device = array{}
    @editable InfoMarker: ?creative_prop = false
    @editable InstanceHarvestTriggers: []trigger_device = array{}

apoc_farming_instance_device := class<concrete>(creative_device, aware_device, custom_device, resetable):
    @editable p_DevId: string = ""
    @editable{ToolTip := TooltipRoutingIdOverride} p_RoutingIdOverride: ?string = false

    @editable p_AccessControl: ?access_control_device = false
    @editable p_FarmingController: apoc_farming_device = apoc_farming_device{}
    @editable p_Fields: []apoc_farming_field_config = array{}

    var m_FieldInstances: []apoc_farming_active_field = array{}

    DeviceMeta<override>()<transacts>:aware_device_meta=
        aware_device_meta{
            DeviceName := "apoc_farming_instance"
            DevId := p_DevId
            RoutingIdOverride := p_RoutingIdOverride
        }

    OnBegin<override>()<suspends>:void=
        AwareInitDevice(Self, OnInit)
    
    OnInit<private>()<suspends>:void=
        if (AccessControl := p_AccessControl?) {
            AccessControl.RegisterUsingDevice(Self)
        }

        for (FieldConfig:p_Fields) {
            if (FieldModel := p_FarmingController.GetFieldModelById[FieldConfig.FieldId]) {
                CropKind := FieldModel.CropKind
                if (CropConfig := p_FarmingController.GetCropById[CropKind]) {
                    InitField(FieldConfig, FieldModel, CropConfig)
                } else {
                    Log(Self, "Crop kind \"{CropKind}\" is not configured. Ignoring field \"{FieldConfig.FieldId}\"", ?Level := log_level.Error)
                }
            } else {
                Log(Self, "Field \"{FieldConfig.FieldId}\" has no matching model in controller", ?Level := log_level.Error)
            }
        }

    Reset<override>(Agent:?agent):void=
        for (Field:m_FieldInstances) {
            if (Field.Model.StartEnabled?) {
                Field.Reset()
                Field.Enable()
            } else {
                Field.Disable()
            }
        }

    InitField<private>(FieldConfig:apoc_farming_field_config, FieldModel:apoc_farming_field_model_config, CropConfig:apoc_farming_crop_config)<suspends>:void=
        Field := apoc_farming_active_field{
            Device := Self
            Config := FieldConfig
            Model := FieldModel
            CropConfig := CropConfig
        }
        for (SproutIndex->SproutMarker:FieldConfig.SproutMarkers) {
            SproutRes := InitSprout(Field, SproutIndex, SproutMarker)
            if (Sprout := SproutRes?) {
                set Field.m_Sprouts += array{Sprout}
            }
        }
        set m_FieldInstances += array{Field}

        for (Button:FieldConfig.Buttons) {
            Button.InteractedWithEvent.Subscribe(Field.OnButtonInteractedWith)
        }

        if (FieldModel.StartEnabled?) {
            Field.Enable()
        } else {
            Field.Disable()
        }

    InitSprout<private>(
        Field:apoc_farming_active_field,
        SproutIndex:int,
        Marker:creative_prop
    ):?apoc_farming_active_sprout=
        if (not Marker.IsValid[]) {
            FieldId := Field.Config.FieldId
            Log(Self, "Aborting init of sprout index {SproutIndex} on field \"{FieldId}\" because of invalid marker", ?Level := log_level.Error)
            return None()
        }

        Transform := Marker.GetTransform()
        Some(apoc_farming_active_sprout{
            Device := Self
            Field := Field
            Transform := Transform
        })

    CustomTrigger<override>(Input:string, Agent:?agent):logic=
        if (
            SplitRes := Input.SplitOnce[':']
            Action := SplitRes(0)
            Field := m_FieldInstances.FindBy[GetFieldId, SplitRes(1)]
        ) {
            if (Action = "enable_field") {
                Field.Enable()
                return true
            } else if (Action = "disable_field") {
                Field.Disable()
                return true
            }
        }
        false

apoc_farming_field_state := enum:
    Disabled
    Idle
    Growing
    ReadyToHarvest
    Harvesting

GetFieldId<internal>(Field:apoc_farming_active_field)<reads>:string = Field.Config.FieldId

apoc_farming_active_field := class<final>:
    Device: apoc_farming_instance_device
    Config: apoc_farming_field_config
    Model: apoc_farming_field_model_config
    CropConfig: apoc_farming_crop_config

    AbortEvent: event() = event(){}

    var m_Sprouts: []apoc_farming_active_sprout = array{}
    var m_State: apoc_farming_field_state = apoc_farming_field_state.Disabled

    var m_InfoProp: ?creative_prop = false

    Enable():void=
        if (m_State = apoc_farming_field_state.Disabled) {
            SetState(apoc_farming_field_state.Idle)
            Log(Device, "Enabled field \"{Config.FieldId}\"", ?Level := log_level.Debug)
        }

    Reset():void=
        AbortEvent.Signal()
        for (Sprout:m_Sprouts) {
            Sprout.Reset()
        }
        Log(Device, "Disabled field \"{Config.FieldId}\"", ?Level := log_level.Debug)

    Disable():void=
        Reset()
        SetState(apoc_farming_field_state.Disabled)

    OnButtonInteractedWith<internal>(Agent:agent):void=
        if (m_State = apoc_farming_field_state.Idle) {
            spawn:
                Grow()
        } else if (m_State = apoc_farming_field_state.ReadyToHarvest) {
            spawn:
                Harvest(Agent)
        }

    SetState<private>(NewState:apoc_farming_field_state):void=
        set m_State = NewState

        if (Prop := m_InfoProp?) {
            Prop.Dispose()
            set m_InfoProp = None()
        }

        FarmingDevice := Device.p_FarmingController
        if (
            InfoMarker := Config.InfoMarker?
            InfoMarker.IsValid[]

            PropAsset := case(NewState):
                apoc_farming_field_state.Idle => FarmingDevice.p_InfoPropPlantMe?
                apoc_farming_field_state.ReadyToHarvest => FarmingDevice.p_InfoPropReadyToHarvest?
        ) {
            Transform := InfoMarker.GetTransform()
            SpawnRes := SpawnProp(PropAsset, Transform)
            set m_InfoProp = SpawnRes(0)
            Log(Device, "Spawned info prop for field \"{Config.FieldId}\" at {Transform}: {SpawnRes(1)}", ?Level := log_level.Verbose)
        } else if (NewState = apoc_farming_field_state.Idle or NewState = apoc_farming_field_state.ReadyToHarvest) {
            Log(Device, "Not spawning info prop for field \"{Config.FieldId}\": info marker is missing/invalid or prop asset wasn't set", ?Level := log_level.Verbose)
        }

        for (Button:Config.Buttons) {
            ButtonState := case(NewState){
                apoc_farming_field_state.Disabled => None()
                apoc_farming_field_state.Idle => Some(FarmingDevice.p_ButtonTextPlant)
                apoc_farming_field_state.Growing => None()
                apoc_farming_field_state.ReadyToHarvest => Some(FarmingDevice.p_ButtonTextHarvest)
                apoc_farming_field_state.Harvesting => None()
            }
            if (Text := ButtonState?) {
                Button.SetInteractionText(ToMessage(Text))
                Button.Enable()
            } else {
                Button.Disable()
            }
        }

    Grow<private>()<suspends>:void=
        race:
            block:
                AbortEvent.Await()
                SetState(apoc_farming_field_state.Idle)
            DoGrow()

    DoGrow<private>()<suspends>:void=
        SetState(apoc_farming_field_state.Growing)

        if (Device.p_FarmingController.p_PlantEffectProp?) {
            for (Sprout:m_Sprouts) {
                spawn:
                    Sprout.PlayPlantEffect()
            }
        }

        for (Stage->StageConfig:CropConfig.Stages) {
            for (Sprout:m_Sprouts) {
                if (PrevProp := Sprout.m_Prop?) {
                    PrevProp.Dispose()
                }
                set Sprout.m_Prop = SpawnProp(StageConfig.Prop, Sprout.Transform)(0)
            }
            Sleep(StageConfig.GrowthDuration)
        }
        Log(Device, "Field \"{Config.FieldId}\" has fully grown", ?Level := log_level.Verbose)
        SetState(apoc_farming_field_state.ReadyToHarvest)

    Harvest<private>(Agent:agent)<suspends>:void=
        if (m_State <> apoc_farming_field_state.ReadyToHarvest) {
            return
        }
        SetState(apoc_farming_field_state.Harvesting)

        TriggersPostEffect := Device.p_FarmingController.p_HarvestTriggersPostEffect
        if (not TriggersPostEffect?) {
            HandlePostHarvest(Agent)
        }
        JoinAll(for (Sprout:m_Sprouts) {
            spawn:
                Sprout.Harvest()
        })
        if (TriggersPostEffect?) {
            HandlePostHarvest(Agent)
        }

        if (Device.p_FarmingController.p_HarvestAutoReplant?) {
            spawn:
                Grow()
        } else {
            SetState(apoc_farming_field_state.Idle)
        }

    HandlePostHarvest<private>(Agent:agent):void=
        Log(Device, MsgHarvest(Agent, Config.FieldId), ?Level := log_level.Debug)
        for (Generate:Model.Generate) {
            Amount := Generate.Amount.GetRandomValue()
            Device.p_FarmingController.p_ResourceManager.Give(Agent, Generate.ResourceId, Amount)
        }
        for (Trigger:Model.HarvestTriggers) {
            Trigger.Trigger(Agent)
        }
        for (Trigger:Config.InstanceHarvestTriggers) {
            Trigger.Trigger(Agent)
        }
    
    MsgHarvest<private><localizes>(Agent:agent, FieldId:string):message=
        "'{Agent}' harvested field \"{FieldId}\""

apoc_farming_active_sprout := class<final>:
    Device: apoc_farming_instance_device
    Field: apoc_farming_active_field
    Transform: transform

    var m_Prop: ?creative_prop = false

    Reset():void=
        if (Prop := m_Prop?) {
            Prop.Dispose()
            set m_Prop = None()
        }

    PlayPlantEffect()<suspends>:void=
        if (
            EffectPropAsset := Device.p_FarmingController.p_PlantEffectProp?
            EffectTransform := Transform.AddTranslation(Device.p_FarmingController.p_PlantEffectOffset)
            EffectProp := SpawnProp(EffectPropAsset, EffectTransform)(0)?
        ) {
            Sleep(Device.p_FarmingController.p_PlantEffectDuration)
            EffectProp.Dispose()
        }

    Harvest()<suspends>:void=
        RattleProp()
        Reset()

        if (
            EffectPropAsset := Device.p_FarmingController.p_HarvestEffectProp?
            EffectTransform := Transform.AddTranslation(Device.p_FarmingController.p_HarvestEffectOffset)
            EffectProp := SpawnProp(EffectPropAsset, EffectTransform)(0)?
        ) {
            Sleep(Device.p_FarmingController.p_HarvestEffectDuration)
            EffectProp.Dispose()
        }

    RattleProp<private>()<suspends>:void=
        if (
            Prop := m_Prop?
            AnimationController := Prop.GetAnimationController[]
            AnimationController.IsValid[]
        ) {
            Keyframes := array{
                keyframe_delta{
                    DeltaLocation := vector3{X := 15.0}
                    DeltaRotation := IdentityRotation()
                    Time := 0.04
                }
                keyframe_delta{
                    DeltaLocation := vector3{X := -20.0}
                    DeltaRotation := IdentityRotation()
                    Time := 0.04
                }
                keyframe_delta{
                    DeltaLocation := vector3{X := 10.0, Y := 5.0}
                    DeltaRotation := IdentityRotation()
                    Time := 0.03
                }
                keyframe_delta{
                    DeltaLocation := vector3{X := 10.0, Y := -12.0}
                    DeltaRotation := IdentityRotation()
                    Time := 0.04
                }
            }

            AnimationController.PlayNewAnimation(Keyframes)
            AnimationController.AwaitDone()
        }
